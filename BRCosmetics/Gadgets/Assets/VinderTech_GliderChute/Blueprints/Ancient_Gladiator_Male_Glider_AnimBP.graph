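# Animation Blueprint for the Ancient Gladiator male glider, driving deploy, glide and closed poses from the owner's skydiving state.
# A Locomotion state machine feeds curve-weighted additive lean, turn and offset layers, then a jitter layer.
# Cosmetic preview swaps in or adds FrontEnd animations and applies a curve-driven Root bone offset.
# The event graph interpolates run and jitter speeds, hides Glider_Base while closed, and turns notifies into deployment and FX signals.
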AnimBlueprint /BRCosmetics/Gadgets/Assets/VinderTech_GliderChute/Blueprints/Ancient_Gladiator_Male_Glider_AnimBP
target_skeleton=/Game/Gadgets/Assets/VinderTech_GliderChute/Glider_AncientGladiatorMale/Meshes/Ancient_Gladiator_Male_Skeleton
graphs [
 AnimGraph {nodes [
   n1 = Use cached pose 'preIK'()
   n2 = Component To Local(ComponentPose="Transform (Modify) Bone - Bone: Root"(
     ComponentPose=Local To Component(LocalPose=@n1)
     Alpha=1
     Translation=Get FrontEnd_Offsets() -> (FrontEnd_Offsets:Vector)
     Rotation="0,0,0"
     Scale="1,1,1"
    ) {translationMode=BMM_Additive, alphaBoolBlend {blendOption=Linear}, boneToModify {boneName=Root}})
   n3 = Use cached pose 'preIK'()
   n4 = Get IsCosmeticPreview!(self:Fort Player Parachute Object Reference=Get OwnerGlider() -> (OwnerGlider:Fort Player Parachute Object Reference)) -> (IsCosmeticPreview:bool)
   Output Pose(Result=Slot 'DefaultSlot'(Source=Blend Poses by bool(BlendPose_0=@n2, BlendPose_1=@n3, bActiveValue=@n4) {blendType=Linear})) {layerGroup=None}
   n5 = Break Fort Anim Input Skydiving!(FortAnimInput_Skydiving:"Fort Anim Input Skydiving Structure (by ref)"=Get SkydivingState() -> (SkydivingState:Fort Anim Input Skydiving Structure)) -> (
    ParaGilderRootOffset:Vector=~
    ParaGliderRootRotationOverride:Rotator=~
    bUseParaGlideRootModifier:bool=~
    bIsSkydivingFromLaunchPad:bool=~
    bIsSkydivingFromBus:bool=~
    bIsInVortex:bool=~
    bIsDiving:bool=~
    bIsParachuteOpen:bool=~
    bIsSkydiving:bool=~
    bIsParachuteLeaning:bool=~
    bIsSkydiveLeaning:bool=~
    bIsLeaning:bool=~
    bIsSkydiveDiveMode:bool=~
    bPlaySkydiveDrift:bool=~
    bSkydiveDriftDelayActive:bool=~
    bSkydiveDriftAnimAllowed:bool=~
    bIsGliderRight:bool=~
    bIsGliderCenter:bool=~
    bIsGliderLeft:bool=~
    bIsGliderForward:bool=~
    bIsGliderBack:bool=~
    bHasDeployedAndIsStandingUpright:bool=~
    LocalAccelForward:float=~
    LocalAccelRight:float=~
    LocalVelocityRight:float=~
    SkydiveAimPitch:float=~
    SkydiveAimYaw:float=~
    DeployChuteAnimRate:float=~
    SkydiveDriftAnimRate:float=~
    SkydiveDriftAnimRateCurrent:float=~
    SkydiveFidgetAnimRate:float=~
    SkydiveFidgetAnimRateCurrent:float=~
    SkydiveAdditiveAlpha:float=~
    SkydiveDriftDelay:float=~
    SkydiveDriftAnim:int
    SkydiveDriftAnimMax:int
    LaunchpadAnim:int
    LocalAccelDir:ESkydivingDirection Enum
    DirectionLast:ESkydivingDirection Enum
    PoseAdditive:Anim Sequence Object Reference
   ) {is=title K2Node_BreakStruct}
   n6 = Blendspace Player 'BS_Ancient_Gladiator_Male_Lean_GLIDER'(X=@n5.LocalAccelRight, Y=@n5.LocalAccelForward) {
    bResetPlayTimeWhenBlendSpaceChanges=false
    blendSpace=/BRCosmetics/Animation/Game/MainPlayer/Skydive/ParaGlide/AcientGladiator/BS_Ancient_Gladiator_Male_Lean_GLIDER
    groupName=Gliding
    groupRole=AlwaysFollower
    method=SyncGroup
   }
   AnimGraphNode_StateMachine_0 = Locomotion() {is=title AnimGraphNode_StateMachine}
   n7 = Apply Additive!(
    Base=Apply Additive(
     Base=Apply Additive(Base=@AnimGraphNode_StateMachine_0, Additive=@n6, AlphaCurveName=ParaGlide_Lean) {alphaInputType=Curve, alphaBoolBlend {blendOption=Linear}, alphaScaleBiasClamp {bias=1, scale=-1}}
     Additive=Use cached pose 'Turn'()
     AlphaCurveName=ParaGlide_Turn
    ) {alphaInputType=Curve, alphaBoolBlend {blendOption=Linear}, alphaScaleBiasClamp {bias=1, scale=-1}}
    Additive="Sequence Player (additive) 'Ancient_Gladiator_Male_Offset_Pose_GLIDER'"() {sequence=/BRCosmetics/Animation/Game/MainPlayer/Skydive/ParaGlide/AcientGladiator/Ancient_Gladiator_Male_Offset_Pose_GLIDER}
    AlphaCurveName:Name=DisableOffset
   ) -> (Pose) {alphaInputType=Curve, alphaBoolBlend {blendOption=Linear}, alphaScaleBiasClamp {bMapRange=true, outRange {max=0, min=1}}}
   Save cached pose 'preIK'(Pose=Apply Additive!(
     Base=@n7
     Additive="Sequence Player (additive) 'Ancient_Gladiator_Male_Jitter_GLIDER'"!(PlayRate:float=1) -> (Pose) {sequence=/BRCosmetics/Animation/Game/MainPlayer/Skydive/ParaGlide/AcientGladiator/Ancient_Gladiator_Male_Jitter_GLIDER}
     Alpha:float
    ) -> (Pose) {alphaBoolBlend {blendOption=Linear}}) {cacheName=preIK}
   Save cached pose 'Turn'(Pose=Blendspace Player 'BS_Ancient_Gladiator_Male_Turn_GLIDER'(X=@n5.SkydiveAimYaw, Y=@n5.LocalAccelForward) {blendSpace=/BRCosmetics/Animation/Game/MainPlayer/Skydive/ParaGlide/AcientGladiator/BS_Ancient_Gladiator_Male_Turn_GLIDER}) {cacheName=Turn}
  ]}
 Locomotion {
  entry=Closed
  path=AnimGraph.AnimGraphNode_StateMachine_0.Locomotion
  ids=id
  nodes [
   AnimStateNode_0 = Idle_Open() {is=title AnimStateNode}
   AnimStateNode_1 = Open_Into() {is=title AnimStateNode, stateEntered {guid=#1, notifyName=DeployEntered}}
   AnimStateNode_3 = Closed() {is=title AnimStateNode, stateEntered {guid=#2, notifyName=ClosedEntered}}
   AnimStateTransitionNode_0 = AnimStateTransitionNode() {bAutomaticRuleBasedOnSequencePlayerInState=true, crossfadeDuration=0}
   AnimStateTransitionNode_10 = AnimStateTransitionNode() {crossfadeDuration=0, priorityOrder=2}
   AnimStateTransitionNode_8 = AnimStateTransitionNode() {crossfadeDuration=0}
   AnimStateTransitionNode_9 = AnimStateTransitionNode() {crossfadeDuration=0}
  ]
  links [
   {
    dst=Open_Into
    rule=AnimGraph.AnimGraphNode_StateMachine_0.Locomotion.AnimStateTransitionNode_8.Transition
    src=Closed
    via=AnimStateTransitionNode_8
   }
   {
    dst=Closed
    rule=AnimGraph.AnimGraphNode_StateMachine_0.Locomotion.AnimStateTransitionNode_9.Transition
    src=Idle_Open
    via=AnimStateTransitionNode_9
   }
   {
    dst=Closed
    rule=AnimGraph.AnimGraphNode_StateMachine_0.Locomotion.AnimStateTransitionNode_10.Transition
    src=Open_Into
    via=AnimStateTransitionNode_10
   }
   {
    dst=Idle_Open
    rule=AnimGraph.AnimGraphNode_StateMachine_0.Locomotion.AnimStateTransitionNode_0.Transition
    src=Open_Into
    via=AnimStateTransitionNode_0
   }
  ]
 }
 Idle_Open {
  path=AnimGraph.AnimGraphNode_StateMachine_0.Locomotion.AnimStateNode_0.Idle_Open
  nodes [
   n1 = Break Fort Anim Input Skydiving(FortAnimInput_Skydiving=Get SkydivingState()) -> (
    ParaGilderRootOffset=~
    ParaGliderRootRotationOverride=~
    bUseParaGlideRootModifier=~
    bIsSkydivingFromLaunchPad=~
    bIsSkydivingFromBus=~
    bIsInVortex=~
    bIsDiving=~
    bIsParachuteOpen=~
    bIsSkydiving=~
    bIsParachuteLeaning=~
    bIsSkydiveLeaning=~
    bIsLeaning=~
    bIsSkydiveDiveMode=~
    bPlaySkydiveDrift=~
    bSkydiveDriftDelayActive=~
    bSkydiveDriftAnimAllowed=~
    bIsGliderRight=~
    bIsGliderCenter=~
    bIsGliderLeft=~
    bIsGliderForward=~
    bIsGliderBack=~
    bHasDeployedAndIsStandingUpright=~
    LocalAccelForward=~
    LocalAccelRight=~
    LocalVelocityRight=~
    SkydiveAimPitch=~
    SkydiveAimYaw=~
    DeployChuteAnimRate=~
    SkydiveDriftAnimRate=~
    SkydiveDriftAnimRateCurrent=~
    SkydiveFidgetAnimRate=~
    SkydiveFidgetAnimRateCurrent=~
    SkydiveAdditiveAlpha=~
    SkydiveDriftDelay=~
   )
   n2 = "Sequence Player (additive) 'Ancient_Gladiator_Male_Glide_FrontEnd_GLIDER'"() {
    groupName=Gliding_FrontEnd
    method=SyncGroup
    sequence=/BRCosmetics/Animation/Game/MainPlayer/Skydive/ParaGlide/AcientGladiator/Ancient_Gladiator_Male_Glide_FrontEnd_GLIDER
   }
   n3 = Get IsCosmeticPreview(self=Get OwnerGlider())
   Output Animation Pose(Result=Apply Additive(
     Base=Blendspace Player 'BS_Ancient_Gladiator_Male_Run_GLIDER'!(X=@n1.SkydiveAimYaw, Y=@n1.LocalAccelForward, PlayRate:float=1) -> (Pose) {blendSpace=/BRCosmetics/Animation/Game/MainPlayer/Skydive/ParaGlide/AcientGladiator/BS_Ancient_Gladiator_Male_Run_GLIDER}
     Additive=@n2
     Alpha="To Float (Boolean)"(InBool=@n3)
    ) {alphaBoolBlend {blendOption=Linear}})
  ]
 }
 Open_Into {
  path=AnimGraph.AnimGraphNode_StateMachine_0.Locomotion.AnimStateNode_1.Open_Into
  nodes [
   n1 = Break Fort Anim Input Skydiving(FortAnimInput_Skydiving=Get SkydivingState()) -> (
    ParaGilderRootOffset=~
    ParaGliderRootRotationOverride=~
    bUseParaGlideRootModifier=~
    bIsSkydivingFromLaunchPad=~
    bIsSkydivingFromBus=~
    bIsInVortex=~
    bIsDiving=~
    bIsParachuteOpen=~
    bIsSkydiving=~
    bIsParachuteLeaning=~
    bIsSkydiveLeaning=~
    bIsLeaning=~
    bIsSkydiveDiveMode=~
    bPlaySkydiveDrift=~
    bSkydiveDriftDelayActive=~
    bSkydiveDriftAnimAllowed=~
    bIsGliderRight=~
    bIsGliderCenter=~
    bIsGliderLeft=~
    bIsGliderForward=~
    bIsGliderBack=~
    bHasDeployedAndIsStandingUpright=~
    LocalAccelForward=~
    LocalAccelRight=~
    LocalVelocityRight=~
    SkydiveAimPitch=~
    SkydiveAimYaw=~
    DeployChuteAnimRate=~
    SkydiveDriftAnimRate=~
    SkydiveDriftAnimRateCurrent=~
    SkydiveFidgetAnimRate=~
    SkydiveFidgetAnimRateCurrent=~
    SkydiveAdditiveAlpha=~
    SkydiveDriftDelay=~
   )
   n2 = Blendspace Player 'BS_Ancient_Gladiator_Male_Into_GLIDER'(X=@n1.SkydiveAimYaw, Y=@n1.LocalAccelForward) {
    bLoop=false
    bResetPlayTimeWhenBlendSpaceChanges=false
    blendSpace=/BRCosmetics/Animation/Game/MainPlayer/Skydive/ParaGlide/AcientGladiator/BS_Ancient_Gladiator_Male_Into_GLIDER
   }
   n3 = Sequence Player 'Ancient_Gladiator_Male_Into_FrontEnd_GLIDER'() {
    bLoopAnimation=false
    sequence=/BRCosmetics/Animation/Game/MainPlayer/Skydive/ParaGlide/AcientGladiator/Ancient_Gladiator_Male_Into_FrontEnd_GLIDER
   }
   n4 = Get OwnerGlider()
   Output Animation Pose(Result=Blend Poses by bool(BlendPose_0=@n3, BlendPose_1=@n2, bActiveValue=Get IsCosmeticPreview(self=@n4)) {blendType=Linear})
  ]
 }
 Closed {
  path=AnimGraph.AnimGraphNode_StateMachine_0.Locomotion.AnimStateNode_3.Closed
  nodes [
   n1 = Get OwnerGlider()
   Output Animation Pose(Result=Blend Poses by bool(
     BlendPose_0=Sequence Player 'Ancient_Gladiator_Male_Closed_FrontEnd_GLIDER'() {sequence=/BRCosmetics/Animation/Game/MainPlayer/Skydive/ParaGlide/AcientGladiator/Ancient_Gladiator_Male_Closed_FrontEnd_GLIDER}
     BlendPose_1=Sequence Player 'Ancient_Gladiator_Male_Closed_GLIDER'() {sequence=/BRCosmetics/Animation/Game/MainPlayer/Skydive/ParaGlide/AcientGladiator/Ancient_Gladiator_Male_Closed_GLIDER}
     bActiveValue=Get IsCosmeticPreview(self=@n1)
    ) {blendType=Linear})
  ]
 }
 Transition {path=AnimGraph.AnimGraphNode_StateMachine_0.Locomotion.AnimStateTransitionNode_0.Transition, nodes [Result()]}
 Transition {
  path=AnimGraph.AnimGraphNode_StateMachine_0.Locomotion.AnimStateTransitionNode_10.Transition
  nodes [
   n1 = Get bForceOpen() -> (bForceOpen:bool)
   n2 = Break Fort Anim Input Skydiving(FortAnimInput_Skydiving=Get SkydivingState()) -> (
    ParaGilderRootOffset=~
    ParaGliderRootRotationOverride=~
    bUseParaGlideRootModifier=~
    bIsSkydivingFromLaunchPad=~
    bIsSkydivingFromBus=~
    bIsInVortex=~
    bIsDiving=~
    bIsParachuteOpen=~
    bIsSkydiving=~
    bIsParachuteLeaning=~
    bIsSkydiveLeaning=~
    bIsLeaning=~
    bIsSkydiveDiveMode=~
    bPlaySkydiveDrift=~
    bSkydiveDriftDelayActive=~
    bSkydiveDriftAnimAllowed=~
    bIsGliderRight=~
    bIsGliderCenter=~
    bIsGliderLeft=~
    bIsGliderForward=~
    bIsGliderBack=~
    bHasDeployedAndIsStandingUpright=~
    LocalAccelForward=~
    LocalAccelRight=~
    LocalVelocityRight=~
    SkydiveAimPitch=~
    SkydiveAimYaw=~
    DeployChuteAnimRate=~
    SkydiveDriftAnimRate=~
    SkydiveDriftAnimRateCurrent=~
    SkydiveFidgetAnimRate=~
    SkydiveFidgetAnimRateCurrent=~
    SkydiveAdditiveAlpha=~
    SkydiveDriftDelay=~
   )
   Result(bCanEnterTransition=AND Boolean(A=NOT Boolean(A=@n2.bIsParachuteOpen), B=NOT Boolean(A=@n1)))
  ]
 }
 Transition {
  path=AnimGraph.AnimGraphNode_StateMachine_0.Locomotion.AnimStateTransitionNode_8.Transition
  nodes [
   n1 = Get SkydivingState()
   Result(bCanEnterTransition=OR Boolean(
     A=Break Fort Anim Input Skydiving(FortAnimInput_Skydiving=@n1) -> (
      ParaGilderRootOffset=~
      ParaGliderRootRotationOverride=~
      bUseParaGlideRootModifier=~
      bIsSkydivingFromLaunchPad=~
      bIsSkydivingFromBus=~
      bIsInVortex=~
      bIsDiving=~
      bIsParachuteOpen=~
      bIsSkydiving=~
      bIsParachuteLeaning=~
      bIsSkydiveLeaning=~
      bIsLeaning=~
      bIsSkydiveDiveMode=~
      bPlaySkydiveDrift=~
      bSkydiveDriftDelayActive=~
      bSkydiveDriftAnimAllowed=~
      bIsGliderRight=~
      bIsGliderCenter=~
      bIsGliderLeft=~
      bIsGliderForward=~
      bIsGliderBack=~
      bHasDeployedAndIsStandingUpright=~
      LocalAccelForward=~
      LocalAccelRight=~
      LocalVelocityRight=~
      SkydiveAimPitch=~
      SkydiveAimYaw=~
      DeployChuteAnimRate=~
      SkydiveDriftAnimRate=~
      SkydiveDriftAnimRateCurrent=~
      SkydiveFidgetAnimRate=~
      SkydiveFidgetAnimRateCurrent=~
      SkydiveAdditiveAlpha=~
      SkydiveDriftDelay=~
     ).bIsParachuteOpen
     B=Get bForceOpen()
    ))
  ]
 }
 Transition {
  path=AnimGraph.AnimGraphNode_StateMachine_0.Locomotion.AnimStateTransitionNode_9.Transition
  nodes [
   n1 = Get bForceOpen()
   n2 = Break Fort Anim Input Skydiving(FortAnimInput_Skydiving=Get SkydivingState()) -> (
    ParaGilderRootOffset=~
    ParaGliderRootRotationOverride=~
    bUseParaGlideRootModifier=~
    bIsSkydivingFromLaunchPad=~
    bIsSkydivingFromBus=~
    bIsInVortex=~
    bIsDiving=~
    bIsParachuteOpen=~
    bIsSkydiving=~
    bIsParachuteLeaning=~
    bIsSkydiveLeaning=~
    bIsLeaning=~
    bIsSkydiveDiveMode=~
    bPlaySkydiveDrift=~
    bSkydiveDriftDelayActive=~
    bSkydiveDriftAnimAllowed=~
    bIsGliderRight=~
    bIsGliderCenter=~
    bIsGliderLeft=~
    bIsGliderForward=~
    bIsGliderBack=~
    bHasDeployedAndIsStandingUpright=~
    LocalAccelForward=~
    LocalAccelRight=~
    LocalVelocityRight=~
    SkydiveAimPitch=~
    SkydiveAimYaw=~
    DeployChuteAnimRate=~
    SkydiveDriftAnimRate=~
    SkydiveDriftAnimRateCurrent=~
    SkydiveFidgetAnimRate=~
    SkydiveFidgetAnimRateCurrent=~
    SkydiveAdditiveAlpha=~
    SkydiveDriftDelay=~
   )
   Result(bCanEnterTransition=AND Boolean(A=NOT Boolean(A=@n2.bIsParachuteOpen), B=NOT Boolean(A=@n1)))
  ]
 }
 ClosedFX {nodes [ClosedFX() {is=title K2Node_FunctionEntry}]}
 EventGraph {nodes [
   n1 = Event BlueprintUpdateAnimation()
   > Set Time(Time:double=@n1.DeltaTimeX) -> (Output_Get:double)
   > n2 = Branch(Condition=Get IsCosmeticPreview(self=Get OwnerGlider()))
   n3 = GetCurveValue(CurveName=Into_FwdBack)
   > n4 = Set FrontEnd_Offsets(FrontEnd_Offsets:Vector=MakeVector(
     X=MapRangeClamped(Value=@n3, InRangeA=1, InRangeB=0, OutRangeA=55, OutRangeB=10)
     Y=MapRangeClamped(Value=@n3, InRangeA=1, InRangeB=0, OutRangeB=-50)
     Z=-65
    )) -> (Output_Get:Vector)
   n5 = Break Fort Anim Input Skydiving(FortAnimInput_Skydiving=Get SkydivingState()) -> (
    ParaGilderRootOffset=~
    ParaGliderRootRotationOverride=~
    bUseParaGlideRootModifier=~
    bIsSkydivingFromLaunchPad=~
    bIsSkydivingFromBus=~
    bIsInVortex=~
    bIsDiving=~
    bIsParachuteOpen=~
    bIsSkydiving=~
    bIsParachuteLeaning=~
    bIsSkydiveLeaning=~
    bIsLeaning=~
    bIsSkydiveDiveMode=~
    bPlaySkydiveDrift=~
    bSkydiveDriftDelayActive=~
    bSkydiveDriftAnimAllowed=~
    bIsGliderRight=~
    bIsGliderCenter=~
    bIsGliderLeft=~
    bIsGliderForward=~
    bIsGliderBack=~
    bHasDeployedAndIsStandingUpright=~
    LocalAccelForward=~
    LocalAccelRight=~
    LocalVelocityRight=~
    SkydiveAimPitch=~
    SkydiveAimYaw=~
    DeployChuteAnimRate=~
    SkydiveDriftAnimRate=~
    SkydiveDriftAnimRateCurrent=~
    SkydiveFidgetAnimRate=~
    SkydiveFidgetAnimRateCurrent=~
    SkydiveAdditiveAlpha=~
    SkydiveDriftDelay=~
   )
   n6 = ForwardBack_Value_Interp!(
    execute=[@n2.else, @n4.then]
    Time:double=Get Time() -> (Time:double)
    Currnet_Value:double=Get RunSpeed() -> (RunSpeed:double)
    Idle_Value:double=1
    Idle_InterpSpeed:double=3
    Forward_Value:double=1
    Forward_InterpSpeed:double=1
    Back_Value:double=1.25
    Back_InterpSpeed:double=3
    LocalAccelForward:double=@n5.LocalAccelForward
   ) -> (then, New_Value:double)
   > Set RunSpeed(RunSpeed:double=@n6) -> (Output_Get:double)
   > n7 = ForwardBack_Value_Interp(
    Time=Get Time()
    Currnet_Value=Get Jitter_Speed() -> (Jitter_Speed:double)
    Idle_Value=1
    Idle_InterpSpeed=1
    Forward_Value=1.5
    Forward_InterpSpeed=1
    Back_Value=-1
    Back_InterpSpeed=3
    LocalAccelForward=@n5.LocalAccelForward
   )
   > Set Jitter_Speed(Jitter_Speed:double=@n7) -> (Output_Get:double)
   AnimNotify_TurnONEngines() {is=title K2Node_Event}
   > SetGliderFullyDeployed!(execute, self:Fort Glider Anim Instance Object Reference, bIsFullyDeployed:bool=true) -> (then)
   AnimNotify_FX_RoarOn() {is=title K2Node_Event}
   > Set FX_Roar(FX_Roar:bool=true) -> (Output_Get:bool)
   AnimNotify_FX_RoarOff() {is=title K2Node_Event}
   > Set FX_Roar(FX_Roar=false)
   n8 = AnimNotify_ClosedEntered() {is=title K2Node_Event}
   n9 = GetOwningComponent()
   HideBoneByName!(
    execute=[@n8.then, @n10.then]
    self:Skinned Mesh Component Object Reference=@n9
    BoneName:Name=Glider_Base
    PhysBodyOption:EPhysBodyOp Enum=PBO_None
   ) -> (then)
   AnimNotify_DeployEntered() {is=title K2Node_Event}
   > Delay(Duration=0.2)
   > UnHideBoneByName!(execute, self:Skinned Mesh Component Object Reference=@n9, BoneName:Name=Glider_Base) -> (then)
   n10 = Event BlueprintInitializeAnimation()
   AnimNotify_ClosedFX() {is=title K2Node_Event}
   > Call Closed FX!(execute:Exec, self:Ancient Gladiator Male Glider Anim BP Object Reference) -> (then:Exec) {is=title K2Node_CallDelegate}
   AnimNotify_ResetFX() {is=title K2Node_Event}
   > Call Reset FX!(execute:Exec, self:Ancient Gladiator Male Glider Anim BP Object Reference) -> (then:Exec) {is=title K2Node_CallDelegate}
   Call Roar Event!(execute:Exec, self:DV Dragon Glider Anim BP Object Reference) -> (then:Exec) {is=title K2Node_CallDelegate}
   Call Roar Event End!(execute:Exec, self:DV Dragon Glider Anim BP Object Reference) -> (then:Exec) {is=title K2Node_CallDelegate}
  ]}
 ResetFX {nodes [ResetFX() {is=title K2Node_FunctionEntry}]}
 RoarEvent {nodes [RoarEvent() {is=title K2Node_FunctionEntry}]}
 RoarEventEnd {nodes [RoarEventEnd() {is=title K2Node_FunctionEntry}]}
]
variables [
 AimPitch {type=number}
 BlockAnimRule {type=boolean}
 ClosedFX {default=()}
 DeployAnimRate {type=number}
 FrontEnd_Offsets {type=Vector, default=<45 0 -65>}
 FX_Roar {type=boolean}
 IsParachuteOpen {type=boolean}
 Jitter_Speed {type=number}
 LeanFromInto {type=boolean}
 ResetFX {default=()}
 RoarEvent {default=()}
 RoarEventEnd {default=()}
 RunSpeed {type=number}
 SkydiveYaw {type=number}
 Time {type=number}
]
