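# Anim Blueprint for the Ashton Boardwalk glider, animating its opening, gliding and closed poses during skydiving.
# A Closed/Open_Into/Idle_Open state machine follows parachute or force-open flags, with cosmetic-preview variants.
# Lean, turn, jitter and turbine additives layer on top, then a curve-driven bone copy moves Frame before the slot.
# The event graph interpolates turbine and jitter rates from forward acceleration, stopping turbines when the ParaGlide_Lean curve reaches one.
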AnimBlueprint /BRCosmetics/Gadgets/Assets/VinderTech_GliderChute/Blueprints/Ashton_Boardwalk_Glider_AnimBP
target_skeleton=/BRCosmetics/Gadgets/Assets/VinderTech_GliderChute/Glider_Ashton_Boardwalk/Meshes/Ashton_Boardwalk_Glider_Skeleton
graphs [
 AnimGraph {nodes [
   n1 = Break Fort Anim Input Skydiving!(FortAnimInput_Skydiving:"Fort Anim Input Skydiving Structure (by ref)"=Get SkydivingState() -> (SkydivingState:Fort Anim Input Skydiving Structure)) -> (
    ParaGilderRootOffset:Vector=~
    ParaGliderRootRotationOverride:Rotator=~
    bUseParaGlideRootModifier:bool=~
    bIsSkydivingFromLaunchPad:bool=~
    bIsSkydivingFromBus:bool=~
    bIsInVortex:bool=~
    bIsDiving:bool=~
    bIsParachuteOpen:bool=~
    bIsSkydiving:bool=~
    bIsParachuteLeaning:bool=~
    bIsSkydiveLeaning:bool=~
    bIsLeaning:bool=~
    bIsSkydiveDiveMode:bool=~
    bPlaySkydiveDrift:bool=~
    bSkydiveDriftDelayActive:bool=~
    bSkydiveDriftAnimAllowed:bool=~
    bIsGliderRight:bool=~
    bIsGliderCenter:bool=~
    bIsGliderLeft:bool=~
    bIsGliderForward:bool=~
    bIsGliderBack:bool=~
    bHasDeployedAndIsStandingUpright:bool=~
    LocalAccelForward:float=~
    LocalAccelRight:float=~
    LocalVelocityRight:float=~
    SkydiveAimPitch:float=~
    SkydiveAimYaw:float=~
    DeployChuteAnimRate:float=~
    SkydiveDriftAnimRate:float=~
    SkydiveDriftAnimRateCurrent:float=~
    SkydiveFidgetAnimRate:float=~
    SkydiveFidgetAnimRateCurrent:float=~
    SkydiveAdditiveAlpha:float=~
    SkydiveDriftDelay:float=~
    SkydiveDriftAnim:int
    SkydiveDriftAnimMax:int
    LaunchpadAnim:int
    LocalAccelDir:ESkydivingDirection Enum
    DirectionLast:ESkydivingDirection Enum
    PoseAdditive:Anim Sequence Object Reference
   ) {is=title K2Node_BreakStruct}
   n2 = Blendspace Player 'BS_Ashton_Boardwalk_Lean_GLIDER'(X=@n1.LocalAccelRight, Y=@n1.LocalAccelForward) {blendSpace=/BRCosmetics/Animation/Game/MainPlayer/Skydive/ParaGlide/AshtonBoardwalk/BS_Ashton_Boardwalk_Lean_GLIDER}
   AnimGraphNode_StateMachine_0 = Locomotion() {is=title AnimGraphNode_StateMachine}
   n3 = Apply Additive(
    Base=Apply Additive(
     Base=Apply Additive!(Base=@AnimGraphNode_StateMachine_0, Additive=@n2, AlphaCurveName:Name=ParaGlide_Lean) -> (Pose) {alphaInputType=Curve, alphaBoolBlend {blendOption=Linear}, alphaScaleBiasClamp {bias=1, scale=-1}}
     Additive=Blendspace Player 'BS_Ashton_Boardwalk_Turn_GLIDER'(X=@n1.SkydiveAimYaw, Y=@n1.SkydiveAimPitch) {blendSpace=/BRCosmetics/Animation/Game/MainPlayer/Skydive/ParaGlide/AshtonBoardwalk/BS_Ashton_Boardwalk_Turn_GLIDER}
    ) {alphaBoolBlend {blendOption=Linear}}
    Additive="Sequence Player (additive) 'Ashton_Boardwalk_Jitter_GLIDER'"!(PlayRate:float=Get JitterSpeed() -> (JitterSpeed:double)) -> (Pose) {sequence=/BRCosmetics/Animation/Game/MainPlayer/Skydive/ParaGlide/AshtonBoardwalk/Ashton_Boardwalk_Jitter_GLIDER}
   ) {alphaBoolBlend {blendOption=Linear}}
   n4 = "Sequence Player (additive) 'Ashton_Boardwalk_Turbines_GLIDER'"!(PlayRate:float=Get TurbineSpeed() -> (TurbineSpeed:double)) -> (Pose) {sequence=/BRCosmetics/Animation/Game/MainPlayer/Skydive/ParaGlide/AshtonBoardwalk/Ashton_Boardwalk_Turbines_GLIDER}
   n5 = Copy Bone - Source Bone: VB Root_Frame - Target Bone: Frame!(
    ComponentPose:Component Space Pose Link Structure=Local To Component(LocalPose=Apply Additive!(Base=@n3, Additive=@n4, Alpha:float) -> (Pose) {alphaBoolBlend {blendOption=Linear}})
    AlphaCurveName:Name=FreeShip
    bCopyTranslation:bool=true
    bCopyRotation:bool=true
    bCopyScale:bool=true
   ) -> (Pose:Component Space Pose Link Structure) {
    is=title AnimGraphNode_CopyBone
    alphaInputType=Curve
    alphaBoolBlend {blendOption=Linear}
    sourceBone {boneName=VB Root_Frame}
    targetBone {boneName=Frame}
   }
   Output Pose(Result=Slot 'DefaultSlot'(Source=Component To Local(ComponentPose=@n5))) {layerGroup=None}
  ]}
 Locomotion {
  entry=Closed
  path=AnimGraph.AnimGraphNode_StateMachine_0.Locomotion
  ids=id
  nodes [
   AnimStateNode_0 = Idle_Open() {is=title AnimStateNode}
   AnimStateNode_1 = Open_Into() {is=title AnimStateNode}
   AnimStateNode_2 = Closed() {is=title AnimStateNode}
   AnimStateTransitionNode_0 = AnimStateTransitionNode() {bAutomaticRuleBasedOnSequencePlayerInState=true, blendMode=Linear, crossfadeDuration=0}
   AnimStateTransitionNode_1 = AnimStateTransitionNode() {crossfadeDuration=0}
   AnimStateTransitionNode_2 = AnimStateTransitionNode() {crossfadeDuration=0}
   AnimStateTransitionNode_3 = AnimStateTransitionNode() {crossfadeDuration=0, priorityOrder=2}
  ]
  links [
   {
    dst=Open_Into
    rule=AnimGraph.AnimGraphNode_StateMachine_0.Locomotion.AnimStateTransitionNode_1.Transition
    src=Closed
    via=AnimStateTransitionNode_1
   }
   {
    dst=Closed
    rule=AnimGraph.AnimGraphNode_StateMachine_0.Locomotion.AnimStateTransitionNode_2.Transition
    src=Idle_Open
    via=AnimStateTransitionNode_2
   }
   {
    dst=Closed
    rule=AnimGraph.AnimGraphNode_StateMachine_0.Locomotion.AnimStateTransitionNode_3.Transition
    src=Open_Into
    via=AnimStateTransitionNode_3
   }
   {
    dst=Idle_Open
    rule=AnimGraph.AnimGraphNode_StateMachine_0.Locomotion.AnimStateTransitionNode_0.Transition
    src=Open_Into
    via=AnimStateTransitionNode_0
   }
  ]
 }
 Idle_Open {
  path=AnimGraph.AnimGraphNode_StateMachine_0.Locomotion.AnimStateNode_0.Idle_Open
  nodes [
   Blend Poses by bool(bActiveValue=Get IsCosmeticPreview!(self:Fort Player Parachute Object Reference=Get OwnerGlider() -> (OwnerGlider:Fort Player Parachute Object Reference)) -> (IsCosmeticPreview:bool)) {blendType=Linear}
   n1 = Break Fort Anim Input Skydiving(FortAnimInput_Skydiving=Get SkydivingState()) -> (
    ParaGilderRootOffset=~
    ParaGliderRootRotationOverride=~
    bUseParaGlideRootModifier=~
    bIsSkydivingFromLaunchPad=~
    bIsSkydivingFromBus=~
    bIsInVortex=~
    bIsDiving=~
    bIsParachuteOpen=~
    bIsSkydiving=~
    bIsParachuteLeaning=~
    bIsSkydiveLeaning=~
    bIsLeaning=~
    bIsSkydiveDiveMode=~
    bPlaySkydiveDrift=~
    bSkydiveDriftDelayActive=~
    bSkydiveDriftAnimAllowed=~
    bIsGliderRight=~
    bIsGliderCenter=~
    bIsGliderLeft=~
    bIsGliderForward=~
    bIsGliderBack=~
    bHasDeployedAndIsStandingUpright=~
    LocalAccelForward=~
    LocalAccelRight=~
    LocalVelocityRight=~
    SkydiveAimPitch=~
    SkydiveAimYaw=~
    DeployChuteAnimRate=~
    SkydiveDriftAnimRate=~
    SkydiveDriftAnimRateCurrent=~
    SkydiveFidgetAnimRate=~
    SkydiveFidgetAnimRateCurrent=~
    SkydiveAdditiveAlpha=~
    SkydiveDriftDelay=~
   )
   n2 = Get OwnerGlider()
   Output Animation Pose(Result=Apply Additive!(
     Base=Blendspace Player 'BS_Ashton_Boardwalk_GLIDER'(X=@n1.SkydiveAimYaw, Y=@n1.SkydiveAimPitch) {blendSpace=/BRCosmetics/Animation/Game/MainPlayer/Skydive/ParaGlide/AshtonBoardwalk/BS_Ashton_Boardwalk_GLIDER}
     Additive="Sequence Player (additive) 'Ashton_Boardwalk_Glide_FrontEnd_GLIDER'"() {sequence=/BRCosmetics/Animation/Game/MainPlayer/Skydive/ParaGlide/AshtonBoardwalk/Ashton_Boardwalk_Glide_FrontEnd_GLIDER}
     bAlphaBoolEnabled:bool=Get IsCosmeticPreview(self=@n2)
    ) -> (Pose) {alphaInputType=Bool, alphaBoolBlend {blendOption=Linear}})
  ]
 }
 Open_Into {
  path=AnimGraph.AnimGraphNode_StateMachine_0.Locomotion.AnimStateNode_1.Open_Into
  nodes [
   Apply Additive(bAlphaBoolEnabled=Get IsCosmeticPreview(self=Get OwnerGlider())) {alphaInputType=Bool, alphaBoolBlend {blendOption=Linear}}
   n1 = Break Fort Anim Input Skydiving(FortAnimInput_Skydiving=Get SkydivingState()) -> (
    ParaGilderRootOffset=~
    ParaGliderRootRotationOverride=~
    bUseParaGlideRootModifier=~
    bIsSkydivingFromLaunchPad=~
    bIsSkydivingFromBus=~
    bIsInVortex=~
    bIsDiving=~
    bIsParachuteOpen=~
    bIsSkydiving=~
    bIsParachuteLeaning=~
    bIsSkydiveLeaning=~
    bIsLeaning=~
    bIsSkydiveDiveMode=~
    bPlaySkydiveDrift=~
    bSkydiveDriftDelayActive=~
    bSkydiveDriftAnimAllowed=~
    bIsGliderRight=~
    bIsGliderCenter=~
    bIsGliderLeft=~
    bIsGliderForward=~
    bIsGliderBack=~
    bHasDeployedAndIsStandingUpright=~
    LocalAccelForward=~
    LocalAccelRight=~
    LocalVelocityRight=~
    SkydiveAimPitch=~
    SkydiveAimYaw=~
    DeployChuteAnimRate=~
    SkydiveDriftAnimRate=~
    SkydiveDriftAnimRateCurrent=~
    SkydiveFidgetAnimRate=~
    SkydiveFidgetAnimRateCurrent=~
    SkydiveAdditiveAlpha=~
    SkydiveDriftDelay=~
   )
   n2 = Sequence Player 'Ashton_Boardwalk_Into_FrontEnd_GLIDER'() {
    bLoopAnimation=false
    sequence=/BRCosmetics/Animation/Game/MainPlayer/Skydive/ParaGlide/AshtonBoardwalk/Ashton_Boardwalk_Into_FrontEnd_GLIDER
   }
   n3 = Get OwnerGlider()
   Output Animation Pose(Result=Blend Poses by bool(
     BlendPose_0=@n2
     BlendPose_1=Blendspace Player 'BS_Ashton_Boardwalk_Into_GLIDER'(X=@n1.SkydiveAimYaw, Y=@n1.SkydiveAimPitch) {bLoop=false, blendSpace=/BRCosmetics/Animation/Game/MainPlayer/Skydive/ParaGlide/AshtonBoardwalk/BS_Ashton_Boardwalk_Into_GLIDER}
     bActiveValue=Get IsCosmeticPreview(self=@n3)
    ) {blendType=Linear})
  ]
 }
 Closed {
  path=AnimGraph.AnimGraphNode_StateMachine_0.Locomotion.AnimStateNode_2.Closed
  nodes [
   n1 = Get OwnerGlider()
   Output Animation Pose(Result=Blend Poses by bool(
     BlendPose_0=Sequence Player 'Ashton_Boardwalk_Closed_GLIDER'() {sequence=/BRCosmetics/Animation/Game/MainPlayer/Skydive/ParaGlide/AshtonBoardwalk/Ashton_Boardwalk_Closed_GLIDER}
     BlendPose_1=Sequence Player 'Ashton_Boardwalk_Closed_GLIDER'() {sequence=&Ashton_Boardwalk_Closed_GLIDER}
     bActiveValue=Get IsCosmeticPreview(self=@n1)
    ) {blendType=Linear})
  ]
 }
 Transition {
  path=AnimGraph.AnimGraphNode_StateMachine_0.Locomotion.AnimStateTransitionNode_0.Transition
  nodes [
   Result()
   "Equal (Float)"!(A:double="Get Relevant Anim Time Remaining (Open_Into)"!() -> (ReturnValue:float) {is=title K2Node_AnimGetter}, B:double=0) -> (ReturnValue:bool) {is=title K2Node_CallFunction}
  ]
 }
 Transition {
  path=AnimGraph.AnimGraphNode_StateMachine_0.Locomotion.AnimStateTransitionNode_1.Transition
  nodes [
   n1 = Get SkydivingState()
   Result(bCanEnterTransition=OR Boolean(
     A=Break Fort Anim Input Skydiving(FortAnimInput_Skydiving=@n1) -> (
      ParaGilderRootOffset=~
      ParaGliderRootRotationOverride=~
      bUseParaGlideRootModifier=~
      bIsSkydivingFromLaunchPad=~
      bIsSkydivingFromBus=~
      bIsInVortex=~
      bIsDiving=~
      bIsParachuteOpen=~
      bIsSkydiving=~
      bIsParachuteLeaning=~
      bIsSkydiveLeaning=~
      bIsLeaning=~
      bIsSkydiveDiveMode=~
      bPlaySkydiveDrift=~
      bSkydiveDriftDelayActive=~
      bSkydiveDriftAnimAllowed=~
      bIsGliderRight=~
      bIsGliderCenter=~
      bIsGliderLeft=~
      bIsGliderForward=~
      bIsGliderBack=~
      bHasDeployedAndIsStandingUpright=~
      LocalAccelForward=~
      LocalAccelRight=~
      LocalVelocityRight=~
      SkydiveAimPitch=~
      SkydiveAimYaw=~
      DeployChuteAnimRate=~
      SkydiveDriftAnimRate=~
      SkydiveDriftAnimRateCurrent=~
      SkydiveFidgetAnimRate=~
      SkydiveFidgetAnimRateCurrent=~
      SkydiveAdditiveAlpha=~
      SkydiveDriftDelay=~
     ).bIsParachuteOpen
     B=Get bForceOpen() -> (bForceOpen:bool)
    ))
  ]
 }
 Transition {
  path=AnimGraph.AnimGraphNode_StateMachine_0.Locomotion.AnimStateTransitionNode_2.Transition
  nodes [
   n1 = Get bForceOpen()
   n2 = Break Fort Anim Input Skydiving(FortAnimInput_Skydiving=Get SkydivingState()) -> (
    ParaGilderRootOffset=~
    ParaGliderRootRotationOverride=~
    bUseParaGlideRootModifier=~
    bIsSkydivingFromLaunchPad=~
    bIsSkydivingFromBus=~
    bIsInVortex=~
    bIsDiving=~
    bIsParachuteOpen=~
    bIsSkydiving=~
    bIsParachuteLeaning=~
    bIsSkydiveLeaning=~
    bIsLeaning=~
    bIsSkydiveDiveMode=~
    bPlaySkydiveDrift=~
    bSkydiveDriftDelayActive=~
    bSkydiveDriftAnimAllowed=~
    bIsGliderRight=~
    bIsGliderCenter=~
    bIsGliderLeft=~
    bIsGliderForward=~
    bIsGliderBack=~
    bHasDeployedAndIsStandingUpright=~
    LocalAccelForward=~
    LocalAccelRight=~
    LocalVelocityRight=~
    SkydiveAimPitch=~
    SkydiveAimYaw=~
    DeployChuteAnimRate=~
    SkydiveDriftAnimRate=~
    SkydiveDriftAnimRateCurrent=~
    SkydiveFidgetAnimRate=~
    SkydiveFidgetAnimRateCurrent=~
    SkydiveAdditiveAlpha=~
    SkydiveDriftDelay=~
   )
   Result(bCanEnterTransition=AND Boolean(A=NOT Boolean(A=@n2.bIsParachuteOpen), B=NOT Boolean(A=@n1)))
  ]
 }
 Transition {
  path=AnimGraph.AnimGraphNode_StateMachine_0.Locomotion.AnimStateTransitionNode_3.Transition
  nodes [
   n1 = Get bForceOpen()
   n2 = Break Fort Anim Input Skydiving(FortAnimInput_Skydiving=Get SkydivingState()) -> (
    ParaGilderRootOffset=~
    ParaGliderRootRotationOverride=~
    bUseParaGlideRootModifier=~
    bIsSkydivingFromLaunchPad=~
    bIsSkydivingFromBus=~
    bIsInVortex=~
    bIsDiving=~
    bIsParachuteOpen=~
    bIsSkydiving=~
    bIsParachuteLeaning=~
    bIsSkydiveLeaning=~
    bIsLeaning=~
    bIsSkydiveDiveMode=~
    bPlaySkydiveDrift=~
    bSkydiveDriftDelayActive=~
    bSkydiveDriftAnimAllowed=~
    bIsGliderRight=~
    bIsGliderCenter=~
    bIsGliderLeft=~
    bIsGliderForward=~
    bIsGliderBack=~
    bHasDeployedAndIsStandingUpright=~
    LocalAccelForward=~
    LocalAccelRight=~
    LocalVelocityRight=~
    SkydiveAimPitch=~
    SkydiveAimYaw=~
    DeployChuteAnimRate=~
    SkydiveDriftAnimRate=~
    SkydiveDriftAnimRateCurrent=~
    SkydiveFidgetAnimRate=~
    SkydiveFidgetAnimRateCurrent=~
    SkydiveAdditiveAlpha=~
    SkydiveDriftDelay=~
   )
   Result(bCanEnterTransition=AND Boolean(A=NOT Boolean(A=@n2.bIsParachuteOpen), B=NOT Boolean(A=@n1)))
  ]
 }
 EventGraph {nodes [
   n1 = Event BlueprintUpdateAnimation()
   > Set Time(Time:double=@n1.DeltaTimeX) -> (Output_Get:double)
   > n2 = Branch(Condition="float >= float"(A=GetCurveValue(CurveName=ParaGlide_Lean), B=1) {is=title K2Node_CallFunction})
   > Set TurbineSpeed(TurbineSpeed:double=0) -> (Output_Get:double)
   n3 = Break Fort Anim Input Skydiving(FortAnimInput_Skydiving=Get SkydivingState()) -> (
    ParaGilderRootOffset=~
    ParaGliderRootRotationOverride=~
    bUseParaGlideRootModifier=~
    bIsSkydivingFromLaunchPad=~
    bIsSkydivingFromBus=~
    bIsInVortex=~
    bIsDiving=~
    bIsParachuteOpen=~
    bIsSkydiving=~
    bIsParachuteLeaning=~
    bIsSkydiveLeaning=~
    bIsLeaning=~
    bIsSkydiveDiveMode=~
    bPlaySkydiveDrift=~
    bSkydiveDriftDelayActive=~
    bSkydiveDriftAnimAllowed=~
    bIsGliderRight=~
    bIsGliderCenter=~
    bIsGliderLeft=~
    bIsGliderForward=~
    bIsGliderBack=~
    bHasDeployedAndIsStandingUpright=~
    LocalAccelForward=~
    LocalAccelRight=~
    LocalVelocityRight=~
    SkydiveAimPitch=~
    SkydiveAimYaw=~
    DeployChuteAnimRate=~
    SkydiveDriftAnimRate=~
    SkydiveDriftAnimRateCurrent=~
    SkydiveFidgetAnimRate=~
    SkydiveFidgetAnimRateCurrent=~
    SkydiveAdditiveAlpha=~
    SkydiveDriftDelay=~
   )
   n4 = Branch(execute=@n2.else, Condition="float > float"(A=@n3.LocalAccelForward, B=500) {is=title K2Node_CallFunction})
   > Set TurbineSpeed(TurbineSpeed=FInterpTo(Current=Get TurbineSpeed(), Target=5, DeltaTime=Get Time() -> (Time:double), InterpSpeed=2))
   > Set JitterSpeed(JitterSpeed:double=FInterpTo(Current=Get JitterSpeed(), Target=2, DeltaTime=Get Time(), InterpSpeed=1)) -> (Output_Get:double)
   n5 = Branch(execute=@n4.else, Condition="Nearly Equal (Float)"(A=@n3.LocalAccelForward, B=0, ErrorTolerance=10))
   > Set TurbineSpeed(TurbineSpeed=FInterpTo(Current=Get TurbineSpeed(), Target=1, DeltaTime=Get Time(), InterpSpeed=1))
   > Set JitterSpeed(JitterSpeed=FInterpTo(Current=Get JitterSpeed(), Target=1, DeltaTime=Get Time(), InterpSpeed=1))
   n6 = Branch(
    execute=@n5.else
    Condition="float < float"!(A:double=@n3.LocalAccelForward, B:double=-500) -> (ReturnValue:bool) {is=title K2Node_CallFunction}
   )
   > Set TurbineSpeed(TurbineSpeed=FInterpTo(Current=Get TurbineSpeed(), Target=-2, DeltaTime=Get Time(), InterpSpeed=2))
   > Set JitterSpeed(JitterSpeed=FInterpTo(Current=Get JitterSpeed(), Target=1, DeltaTime=Get Time(), InterpSpeed=1))
   Set TurbineSpeed(execute=@n6.else, TurbineSpeed=FInterpTo(Current=Get TurbineSpeed(), Target=1, DeltaTime=Get Time(), InterpSpeed=15))
   > Set JitterSpeed(JitterSpeed=FInterpTo(Current=Get JitterSpeed(), Target=1, DeltaTime=Get Time(), InterpSpeed=1))
   AnimNotify_TurnONEngines() {is=title K2Node_Event}
   > SetGliderFullyDeployed!(execute, self:Fort Glider Anim Instance Object Reference, bIsFullyDeployed:bool=true) -> (then)
   Break Fort Anim Input Skydiving(FortAnimInput_Skydiving=Get SkydivingState()) -> (
    ParaGilderRootOffset=~
    ParaGliderRootRotationOverride=~
    bUseParaGlideRootModifier=~
    bIsSkydivingFromLaunchPad=~
    bIsSkydivingFromBus=~
    bIsInVortex=~
    bIsDiving=~
    bIsParachuteOpen=~
    bIsSkydiving=~
    bIsParachuteLeaning=~
    bIsSkydiveLeaning=~
    bIsLeaning=~
    bIsSkydiveDiveMode=~
    bPlaySkydiveDrift=~
    bSkydiveDriftDelayActive=~
    bSkydiveDriftAnimAllowed=~
    bIsGliderRight=~
    bIsGliderCenter=~
    bIsGliderLeft=~
    bIsGliderForward=~
    bIsGliderBack=~
    bHasDeployedAndIsStandingUpright=~
    LocalAccelForward=~
    LocalAccelRight=~
    LocalVelocityRight=~
    SkydiveAimPitch=~
    SkydiveAimYaw=~
    DeployChuteAnimRate=~
    SkydiveDriftAnimRate=~
    SkydiveDriftAnimRateCurrent=~
    SkydiveFidgetAnimRate=~
    SkydiveFidgetAnimRateCurrent=~
    SkydiveAdditiveAlpha=~
    SkydiveDriftDelay=~
   )
  ]}
]
variables [
 JitterSpeed {type=number}
 Time {type=number}
 TurbineSpeed {type=number}
]
