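# Anim Blueprint for the BirdNest bird glider, animating deployment, gliding and flapping from skydiving state.
# A Locomotion state machine feeds lean, turn, sway and feather additives, curve-driven Leg IK to the handles, then head-dart, caw and blink additives.
# Update logic and anim notifies re-arm randomized flap and idle behaviors through Do Once/Delay loops.
# It also scales feather and scarf values by forward acceleration and broadcasts the Flap curve state through FlappingStateToggled.
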AnimBlueprint /BRCosmetics/Gadgets/Assets/VinderTech_GliderChute/Blueprints/BirdNest_Glider_AnimBP
target_skeleton=/BRCosmetics/Gadgets/Assets/VinderTech_GliderChute/Glider_BirdNest/Meshes/BirdNest_Glider_Skeleton
graphs [
 AnimGraph {nodes [
   n1 = Break Fort Anim Input Skydiving!(FortAnimInput_Skydiving:"Fort Anim Input Skydiving Structure (by ref)"=Get SkydivingState() -> (SkydivingState:Fort Anim Input Skydiving Structure)) -> (
    ParaGilderRootOffset:Vector=~
    ParaGliderRootRotationOverride:Rotator=~
    bUseParaGlideRootModifier:bool=~
    bIsSkydivingFromLaunchPad:bool=~
    bIsSkydivingFromBus:bool=~
    bIsInVortex:bool=~
    bIsDiving:bool=~
    bIsParachuteOpen:bool=~
    bIsSkydiving:bool=~
    bIsParachuteLeaning:bool=~
    bIsSkydiveLeaning:bool=~
    bIsLeaning:bool=~
    bIsSkydiveDiveMode:bool=~
    bPlaySkydiveDrift:bool=~
    bSkydiveDriftDelayActive:bool=~
    bSkydiveDriftAnimAllowed:bool=~
    bIsGliderRight:bool=~
    bIsGliderCenter:bool=~
    bIsGliderLeft:bool=~
    bIsGliderForward:bool=~
    bIsGliderBack:bool=~
    bHasDeployedAndIsStandingUpright:bool=~
    LocalAccelForward:float=~
    LocalAccelRight:float=~
    LocalVelocityRight:float=~
    SkydiveAimPitch:float=~
    SkydiveAimYaw:float=~
    DeployChuteAnimRate:float=~
    SkydiveDriftAnimRate:float=~
    SkydiveDriftAnimRateCurrent:float=~
    SkydiveFidgetAnimRate:float=~
    SkydiveFidgetAnimRateCurrent:float=~
    SkydiveAdditiveAlpha:float=~
    SkydiveDriftDelay:float=~
    SkydiveDriftAnim:int
    SkydiveDriftAnimMax:int
    LaunchpadAnim:int
    LocalAccelDir:ESkydivingDirection Enum
    DirectionLast:ESkydivingDirection Enum
    PoseAdditive:Anim Sequence Object Reference
   ) {is=title K2Node_BreakStruct}
   n2 = Blendspace Player 'BS_BirdNest_Lean_GLIDER'(X=@n1.LocalAccelRight, Y=@n1.LocalAccelForward) {blendSpace=/BRCosmetics/Animation/Game/MainPlayer/Skydive/ParaGlide/BirdNest/BS_BirdNest_Lean_GLIDER}
   AnimGraphNode_StateMachine_0 = Locomotion() {is=title AnimGraphNode_StateMachine}
   n3 = Break Fort Anim Input Skydiving(FortAnimInput_Skydiving=Get SkydivingState()) -> (
    ParaGilderRootOffset=~
    ParaGliderRootRotationOverride=~
    bUseParaGlideRootModifier=~
    bIsSkydivingFromLaunchPad=~
    bIsSkydivingFromBus=~
    bIsInVortex=~
    bIsDiving=~
    bIsParachuteOpen=~
    bIsSkydiving=~
    bIsParachuteLeaning=~
    bIsSkydiveLeaning=~
    bIsLeaning=~
    bIsSkydiveDiveMode=~
    bPlaySkydiveDrift=~
    bSkydiveDriftDelayActive=~
    bSkydiveDriftAnimAllowed=~
    bIsGliderRight=~
    bIsGliderCenter=~
    bIsGliderLeft=~
    bIsGliderForward=~
    bIsGliderBack=~
    bHasDeployedAndIsStandingUpright=~
    LocalAccelForward=~
    LocalAccelRight=~
    LocalVelocityRight=~
    SkydiveAimPitch=~
    SkydiveAimYaw=~
    DeployChuteAnimRate=~
    SkydiveDriftAnimRate=~
    SkydiveDriftAnimRateCurrent=~
    SkydiveFidgetAnimRate=~
    SkydiveFidgetAnimRateCurrent=~
    SkydiveAdditiveAlpha=~
    SkydiveDriftDelay=~
   )
   n4 = Apply Additive(
    Base=Apply Additive(
     Base=Apply Additive!(Base=@AnimGraphNode_StateMachine_0, Additive=@n2, AlphaCurveName:Name=ParaGlide_Lean) -> (Pose) {alphaInputType=Curve, alphaBoolBlend {blendOption=Linear}, alphaScaleBiasClamp {bias=1, scale=-1}}
     Additive=Blendspace Player 'BS_BirdNest_Turn_GLIDER'(X=@n3.SkydiveAimYaw, Y=@n3.SkydiveAimPitch) {blendSpace=/BRCosmetics/Animation/Game/MainPlayer/Skydive/ParaGlide/BirdNest/BS_BirdNest_Turn_GLIDER}
    ) {alphaBoolBlend {blendOption=Linear}}
    Additive="Sequence Player (additive) 'BirdNest_WingSway_GLIDER'"() {sequence=/BRCosmetics/Animation/Game/MainPlayer/Skydive/ParaGlide/BirdNest/BirdNest_WingSway_GLIDER}
   ) {alphaBoolBlend {blendOption=Linear}}
   n5 = "Sequence Player (additive) 'BirdNest_Feathers_GLIDER'"!(PlayRate:float=Get FeatherSpeed() -> (FeatherSpeed:double)) -> (Pose) {sequence=/BRCosmetics/Animation/Game/MainPlayer/Skydive/ParaGlide/BirdNest/BirdNest_Feathers_GLIDER}
   n6 = Leg IK!(
    ComponentPose:Component Space Pose Link Structure=Local To Component(LocalPose=Apply Additive!(Base=@n4, Additive=@n5, Alpha:float) -> (Pose) {alphaBoolBlend {blendOption=Linear}})
    AlphaCurveName:Name=EnableIK
   ) -> (Pose:Component Space Pose Link Structure) {
    is=title AnimGraphNode_LegIK
    alphaInputType=Curve
    alphaBoolBlend {blendOption=Linear}
    legsDefinition [
     {
      bEnableKneeTwistCorrection=true
      bEnableRotationLimit=false
      footBoneForwardAxis=Y
      hingeRotationAxis=None
      minRotationAngle=15
      numBonesInLimb=2
      twistOffsetCurveName=None
      fKFootBone {boneName=ankle_eff_l}
      iKFootBone {boneName=VB Handles_Base_ankle_eff_l}
     }
     {
      bEnableKneeTwistCorrection=true
      bEnableRotationLimit=false
      footBoneForwardAxis=Y
      hingeRotationAxis=None
      minRotationAngle=15
      numBonesInLimb=2
      twistOffsetCurveName=None
      fKFootBone {boneName=ankle_eff_r}
      iKFootBone {boneName=VB Handles_Base_ankle_eff_r}
     }
    ]
   }
   n7 = Random Sequence Player!() -> (Pose:Pose Link Structure) {
    is=title AnimGraphNode_RandomPlayer
    bShuffleMode=true
    entries [
     {
      chanceToPlay=1
      maxLoopCount=1
      maxPlayRate=1
      minLoopCount=0
      minPlayRate=1
      sequence=/BRCosmetics/Animation/Game/MainPlayer/Skydive/ParaGlide/BirdNest/BirdNest_Head_Darts_Left_GLIDER
      blendIn {blendOption=Linear, blendTime=0.2, customCurve=null}
     }
     {
      chanceToPlay=1
      maxLoopCount=1
      maxPlayRate=1
      minLoopCount=0
      minPlayRate=1
      sequence=/BRCosmetics/Animation/Game/MainPlayer/Skydive/ParaGlide/BirdNest/BirdNest_Head_Darts_Right_GLIDER
      blendIn {blendOption=Linear, blendTime=0.2, customCurve=null}
     }
    ]
   }
   AnimGraphNode_StateMachine_1 = CuhCAW() {is=title AnimGraphNode_StateMachine}
   n8 = Apply Additive!(
    Base=Apply Additive(Base=Component To Local(ComponentPose=@n6), Additive=@n7, bAlphaBoolEnabled=Get CanLook() -> (CanLook:bool)) {alphaInputType=Bool, alphaBoolBlend {blendInTime=0.1, blendOption=Linear, blendOutTime=0.1}}
    Additive=@AnimGraphNode_StateMachine_1
    bAlphaBoolEnabled:bool=Get CanCaw() -> (CanCaw:bool)
   ) -> (Pose) {alphaInputType=Bool, alphaBoolBlend {blendInTime=0.1, blendOption=Linear, blendOutTime=0.1}}
   n9 = Random Sequence Player() {
    bShuffleMode=true
    entries [
     {
      chanceToPlay=1
      maxLoopCount=1
      maxPlayRate=1
      minLoopCount=0
      minPlayRate=1
      sequence=/BRCosmetics/Animation/Game/MainPlayer/Skydive/ParaGlide/BirdNest/BirdNest_Blink_GLIDER
      blendIn {blendOption=Linear, blendTime=0.2, customCurve=null}
     }
     {
      chanceToPlay=1
      maxLoopCount=1
      maxPlayRate=1
      minLoopCount=0
      minPlayRate=1
      sequence=/BRCosmetics/Animation/Game/MainPlayer/Skydive/ParaGlide/BirdNest/BirdNest_BlinkDart_GLIDER
      blendIn {blendOption=Linear, blendTime=0.2, customCurve=null}
     }
    ]
   }
   n10 = AND Boolean(A=Get CanBlink() -> (CanBlink:bool), B=NOT Boolean(A=Get CanCaw()))
   Output Pose(Result=Slot 'DefaultSlot'(Source=Apply Additive(Base=@n8, Additive=@n9, bAlphaBoolEnabled=@n10) {alphaInputType=Bool, alphaBoolBlend {blendOption=Linear}})) {layerGroup=None}
  ]}
 Locomotion {
  entry=Closed
  path=AnimGraph.AnimGraphNode_StateMachine_0.Locomotion
  ids=id
  nodes [
   AnimStateNode_0 = Idle_Open() {is=title AnimStateNode}
   AnimStateNode_1 = Open_Into() {is=title AnimStateNode}
   AnimStateNode_2 = Closed() {is=title AnimStateNode}
   AnimStateNode_3 = Flap_Transition() {is=title AnimStateNode}
   AnimStateNode_4 = Flap() {is=title AnimStateNode}
   AnimStateTransitionNode_0 = AnimStateTransitionNode() {bAutomaticRuleBasedOnSequencePlayerInState=true, blendMode=Linear, crossfadeDuration=0}
   AnimStateTransitionNode_1 = AnimStateTransitionNode() {crossfadeDuration=0}
   AnimStateTransitionNode_10 = AnimStateTransitionNode() {crossfadeDuration=0}
   AnimStateTransitionNode_2 = AnimStateTransitionNode() {crossfadeDuration=0}
   AnimStateTransitionNode_3 = AnimStateTransitionNode() {crossfadeDuration=0, priorityOrder=2}
   AnimStateTransitionNode_4 = AnimStateTransitionNode()
   AnimStateTransitionNode_5 = AnimStateTransitionNode() {bAutomaticRuleBasedOnSequencePlayerInState=true, blendMode=Linear, crossfadeDuration=0.35}
   AnimStateTransitionNode_6 = AnimStateTransitionNode() {priorityOrder=2}
   AnimStateTransitionNode_7 = AnimStateTransitionNode() {crossfadeDuration=0}
   AnimStateTransitionNode_8 = AnimStateTransitionNode()
   AnimStateTransitionNode_9 = AnimStateTransitionNode() {bAutomaticRuleBasedOnSequencePlayerInState=true, blendMode=Linear, crossfadeDuration=0.35}
  ]
  links [
   {
    dst=Open_Into
    rule=AnimGraph.AnimGraphNode_StateMachine_0.Locomotion.AnimStateTransitionNode_1.Transition
    src=Closed
    via=AnimStateTransitionNode_1
   }
   {
    dst=Closed
    rule=AnimGraph.AnimGraphNode_StateMachine_0.Locomotion.AnimStateTransitionNode_7.Transition
    src=Flap
    via=AnimStateTransitionNode_7
   }
   {
    dst=Idle_Open
    rule=AnimGraph.AnimGraphNode_StateMachine_0.Locomotion.AnimStateTransitionNode_5.Transition
    src=Flap
    via=AnimStateTransitionNode_5
   }
   {
    dst=Closed
    rule=AnimGraph.AnimGraphNode_StateMachine_0.Locomotion.AnimStateTransitionNode_10.Transition
    src=Flap_Transition
    via=AnimStateTransitionNode_10
   }
   {
    dst=Idle_Open
    rule=AnimGraph.AnimGraphNode_StateMachine_0.Locomotion.AnimStateTransitionNode_9.Transition
    src=Flap_Transition
    via=AnimStateTransitionNode_9
   }
   {
    dst=Closed
    rule=AnimGraph.AnimGraphNode_StateMachine_0.Locomotion.AnimStateTransitionNode_2.Transition
    src=Idle_Open
    via=AnimStateTransitionNode_2
   }
   {
    dst=Flap
    rule=AnimGraph.AnimGraphNode_StateMachine_0.Locomotion.AnimStateTransitionNode_4.Transition
    src=Idle_Open
    via=AnimStateTransitionNode_4
   }
   {
    dst=Flap_Transition
    rule=AnimGraph.AnimGraphNode_StateMachine_0.Locomotion.AnimStateTransitionNode_8.Transition
    src=Idle_Open
    via=AnimStateTransitionNode_8
   }
   {
    dst=Closed
    rule=AnimGraph.AnimGraphNode_StateMachine_0.Locomotion.AnimStateTransitionNode_3.Transition
    src=Open_Into
    via=AnimStateTransitionNode_3
   }
   {
    dst=Flap
    rule=AnimGraph.AnimGraphNode_StateMachine_0.Locomotion.AnimStateTransitionNode_6.Transition
    src=Open_Into
    via=AnimStateTransitionNode_6
   }
   {
    dst=Idle_Open
    rule=AnimGraph.AnimGraphNode_StateMachine_0.Locomotion.AnimStateTransitionNode_0.Transition
    src=Open_Into
    via=AnimStateTransitionNode_0
   }
  ]
 }
 Idle_Open {
  path=AnimGraph.AnimGraphNode_StateMachine_0.Locomotion.AnimStateNode_0.Idle_Open
  nodes [
   Blend Poses by bool(bActiveValue=Get IsCosmeticPreview!(self:Fort Player Parachute Object Reference=Get OwnerGlider() -> (OwnerGlider:Fort Player Parachute Object Reference)) -> (IsCosmeticPreview:bool)) {blendType=Linear}
   n1 = Break Fort Anim Input Skydiving(FortAnimInput_Skydiving=Get SkydivingState()) -> (
    ParaGilderRootOffset=~
    ParaGliderRootRotationOverride=~
    bUseParaGlideRootModifier=~
    bIsSkydivingFromLaunchPad=~
    bIsSkydivingFromBus=~
    bIsInVortex=~
    bIsDiving=~
    bIsParachuteOpen=~
    bIsSkydiving=~
    bIsParachuteLeaning=~
    bIsSkydiveLeaning=~
    bIsLeaning=~
    bIsSkydiveDiveMode=~
    bPlaySkydiveDrift=~
    bSkydiveDriftDelayActive=~
    bSkydiveDriftAnimAllowed=~
    bIsGliderRight=~
    bIsGliderCenter=~
    bIsGliderLeft=~
    bIsGliderForward=~
    bIsGliderBack=~
    bHasDeployedAndIsStandingUpright=~
    LocalAccelForward=~
    LocalAccelRight=~
    LocalVelocityRight=~
    SkydiveAimPitch=~
    SkydiveAimYaw=~
    DeployChuteAnimRate=~
    SkydiveDriftAnimRate=~
    SkydiveDriftAnimRateCurrent=~
    SkydiveFidgetAnimRate=~
    SkydiveFidgetAnimRateCurrent=~
    SkydiveAdditiveAlpha=~
    SkydiveDriftDelay=~
   )
   n2 = Blendspace Player 'BS_BirdNest_GLIDER'(X=@n1.SkydiveAimYaw, Y=@n1.LocalAccelForward) {blendSpace=/BRCosmetics/Animation/Game/MainPlayer/Skydive/ParaGlide/BirdNest/BS_BirdNest_GLIDER, groupName=Glide, method=SyncGroup}
   n3 = "Sequence Player (additive) 'BirdNest_Glide_FrontEnd_GLIDER'"() {
    groupName=FrontEndGlide
    method=SyncGroup
    sequence=/BRCosmetics/Animation/Game/MainPlayer/Skydive/ParaGlide/BirdNest/BirdNest_Glide_FrontEnd_GLIDER
   }
   n4 = Get OwnerGlider()
   Output Animation Pose(Result=Apply Additive(Base=@n2, Additive=@n3, bAlphaBoolEnabled=Get IsCosmeticPreview(self=@n4)) {alphaInputType=Bool, alphaBoolBlend {blendOption=Linear}})
  ]
 }
 Open_Into {
  path=AnimGraph.AnimGraphNode_StateMachine_0.Locomotion.AnimStateNode_1.Open_Into
  nodes [
   n1 = Break Fort Anim Input Skydiving(FortAnimInput_Skydiving=Get SkydivingState()) -> (
    ParaGilderRootOffset=~
    ParaGliderRootRotationOverride=~
    bUseParaGlideRootModifier=~
    bIsSkydivingFromLaunchPad=~
    bIsSkydivingFromBus=~
    bIsInVortex=~
    bIsDiving=~
    bIsParachuteOpen=~
    bIsSkydiving=~
    bIsParachuteLeaning=~
    bIsSkydiveLeaning=~
    bIsLeaning=~
    bIsSkydiveDiveMode=~
    bPlaySkydiveDrift=~
    bSkydiveDriftDelayActive=~
    bSkydiveDriftAnimAllowed=~
    bIsGliderRight=~
    bIsGliderCenter=~
    bIsGliderLeft=~
    bIsGliderForward=~
    bIsGliderBack=~
    bHasDeployedAndIsStandingUpright=~
    LocalAccelForward=~
    LocalAccelRight=~
    LocalVelocityRight=~
    SkydiveAimPitch=~
    SkydiveAimYaw=~
    DeployChuteAnimRate=~
    SkydiveDriftAnimRate=~
    SkydiveDriftAnimRateCurrent=~
    SkydiveFidgetAnimRate=~
    SkydiveFidgetAnimRateCurrent=~
    SkydiveAdditiveAlpha=~
    SkydiveDriftDelay=~
   )
   n2 = Get OwnerGlider()
   Output Animation Pose(Result=Apply Additive(
     Base=Blendspace Player 'BS_BirdNest_Into_GLIDER'(X=@n1.SkydiveAimYaw, Y=@n1.SkydiveAimPitch) {bLoop=false, blendSpace=/BRCosmetics/Animation/Game/MainPlayer/Skydive/ParaGlide/BirdNest/BS_BirdNest_Into_GLIDER}
     Additive="Sequence Player (additive) 'BirdNest_Into_FrontEnd_GLIDER'"() {bLoopAnimation=false, sequence=/BRCosmetics/Animation/Game/MainPlayer/Skydive/ParaGlide/BirdNest/BirdNest_Into_FrontEnd_GLIDER}
     bAlphaBoolEnabled=Get IsCosmeticPreview(self=@n2)
    ) {alphaInputType=Bool, alphaBoolBlend {blendOption=Linear}})
  ]
 }
 Closed {
  path=AnimGraph.AnimGraphNode_StateMachine_0.Locomotion.AnimStateNode_2.Closed
  nodes [
   n1 = Get OwnerGlider()
   Output Animation Pose(Result=Blend Poses by bool(
     BlendPose_0=Sequence Player 'BirdNest_Closed_GLIDER'() {sequence=/BRCosmetics/Animation/Game/MainPlayer/Skydive/ParaGlide/BirdNest/BirdNest_Closed_GLIDER}
     BlendPose_1=Sequence Player 'BirdNest_Closed_GLIDER'() {sequence=&BirdNest_Closed_GLIDER}
     bActiveValue=Get IsCosmeticPreview(self=@n1)
    ) {blendType=Linear})
  ]
 }
 Flap_Transition {
  path=AnimGraph.AnimGraphNode_StateMachine_0.Locomotion.AnimStateNode_3.Flap_Transition
  nodes [
   n1 = Break Fort Anim Input Skydiving(FortAnimInput_Skydiving=Get SkydivingState()) -> (
    ParaGilderRootOffset=~
    ParaGliderRootRotationOverride=~
    bUseParaGlideRootModifier=~
    bIsSkydivingFromLaunchPad=~
    bIsSkydivingFromBus=~
    bIsInVortex=~
    bIsDiving=~
    bIsParachuteOpen=~
    bIsSkydiving=~
    bIsParachuteLeaning=~
    bIsSkydiveLeaning=~
    bIsLeaning=~
    bIsSkydiveDiveMode=~
    bPlaySkydiveDrift=~
    bSkydiveDriftDelayActive=~
    bSkydiveDriftAnimAllowed=~
    bIsGliderRight=~
    bIsGliderCenter=~
    bIsGliderLeft=~
    bIsGliderForward=~
    bIsGliderBack=~
    bHasDeployedAndIsStandingUpright=~
    LocalAccelForward=~
    LocalAccelRight=~
    LocalVelocityRight=~
    SkydiveAimPitch=~
    SkydiveAimYaw=~
    DeployChuteAnimRate=~
    SkydiveDriftAnimRate=~
    SkydiveDriftAnimRateCurrent=~
    SkydiveFidgetAnimRate=~
    SkydiveFidgetAnimRateCurrent=~
    SkydiveAdditiveAlpha=~
    SkydiveDriftDelay=~
   )
   Output Animation Pose(Result=Blendspace Player 'BS_BirdNest_Flap_GLIDER'!(X=@n1.SkydiveAimYaw, Y=@n1.LocalAccelForward, PlayRate:float=Get Flap_Rate() -> (Flap_Rate:double)) -> (Pose) {blendSpace=/BRCosmetics/Animation/Game/MainPlayer/Skydive/ParaGlide/BirdNest/BS_BirdNest_Flap_GLIDER})
  ]
 }
 Flap {
  path=AnimGraph.AnimGraphNode_StateMachine_0.Locomotion.AnimStateNode_4.Flap
  nodes [
   n1 = Break Fort Anim Input Skydiving(FortAnimInput_Skydiving=Get SkydivingState()) -> (
    ParaGilderRootOffset=~
    ParaGliderRootRotationOverride=~
    bUseParaGlideRootModifier=~
    bIsSkydivingFromLaunchPad=~
    bIsSkydivingFromBus=~
    bIsInVortex=~
    bIsDiving=~
    bIsParachuteOpen=~
    bIsSkydiving=~
    bIsParachuteLeaning=~
    bIsSkydiveLeaning=~
    bIsLeaning=~
    bIsSkydiveDiveMode=~
    bPlaySkydiveDrift=~
    bSkydiveDriftDelayActive=~
    bSkydiveDriftAnimAllowed=~
    bIsGliderRight=~
    bIsGliderCenter=~
    bIsGliderLeft=~
    bIsGliderForward=~
    bIsGliderBack=~
    bHasDeployedAndIsStandingUpright=~
    LocalAccelForward=~
    LocalAccelRight=~
    LocalVelocityRight=~
    SkydiveAimPitch=~
    SkydiveAimYaw=~
    DeployChuteAnimRate=~
    SkydiveDriftAnimRate=~
    SkydiveDriftAnimRateCurrent=~
    SkydiveFidgetAnimRate=~
    SkydiveFidgetAnimRateCurrent=~
    SkydiveAdditiveAlpha=~
    SkydiveDriftDelay=~
   )
   n2 = Blendspace Player 'BS_BirdNest_GLIDER'(X=@n1.SkydiveAimYaw, Y=@n1.SkydiveAimPitch) {blendSpace=&BS_BirdNest_GLIDER, groupName=Glide, method=SyncGroup}
   Apply Additive(Base=@n2, bAlphaBoolEnabled=True) {alphaInputType=Bool, alphaBoolBlend {blendInTime=0.05, blendOption=Linear, blendOutTime=0.05}}
   n3 = Get FlapBS() -> (FlapBS:Blend Space Object Reference)
   n4 = Get Flap_Rate()
   n5 = "Sequence Player (additive) 'BirdNest_Glide_FrontEnd_GLIDER'"() {
    groupName=FrontEndGlide
    method=SyncGroup
    sequence=&BirdNest_Glide_FrontEnd_GLIDER
   }
   n6 = Get OwnerGlider()
   Output Animation Pose(Result=Apply Additive(
     Base=Blendspace Player!(X=@n1.SkydiveAimYaw, Y=@n1.LocalAccelForward, PlayRate:float=@n4, BlendSpace:Blend Space Object Reference=@n3) -> (Pose)
     Additive=@n5
     bAlphaBoolEnabled=Get IsCosmeticPreview(self=@n6)
    ) {alphaInputType=Bool, alphaBoolBlend {blendOption=Linear}})
  ]
 }
 Transition {
  path=AnimGraph.AnimGraphNode_StateMachine_0.Locomotion.AnimStateTransitionNode_0.Transition
  nodes [
   Result()
   "Equal (Float)"!(A:double="Get Relevant Anim Time Remaining (Open_Into)"!() -> (ReturnValue:float) {is=title K2Node_AnimGetter}, B:double=0) -> (ReturnValue:bool) {is=title K2Node_CallFunction}
  ]
 }
 Transition {
  path=AnimGraph.AnimGraphNode_StateMachine_0.Locomotion.AnimStateTransitionNode_1.Transition
  nodes [
   n1 = Get SkydivingState()
   Result(bCanEnterTransition=OR Boolean(
     A=Break Fort Anim Input Skydiving(FortAnimInput_Skydiving=@n1) -> (
      ParaGilderRootOffset=~
      ParaGliderRootRotationOverride=~
      bUseParaGlideRootModifier=~
      bIsSkydivingFromLaunchPad=~
      bIsSkydivingFromBus=~
      bIsInVortex=~
      bIsDiving=~
      bIsParachuteOpen=~
      bIsSkydiving=~
      bIsParachuteLeaning=~
      bIsSkydiveLeaning=~
      bIsLeaning=~
      bIsSkydiveDiveMode=~
      bPlaySkydiveDrift=~
      bSkydiveDriftDelayActive=~
      bSkydiveDriftAnimAllowed=~
      bIsGliderRight=~
      bIsGliderCenter=~
      bIsGliderLeft=~
      bIsGliderForward=~
      bIsGliderBack=~
      bHasDeployedAndIsStandingUpright=~
      LocalAccelForward=~
      LocalAccelRight=~
      LocalVelocityRight=~
      SkydiveAimPitch=~
      SkydiveAimYaw=~
      DeployChuteAnimRate=~
      SkydiveDriftAnimRate=~
      SkydiveDriftAnimRateCurrent=~
      SkydiveFidgetAnimRate=~
      SkydiveFidgetAnimRateCurrent=~
      SkydiveAdditiveAlpha=~
      SkydiveDriftDelay=~
     ).bIsParachuteOpen
     B=Get bForceOpen() -> (bForceOpen:bool)
    ))
  ]
 }
 Transition {
  path=AnimGraph.AnimGraphNode_StateMachine_0.Locomotion.AnimStateTransitionNode_10.Transition
  nodes [
   n1 = Get bForceOpen()
   n2 = Break Fort Anim Input Skydiving(FortAnimInput_Skydiving=Get SkydivingState()) -> (
    ParaGilderRootOffset=~
    ParaGliderRootRotationOverride=~
    bUseParaGlideRootModifier=~
    bIsSkydivingFromLaunchPad=~
    bIsSkydivingFromBus=~
    bIsInVortex=~
    bIsDiving=~
    bIsParachuteOpen=~
    bIsSkydiving=~
    bIsParachuteLeaning=~
    bIsSkydiveLeaning=~
    bIsLeaning=~
    bIsSkydiveDiveMode=~
    bPlaySkydiveDrift=~
    bSkydiveDriftDelayActive=~
    bSkydiveDriftAnimAllowed=~
    bIsGliderRight=~
    bIsGliderCenter=~
    bIsGliderLeft=~
    bIsGliderForward=~
    bIsGliderBack=~
    bHasDeployedAndIsStandingUpright=~
    LocalAccelForward=~
    LocalAccelRight=~
    LocalVelocityRight=~
    SkydiveAimPitch=~
    SkydiveAimYaw=~
    DeployChuteAnimRate=~
    SkydiveDriftAnimRate=~
    SkydiveDriftAnimRateCurrent=~
    SkydiveFidgetAnimRate=~
    SkydiveFidgetAnimRateCurrent=~
    SkydiveAdditiveAlpha=~
    SkydiveDriftDelay=~
   )
   Result(bCanEnterTransition=AND Boolean(A=NOT Boolean(A=@n2.bIsParachuteOpen), B=NOT Boolean(A=@n1)))
  ]
 }
 Transition {
  path=AnimGraph.AnimGraphNode_StateMachine_0.Locomotion.AnimStateTransitionNode_2.Transition
  nodes [
   n1 = Get bForceOpen()
   n2 = Break Fort Anim Input Skydiving(FortAnimInput_Skydiving=Get SkydivingState()) -> (
    ParaGilderRootOffset=~
    ParaGliderRootRotationOverride=~
    bUseParaGlideRootModifier=~
    bIsSkydivingFromLaunchPad=~
    bIsSkydivingFromBus=~
    bIsInVortex=~
    bIsDiving=~
    bIsParachuteOpen=~
    bIsSkydiving=~
    bIsParachuteLeaning=~
    bIsSkydiveLeaning=~
    bIsLeaning=~
    bIsSkydiveDiveMode=~
    bPlaySkydiveDrift=~
    bSkydiveDriftDelayActive=~
    bSkydiveDriftAnimAllowed=~
    bIsGliderRight=~
    bIsGliderCenter=~
    bIsGliderLeft=~
    bIsGliderForward=~
    bIsGliderBack=~
    bHasDeployedAndIsStandingUpright=~
    LocalAccelForward=~
    LocalAccelRight=~
    LocalVelocityRight=~
    SkydiveAimPitch=~
    SkydiveAimYaw=~
    DeployChuteAnimRate=~
    SkydiveDriftAnimRate=~
    SkydiveDriftAnimRateCurrent=~
    SkydiveFidgetAnimRate=~
    SkydiveFidgetAnimRateCurrent=~
    SkydiveAdditiveAlpha=~
    SkydiveDriftDelay=~
   )
   Result(bCanEnterTransition=AND Boolean(A=NOT Boolean(A=@n2.bIsParachuteOpen), B=NOT Boolean(A=@n1)))
  ]
 }
 Transition {
  path=AnimGraph.AnimGraphNode_StateMachine_0.Locomotion.AnimStateTransitionNode_3.Transition
  nodes [
   n1 = Get bForceOpen()
   n2 = Break Fort Anim Input Skydiving(FortAnimInput_Skydiving=Get SkydivingState()) -> (
    ParaGilderRootOffset=~
    ParaGliderRootRotationOverride=~
    bUseParaGlideRootModifier=~
    bIsSkydivingFromLaunchPad=~
    bIsSkydivingFromBus=~
    bIsInVortex=~
    bIsDiving=~
    bIsParachuteOpen=~
    bIsSkydiving=~
    bIsParachuteLeaning=~
    bIsSkydiveLeaning=~
    bIsLeaning=~
    bIsSkydiveDiveMode=~
    bPlaySkydiveDrift=~
    bSkydiveDriftDelayActive=~
    bSkydiveDriftAnimAllowed=~
    bIsGliderRight=~
    bIsGliderCenter=~
    bIsGliderLeft=~
    bIsGliderForward=~
    bIsGliderBack=~
    bHasDeployedAndIsStandingUpright=~
    LocalAccelForward=~
    LocalAccelRight=~
    LocalVelocityRight=~
    SkydiveAimPitch=~
    SkydiveAimYaw=~
    DeployChuteAnimRate=~
    SkydiveDriftAnimRate=~
    SkydiveDriftAnimRateCurrent=~
    SkydiveFidgetAnimRate=~
    SkydiveFidgetAnimRateCurrent=~
    SkydiveAdditiveAlpha=~
    SkydiveDriftDelay=~
   )
   Result(bCanEnterTransition=AND Boolean(A=NOT Boolean(A=@n2.bIsParachuteOpen), B=NOT Boolean(A=@n1)))
  ]
 }
 Transition {
  path=AnimGraph.AnimGraphNode_StateMachine_0.Locomotion.AnimStateTransitionNode_4.Transition
  nodes [
   n1 = "Equal (Float)"(A=GetCurveValue(CurveName=DisableFlap), B=1)
   Result(bCanEnterTransition=AND Boolean(A=Get CanFlap() -> (CanFlap:bool), B=NOT Boolean(A=@n1)))
   OR Boolean(A=false, B=Get TransitionFlap() -> (TransitionFlap:bool))
  ]
 }
 Transition {
  path=AnimGraph.AnimGraphNode_StateMachine_0.Locomotion.AnimStateTransitionNode_5.Transition
  nodes [
   Result()
   NOT Boolean(A=Get CanFlap())
  ]
 }
 Transition {
  path=AnimGraph.AnimGraphNode_StateMachine_0.Locomotion.AnimStateTransitionNode_6.Transition
  nodes [
   n1 = "Equal (Float)"(A=GetCurveValue(CurveName=DisableFlap), B=1)
   n2 = Get CanFlap()
   n3 = Get TransitionFlap()
   Result(bCanEnterTransition=AND Boolean(A=OR Boolean(A=@n2, B=@n3), B=NOT Boolean(A=@n1)))
  ]
 }
 Transition {
  path=AnimGraph.AnimGraphNode_StateMachine_0.Locomotion.AnimStateTransitionNode_7.Transition
  nodes [
   n1 = Get bForceOpen()
   n2 = Break Fort Anim Input Skydiving(FortAnimInput_Skydiving=Get SkydivingState()) -> (
    ParaGilderRootOffset=~
    ParaGliderRootRotationOverride=~
    bUseParaGlideRootModifier=~
    bIsSkydivingFromLaunchPad=~
    bIsSkydivingFromBus=~
    bIsInVortex=~
    bIsDiving=~
    bIsParachuteOpen=~
    bIsSkydiving=~
    bIsParachuteLeaning=~
    bIsSkydiveLeaning=~
    bIsLeaning=~
    bIsSkydiveDiveMode=~
    bPlaySkydiveDrift=~
    bSkydiveDriftDelayActive=~
    bSkydiveDriftAnimAllowed=~
    bIsGliderRight=~
    bIsGliderCenter=~
    bIsGliderLeft=~
    bIsGliderForward=~
    bIsGliderBack=~
    bHasDeployedAndIsStandingUpright=~
    LocalAccelForward=~
    LocalAccelRight=~
    LocalVelocityRight=~
    SkydiveAimPitch=~
    SkydiveAimYaw=~
    DeployChuteAnimRate=~
    SkydiveDriftAnimRate=~
    SkydiveDriftAnimRateCurrent=~
    SkydiveFidgetAnimRate=~
    SkydiveFidgetAnimRateCurrent=~
    SkydiveAdditiveAlpha=~
    SkydiveDriftDelay=~
   )
   Result(bCanEnterTransition=AND Boolean(A=NOT Boolean(A=@n2.bIsParachuteOpen), B=NOT Boolean(A=@n1)))
  ]
 }
 Transition {
  path=AnimGraph.AnimGraphNode_StateMachine_0.Locomotion.AnimStateTransitionNode_8.Transition
  nodes [
   Result(bCanEnterTransition=Get TransitionFlap())
   AND Boolean!(A=NOT Boolean(A=Get Flapping() -> (Flapping:bool)), B=Get bIsAccelerating2D(), C:bool) -> (ReturnValue)
   Get BlockAnimRule() -> (BlockAnimRule:bool)
  ]
 }
 Transition {
  path=AnimGraph.AnimGraphNode_StateMachine_0.Locomotion.AnimStateTransitionNode_9.Transition
  nodes [
   Result()
   OR Boolean(A=NOT Boolean(A=Get CanFlap()), B=NOT Boolean(A=Get TransitionFlap()))
  ]
 }
 CuhCAW {entry=Caw, path=AnimGraph.AnimGraphNode_StateMachine_1.CuhCAW, ids=id, nodes [AnimStateNode_0 = Caw() {is=title AnimStateNode}]}
 Caw {
  path=AnimGraph.AnimGraphNode_StateMachine_1.CuhCAW.AnimStateNode_0.Caw
  nodes [Output Animation Pose(Result="Sequence Player (additive) 'BirdNest_Caw_GLIDER'"() {sequence=/BRCosmetics/Animation/Game/MainPlayer/Skydive/ParaGlide/BirdNest/BirdNest_Caw_GLIDER})]
 }
 EventGraph {nodes [
   n1 = Event BlueprintUpdateAnimation()
   > Set Time(Time:double=@n1.DeltaTimeX) -> (Output_Get:double)
   > n2 = Sequence!(execute) -> (then_0, then_1, then_2:Exec, then_3:Exec, then_4:Exec, then_5:Exec)
   Do Once(execute=@n2.then_4, Reset=[@n29.then, @n27.then, @n30.Completed], Start Closed=true)
   > Set Flap_Rate(Flap_Rate:double=Select(Option 0:double=1, Option 1:double=1.1, Index:int=RandomInteger!(Max:int=2) -> (ReturnValue:int=0)) -> (ReturnValue:double)) -> (Output_Get:double)
   > n3 = Set RandomFlapIndexMontage(RandomFlapIndexMontage:int=RandomInteger(Max=2) -> (ReturnValue=0)) -> (Output_Get:int=0)
   > Set FlapBS(FlapBS:Blend Space Object Reference=Select(
     Option 0:Blend Space Object Reference=&BS_BirdNest_Flap_GLIDER
     Option 1:Blend Space Object Reference=/BRCosmetics/Animation/Game/MainPlayer/Skydive/ParaGlide/BirdNest/BS_BirdNest_Flap2x_GLIDER
     Index=@n3
    ) -> (ReturnValue:Blend Space Object Reference)) -> (Output_Get:Blend Space Object Reference)
   n4 = Get SkydivingState()
   > n5 = Branch(Condition="float < float"!(
     A:double=Break Fort Anim Input Skydiving(FortAnimInput_Skydiving=@n4) -> (
      ParaGilderRootOffset=~
      ParaGliderRootRotationOverride=~
      bUseParaGlideRootModifier=~
      bIsSkydivingFromLaunchPad=~
      bIsSkydivingFromBus=~
      bIsInVortex=~
      bIsDiving=~
      bIsParachuteOpen=~
      bIsSkydiving=~
      bIsParachuteLeaning=~
      bIsSkydiveLeaning=~
      bIsLeaning=~
      bIsSkydiveDiveMode=~
      bPlaySkydiveDrift=~
      bSkydiveDriftDelayActive=~
      bSkydiveDriftAnimAllowed=~
      bIsGliderRight=~
      bIsGliderCenter=~
      bIsGliderLeft=~
      bIsGliderForward=~
      bIsGliderBack=~
      bHasDeployedAndIsStandingUpright=~
      LocalAccelForward=~
      LocalAccelRight=~
      LocalVelocityRight=~
      SkydiveAimPitch=~
      SkydiveAimYaw=~
      DeployChuteAnimRate=~
      SkydiveDriftAnimRate=~
      SkydiveDriftAnimRateCurrent=~
      SkydiveFidgetAnimRate=~
      SkydiveFidgetAnimRateCurrent=~
      SkydiveAdditiveAlpha=~
      SkydiveDriftDelay=~
     ).LocalAccelForward
     B:double=-250
    ) -> (ReturnValue:bool) {is=title K2Node_CallFunction})
   > n6 = Set Flap_Rate(Flap_Rate=1.1)
   Set canFlap(execute=[@n5.else, @n6.then], canFlap:bool=true) -> (Output_Get:bool)
   > Set TransitionFlap(TransitionFlap:bool=false) -> (Output_Get:bool)
   Do Once(execute=@n2.then_5, Reset=@n33.then)
   > Set CanBlink(CanBlink:bool=true) -> (Output_Get:bool)
   n7 = Branch(execute=@n2.then_0, Condition="Equal (Float)"(A=GetCurveValue(CurveName=DisableFlap), B=1))
   > n8 = Set ScarfAlpha(ScarfAlpha:double=0) -> (Output_Get:double)
   n9 = GetCurveValue(CurveName=Flap)
   n10 = Set Flapping(
    execute=[@n8.then, @n7.else]
    Flapping:bool="Not Equal (Float)"!(A:double=@n9, B:double=0) -> (ReturnValue:bool) {is=title K2Node_CallFunction}
   ) -> (Output_Get:bool)
   > Call Flapping State Toggled!(execute:Exec, self:Bird Nest Glider Anim BP Object Reference, Flapping:bool=@n10, FlapCurve:double=@n9, MontageIndex:int=@n3) -> (then:Exec) {is=title K2Node_CallDelegate}
   n11 = Break Fort Anim Input Skydiving(FortAnimInput_Skydiving=Get SkydivingState()) -> (
    ParaGilderRootOffset=~
    ParaGliderRootRotationOverride=~
    bUseParaGlideRootModifier=~
    bIsSkydivingFromLaunchPad=~
    bIsSkydivingFromBus=~
    bIsInVortex=~
    bIsDiving=~
    bIsParachuteOpen=~
    bIsSkydiving=~
    bIsParachuteLeaning=~
    bIsSkydiveLeaning=~
    bIsLeaning=~
    bIsSkydiveDiveMode=~
    bPlaySkydiveDrift=~
    bSkydiveDriftDelayActive=~
    bSkydiveDriftAnimAllowed=~
    bIsGliderRight=~
    bIsGliderCenter=~
    bIsGliderLeft=~
    bIsGliderForward=~
    bIsGliderBack=~
    bHasDeployedAndIsStandingUpright=~
    LocalAccelForward=~
    LocalAccelRight=~
    LocalVelocityRight=~
    SkydiveAimPitch=~
    SkydiveAimYaw=~
    DeployChuteAnimRate=~
    SkydiveDriftAnimRate=~
    SkydiveDriftAnimRateCurrent=~
    SkydiveFidgetAnimRate=~
    SkydiveFidgetAnimRateCurrent=~
    SkydiveAdditiveAlpha=~
    SkydiveDriftDelay=~
   )
   > n12 = Branch(Condition="float > float"(A=@n11.LocalAccelForward, B=500) {is=title K2Node_CallFunction})
   > n13 = Set FeatherSpeed(FeatherSpeed:double=FInterpTo(Current=Get FeatherSpeed(), Target=1.75, DeltaTime=Get Time() -> (Time:double), InterpSpeed=1)) -> (Output_Get:double)
   > Set ScarfSpeed(ScarfSpeed:double=float * float(A=@n13, B=3.25)) -> (Output_Get:double)
   > Set ScarfAlpha(ScarfAlpha=FInterpTo(Current=Get ScarfAlpha() -> (ScarfAlpha:double), Target=0.4, DeltaTime=Get Time(), InterpSpeed=2))
   n14 = Branch(execute=@n12.else, Condition="Nearly Equal (Float)"(A=@n11.LocalAccelForward, B=0, ErrorTolerance=10))
   > n15 = Set FeatherSpeed(FeatherSpeed=FInterpTo(Current=Get FeatherSpeed(), Target=1, DeltaTime=Get Time(), InterpSpeed=1))
   > Set ScarfSpeed(ScarfSpeed=float * float(A=@n15, B=2))
   > Set ScarfAlpha(ScarfAlpha=FInterpTo(Current=Get ScarfAlpha(), Target=0.8, DeltaTime=Get Time(), InterpSpeed=1))
   n16 = Branch(execute=@n14.else, Condition="float < float"(A=@n11.LocalAccelForward, B=-500))
   > n17 = Set FeatherSpeed(FeatherSpeed=FInterpTo(Current=Get FeatherSpeed(), Target=0.75, DeltaTime=Get Time(), InterpSpeed=1))
   > Set ScarfSpeed(ScarfSpeed=float * float(A=@n17, B=1.75))
   > Set ScarfAlpha(ScarfAlpha=FInterpTo(Current=Get ScarfAlpha(), Target=1, DeltaTime=Get Time(), InterpSpeed=1))
   n18 = Set FeatherSpeed(execute=@n16.else, FeatherSpeed=FInterpTo(Current=Get FeatherSpeed(), Target=1, DeltaTime=Get Time(), InterpSpeed=1))
   > Set ScarfSpeed(ScarfSpeed=float * float(A=@n18, B=1.75))
   > Set ScarfAlpha(ScarfAlpha=FInterpTo(Current=Get ScarfAlpha(), Target=0.8, DeltaTime=Get Time(), InterpSpeed=1))
   n19 = Branch(execute=@n2.then_1, Condition=Get bIsAccelerating2D())
   > n20 = Do Once(Reset=@n23.then, Start Closed=true)
   Set TransitionFlap(execute=[@n20.Completed, @n21.Completed], TransitionFlap=true)
   > Set CanFlap(CanFlap:bool=false) -> (Output_Get:bool)
   > Set Flap_Rate(Flap_Rate=1.1)
   n21 = Do Once(execute=@n19.else, Reset=@n23.else, Start Closed=true)
   n22 = Get SkydivingState()
   n23 = Branch(
    execute=@n2.then_2
    Condition="Equal (Enum)"(
     A:ESkydivingDirection Enum=Break Fort Anim Input Skydiving(FortAnimInput_Skydiving=@n22) -> (
      ParaGilderRootOffset=~
      ParaGliderRootRotationOverride=~
      bUseParaGlideRootModifier=~
      bIsSkydivingFromLaunchPad=~
      bIsSkydivingFromBus=~
      bIsInVortex=~
      bIsDiving=~
      bIsParachuteOpen=~
      bIsSkydiving=~
      bIsParachuteLeaning=~
      bIsSkydiveLeaning=~
      bIsLeaning=~
      bIsSkydiveDiveMode=~
      bPlaySkydiveDrift=~
      bSkydiveDriftDelayActive=~
      bSkydiveDriftAnimAllowed=~
      bIsGliderRight=~
      bIsGliderCenter=~
      bIsGliderLeft=~
      bIsGliderForward=~
      bIsGliderBack=~
      bHasDeployedAndIsStandingUpright=~
      LocalAccelForward=~
      LocalAccelRight=~
      LocalVelocityRight=~
      SkydiveAimPitch=~
      SkydiveAimYaw=~
      DeployChuteAnimRate=~
      SkydiveDriftAnimRate=~
      SkydiveDriftAnimRateCurrent=~
      SkydiveFidgetAnimRate=~
      SkydiveFidgetAnimRateCurrent=~
      SkydiveAdditiveAlpha=~
      SkydiveDriftDelay=~
     ).LocalAccelDir
     B:ESkydivingDirection Enum=Center
    )
   )
   Do Once(execute=@n2.then_3, Reset=@n31.then)
   > Set CanLook(CanLook:bool=true) -> (Output_Get:bool)
   > Delay(Duration=0.4)
   > Set CanCaw(CanCaw:bool=Select(Option 0:bool=false, Option 1:bool=true, Index=RandomInteger(Max=2) -> (ReturnValue=0)) -> (ReturnValue:bool)) -> (Output_Get:bool)
   AnimNotify_TurnONEngines() {is=title K2Node_Event}
   > SetGliderFullyDeployed!(execute, self:Fort Glider Anim Instance Object Reference, bIsFullyDeployed:bool=true) -> (then)
   AnimNotify_FlapOff() {is=title K2Node_Event}
   > n24 = Sequence!(execute) -> (then_0, then_1)
   > Set TransitionFlap(TransitionFlap=false)
   Set canFlap(execute=@n24.then_1, canFlap=false)
   n25 = Get SkydivingState()
   > n26 = Branch(Condition="float < float"(
     A=Break Fort Anim Input Skydiving(FortAnimInput_Skydiving=@n25) -> (
      ParaGilderRootOffset=~
      ParaGliderRootRotationOverride=~
      bUseParaGlideRootModifier=~
      bIsSkydivingFromLaunchPad=~
      bIsSkydivingFromBus=~
      bIsInVortex=~
      bIsDiving=~
      bIsParachuteOpen=~
      bIsSkydiving=~
      bIsParachuteLeaning=~
      bIsSkydiveLeaning=~
      bIsLeaning=~
      bIsSkydiveDiveMode=~
      bPlaySkydiveDrift=~
      bSkydiveDriftDelayActive=~
      bSkydiveDriftAnimAllowed=~
      bIsGliderRight=~
      bIsGliderCenter=~
      bIsGliderLeft=~
      bIsGliderForward=~
      bIsGliderBack=~
      bHasDeployedAndIsStandingUpright=~
      LocalAccelForward=~
      LocalAccelRight=~
      LocalVelocityRight=~
      SkydiveAimPitch=~
      SkydiveAimYaw=~
      DeployChuteAnimRate=~
      SkydiveDriftAnimRate=~
      SkydiveDriftAnimRateCurrent=~
      SkydiveFidgetAnimRate=~
      SkydiveFidgetAnimRateCurrent=~
      SkydiveAdditiveAlpha=~
      SkydiveDriftDelay=~
     ).LocalAccelForward
     B=-250
    ))
   > n27 = Delay(Duration=0)
   n28 = Set Flap_Delay(execute=@n26.else, Flap_Delay:double=RandomFloatInRange(Min=3, Max=5)) -> (Output_Get:double)
   > n29 = Delay(Duration=@n28)
   AnimNotify_FlapEnable() {is=title K2Node_Event}
   > n30 = Do Once()
   AnimNotify_LookOff() {is=title K2Node_Event}
   > Set CanLook(CanLook=false)
   > n31 = Delay(Duration=4)
   AnimNotify_CawOff() {is=title K2Node_Event}
   > Set CanCaw(CanCaw=false)
   > Delay(Duration=Select(Option 0:double=4, Option 1:double=8, Index=RandomInteger(Max=2) -> (ReturnValue=0)) -> (ReturnValue:double))
   > n32 = Branch(Condition=Get CanLook())
   Set CanCaw(execute=@n32.else, CanCaw=true)
   AnimNotify_BlinkOff() {is=title K2Node_Event}
   > Set CanBlink(CanBlink=false)
   > n33 = Delay(Duration=RandomFloatInRange(Min=1, Max=3))
  ]}
 FlappingStateToggled {nodes [FlappingStateToggled!() -> (then, Flapping:bool=~, FlapCurve:double=~, MontageIndex:int) {is=title K2Node_FunctionEntry}]}
]
variables [
 BlockAnimRule {type=boolean}
 CanBlink {type=boolean}
 CanCaw {type=boolean}
 CanFlap {type=boolean}
 CanLook {type=boolean}
 FeatherSpeed {type=number}
 Flap_Delay {type=number}
 Flap_Rate {type=number}
 FlapBS {type=BlendSpace, default=&BS_BirdNest_Flap_GLIDER}
 Flapping {type=boolean}
 FlappingStateToggled {default=()}
 RandomFlapIndexMontage {type=integer}
 ScarfAlpha {type=number}
 ScarfSpeed {type=number}
 Time {type=number}
 TransitionFlap {type=boolean}
]
