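# Anim Blueprint for the Black Widow glider cosmetic, animating its spider rig through deployment and skydive gliding.
# A Locomotion state machine (Closed, Open_Into, Idle_Open, PopUp) layers aim, turn and curve-weighted lean blendspaces, with front-end clips for cosmetic preview.
# Additive leg sway and streamers sit on top, with sway sync-grouped to the glide.
# Update logic triggers timed, eased scurries and pop-ups from acceleration direction and curves, while notifies end them or flag deployment.
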AnimBlueprint /BRCosmetics/Gadgets/Assets/VinderTech_GliderChute/Blueprints/BlackWidow_Glider_AnimBP
target_skeleton=/BRCosmetics/Gadgets/Assets/VinderTech_GliderChute/Glider_BlackWidow/Meshes/Glider_BlackWidow_Skeleton
graphs [
 AnimGraph {nodes [
   n1 = "Sequence Player (additive) 'BlackWidow_LegBodySway_GLIDER'"() {
    groupName=GlideSway
    method=SyncGroup
    sequence=/BRCosmetics/Animation/Game/MainPlayer/Skydive/ParaGlide/BlackWidow/BlackWidow_LegBodySway_GLIDER
   }
   AnimGraphNode_StateMachine_0 = Locomotion() {is=title AnimGraphNode_StateMachine}
   n2 = Apply Additive(Base=@AnimGraphNode_StateMachine_0, Additive=@n1) {alphaBoolBlend {blendOption=Linear}}
   n3 = "Sequence Player (additive) 'BlackWidow_Streamers_GLIDER'"!(PlayRate:float=1) -> (Pose) {sequence=/BRCosmetics/Animation/Game/MainPlayer/Skydive/ParaGlide/BlackWidow/BlackWidow_Streamers_GLIDER}
   Output Pose(Result=Slot 'DefaultSlot'(Source=Apply Additive(Base=@n2, Additive=@n3) {alphaBoolBlend {blendOption=Linear}})) {layerGroup=None}
  ]}
 Locomotion {
  entry=Closed
  path=AnimGraph.AnimGraphNode_StateMachine_0.Locomotion
  ids=id
  nodes [
   AnimStateNode_0 = Idle_Open() {is=title AnimStateNode}
   AnimStateNode_1 = Open_Into() {is=title AnimStateNode}
   AnimStateNode_2 = Closed() {is=title AnimStateNode}
   AnimStateNode_3 = PopUp() {is=title AnimStateNode}
   AnimStateTransitionNode_0 = AnimStateTransitionNode() {bAutomaticRuleBasedOnSequencePlayerInState=true, blendMode=Linear, crossfadeDuration=0}
   AnimStateTransitionNode_1 = AnimStateTransitionNode() {crossfadeDuration=0}
   AnimStateTransitionNode_2 = AnimStateTransitionNode() {crossfadeDuration=0}
   AnimStateTransitionNode_3 = AnimStateTransitionNode() {crossfadeDuration=0, priorityOrder=2}
   AnimStateTransitionNode_4 = AnimStateTransitionNode()
   AnimStateTransitionNode_5 = AnimStateTransitionNode() {blendMode=Linear, crossfadeDuration=0.3}
   AnimStateTransitionNode_6 = AnimStateTransitionNode() {crossfadeDuration=0}
  ]
  links [
   {
    dst=Open_Into
    rule=AnimGraph.AnimGraphNode_StateMachine_0.Locomotion.AnimStateTransitionNode_1.Transition
    src=Closed
    via=AnimStateTransitionNode_1
   }
   {
    dst=Closed
    rule=AnimGraph.AnimGraphNode_StateMachine_0.Locomotion.AnimStateTransitionNode_2.Transition
    src=Idle_Open
    via=AnimStateTransitionNode_2
   }
   {
    dst=PopUp
    rule=AnimGraph.AnimGraphNode_StateMachine_0.Locomotion.AnimStateTransitionNode_4.Transition
    src=Idle_Open
    via=AnimStateTransitionNode_4
   }
   {
    dst=Closed
    rule=AnimGraph.AnimGraphNode_StateMachine_0.Locomotion.AnimStateTransitionNode_3.Transition
    src=Open_Into
    via=AnimStateTransitionNode_3
   }
   {
    dst=Idle_Open
    rule=AnimGraph.AnimGraphNode_StateMachine_0.Locomotion.AnimStateTransitionNode_0.Transition
    src=Open_Into
    via=AnimStateTransitionNode_0
   }
   {
    dst=Closed
    rule=AnimGraph.AnimGraphNode_StateMachine_0.Locomotion.AnimStateTransitionNode_6.Transition
    src=PopUp
    via=AnimStateTransitionNode_6
   }
   {
    dst=Idle_Open
    rule=AnimGraph.AnimGraphNode_StateMachine_0.Locomotion.AnimStateTransitionNode_5.Transition
    src=PopUp
    via=AnimStateTransitionNode_5
   }
  ]
 }
 Idle_Open {
  path=AnimGraph.AnimGraphNode_StateMachine_0.Locomotion.AnimStateNode_0.Idle_Open
  nodes [
   n1 = Break Fort Anim Input Skydiving!(FortAnimInput_Skydiving:"Fort Anim Input Skydiving Structure (by ref)"=Get SkydivingState() -> (SkydivingState:Fort Anim Input Skydiving Structure)) -> (
    ParaGilderRootOffset:Vector=~
    ParaGliderRootRotationOverride:Rotator=~
    bUseParaGlideRootModifier:bool=~
    bIsSkydivingFromLaunchPad:bool=~
    bIsSkydivingFromBus:bool=~
    bIsInVortex:bool=~
    bIsDiving:bool=~
    bIsParachuteOpen:bool=~
    bIsSkydiving:bool=~
    bIsParachuteLeaning:bool=~
    bIsSkydiveLeaning:bool=~
    bIsLeaning:bool=~
    bIsSkydiveDiveMode:bool=~
    bPlaySkydiveDrift:bool=~
    bSkydiveDriftDelayActive:bool=~
    bSkydiveDriftAnimAllowed:bool=~
    bIsGliderRight:bool=~
    bIsGliderCenter:bool=~
    bIsGliderLeft:bool=~
    bIsGliderForward:bool=~
    bIsGliderBack:bool=~
    bHasDeployedAndIsStandingUpright:bool=~
    LocalAccelForward:float=~
    LocalAccelRight:float=~
    LocalVelocityRight:float=~
    SkydiveAimPitch:float=~
    SkydiveAimYaw:float=~
    DeployChuteAnimRate:float=~
    SkydiveDriftAnimRate:float=~
    SkydiveDriftAnimRateCurrent:float=~
    SkydiveFidgetAnimRate:float=~
    SkydiveFidgetAnimRateCurrent:float=~
    SkydiveAdditiveAlpha:float=~
    SkydiveDriftDelay:float=~
    SkydiveDriftAnim:int
    SkydiveDriftAnimMax:int
    LaunchpadAnim:int
    LocalAccelDir:ESkydivingDirection Enum
    DirectionLast:ESkydivingDirection Enum
    PoseAdditive:Anim Sequence Object Reference
   ) {is=title K2Node_BreakStruct}
   n2 = Blendspace Player 'BS_BlackWidow_GLIDER'(X=@n1.SkydiveAimYaw, Y=@n1.SkydiveAimPitch) {blendSpace=/BRCosmetics/Animation/Game/MainPlayer/Skydive/ParaGlide/BlackWidow/BS_BlackWidow_GLIDER, groupName=GlideSway, method=SyncGroup}
   n3 = Get OwnerGlider() -> (OwnerGlider:Fort Player Parachute Object Reference)
   n4 = Break Fort Anim Input Skydiving(FortAnimInput_Skydiving=Get SkydivingState()) -> (
    ParaGilderRootOffset=~
    ParaGliderRootRotationOverride=~
    bUseParaGlideRootModifier=~
    bIsSkydivingFromLaunchPad=~
    bIsSkydivingFromBus=~
    bIsInVortex=~
    bIsDiving=~
    bIsParachuteOpen=~
    bIsSkydiving=~
    bIsParachuteLeaning=~
    bIsSkydiveLeaning=~
    bIsLeaning=~
    bIsSkydiveDiveMode=~
    bPlaySkydiveDrift=~
    bSkydiveDriftDelayActive=~
    bSkydiveDriftAnimAllowed=~
    bIsGliderRight=~
    bIsGliderCenter=~
    bIsGliderLeft=~
    bIsGliderForward=~
    bIsGliderBack=~
    bHasDeployedAndIsStandingUpright=~
    LocalAccelForward=~
    LocalAccelRight=~
    LocalVelocityRight=~
    SkydiveAimPitch=~
    SkydiveAimYaw=~
    DeployChuteAnimRate=~
    SkydiveDriftAnimRate=~
    SkydiveDriftAnimRateCurrent=~
    SkydiveFidgetAnimRate=~
    SkydiveFidgetAnimRateCurrent=~
    SkydiveAdditiveAlpha=~
    SkydiveDriftDelay=~
   )
   n5 = Apply Additive!(
    Base=Blend Poses by bool(
     BlendPose_0=Sequence Player 'BlackWidow_Glide_FrontEnd_GLIDER'() {sequence=/BRCosmetics/Animation/Game/MainPlayer/Skydive/ParaGlide/BlackWidow/BlackWidow_Glide_FrontEnd_GLIDER}
     BlendPose_1=@n2
     bActiveValue=Get IsCosmeticPreview!(self:Fort Player Parachute Object Reference=@n3) -> (IsCosmeticPreview:bool)
    ) {blendType=Linear}
    Additive=Blendspace Player 'BS_BlackWidow_Lean_GLIDER'(X=@n4.LocalAccelRight, Y=@n4.LocalAccelForward) {blendSpace=/BRCosmetics/Animation/Game/MainPlayer/Skydive/ParaGlide/BlackWidow/BS_BlackWidow_Lean_GLIDER}
    AlphaCurveName:Name=ParaGlide_Lean
   ) -> (Pose) {alphaInputType=Curve, alphaBoolBlend {blendOption=Linear}, alphaScaleBiasClamp {bias=1, scale=-1}}
   n6 = Blendspace Player 'BS_BlackWidow_Turn_GLIDER'(X=@n4.SkydiveAimYaw, Y=@n4.SkydiveAimPitch) {blendSpace=/BRCosmetics/Animation/Game/MainPlayer/Skydive/ParaGlide/BlackWidow/BS_BlackWidow_Turn_GLIDER}
   n7 = Get ScurrySpeed() -> (ScurrySpeed:double)
   Output Animation Pose(Result=Two Way Blend(
     A=Apply Additive!(Base=@n5, Additive=@n6, Alpha:float) -> (Pose) {alphaBoolBlend {blendOption=Linear}}
     B=Blendspace Player 'BS_BlackWidow_Scurry_GLIDER'!(X=@n4.LocalAccelRight, Y=@n4.LocalAccelForward, PlayRate:float=@n7) -> (Pose) {blendSpace=/BRCosmetics/Animation/Game/MainPlayer/Skydive/ParaGlide/BlackWidow/BS_BlackWidow_Scurry_GLIDER}
     Alpha=Get ScurryAlpha() -> (ScurryAlpha:double)
    ) {blendNode {alphaBoolBlend {blendOption=Linear}}})
  ]
 }
 Open_Into {
  path=AnimGraph.AnimGraphNode_StateMachine_0.Locomotion.AnimStateNode_1.Open_Into
  nodes [
   n1 = Break Fort Anim Input Skydiving(FortAnimInput_Skydiving=Get SkydivingState()) -> (
    ParaGilderRootOffset=~
    ParaGliderRootRotationOverride=~
    bUseParaGlideRootModifier=~
    bIsSkydivingFromLaunchPad=~
    bIsSkydivingFromBus=~
    bIsInVortex=~
    bIsDiving=~
    bIsParachuteOpen=~
    bIsSkydiving=~
    bIsParachuteLeaning=~
    bIsSkydiveLeaning=~
    bIsLeaning=~
    bIsSkydiveDiveMode=~
    bPlaySkydiveDrift=~
    bSkydiveDriftDelayActive=~
    bSkydiveDriftAnimAllowed=~
    bIsGliderRight=~
    bIsGliderCenter=~
    bIsGliderLeft=~
    bIsGliderForward=~
    bIsGliderBack=~
    bHasDeployedAndIsStandingUpright=~
    LocalAccelForward=~
    LocalAccelRight=~
    LocalVelocityRight=~
    SkydiveAimPitch=~
    SkydiveAimYaw=~
    DeployChuteAnimRate=~
    SkydiveDriftAnimRate=~
    SkydiveDriftAnimRateCurrent=~
    SkydiveFidgetAnimRate=~
    SkydiveFidgetAnimRateCurrent=~
    SkydiveAdditiveAlpha=~
    SkydiveDriftDelay=~
   )
   n2 = Get OwnerGlider()
   n3 = Break Fort Anim Input Skydiving(FortAnimInput_Skydiving=Get SkydivingState()) -> (
    ParaGilderRootOffset=~
    ParaGliderRootRotationOverride=~
    bUseParaGlideRootModifier=~
    bIsSkydivingFromLaunchPad=~
    bIsSkydivingFromBus=~
    bIsInVortex=~
    bIsDiving=~
    bIsParachuteOpen=~
    bIsSkydiving=~
    bIsParachuteLeaning=~
    bIsSkydiveLeaning=~
    bIsLeaning=~
    bIsSkydiveDiveMode=~
    bPlaySkydiveDrift=~
    bSkydiveDriftDelayActive=~
    bSkydiveDriftAnimAllowed=~
    bIsGliderRight=~
    bIsGliderCenter=~
    bIsGliderLeft=~
    bIsGliderForward=~
    bIsGliderBack=~
    bHasDeployedAndIsStandingUpright=~
    LocalAccelForward=~
    LocalAccelRight=~
    LocalVelocityRight=~
    SkydiveAimPitch=~
    SkydiveAimYaw=~
    DeployChuteAnimRate=~
    SkydiveDriftAnimRate=~
    SkydiveDriftAnimRateCurrent=~
    SkydiveFidgetAnimRate=~
    SkydiveFidgetAnimRateCurrent=~
    SkydiveAdditiveAlpha=~
    SkydiveDriftDelay=~
   )
   n4 = Apply Additive!(
    Base=Blend Poses by bool(
     BlendPose_0=Sequence Player 'BlackWidow_Into_FrontEnd_GLIDER'() {bLoopAnimation=false, sequence=/BRCosmetics/Animation/Game/MainPlayer/Skydive/ParaGlide/BlackWidow/BlackWidow_Into_FrontEnd_GLIDER}
     BlendPose_1=Blendspace Player 'BS_BlackWidow_Into_GLIDER'(X=@n1.SkydiveAimYaw, Y=@n1.SkydiveAimPitch) {bLoop=false, blendSpace=/BRCosmetics/Animation/Game/MainPlayer/Skydive/ParaGlide/BlackWidow/BS_BlackWidow_Into_GLIDER}
     bActiveValue=Get IsCosmeticPreview(self=@n2)
    ) {blendType=Linear}
    Additive=Blendspace Player 'BS_BlackWidow_Lean_GLIDER'(X=@n3.LocalAccelRight, Y=@n3.LocalAccelForward) {blendSpace=&BS_BlackWidow_Lean_GLIDER}
    AlphaCurveName:Name=ParaGlide_Lean
   ) -> (Pose) {alphaInputType=Curve, alphaBoolBlend {blendOption=Linear}, alphaScaleBiasClamp {bias=1, scale=-1}}
   n5 = Blendspace Player 'BS_BlackWidow_Turn_GLIDER'(X=@n3.SkydiveAimYaw, Y=@n3.SkydiveAimPitch) {blendSpace=&BS_BlackWidow_Turn_GLIDER}
   n6 = Get ScurrySpeed()
   Output Animation Pose(Result=Two Way Blend(
     A=Apply Additive!(Base=@n4, Additive=@n5, Alpha:float) -> (Pose) {alphaBoolBlend {blendOption=Linear}}
     B=Blendspace Player 'BS_BlackWidow_Scurry_GLIDER'(X=@n3.LocalAccelRight, Y=@n3.LocalAccelForward, PlayRate=@n6) {blendSpace=&BS_BlackWidow_Scurry_GLIDER}
     Alpha=Get ScurryAlpha()
    ) {blendNode {alphaBoolBlend {blendOption=Linear}}})
  ]
 }
 Closed {
  path=AnimGraph.AnimGraphNode_StateMachine_0.Locomotion.AnimStateNode_2.Closed
  nodes [
   n1 = Get OwnerGlider()
   Output Animation Pose(Result=Blend Poses by bool(
     BlendPose_0=Sequence Player 'BlackWidow_Closed_GLIDER'() {sequence=/BRCosmetics/Animation/Game/MainPlayer/Skydive/ParaGlide/BlackWidow/BlackWidow_Closed_GLIDER}
     BlendPose_1=Sequence Player 'BlackWidow_Closed_GLIDER'() {sequence=&BlackWidow_Closed_GLIDER}
     bActiveValue=Get IsCosmeticPreview(self=@n1)
    ) {blendType=Linear})
  ]
 }
 PopUp {
  path=AnimGraph.AnimGraphNode_StateMachine_0.Locomotion.AnimStateNode_3.PopUp
  nodes [
   n1 = Break Fort Anim Input Skydiving(FortAnimInput_Skydiving=Get SkydivingState()) -> (
    ParaGilderRootOffset=~
    ParaGliderRootRotationOverride=~
    bUseParaGlideRootModifier=~
    bIsSkydivingFromLaunchPad=~
    bIsSkydivingFromBus=~
    bIsInVortex=~
    bIsDiving=~
    bIsParachuteOpen=~
    bIsSkydiving=~
    bIsParachuteLeaning=~
    bIsSkydiveLeaning=~
    bIsLeaning=~
    bIsSkydiveDiveMode=~
    bPlaySkydiveDrift=~
    bSkydiveDriftDelayActive=~
    bSkydiveDriftAnimAllowed=~
    bIsGliderRight=~
    bIsGliderCenter=~
    bIsGliderLeft=~
    bIsGliderForward=~
    bIsGliderBack=~
    bHasDeployedAndIsStandingUpright=~
    LocalAccelForward=~
    LocalAccelRight=~
    LocalVelocityRight=~
    SkydiveAimPitch=~
    SkydiveAimYaw=~
    DeployChuteAnimRate=~
    SkydiveDriftAnimRate=~
    SkydiveDriftAnimRateCurrent=~
    SkydiveFidgetAnimRate=~
    SkydiveFidgetAnimRateCurrent=~
    SkydiveAdditiveAlpha=~
    SkydiveDriftDelay=~
   )
   n2 = Break Fort Anim Input Skydiving(FortAnimInput_Skydiving=Get SkydivingState()) -> (
    ParaGilderRootOffset=~
    ParaGliderRootRotationOverride=~
    bUseParaGlideRootModifier=~
    bIsSkydivingFromLaunchPad=~
    bIsSkydivingFromBus=~
    bIsInVortex=~
    bIsDiving=~
    bIsParachuteOpen=~
    bIsSkydiving=~
    bIsParachuteLeaning=~
    bIsSkydiveLeaning=~
    bIsLeaning=~
    bIsSkydiveDiveMode=~
    bPlaySkydiveDrift=~
    bSkydiveDriftDelayActive=~
    bSkydiveDriftAnimAllowed=~
    bIsGliderRight=~
    bIsGliderCenter=~
    bIsGliderLeft=~
    bIsGliderForward=~
    bIsGliderBack=~
    bHasDeployedAndIsStandingUpright=~
    LocalAccelForward=~
    LocalAccelRight=~
    LocalVelocityRight=~
    SkydiveAimPitch=~
    SkydiveAimYaw=~
    DeployChuteAnimRate=~
    SkydiveDriftAnimRate=~
    SkydiveDriftAnimRateCurrent=~
    SkydiveFidgetAnimRate=~
    SkydiveFidgetAnimRateCurrent=~
    SkydiveAdditiveAlpha=~
    SkydiveDriftDelay=~
   )
   n3 = Blendspace Player 'BS_BlackWidow_GLIDER'(X=@n2.SkydiveAimYaw, Y=@n2.SkydiveAimPitch) {blendSpace=&BS_BlackWidow_GLIDER, groupName=GlideSway, method=SyncGroup}
   n4 = Apply Additive!(
    Base=@n3
    Additive=Blendspace Player 'BS_BlackWidow_Lean_GLIDER'(X=@n1.LocalAccelRight, Y=@n1.LocalAccelForward) {blendSpace=&BS_BlackWidow_Lean_GLIDER}
    AlphaCurveName:Name=ParaGlide_Lean
   ) -> (Pose) {alphaInputType=Curve, alphaBoolBlend {blendOption=Linear}, alphaScaleBiasClamp {bias=1, scale=-1}}
   n5 = Blendspace Player 'BS_BlackWidow_Turn_GLIDER'(X=@n1.SkydiveAimYaw, Y=@n1.SkydiveAimPitch) {blendSpace=&BS_BlackWidow_Turn_GLIDER}
   Output Animation Pose(Result=Apply Additive!(
     Base=Apply Additive(Base=@n4, Additive=@n5) {alphaBoolBlend {blendOption=Linear}}
     Additive="Sequence Player (additive) 'BlackWidow_PopUp_GLIDER'"() {bLoopAnimation=false, sequence=/BRCosmetics/Animation/Game/MainPlayer/Skydive/ParaGlide/BlackWidow/BlackWidow_PopUp_GLIDER}
     Alpha:float=0.75
    ) -> (Pose) {alphaBoolBlend {blendOption=Linear}})
  ]
 }
 Transition {
  path=AnimGraph.AnimGraphNode_StateMachine_0.Locomotion.AnimStateTransitionNode_0.Transition
  nodes [
   Result()
   "Equal (Float)"!(A:double="Get Relevant Anim Time Remaining (Open_Into)"!() -> (ReturnValue:float) {is=title K2Node_AnimGetter}, B:double=0) -> (ReturnValue:bool) {is=title K2Node_CallFunction}
  ]
 }
 Transition {
  path=AnimGraph.AnimGraphNode_StateMachine_0.Locomotion.AnimStateTransitionNode_1.Transition
  nodes [
   n1 = Get SkydivingState()
   Result(bCanEnterTransition=OR Boolean(
     A=Break Fort Anim Input Skydiving(FortAnimInput_Skydiving=@n1) -> (
      ParaGilderRootOffset=~
      ParaGliderRootRotationOverride=~
      bUseParaGlideRootModifier=~
      bIsSkydivingFromLaunchPad=~
      bIsSkydivingFromBus=~
      bIsInVortex=~
      bIsDiving=~
      bIsParachuteOpen=~
      bIsSkydiving=~
      bIsParachuteLeaning=~
      bIsSkydiveLeaning=~
      bIsLeaning=~
      bIsSkydiveDiveMode=~
      bPlaySkydiveDrift=~
      bSkydiveDriftDelayActive=~
      bSkydiveDriftAnimAllowed=~
      bIsGliderRight=~
      bIsGliderCenter=~
      bIsGliderLeft=~
      bIsGliderForward=~
      bIsGliderBack=~
      bHasDeployedAndIsStandingUpright=~
      LocalAccelForward=~
      LocalAccelRight=~
      LocalVelocityRight=~
      SkydiveAimPitch=~
      SkydiveAimYaw=~
      DeployChuteAnimRate=~
      SkydiveDriftAnimRate=~
      SkydiveDriftAnimRateCurrent=~
      SkydiveFidgetAnimRate=~
      SkydiveFidgetAnimRateCurrent=~
      SkydiveAdditiveAlpha=~
      SkydiveDriftDelay=~
     ).bIsParachuteOpen
     B=Get bForceOpen() -> (bForceOpen:bool)
    ))
  ]
 }
 Transition {
  path=AnimGraph.AnimGraphNode_StateMachine_0.Locomotion.AnimStateTransitionNode_2.Transition
  nodes [
   n1 = Get bForceOpen()
   n2 = Break Fort Anim Input Skydiving(FortAnimInput_Skydiving=Get SkydivingState()) -> (
    ParaGilderRootOffset=~
    ParaGliderRootRotationOverride=~
    bUseParaGlideRootModifier=~
    bIsSkydivingFromLaunchPad=~
    bIsSkydivingFromBus=~
    bIsInVortex=~
    bIsDiving=~
    bIsParachuteOpen=~
    bIsSkydiving=~
    bIsParachuteLeaning=~
    bIsSkydiveLeaning=~
    bIsLeaning=~
    bIsSkydiveDiveMode=~
    bPlaySkydiveDrift=~
    bSkydiveDriftDelayActive=~
    bSkydiveDriftAnimAllowed=~
    bIsGliderRight=~
    bIsGliderCenter=~
    bIsGliderLeft=~
    bIsGliderForward=~
    bIsGliderBack=~
    bHasDeployedAndIsStandingUpright=~
    LocalAccelForward=~
    LocalAccelRight=~
    LocalVelocityRight=~
    SkydiveAimPitch=~
    SkydiveAimYaw=~
    DeployChuteAnimRate=~
    SkydiveDriftAnimRate=~
    SkydiveDriftAnimRateCurrent=~
    SkydiveFidgetAnimRate=~
    SkydiveFidgetAnimRateCurrent=~
    SkydiveAdditiveAlpha=~
    SkydiveDriftDelay=~
   )
   Result(bCanEnterTransition=AND Boolean(A=NOT Boolean(A=@n2.bIsParachuteOpen), B=NOT Boolean(A=@n1)))
  ]
 }
 Transition {
  path=AnimGraph.AnimGraphNode_StateMachine_0.Locomotion.AnimStateTransitionNode_3.Transition
  nodes [
   n1 = Get bForceOpen()
   n2 = Break Fort Anim Input Skydiving(FortAnimInput_Skydiving=Get SkydivingState()) -> (
    ParaGilderRootOffset=~
    ParaGliderRootRotationOverride=~
    bUseParaGlideRootModifier=~
    bIsSkydivingFromLaunchPad=~
    bIsSkydivingFromBus=~
    bIsInVortex=~
    bIsDiving=~
    bIsParachuteOpen=~
    bIsSkydiving=~
    bIsParachuteLeaning=~
    bIsSkydiveLeaning=~
    bIsLeaning=~
    bIsSkydiveDiveMode=~
    bPlaySkydiveDrift=~
    bSkydiveDriftDelayActive=~
    bSkydiveDriftAnimAllowed=~
    bIsGliderRight=~
    bIsGliderCenter=~
    bIsGliderLeft=~
    bIsGliderForward=~
    bIsGliderBack=~
    bHasDeployedAndIsStandingUpright=~
    LocalAccelForward=~
    LocalAccelRight=~
    LocalVelocityRight=~
    SkydiveAimPitch=~
    SkydiveAimYaw=~
    DeployChuteAnimRate=~
    SkydiveDriftAnimRate=~
    SkydiveDriftAnimRateCurrent=~
    SkydiveFidgetAnimRate=~
    SkydiveFidgetAnimRateCurrent=~
    SkydiveAdditiveAlpha=~
    SkydiveDriftDelay=~
   )
   Result(bCanEnterTransition=AND Boolean(A=NOT Boolean(A=@n2.bIsParachuteOpen), B=NOT Boolean(A=@n1)))
  ]
 }
 Transition {
  path=AnimGraph.AnimGraphNode_StateMachine_0.Locomotion.AnimStateTransitionNode_4.Transition
  nodes [
   n1 = Get ScurryAnim() -> (ScurryAnim:bool)
   Result(bCanEnterTransition=AND Boolean(A=Get CanPopUp() -> (CanPopUp:bool), B=NOT Boolean(A=@n1)))
   Get BlockAnimRule() -> (BlockAnimRule:bool)
  ]
 }
 Transition {
  path=AnimGraph.AnimGraphNode_StateMachine_0.Locomotion.AnimStateTransitionNode_5.Transition
  nodes [Result(bCanEnterTransition=NOT Boolean(A=Get CanPopUp()))]
 }
 Transition {
  path=AnimGraph.AnimGraphNode_StateMachine_0.Locomotion.AnimStateTransitionNode_6.Transition
  nodes [
   n1 = Get bForceOpen()
   n2 = Break Fort Anim Input Skydiving(FortAnimInput_Skydiving=Get SkydivingState()) -> (
    ParaGilderRootOffset=~
    ParaGliderRootRotationOverride=~
    bUseParaGlideRootModifier=~
    bIsSkydivingFromLaunchPad=~
    bIsSkydivingFromBus=~
    bIsInVortex=~
    bIsDiving=~
    bIsParachuteOpen=~
    bIsSkydiving=~
    bIsParachuteLeaning=~
    bIsSkydiveLeaning=~
    bIsLeaning=~
    bIsSkydiveDiveMode=~
    bPlaySkydiveDrift=~
    bSkydiveDriftDelayActive=~
    bSkydiveDriftAnimAllowed=~
    bIsGliderRight=~
    bIsGliderCenter=~
    bIsGliderLeft=~
    bIsGliderForward=~
    bIsGliderBack=~
    bHasDeployedAndIsStandingUpright=~
    LocalAccelForward=~
    LocalAccelRight=~
    LocalVelocityRight=~
    SkydiveAimPitch=~
    SkydiveAimYaw=~
    DeployChuteAnimRate=~
    SkydiveDriftAnimRate=~
    SkydiveDriftAnimRateCurrent=~
    SkydiveFidgetAnimRate=~
    SkydiveFidgetAnimRateCurrent=~
    SkydiveAdditiveAlpha=~
    SkydiveDriftDelay=~
   )
   Result(bCanEnterTransition=AND Boolean(A=NOT Boolean(A=@n2.bIsParachuteOpen), B=NOT Boolean(A=@n1)))
  ]
 }
 EventGraph {nodes [
   n1 = Event BlueprintUpdateAnimation()
   > Set Time(Time:double=@n1.DeltaTimeX) -> (Output_Get:double)
   > Set ScurryAnim(ScurryAnim:bool="float >= float"(A=GetCurveValue(CurveName=Scurry), B=0.35) {is=title K2Node_CallFunction}) -> (Output_Get:bool)
   > n2 = Sequence!(execute) -> (then_0, then_1, then_2:Exec)
   n3 = GetCurveValue(CurveName=ParaGlide_Lean)
   n4 = Break Fort Anim Input Skydiving(FortAnimInput_Skydiving=Get SkydivingState()) -> (
    ParaGilderRootOffset=~
    ParaGliderRootRotationOverride=~
    bUseParaGlideRootModifier=~
    bIsSkydivingFromLaunchPad=~
    bIsSkydivingFromBus=~
    bIsInVortex=~
    bIsDiving=~
    bIsParachuteOpen=~
    bIsSkydiving=~
    bIsParachuteLeaning=~
    bIsSkydiveLeaning=~
    bIsLeaning=~
    bIsSkydiveDiveMode=~
    bPlaySkydiveDrift=~
    bSkydiveDriftDelayActive=~
    bSkydiveDriftAnimAllowed=~
    bIsGliderRight=~
    bIsGliderCenter=~
    bIsGliderLeft=~
    bIsGliderForward=~
    bIsGliderBack=~
    bHasDeployedAndIsStandingUpright=~
    LocalAccelForward=~
    LocalAccelRight=~
    LocalVelocityRight=~
    SkydiveAimPitch=~
    SkydiveAimYaw=~
    DeployChuteAnimRate=~
    SkydiveDriftAnimRate=~
    SkydiveDriftAnimRateCurrent=~
    SkydiveFidgetAnimRate=~
    SkydiveFidgetAnimRateCurrent=~
    SkydiveAdditiveAlpha=~
    SkydiveDriftDelay=~
   )
   > n5 = Branch(Condition=AND Boolean!(
     A="Not Equal (Enum)"!(A:ESkydivingDirection Enum=@n4.LocalAccelDir, B:ESkydivingDirection Enum=Left) -> (ReturnValue:bool)
     B="Not Equal (Enum)"(A=@n4.LocalAccelDir, B=Right)
     C:bool=Get bIsAccelerating2D()
     D:bool="float <= float"!(A:double=@n3, B:double=0.85) -> (ReturnValue:bool) {is=title K2Node_CallFunction}
    ) -> (ReturnValue))
   > Do Once(Reset=@n12.then, Start Closed=true)
   > Set CanScurry(CanScurry:bool=true) -> (Output_Get:bool)
   > Set ScurryInt(ScurryInt:int=RandomIntegerInRange!(Min:int=0, Max:int=1) -> (ReturnValue:int=0)) -> (Output_Get:int=0)
   > Delay(Duration=RandomFloatInRange(Min=2, Max=3))
   > Set CanScurry(CanScurry=false)
   Set ScurryAlpha(
    execute=[@n5.else, @n13.else]
    ScurryAlpha:double=FInterpTo(Current=Get ScurryAlpha(), Target=0, DeltaTime=Get Time() -> (Time:double), InterpSpeed=2)
   ) -> (Output_Get:double)
   > n6 = Set ScurrySpeed(ScurrySpeed:double=FInterpTo(Current=Get ScurrySpeed(), Target=0, DeltaTime=Get Time(), InterpSpeed=2)) -> (Output_Get:double)
   n7 = Break Fort Anim Input Skydiving(FortAnimInput_Skydiving=Get SkydivingState()) -> (
    ParaGilderRootOffset=~
    ParaGliderRootRotationOverride=~
    bUseParaGlideRootModifier=~
    bIsSkydivingFromLaunchPad=~
    bIsSkydivingFromBus=~
    bIsInVortex=~
    bIsDiving=~
    bIsParachuteOpen=~
    bIsSkydiving=~
    bIsParachuteLeaning=~
    bIsSkydiveLeaning=~
    bIsLeaning=~
    bIsSkydiveDiveMode=~
    bPlaySkydiveDrift=~
    bSkydiveDriftDelayActive=~
    bSkydiveDriftAnimAllowed=~
    bIsGliderRight=~
    bIsGliderCenter=~
    bIsGliderLeft=~
    bIsGliderForward=~
    bIsGliderBack=~
    bHasDeployedAndIsStandingUpright=~
    LocalAccelForward=~
    LocalAccelRight=~
    LocalVelocityRight=~
    SkydiveAimPitch=~
    SkydiveAimYaw=~
    DeployChuteAnimRate=~
    SkydiveDriftAnimRate=~
    SkydiveDriftAnimRateCurrent=~
    SkydiveFidgetAnimRate=~
    SkydiveFidgetAnimRateCurrent=~
    SkydiveAdditiveAlpha=~
    SkydiveDriftDelay=~
   )
   n8 = Branch(execute=[@n15.then, @n6.then], Condition="Equal (Enum)"(A:ESkydivingDirection Enum=@n7.LocalAccelDir, B:ESkydivingDirection Enum=Center))
   > Do Once(Reset=@n8.else, Start Closed=true)
   > Set CanPopUp(CanPopUp:bool=true) -> (Output_Get:bool)
   n9 = "float >= float"(A=@n3, B=0.85)
   n10 = "float > float"(A=GetCurveValue(CurveName=DeployAnim), B=0.85) {is=title K2Node_CallFunction}
   n11 = "Not Equal (Enum)"(A=@n4.LocalAccelDir, B=Center)
   n12 = Branch(
    execute=@n2.then_1
    Condition=OR Boolean(A="Equal (Enum)"(A=@n4.LocalAccelDir, B=Center), B=AND Boolean!(A=@n9, B=@n11, C=@n10) -> (ReturnValue))
   )
   n13 = Branch(execute=@n2.then_2, Condition=Get CanScurry() -> (CanScurry:bool))
   n14 = Get ScurryInt() -> (ScurryInt:int=0)
   > Set ScurryAlpha(ScurryAlpha=FInterpTo(
     Current=Get ScurryAlpha()
     Target=Select(Option 0:double=0.98, Option 1:double=0.98, Index:int=@n14) -> (ReturnValue:double)
     DeltaTime=Get Time()
     InterpSpeed=5
    ))
   > n15 = Set ScurrySpeed(ScurrySpeed=FInterpTo(Current=Get ScurrySpeed(), Target=Select(Option 0=1, Option 1=1.25, Index=@n14), DeltaTime=Get Time(), InterpSpeed=5))
   n16 = AnimNotify_TurnONEngines() {is=title K2Node_Event}
   SetGliderFullyDeployed!(execute=[@n16.then, @n17.then], self:Fort Glider Anim Instance Object Reference, bIsFullyDeployed:bool=true) -> (then)
   n17 = AnimNotify_SpiderDeployed() {is=title K2Node_Event}
   AnimNotify_PopUpOver() {is=title K2Node_Event}
   > Set CanPopUp(CanPopUp=false)
   AnimNotify_ScurryOver() {is=title K2Node_Event}
   > Set CanScurry(CanScurry=false)
   > Set ScurryAlpha(ScurryAlpha=0)
   > Set ScurrySpeed(ScurrySpeed=0)
   "Not Equal (Enum)"(A=@n7.LocalAccelDir, B=Center)
  ]}
]
variables [
 BlockAnimRule {type=boolean}
 CanPopUp {type=boolean}
 CanScurry {type=boolean}
 ScurryAlpha {type=number}
 ScurryAnim {type=boolean}
 ScurryInt {type=integer}
 ScurrySpeed {type=number, default=1}
 Time {type=number}
]
