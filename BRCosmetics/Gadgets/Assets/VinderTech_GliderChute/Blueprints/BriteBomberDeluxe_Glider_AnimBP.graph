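# Animation Blueprint for the Brite Bomber Deluxe glider, with a hidden glitter-rocket ("Fox2") easter egg.
# A Closed/Open_Into/Idle_Open state machine feeds additive lean, turn, sway, tassel and jitter layers; forward acceleration drives tassels and jitter.
# A one-time date and cosmetic-loadout check sets WeaponsFree, letting fire input trigger rocket, low-altitude warning or dry-fire additives.
# Curves track playing animations, while notifies reset flags and pick random rocket variants.
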AnimBlueprint /BRCosmetics/Gadgets/Assets/VinderTech_GliderChute/Blueprints/BriteBomberDeluxe_Glider_AnimBP
target_skeleton=/BRCosmetics/Gadgets/Assets/VinderTech_GliderChute/Glider_BriteBomberDeluxe/Meshes/Glider_BriteBomberDeluxe_Skeleton
graphs [
 Adventure {nodes [
   Inputs()
   > Set ReadyPlayer(ReadyPlayer:bool="DateTime >= DateTime"!(A:Date Time Structure=Now!() -> (ReturnValue:Date Time Structure), B:Date Time Structure=Get Player12() -> (Player12:Date Time Structure)) -> (ReturnValue:bool)) -> (Output_Get:bool)
   > Outputs!(then:Exec)
   Get Player11() -> (Player11:Date Time Structure)
   Get PlayerID4() -> (PlayerID4:Date Time Structure)
  ]}
 AnimGraph {nodes [
   n1 = Break Fort Anim Input Skydiving!(FortAnimInput_Skydiving:"Fort Anim Input Skydiving Structure (by ref)"=Get SkydivingState() -> (SkydivingState:Fort Anim Input Skydiving Structure)) -> (
    ParaGilderRootOffset:Vector=~
    ParaGliderRootRotationOverride:Rotator=~
    bUseParaGlideRootModifier:bool=~
    bIsSkydivingFromLaunchPad:bool=~
    bIsSkydivingFromBus:bool=~
    bIsInVortex:bool=~
    bIsDiving:bool=~
    bIsParachuteOpen:bool=~
    bIsSkydiving:bool=~
    bIsParachuteLeaning:bool=~
    bIsSkydiveLeaning:bool=~
    bIsLeaning:bool=~
    bIsSkydiveDiveMode:bool=~
    bPlaySkydiveDrift:bool=~
    bSkydiveDriftDelayActive:bool=~
    bSkydiveDriftAnimAllowed:bool=~
    bIsGliderRight:bool=~
    bIsGliderCenter:bool=~
    bIsGliderLeft:bool=~
    bIsGliderForward:bool=~
    bIsGliderBack:bool=~
    bHasDeployedAndIsStandingUpright:bool=~
    LocalAccelForward:float=~
    LocalAccelRight:float=~
    LocalVelocityRight:float=~
    SkydiveAimPitch:float=~
    SkydiveAimYaw:float=~
    DeployChuteAnimRate:float=~
    SkydiveDriftAnimRate:float=~
    SkydiveDriftAnimRateCurrent:float=~
    SkydiveFidgetAnimRate:float=~
    SkydiveFidgetAnimRateCurrent:float=~
    SkydiveAdditiveAlpha:float=~
    SkydiveDriftDelay:float=~
    SkydiveDriftAnim:int
    SkydiveDriftAnimMax:int
    LaunchpadAnim:int
    LocalAccelDir:ESkydivingDirection Enum
    DirectionLast:ESkydivingDirection Enum
    PoseAdditive:Anim Sequence Object Reference
   ) {is=title K2Node_BreakStruct}
   n2 = Blendspace Player 'BS_BriteBomberDeluxe_Lean_GLIDER'(X=@n1.LocalAccelRight, Y=@n1.LocalAccelForward) {blendSpace=/BRCosmetics/Animation/Game/MainPlayer/Skydive/ParaGlide/BriteBomberDeluxe/BS_BriteBomberDeluxe_Lean_GLIDER}
   AnimGraphNode_StateMachine_0 = Locomotion() {is=title AnimGraphNode_StateMachine}
   n3 = Get Tassle_Speed() -> (Tassle_Speed:double)
   n4 = Apply Additive(
    Base=Apply Additive(
     Base=Apply Additive!(Base=@AnimGraphNode_StateMachine_0, Additive=@n2, AlphaCurveName:Name=ParaGlide_Lean) -> (Pose) {alphaInputType=Curve, alphaBoolBlend {blendOption=Linear}, alphaScaleBiasClamp {bias=1, scale=-1}}
     Additive=Blendspace Player 'BS_BriteBomberDeluxe_Turn_GLIDER'(X=@n1.SkydiveAimYaw, Y=@n1.SkydiveAimPitch) {blendSpace=/BRCosmetics/Animation/Game/MainPlayer/Skydive/ParaGlide/BriteBomberDeluxe/BS_BriteBomberDeluxe_Turn_GLIDER}
    ) {alphaBoolBlend {blendOption=Linear}}
    Additive="Sequence Player (additive) 'BriteBomberDeluxe_Tassles_Tail_GLIDER'"!(PlayRate:float=@n3) -> (Pose) {sequence=/BRCosmetics/Animation/Game/MainPlayer/Skydive/ParaGlide/BriteBomberDeluxe/BriteBomberDeluxe_Tassles_Tail_GLIDER}
   ) {alphaBoolBlend {blendOption=Linear}}
   n5 = "Sequence Player (additive) 'BriteBomberDeluxe_Tassles_HandleBar_GLIDER'"!(PlayRate:float=@n3) -> (Pose) {sequence=/BRCosmetics/Animation/Game/MainPlayer/Skydive/ParaGlide/BriteBomberDeluxe/BriteBomberDeluxe_Tassles_HandleBar_GLIDER}
   n6 = Break Fort Anim Input Skydiving(FortAnimInput_Skydiving=Get SkydivingState()) -> (
    ParaGilderRootOffset=~
    ParaGliderRootRotationOverride=~
    bUseParaGlideRootModifier=~
    bIsSkydivingFromLaunchPad=~
    bIsSkydivingFromBus=~
    bIsInVortex=~
    bIsDiving=~
    bIsParachuteOpen=~
    bIsSkydiving=~
    bIsParachuteLeaning=~
    bIsSkydiveLeaning=~
    bIsLeaning=~
    bIsSkydiveDiveMode=~
    bPlaySkydiveDrift=~
    bSkydiveDriftDelayActive=~
    bSkydiveDriftAnimAllowed=~
    bIsGliderRight=~
    bIsGliderCenter=~
    bIsGliderLeft=~
    bIsGliderForward=~
    bIsGliderBack=~
    bHasDeployedAndIsStandingUpright=~
    LocalAccelForward=~
    LocalAccelRight=~
    LocalVelocityRight=~
    SkydiveAimPitch=~
    SkydiveAimYaw=~
    DeployChuteAnimRate=~
    SkydiveDriftAnimRate=~
    SkydiveDriftAnimRateCurrent=~
    SkydiveFidgetAnimRate=~
    SkydiveFidgetAnimRateCurrent=~
    SkydiveAdditiveAlpha=~
    SkydiveDriftDelay=~
   )
   n7 = Apply Additive(
    Base=Apply Additive!(
     Base=Apply Additive(Base=@n4, Additive=@n5) {alphaBoolBlend {blendOption=Linear}}
     Additive="Sequence Player (additive) 'BriteBomberDeluxe_Jitter_GLIDER'"!(PlayRate:float=1.1) -> (Pose) {sequence=/BRCosmetics/Animation/Game/MainPlayer/Skydive/ParaGlide/BriteBomberDeluxe/BriteBomberDeluxe_Jitter_GLIDER}
     Alpha:float=Get Jitter_Alpha() -> (Jitter_Alpha:double)
    ) -> (Pose) {alphaBoolBlend {blendOption=Linear}}
    Additive=Blendspace Player 'BS_BriteBomberDeluxe_Sway_GLIDER'(X=@n6.SkydiveAimYaw, Y=@n6.LocalAccelForward) {blendSpace=/BRCosmetics/Animation/Game/MainPlayer/Skydive/ParaGlide/BriteBomberDeluxe/BS_BriteBomberDeluxe_Sway_GLIDER}
    AlphaCurveName=Sway
   ) {alphaCurveName=ParaGlide_Lean, alphaInputType=Curve, alphaBoolBlend {blendOption=Linear}, alphaScaleBiasClamp {bias=1, scale=-1}}
   AnimGraphNode_StateMachine_1 = Fox() {is=title AnimGraphNode_StateMachine}
   AnimGraphNode_StateMachine_2 = FoxWarning() {is=title AnimGraphNode_StateMachine}
   n8 = Apply Additive!(
    Base=Apply Additive(Base=@n7, Additive=@AnimGraphNode_StateMachine_1, bAlphaBoolEnabled=Get Fox2_Away() -> (Fox2_Away:bool)) {alphaInputType=Bool, alphaBoolBlend {blendOption=Linear}}
    Additive=@AnimGraphNode_StateMachine_2
    bAlphaBoolEnabled:bool=Get Fox2_Warning() -> (Fox2_Warning:bool)
   ) -> (Pose) {alphaInputType=Bool, alphaBoolBlend {blendOption=Linear}}
   AnimGraphNode_StateMachine_3 = FoxDryFire() {is=title AnimGraphNode_StateMachine}
   n9 = Get Fox2_DryFire() -> (Fox2_DryFire:bool)
   Output Pose(Result=Slot 'DefaultSlot'(Source=Apply Additive(Base=@n8, Additive=@AnimGraphNode_StateMachine_3, bAlphaBoolEnabled=@n9) {alphaInputType=Bool, alphaBoolBlend {blendOption=Linear}})) {layerGroup=None}
  ]}
 Locomotion {
  entry=Closed
  path=AnimGraph.AnimGraphNode_StateMachine_0.Locomotion
  ids=id
  nodes [
   AnimStateNode_0 = Idle_Open() {is=title AnimStateNode}
   AnimStateNode_1 = Open_Into() {is=title AnimStateNode}
   AnimStateNode_2 = Closed() {is=title AnimStateNode}
   AnimStateTransitionNode_0 = AnimStateTransitionNode() {bAutomaticRuleBasedOnSequencePlayerInState=true, blendMode=Linear, crossfadeDuration=0}
   AnimStateTransitionNode_1 = AnimStateTransitionNode() {crossfadeDuration=0}
   AnimStateTransitionNode_2 = AnimStateTransitionNode() {crossfadeDuration=0}
   AnimStateTransitionNode_3 = AnimStateTransitionNode() {crossfadeDuration=0, priorityOrder=2}
  ]
  links [
   {
    dst=Open_Into
    rule=AnimGraph.AnimGraphNode_StateMachine_0.Locomotion.AnimStateTransitionNode_1.Transition
    src=Closed
    via=AnimStateTransitionNode_1
   }
   {
    dst=Closed
    rule=AnimGraph.AnimGraphNode_StateMachine_0.Locomotion.AnimStateTransitionNode_2.Transition
    src=Idle_Open
    via=AnimStateTransitionNode_2
   }
   {
    dst=Closed
    rule=AnimGraph.AnimGraphNode_StateMachine_0.Locomotion.AnimStateTransitionNode_3.Transition
    src=Open_Into
    via=AnimStateTransitionNode_3
   }
   {
    dst=Idle_Open
    rule=AnimGraph.AnimGraphNode_StateMachine_0.Locomotion.AnimStateTransitionNode_0.Transition
    src=Open_Into
    via=AnimStateTransitionNode_0
   }
  ]
 }
 Idle_Open {
  path=AnimGraph.AnimGraphNode_StateMachine_0.Locomotion.AnimStateNode_0.Idle_Open
  nodes [
   Blend Poses by bool(bActiveValue=Get IsCosmeticPreview!(self:Fort Player Parachute Object Reference=Get OwnerGlider() -> (OwnerGlider:Fort Player Parachute Object Reference)) -> (IsCosmeticPreview:bool)) {blendType=Linear}
   n1 = Break Fort Anim Input Skydiving(FortAnimInput_Skydiving=Get SkydivingState()) -> (
    ParaGilderRootOffset=~
    ParaGliderRootRotationOverride=~
    bUseParaGlideRootModifier=~
    bIsSkydivingFromLaunchPad=~
    bIsSkydivingFromBus=~
    bIsInVortex=~
    bIsDiving=~
    bIsParachuteOpen=~
    bIsSkydiving=~
    bIsParachuteLeaning=~
    bIsSkydiveLeaning=~
    bIsLeaning=~
    bIsSkydiveDiveMode=~
    bPlaySkydiveDrift=~
    bSkydiveDriftDelayActive=~
    bSkydiveDriftAnimAllowed=~
    bIsGliderRight=~
    bIsGliderCenter=~
    bIsGliderLeft=~
    bIsGliderForward=~
    bIsGliderBack=~
    bHasDeployedAndIsStandingUpright=~
    LocalAccelForward=~
    LocalAccelRight=~
    LocalVelocityRight=~
    SkydiveAimPitch=~
    SkydiveAimYaw=~
    DeployChuteAnimRate=~
    SkydiveDriftAnimRate=~
    SkydiveDriftAnimRateCurrent=~
    SkydiveFidgetAnimRate=~
    SkydiveFidgetAnimRateCurrent=~
    SkydiveAdditiveAlpha=~
    SkydiveDriftDelay=~
   )
   n2 = Blendspace Player 'BS_BriteBomberDeluxe_GLIDER'(X=@n1.SkydiveAimYaw, Y=@n1.LocalAccelForward) {
    blendSpace=/BRCosmetics/Animation/Game/MainPlayer/Skydive/ParaGlide/BriteBomberDeluxe/BS_BriteBomberDeluxe_GLIDER
    groupName=Glide
    method=SyncGroup
   }
   n3 = Get OwnerGlider()
   Output Animation Pose(Result=Apply Additive(
     Base=@n2
     Additive="Sequence Player (additive) 'BriteBomberDeluxe_Glide_FrontEnd_GLIDER'"() {sequence=/BRCosmetics/Animation/Game/MainPlayer/Skydive/ParaGlide/BriteBomberDeluxe/BriteBomberDeluxe_Glide_FrontEnd_GLIDER}
     bAlphaBoolEnabled=Get IsCosmeticPreview(self=@n3)
    ) {alphaInputType=Bool, alphaBoolBlend {blendOption=Linear}})
  ]
 }
 Open_Into {
  path=AnimGraph.AnimGraphNode_StateMachine_0.Locomotion.AnimStateNode_1.Open_Into
  nodes [
   n1 = Break Fort Anim Input Skydiving(FortAnimInput_Skydiving=Get SkydivingState()) -> (
    ParaGilderRootOffset=~
    ParaGliderRootRotationOverride=~
    bUseParaGlideRootModifier=~
    bIsSkydivingFromLaunchPad=~
    bIsSkydivingFromBus=~
    bIsInVortex=~
    bIsDiving=~
    bIsParachuteOpen=~
    bIsSkydiving=~
    bIsParachuteLeaning=~
    bIsSkydiveLeaning=~
    bIsLeaning=~
    bIsSkydiveDiveMode=~
    bPlaySkydiveDrift=~
    bSkydiveDriftDelayActive=~
    bSkydiveDriftAnimAllowed=~
    bIsGliderRight=~
    bIsGliderCenter=~
    bIsGliderLeft=~
    bIsGliderForward=~
    bIsGliderBack=~
    bHasDeployedAndIsStandingUpright=~
    LocalAccelForward=~
    LocalAccelRight=~
    LocalVelocityRight=~
    SkydiveAimPitch=~
    SkydiveAimYaw=~
    DeployChuteAnimRate=~
    SkydiveDriftAnimRate=~
    SkydiveDriftAnimRateCurrent=~
    SkydiveFidgetAnimRate=~
    SkydiveFidgetAnimRateCurrent=~
    SkydiveAdditiveAlpha=~
    SkydiveDriftDelay=~
   )
   n2 = "Sequence Player (additive) 'BriteBomberDeluxe_Into_FrontEnd_GLIDER'"() {
    bLoopAnimation=false
    sequence=/BRCosmetics/Animation/Game/MainPlayer/Skydive/ParaGlide/BriteBomberDeluxe/BriteBomberDeluxe_Into_FrontEnd_GLIDER
   }
   n3 = Get OwnerGlider()
   Output Animation Pose(Result=Apply Additive(
     Base=Blendspace Player 'BS_BriteBomberDeluxe_Into_GLIDER'(X=@n1.SkydiveAimYaw, Y=@n1.SkydiveAimPitch) {bLoop=false, blendSpace=/BRCosmetics/Animation/Game/MainPlayer/Skydive/ParaGlide/BriteBomberDeluxe/BS_BriteBomberDeluxe_Into_GLIDER}
     Additive=@n2
     bAlphaBoolEnabled=Get IsCosmeticPreview(self=@n3)
    ) {alphaInputType=Bool, alphaBoolBlend {blendOption=Linear}})
  ]
 }
 Closed {
  path=AnimGraph.AnimGraphNode_StateMachine_0.Locomotion.AnimStateNode_2.Closed
  nodes [
   n1 = Get OwnerGlider()
   Output Animation Pose(Result=Blend Poses by bool(
     BlendPose_0=Sequence Player 'BriteBomberDeluxe_Closed_GLIDER'() {sequence=/BRCosmetics/Animation/Game/MainPlayer/Skydive/ParaGlide/BriteBomberDeluxe/BriteBomberDeluxe_Closed_GLIDER}
     BlendPose_1=Sequence Player 'BriteBomberDeluxe_Closed_GLIDER'() {sequence=&BriteBomberDeluxe_Closed_GLIDER}
     bActiveValue=Get IsCosmeticPreview(self=@n1)
    ) {blendType=Linear})
  ]
 }
 Transition {
  path=AnimGraph.AnimGraphNode_StateMachine_0.Locomotion.AnimStateTransitionNode_0.Transition
  nodes [
   Result()
   "Equal (Float)"!(A:double="Get Relevant Anim Time Remaining (Open_Into)"!() -> (ReturnValue:float) {is=title K2Node_AnimGetter}, B:double=0) -> (ReturnValue:bool) {is=title K2Node_CallFunction}
  ]
 }
 Transition {
  path=AnimGraph.AnimGraphNode_StateMachine_0.Locomotion.AnimStateTransitionNode_1.Transition
  nodes [
   n1 = Get SkydivingState()
   Result(bCanEnterTransition=OR Boolean(
     A=Break Fort Anim Input Skydiving(FortAnimInput_Skydiving=@n1) -> (
      ParaGilderRootOffset=~
      ParaGliderRootRotationOverride=~
      bUseParaGlideRootModifier=~
      bIsSkydivingFromLaunchPad=~
      bIsSkydivingFromBus=~
      bIsInVortex=~
      bIsDiving=~
      bIsParachuteOpen=~
      bIsSkydiving=~
      bIsParachuteLeaning=~
      bIsSkydiveLeaning=~
      bIsLeaning=~
      bIsSkydiveDiveMode=~
      bPlaySkydiveDrift=~
      bSkydiveDriftDelayActive=~
      bSkydiveDriftAnimAllowed=~
      bIsGliderRight=~
      bIsGliderCenter=~
      bIsGliderLeft=~
      bIsGliderForward=~
      bIsGliderBack=~
      bHasDeployedAndIsStandingUpright=~
      LocalAccelForward=~
      LocalAccelRight=~
      LocalVelocityRight=~
      SkydiveAimPitch=~
      SkydiveAimYaw=~
      DeployChuteAnimRate=~
      SkydiveDriftAnimRate=~
      SkydiveDriftAnimRateCurrent=~
      SkydiveFidgetAnimRate=~
      SkydiveFidgetAnimRateCurrent=~
      SkydiveAdditiveAlpha=~
      SkydiveDriftDelay=~
     ).bIsParachuteOpen
     B=Get bForceOpen() -> (bForceOpen:bool)
    ))
  ]
 }
 Transition {
  path=AnimGraph.AnimGraphNode_StateMachine_0.Locomotion.AnimStateTransitionNode_2.Transition
  nodes [
   n1 = Get bForceOpen()
   n2 = Break Fort Anim Input Skydiving(FortAnimInput_Skydiving=Get SkydivingState()) -> (
    ParaGilderRootOffset=~
    ParaGliderRootRotationOverride=~
    bUseParaGlideRootModifier=~
    bIsSkydivingFromLaunchPad=~
    bIsSkydivingFromBus=~
    bIsInVortex=~
    bIsDiving=~
    bIsParachuteOpen=~
    bIsSkydiving=~
    bIsParachuteLeaning=~
    bIsSkydiveLeaning=~
    bIsLeaning=~
    bIsSkydiveDiveMode=~
    bPlaySkydiveDrift=~
    bSkydiveDriftDelayActive=~
    bSkydiveDriftAnimAllowed=~
    bIsGliderRight=~
    bIsGliderCenter=~
    bIsGliderLeft=~
    bIsGliderForward=~
    bIsGliderBack=~
    bHasDeployedAndIsStandingUpright=~
    LocalAccelForward=~
    LocalAccelRight=~
    LocalVelocityRight=~
    SkydiveAimPitch=~
    SkydiveAimYaw=~
    DeployChuteAnimRate=~
    SkydiveDriftAnimRate=~
    SkydiveDriftAnimRateCurrent=~
    SkydiveFidgetAnimRate=~
    SkydiveFidgetAnimRateCurrent=~
    SkydiveAdditiveAlpha=~
    SkydiveDriftDelay=~
   )
   Result(bCanEnterTransition=AND Boolean(A=NOT Boolean(A=@n2.bIsParachuteOpen), B=NOT Boolean(A=@n1)))
  ]
 }
 Transition {
  path=AnimGraph.AnimGraphNode_StateMachine_0.Locomotion.AnimStateTransitionNode_3.Transition
  nodes [
   n1 = Get bForceOpen()
   n2 = Break Fort Anim Input Skydiving(FortAnimInput_Skydiving=Get SkydivingState()) -> (
    ParaGilderRootOffset=~
    ParaGliderRootRotationOverride=~
    bUseParaGlideRootModifier=~
    bIsSkydivingFromLaunchPad=~
    bIsSkydivingFromBus=~
    bIsInVortex=~
    bIsDiving=~
    bIsParachuteOpen=~
    bIsSkydiving=~
    bIsParachuteLeaning=~
    bIsSkydiveLeaning=~
    bIsLeaning=~
    bIsSkydiveDiveMode=~
    bPlaySkydiveDrift=~
    bSkydiveDriftDelayActive=~
    bSkydiveDriftAnimAllowed=~
    bIsGliderRight=~
    bIsGliderCenter=~
    bIsGliderLeft=~
    bIsGliderForward=~
    bIsGliderBack=~
    bHasDeployedAndIsStandingUpright=~
    LocalAccelForward=~
    LocalAccelRight=~
    LocalVelocityRight=~
    SkydiveAimPitch=~
    SkydiveAimYaw=~
    DeployChuteAnimRate=~
    SkydiveDriftAnimRate=~
    SkydiveDriftAnimRateCurrent=~
    SkydiveFidgetAnimRate=~
    SkydiveFidgetAnimRateCurrent=~
    SkydiveAdditiveAlpha=~
    SkydiveDriftDelay=~
   )
   Result(bCanEnterTransition=AND Boolean(A=NOT Boolean(A=@n2.bIsParachuteOpen), B=NOT Boolean(A=@n1)))
  ]
 }
 Fox {
  entry=Fox2
  path=AnimGraph.AnimGraphNode_StateMachine_1.Fox
  ids=id
  nodes [
   AnimStateNode_0 = Fox2() {is=title AnimStateNode}
   AnimStateNode_1 = Reload() {is=title AnimStateNode}
   AnimStateTransitionNode_2 = AnimStateTransitionNode() {crossfadeDuration=0}
  ]
  links [{dst=Reload, rule=AnimGraph.AnimGraphNode_StateMachine_1.Fox.AnimStateTransitionNode_2.Transition, src=Fox2, via=AnimStateTransitionNode_2}]
 }
 Fox2 {
  path=AnimGraph.AnimGraphNode_StateMachine_1.Fox.AnimStateNode_0.Fox2
  nodes [Output Animation Pose(Result=Sequence Player!(Sequence:Anim Sequence Base Object Reference=Get Fox2_Anim() -> (Fox2_Anim:Anim Sequence Object Reference), PlayRate:float=1.4) -> (Pose) {bLoopAnimation=false, playRate=1.25})]
 }
 Reload {
  path=AnimGraph.AnimGraphNode_StateMachine_1.Fox.AnimStateNode_1.Reload
  nodes [
   n1 = "Sequence Player (additive) 'BriteBomberDeluxe_Fox2_Reload_GLIDER'"() {
    bLoopAnimation=false
    playRate=1.5
    sequence=/BRCosmetics/Animation/Game/MainPlayer/Skydive/ParaGlide/BriteBomberDeluxe/BriteBomberDeluxe_Fox2_Reload_GLIDER
   }
   Output Animation Pose(Result=@n1)
  ]
 }
 Transition {
  path=AnimGraph.AnimGraphNode_StateMachine_1.Fox.AnimStateTransitionNode_2.Transition
  nodes [Result(bCanEnterTransition=Get ReloadGlitterRockets() -> (ReloadGlitterRockets:bool))]
 }
 FoxWarning {
  entry=Fox2_Warning
  path=AnimGraph.AnimGraphNode_StateMachine_2.FoxWarning
  ids=id
  nodes [AnimStateNode_0 = Fox2_Warning() {is=title AnimStateNode}]
 }
 Fox2_Warning {
  path=AnimGraph.AnimGraphNode_StateMachine_2.FoxWarning.AnimStateNode_0.Fox2_Warning
  nodes [
   n1 = "Sequence Player (additive) 'BriteBomberDeluxe_Fox2_LowAltitude_GLIDER'"() {
    bLoopAnimation=false
    sequence=/BRCosmetics/Animation/Game/MainPlayer/Skydive/ParaGlide/BriteBomberDeluxe/BriteBomberDeluxe_Fox2_LowAltitude_GLIDER
   }
   Output Animation Pose(Result=@n1)
  ]
 }
 FoxDryFire {entry=DryFire, path=AnimGraph.AnimGraphNode_StateMachine_3.FoxDryFire, ids=id, nodes [AnimStateNode_0 = DryFire() {is=title AnimStateNode}]}
 DryFire {
  path=AnimGraph.AnimGraphNode_StateMachine_3.FoxDryFire.AnimStateNode_0.DryFire
  nodes [
   n1 = "Sequence Player (additive) 'BriteBomberDeluxe_Fox2_DryFire_GLIDER'"() {
    bLoopAnimation=false
    sequence=/BRCosmetics/Animation/Game/MainPlayer/Skydive/ParaGlide/BriteBomberDeluxe/BriteBomberDeluxe_Fox2_DryFire_GLIDER
   }
   Output Animation Pose(Result=@n1)
  ]
 }
 CheckCosmetics {nodes [
   Inputs()
   > n1 = Branch(Condition=Get ReadyPlayer() -> (ReadyPlayer:bool))
   > Set ClassicStyleOnly(ClassicStyleOnly:bool=true) -> (Output_Get:bool)
   > Set BB_AnyChar(BB_AnyChar:bool=false) -> (Output_Get:bool)
   > n2 = Sequence!(execute) -> (then_0, then_1, then_2:Exec, then_3:Exec)
   > n3 = Cast To FortPlayerPawnAthena(Object=Get OwnerFortPawn() -> (OwnerFortPawn:Fort Player Pawn Object Reference))
   > Set Pickaxe(Pickaxe:Athena Pickaxe Item Definition Object Reference=GetItemDefinitionTypeFromLoadoutWithActor!(
     Actor:Actor Object Reference=@n3
     LoadoutSlot:Gameplay Tag Structure=(TagName="LoadoutSlot.Pickaxe")
     CosmeticType:Athena Cosmetic Item Definition Class Reference=/Script/FortniteGame.AthenaPickaxeItemDefinition
    ) -> (OutCosmeticItemDef:Athena Pickaxe Item Definition Object Reference)) -> (Output_Get:Athena Pickaxe Item Definition Object Reference)
   > Set Character(Character:Athena Character Item Definition Object Reference=GetItemDefinitionTypeFromLoadoutWithActor(Actor=@n3, LoadoutSlot=(TagName="LoadoutSlot.Character"), CosmeticType=/Script/FortniteGame.AthenaCharacterItemDefinition) -> (OutCosmeticItemDef:Athena Character Item Definition Object Reference)) -> (Output_Get:Athena Character Item Definition Object Reference)
   > Set Backpack(Backpack:Athena Backpack Item Definition Object Reference=GetItemDefinitionTypeFromLoadoutWithActor(Actor=@n3, LoadoutSlot=(TagName="LoadoutSlot.Backpack"), CosmeticType=/Script/FortniteGame.AthenaBackpackItemDefinition) -> (OutCosmeticItemDef:Athena Backpack Item Definition Object Reference)) -> (Output_Get:Athena Backpack Item Definition Object Reference)
   > n4 = For Each Loop with Break!(
    Exec:Exec
    Array:Array of Strings=Select(
     Option 0:Array of Strings=Make Array!("[0]":String=CID_044, "[1]":String=CID_069, "[2]":String=CID_112, "[3]":String=CID_139, "[4]":String=CID_460, "[5]":String=CID_341) -> (Array:Array of Strings)
     Option 1:Array of Strings=Make Array!("[0]"=CID_044, "[1]"=CID_112, "[2]"=CID_069) -> (Array)
     Index=Get ClassicStyleOnly() -> (ClassicStyleOnly:bool)
    ) -> (ReturnValue:Array of Strings)
    Break:Exec=@n7.then
   ) -> (LoopBody:Exec, Array Element:String, Array Index:int, Completed:Exec)
   n5 = Contains(
    SearchIn=GetDisplayName(Object=Get Character() -> (Character:Athena Character Item Definition Object Reference))
    Substring=@n4
    bSearchFromEnd=true
   )
   > n6 = Set BB_Pilot(BB_Pilot:bool=@n5) -> (Output_Get:bool)
   > n7 = Branch(Condition=@n6)
   n8 = For Each Loop with Break(
    Exec=@n2.then_1
    Array=Select(Option 0=Make Array!("[0]"=ID_026, "[1]"=ID_162) -> (Array), Option 1=Make Array("[0]"=ID_026, "[1]"=ID_026), Index=Get ClassicStyleOnly())
    Break=@n11.then
   )
   n9 = Get Pickaxe() -> (Pickaxe:Athena Pickaxe Item Definition Object Reference)
   > n10 = Set BB_Pickaxe(BB_Pickaxe:bool=Contains(SearchIn=GetDisplayName(Object=@n9), Substring=@n8, bSearchFromEnd=true)) -> (Output_Get:bool)
   > n11 = Branch(Condition=@n10)
   n12 = For Each Loop with Break(
    Exec=@n2.then_2
    Array=Select(
     Option 0=Make Array!("[0]"=BID_026, "[1]"=BID_211, "[2]":String=BID_299) -> (Array)
     Option 1=Make Array!("[0]"=BID_026, "[1]"=BID_026) -> (Array)
     Index=Get ClassicStyleOnly()
    )
    Break=@n15.then
   )
   n13 = Get Backpack() -> (Backpack:Athena Backpack Item Definition Object Reference)
   > n14 = Set BB_Backpack(BB_Backpack:bool=Contains(SearchIn=GetDisplayName(Object=@n13), Substring=@n12, bSearchFromEnd=true)) -> (Output_Get:bool)
   > n15 = Branch(Condition=@n14)
   n16 = GetDisplayName(Object=Get Character())
   n17 = Get BB_AnyChar() -> (BB_AnyChar:bool)
   n18 = Get BB_Pickaxe() -> (BB_Pickaxe:bool)
   n19 = Get BB_Backpack() -> (BB_Backpack:bool)
   n20 = Get BB_Pickaxe()
   n21 = Get BB_Pilot() -> (BB_Pilot:bool)
   n22 = Get BB_Backpack()
   Set WeaponsFree(
    execute=@n2.then_3
    WeaponsFree:bool=OR Boolean!(
     A=AND Boolean!(A=@n20, B=@n21, C:bool=@n22) -> (ReturnValue)
     B=Contains(SearchIn=@n16, Substring=CID_139, bSearchFromEnd=true)
     C:bool=AND Boolean(A=@n18, B=@n17, C=@n19)
    ) -> (ReturnValue)
   ) -> (Output_Get:bool)
   Set WeaponsFree(execute=@n1.else, WeaponsFree=false)
   Make Array!(
    "[0]"=SciPop
    "[1]"=Brite
    "[2]":String=Disco
    "[3]":String=PinkBear
    "[4]":String=CID_084
    "[5]":String=RockerPunk
    "[6]":String=Bunny
    "[7]":String=CID_108
    "[8]":String=F_Candy
    "[9]":String=M_FighterPilot
    "[10]":String=CID_142
    "[11]":String=F_Birthday2018
    "[12]":String=Exercise
    "[13]":String=Dieselpunk
    "[14]":String=FortniteDJ
    "[15]":String=F_NeonCat
    "[16]":String=M_Squishy
    "[17]":String=M_BlueAce
    "[18]":String=M_IceCream
   ) -> (Array)
   Make Array!("[0]"=SciPop, "[1]"=PinkBear, "[2]"=M_Brite, "[3]"=FighterPilot, "[4]"=F_BriteBomberSummer, "[5]"=SkullBrite) -> (Array)
   Outputs!()
   Set WeaponsFree(WeaponsFree=@n5)
  ]}
 EventGraph {nodes [
   n1 = Event BlueprintUpdateAnimation()
   > Set Time(Time:double=@n1.DeltaTimeX) -> (Output_Get:double)
   > Set GlitterBomb(GlitterBomb:bool=true) -> (Output_Get:bool)
   n2 = Get OwnerGlider()
   > n3 = Branch(Condition=NOT Boolean(A=Get IsCosmeticPreview(self=@n2)))
   > n4 = Sequence!(execute) -> (then_0, then_1)
   > Do Once()
   > Adventure!(execute:Exec) -> (then:Exec) {is=title K2Node_MacroInstance}
   > Check Cosmetics!(execute:Exec) {is=title K2Node_MacroInstance}
   n5 = Get OwnerFortPawn()
   n6 = Set RawHeight(execute=@n4.then_1, RawHeight:double=BreakVector(InVec=Get Actor Location(self=@n5)).Z) -> (Output_Get:double)
   > Set HardDeck(HardDeck:bool="float <= float"!(A:double=@n6, B:double=2500) -> (ReturnValue:bool) {is=title K2Node_CallFunction}) -> (Output_Get:bool)
   n7 = GetFortPlayerControllerFromActor!(Actor:Actor Object Reference=Get OwnerFortPawn()) -> (ReturnValue:Fort Player Controller Object Reference)
   n8 = IsInputKeyDown!(self:Player Controller Object Reference=@n7, Key:Key Structure=LeftMouseButton) -> (ReturnValue:bool)
   n9 = IsInputKeyDown(self=@n7, Key=Gamepad_RightTrigger)
   > Set FireButton(FireButton:bool=AND Boolean!(A=OR Boolean!(A=@n9, B=@n8) -> (ReturnValue), B=Get GlitterBomb() -> (GlitterBomb:bool)) -> (ReturnValue)) -> (Output_Get:bool)
   > Set Fox2_Playing(Fox2_Playing:bool="Equal (Float)"(A=GetCurveValue(CurveName=Fox2Playing), B=1)) -> (Output_Get:bool)
   > Set Fox2_WarningPlaying(Fox2_WarningPlaying:bool="Equal (Float)"(A=GetCurveValue(CurveName=Fox2Warning), B=1)) -> (Output_Get:bool)
   > n10 = Cast To FortPlayerController(Object=GetControllerFromActor!(Actor:Actor Object Reference=Get OwnerFortPawn()) -> (ReturnValue:Controller Object Reference))
   n11 = GetMiniMapComponent!(self:Fort Player Controller Object Reference=@n10) -> (ReturnValue:Fort Controller Component Minimap Object Reference)
   > Is Valid(InputObject=@n11)
   > n12 = Set IsMapOpen(IsMapOpen:bool=IsFullscreenMapVisible!(self:Fort Controller Component Minimap Object Reference=@n11) -> (ReturnValue:bool)) -> (Output_Get:bool)
   n13 = Get IsCosmeticPreview(self=Get OwnerGlider())
   n14 = Get IsMapOpen() -> (IsMapOpen:bool)
   n15 = Get Fox2_Playing() -> (Fox2_Playing:bool)
   n16 = Branch(
    execute=[@n12.then, @n3.else]
    Condition=AND Boolean!(
     A=Get FireButton() -> (FireButton:bool)
     B=NOT Boolean(A=@n15)
     C:bool=NOT Boolean(A=@n13)
     D:bool=NOT Boolean(A=@n14)
     E:bool=Get WeaponsFree() -> (WeaponsFree:bool)
    ) -> (ReturnValue)
   )
   > n17 = Branch(Condition=NOT Boolean(A=Get HardDeck() -> (HardDeck:bool)))
   > Set Fox2_Away(Fox2_Away:bool=true) -> (Output_Get:bool)
   > n18 = Sequence!(execute) -> (then_0, then_1, then_2:Exec)
   > Set ReloadGlitterRockets(ReloadGlitterRockets:bool=false) -> (Output_Get:bool)
   n19 = Branch(
    execute=[@n18.then_1, @n28.then, @n32.then]
    Condition="float >= float"(A=GetCurveValue(CurveName=Prop_Start), B=1) {is=title K2Node_CallFunction}
   )
   > Set Tassle_Speed(Tassle_Speed:double=0) -> (Output_Get:double)
   > Set Jitter_Alpha(Jitter_Alpha:double=0) -> (Output_Get:double)
   n20 = Break Fort Anim Input Skydiving(FortAnimInput_Skydiving=Get SkydivingState()) -> (
    ParaGilderRootOffset=~
    ParaGliderRootRotationOverride=~
    bUseParaGlideRootModifier=~
    bIsSkydivingFromLaunchPad=~
    bIsSkydivingFromBus=~
    bIsInVortex=~
    bIsDiving=~
    bIsParachuteOpen=~
    bIsSkydiving=~
    bIsParachuteLeaning=~
    bIsSkydiveLeaning=~
    bIsLeaning=~
    bIsSkydiveDiveMode=~
    bPlaySkydiveDrift=~
    bSkydiveDriftDelayActive=~
    bSkydiveDriftAnimAllowed=~
    bIsGliderRight=~
    bIsGliderCenter=~
    bIsGliderLeft=~
    bIsGliderForward=~
    bIsGliderBack=~
    bHasDeployedAndIsStandingUpright=~
    LocalAccelForward=~
    LocalAccelRight=~
    LocalVelocityRight=~
    SkydiveAimPitch=~
    SkydiveAimYaw=~
    DeployChuteAnimRate=~
    SkydiveDriftAnimRate=~
    SkydiveDriftAnimRateCurrent=~
    SkydiveFidgetAnimRate=~
    SkydiveFidgetAnimRateCurrent=~
    SkydiveAdditiveAlpha=~
    SkydiveDriftDelay=~
   )
   n21 = Branch(execute=@n19.else, Condition="float > float"(A=@n20.LocalAccelForward, B=500) {is=title K2Node_CallFunction})
   > n22 = Set Tassle_Speed(Tassle_Speed=FInterpTo(Current=Get Tassle_Speed(), Target=3, DeltaTime=Get Time() -> (Time:double), InterpSpeed=5))
   > Set Jitter_Alpha(Jitter_Alpha=float - float(A=@n22, B=2) {is=title K2Node_CallFunction})
   n23 = Branch(execute=@n21.else, Condition="Nearly Equal (Float)"(A=@n20.LocalAccelForward, B=0, ErrorTolerance=10))
   > n24 = Set Tassle_Speed(Tassle_Speed=FInterpTo(Current=Get Tassle_Speed(), Target=1, DeltaTime=Get Time(), InterpSpeed=5))
   > Set Jitter_Alpha(Jitter_Alpha=float - float(A=@n24, B=0.75))
   n25 = Branch(
    execute=@n23.else
    Condition="float < float"!(A:double=@n20.LocalAccelForward, B:double=-500) -> (ReturnValue:bool) {is=title K2Node_CallFunction}
   )
   > n26 = Set Tassle_Speed(Tassle_Speed=FInterpTo(Current=Get Tassle_Speed(), Target=1, DeltaTime=Get Time(), InterpSpeed=5))
   > Set Jitter_Alpha(Jitter_Alpha=float - float(A=@n26, B=0.75))
   n27 = Set Tassle_Speed(execute=@n25.else, Tassle_Speed=FInterpTo(Current=Get Tassle_Speed(), Target=1, DeltaTime=Get Time(), InterpSpeed=5))
   > Set Jitter_Alpha(Jitter_Alpha=float - float(A=@n27, B=0.75))
   Delay(execute=@n18.then_2, Duration=4)
   > Set ReloadGlitterRockets(ReloadGlitterRockets=true)
   n28 = Set Fox2_Warning(execute=@n17.else, Fox2_Warning:bool=true) -> (Output_Get:bool)
   n29 = Get IsMapOpen()
   n30 = Get Fox2_WarningPlaying() -> (Fox2_WarningPlaying:bool)
   n31 = Get IsCosmeticPreview(self=Get OwnerGlider())
   n32 = Set Fox2_DryFire(
    execute=@n16.else
    Fox2_DryFire:bool=AND Boolean!(A=Get FireButton(), B=Get Fox2_Playing(), C=NOT Boolean(A=@n30), D=NOT Boolean(A=@n31), E=NOT Boolean(A=@n29), F:bool=Get WeaponsFree()) -> (ReturnValue)
   ) -> (Output_Get:bool)
   AnimNotify_TurnONEngines() {is=title K2Node_Event}
   > SetGliderFullyDeployed!(execute, self:Fort Glider Anim Instance Object Reference, bIsFullyDeployed:bool=true) -> (then)
   AnimNotify_FE_IdleStart() {is=title K2Node_Event}
   > Call FE Idle Start!(execute:Exec, self:Brite Bomber Deluxe Glider Anim BP Object Reference) -> (then:Exec) {is=title K2Node_CallDelegate}
   AnimNotify_Fox2_Off() {is=title K2Node_Event}
   > Set Fox2_Warning(Fox2_Warning=false)
   > Set Fox2_Away(Fox2_Away=false)
   > Set Fox2_Anim(Fox2_Anim:Anim Sequence Object Reference=Select!(
     Option 0:Anim Sequence Object Reference=/BRCosmetics/Animation/Game/MainPlayer/Skydive/ParaGlide/BriteBomberDeluxe/BriteBomberDeluxe_Fox2_A_GLIDER
     Option 1:Anim Sequence Object Reference=/BRCosmetics/Animation/Game/MainPlayer/Skydive/ParaGlide/BriteBomberDeluxe/BriteBomberDeluxe_Fox2_B_GLIDER
     Option 2:Anim Sequence Object Reference=/BRCosmetics/Animation/Game/MainPlayer/Skydive/ParaGlide/BriteBomberDeluxe/BriteBomberDeluxe_Fox2_C_GLIDER
     Option 3:Anim Sequence Object Reference=/BRCosmetics/Animation/Game/MainPlayer/Skydive/ParaGlide/BriteBomberDeluxe/BriteBomberDeluxe_Fox2_D_GLIDER
     Option 4:Anim Sequence Object Reference=/BRCosmetics/Animation/Game/MainPlayer/Skydive/ParaGlide/BriteBomberDeluxe/BriteBomberDeluxe_Fox2_E_GLIDER
     Index:int=RandomIntegerInRange!(Min:int=0, Max:int=4) -> (ReturnValue:int=0)
    ) -> (ReturnValue:Anim Sequence Object Reference)) -> (Output_Get:Anim Sequence Object Reference)
   AnimNotify_Fox2_DryFireOff() {is=title K2Node_Event}
   > Set Fox2_DryFire(Fox2_DryFire=false)
   GetPlayerController!(PlayerIndex:int=0) -> (ReturnValue:Player Controller Object Reference)
   "float <= float"(A=@n6, B=5000)
   IsInputKeyDown(Key=Touch1)
  ]}
 FE_IdleStart {nodes [FE_IdleStart() {is=title K2Node_FunctionEntry}]}
]
variables [
 Backpack {type=AthenaBackpackItemDefinition, default=null}
 BB_AnyChar {type=boolean}
 BB_Backpack {type=boolean}
 BB_Pickaxe {type=boolean}
 BB_Pilot {type=boolean}
 Character {type=AthenaCharacterItemDefinition, default=null}
 ClassicStyleOnly {type=boolean}
 FE_IdleStart {default=()}
 FireButton {type=boolean}
 Fox2_Anim {type=AnimSequence, default=&BriteBomberDeluxe_Fox2_A_GLIDER}
 Fox2_Away {type=boolean}
 Fox2_DryFire {type=boolean}
 Fox2_Playing {type=boolean}
 Fox2_Warning {type=boolean}
 Fox2_WarningPlaying {type=boolean}
 GlitterBomb {type=boolean}
 HardDeck {type=boolean}
 IsMapOpen {type=boolean}
 Jitter_Alpha {type=number, default=1}
 Pickaxe {type=AthenaPickaxeItemDefinition, default=null}
 Player11 {type=DateTime, default=1974.12.11-11.00.00}
 Player12 {type=DateTime, default=2022.12.11-11.00.00}
 PlayerID4 {type=DateTime, default=2022.07.04-11.00.00}
 RawHeight {type=number}
 ReadyPlayer {type=boolean}
 ReloadGlitterRockets {type=boolean}
 Tassle_Speed {type=number, default=1}
 Time {type=number}
 WeaponsFree {type=boolean}
]
