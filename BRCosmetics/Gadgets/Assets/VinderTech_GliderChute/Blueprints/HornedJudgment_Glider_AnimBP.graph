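# Animation Blueprint for the Horned Judgment glider, animating deploy, glide, lean, wing flaps and roars from skydiving state, with front-end cosmetic preview alternatives.
# A Locomotion state machine feeds stacked additive lean, turn, roar, flap and mouth layers.
# Update macros use anim curves, direction changes and randomness to gate flaps and roars, with notifies resetting them and firing roar events.
# The equipped glider's name selects the variant index and roar.
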AnimBlueprint /BRCosmetics/Gadgets/Assets/VinderTech_GliderChute/Blueprints/HornedJudgment_Glider_AnimBP
target_skeleton=/BRCosmetics/Gadgets/Assets/VinderTech_GliderChute/Glider_HornedJudgment_Midgard/Meshes/HornedJudgment_Glider_Skeleton
graphs [
 AnimGraph {nodes [
   n1 = Break Fort Anim Input Skydiving!(FortAnimInput_Skydiving:"Fort Anim Input Skydiving Structure (by ref)"=Get SkydivingState() -> (SkydivingState:Fort Anim Input Skydiving Structure)) -> (
    ParaGilderRootOffset:Vector=~
    ParaGliderRootRotationOverride:Rotator=~
    bUseParaGlideRootModifier:bool=~
    bIsSkydivingFromLaunchPad:bool=~
    bIsSkydivingFromBus:bool=~
    bIsInVortex:bool=~
    bIsDiving:bool=~
    bIsParachuteOpen:bool=~
    bIsSkydiving:bool=~
    bIsParachuteLeaning:bool=~
    bIsSkydiveLeaning:bool=~
    bIsLeaning:bool=~
    bIsSkydiveDiveMode:bool=~
    bPlaySkydiveDrift:bool=~
    bSkydiveDriftDelayActive:bool=~
    bSkydiveDriftAnimAllowed:bool=~
    bIsGliderRight:bool=~
    bIsGliderCenter:bool=~
    bIsGliderLeft:bool=~
    bIsGliderForward:bool=~
    bIsGliderBack:bool=~
    bHasDeployedAndIsStandingUpright:bool=~
    LocalAccelForward:float=~
    LocalAccelRight:float=~
    LocalVelocityRight:float=~
    SkydiveAimPitch:float=~
    SkydiveAimYaw:float=~
    DeployChuteAnimRate:float=~
    SkydiveDriftAnimRate:float=~
    SkydiveDriftAnimRateCurrent:float=~
    SkydiveFidgetAnimRate:float=~
    SkydiveFidgetAnimRateCurrent:float=~
    SkydiveAdditiveAlpha:float=~
    SkydiveDriftDelay:float=~
    SkydiveDriftAnim:int
    SkydiveDriftAnimMax:int
    LaunchpadAnim:int
    LocalAccelDir:ESkydivingDirection Enum
    DirectionLast:ESkydivingDirection Enum
    PoseAdditive:Anim Sequence Object Reference
   ) {is=title K2Node_BreakStruct}
   n2 = Blendspace Player 'BS_HornedJudgment_Lean_GLIDER'(X=@n1.LocalAccelRight, Y=@n1.LocalAccelForward) {
    bResetPlayTimeWhenBlendSpaceChanges=false
    blendSpace=/BRCosmetics/Animation/Game/MainPlayer/Skydive/ParaGlide/HornedJudgment/BS_HornedJudgment_Lean_GLIDER
    groupName=Gliding
    groupRole=AlwaysFollower
    method=SyncGroup
   }
   AnimGraphNode_StateMachine_0 = Locomotion() {is=title AnimGraphNode_StateMachine}
   AnimGraphNode_StateMachine_1 = Variant_Roar() {is=title AnimGraphNode_StateMachine}
   AnimGraphNode_StateMachine_3 = Original_Roar() {is=title AnimGraphNode_StateMachine}
   n3 = Apply Additive!(
    Base=Apply Additive!(
     Base=Apply Additive(Base=@AnimGraphNode_StateMachine_0, Additive=@n2, AlphaCurveName=ParaGlide_Lean) {alphaInputType=Curve, alphaBoolBlend {blendOption=Linear}, alphaScaleBiasClamp {bias=1, scale=-1}}
     Additive=Use cached pose 'Turn'()
     AlphaCurveName:Name=ParaGlide_Turn
    ) -> (Pose) {alphaInputType=Curve, alphaBoolBlend {blendOption=Linear}, alphaScaleBiasClamp {bias=1, scale=-1}}
    Additive=Blend Poses by bool(
     BlendPose_0=@AnimGraphNode_StateMachine_1
     BlendPose_1=@AnimGraphNode_StateMachine_3
     BlendTime_0=0
     BlendTime_1=0
     bActiveValue=Get UseVariantRoar() -> (UseVariantRoar:bool)
    ) {blendType=Linear}
    bAlphaBoolEnabled:bool=Get canRoar() -> (canRoar:bool)
   ) -> (Pose) {alphaInputType=Bool, alphaBoolBlend {blendInTime=0.2, blendOption=Linear}, alphaScaleBiasClamp {bias=1, scale=-1}}
   AnimGraphNode_StateMachine_2 = Flap() {is=title AnimGraphNode_StateMachine}
   n4 = Apply Additive!(
    Base=Apply Additive!(Base=@n3, Additive=@AnimGraphNode_StateMachine_2, Alpha:float=Get Flap_Alpha() -> (Flap_Alpha:double)) -> (Pose) {alphaBoolBlend {blendInTime=0.2, blendOption=Linear}}
    Additive="Sequence Player (additive) 'HornedJudgment_Mouth_Pose_GLIDER'"() {sequence=/BRCosmetics/Animation/Game/MainPlayer/Skydive/ParaGlide/HornedJudgment/HornedJudgment_Mouth_Pose_GLIDER}
    bAlphaBoolEnabled:bool=True
   ) -> (Pose) {alphaInputType=Bool, alphaBoolBlend {blendInTime=0.2, blendOption=Linear}, alphaScaleBiasClamp {bias=1, scale=-1}}
   Output Pose(Result=Slot 'DefaultSlot'(Source=@n4)) {layerGroup=None}
   Save cached pose 'Turn'(Pose=Blendspace Player 'BS_HornedJudgment_Turn_GLIDER'(X=@n1.SkydiveAimYaw, Y=@n1.LocalAccelForward) {blendSpace=/BRCosmetics/Animation/Game/MainPlayer/Skydive/ParaGlide/HornedJudgment/BS_HornedJudgment_Turn_GLIDER}) {cacheName=Turn}
  ]}
 Locomotion {
  entry=Closed
  path=AnimGraph.AnimGraphNode_StateMachine_0.Locomotion
  ids=id
  nodes [
   AnimStateNode_0 = Idle_Open() {is=title AnimStateNode}
   AnimStateNode_1 = Open_Into() {is=title AnimStateNode}
   AnimStateNode_3 = Closed() {is=title AnimStateNode}
   AnimStateNode_4 = Lean() {is=title AnimStateNode}
   AnimStateNode_9 = Lean_Into() {is=title AnimStateNode}
   AnimStateTransitionNode_10 = AnimStateTransitionNode() {crossfadeDuration=0, priorityOrder=2}
   AnimStateTransitionNode_11 = AnimStateTransitionNode()
   AnimStateTransitionNode_12 = AnimStateTransitionNode() {priorityOrder=2}
   AnimStateTransitionNode_13 = AnimStateTransitionNode()
   AnimStateTransitionNode_14 = AnimStateTransitionNode()
   AnimStateTransitionNode_15 = AnimStateTransitionNode()
   AnimStateTransitionNode_21 = AnimStateTransitionNode() {blendMode=Linear}
   AnimStateTransitionNode_5 = AnimStateTransitionNode()
   AnimStateTransitionNode_6 = AnimStateTransitionNode() {bAutomaticRuleBasedOnSequencePlayerInState=true, blendMode=Linear, priorityOrder=2}
   AnimStateTransitionNode_7 = AnimStateTransitionNode() {bAutomaticRuleBasedOnSequencePlayerInState=true, blendMode=Linear, crossfadeDuration=0}
   AnimStateTransitionNode_8 = AnimStateTransitionNode() {crossfadeDuration=0}
   AnimStateTransitionNode_9 = AnimStateTransitionNode() {crossfadeDuration=0}
  ]
  links [
   {
    dst=Open_Into
    rule=AnimGraph.AnimGraphNode_StateMachine_0.Locomotion.AnimStateTransitionNode_8.Transition
    src=Closed
    via=AnimStateTransitionNode_8
   }
   {
    dst=Closed
    rule=AnimGraph.AnimGraphNode_StateMachine_0.Locomotion.AnimStateTransitionNode_9.Transition
    src=Idle_Open
    via=AnimStateTransitionNode_9
   }
   {
    dst=Lean
    rule=AnimGraph.AnimGraphNode_StateMachine_0.Locomotion.AnimStateTransitionNode_15.Transition
    src=Idle_Open
    via=AnimStateTransitionNode_15
   }
   {
    dst=Lean_Into
    rule=AnimGraph.AnimGraphNode_StateMachine_0.Locomotion.AnimStateTransitionNode_14.Transition
    src=Idle_Open
    via=AnimStateTransitionNode_14
   }
   {
    dst=Closed
    rule=AnimGraph.AnimGraphNode_StateMachine_0.Locomotion.AnimStateTransitionNode_5.Transition
    src=Lean
    via=AnimStateTransitionNode_5
   }
   {
    dst=Idle_Open
    rule=AnimGraph.AnimGraphNode_StateMachine_0.Locomotion.AnimStateTransitionNode_13.Transition
    src=Lean
    via=AnimStateTransitionNode_13
   }
   {
    dst=Closed
    rule=AnimGraph.AnimGraphNode_StateMachine_0.Locomotion.AnimStateTransitionNode_11.Transition
    src=Lean_Into
    via=AnimStateTransitionNode_11
   }
   {
    dst=Idle_Open
    rule=AnimGraph.AnimGraphNode_StateMachine_0.Locomotion.AnimStateTransitionNode_21.Transition
    src=Lean_Into
    via=AnimStateTransitionNode_21
   }
   {
    dst=Lean
    rule=AnimGraph.AnimGraphNode_StateMachine_0.Locomotion.AnimStateTransitionNode_6.Transition
    src=Lean_Into
    via=AnimStateTransitionNode_6
   }
   {
    dst=Closed
    rule=AnimGraph.AnimGraphNode_StateMachine_0.Locomotion.AnimStateTransitionNode_10.Transition
    src=Open_Into
    via=AnimStateTransitionNode_10
   }
   {
    dst=Idle_Open
    rule=AnimGraph.AnimGraphNode_StateMachine_0.Locomotion.AnimStateTransitionNode_7.Transition
    src=Open_Into
    via=AnimStateTransitionNode_7
   }
   {
    dst=Lean_Into
    rule=AnimGraph.AnimGraphNode_StateMachine_0.Locomotion.AnimStateTransitionNode_12.Transition
    src=Open_Into
    via=AnimStateTransitionNode_12
   }
  ]
 }
 Idle_Open {
  path=AnimGraph.AnimGraphNode_StateMachine_0.Locomotion.AnimStateNode_0.Idle_Open
  nodes [
   n1 = Get IsCosmeticPreview!(self:Fort Player Parachute Object Reference=Get OwnerGlider() -> (OwnerGlider:Fort Player Parachute Object Reference)) -> (IsCosmeticPreview:bool)
   Apply Additive(Alpha="To Float (Boolean)"(InBool=@n1)) {alphaBoolBlend {blendOption=Linear}}
   n2 = Break Fort Anim Input Skydiving(FortAnimInput_Skydiving=Get SkydivingState()) -> (
    ParaGilderRootOffset=~
    ParaGliderRootRotationOverride=~
    bUseParaGlideRootModifier=~
    bIsSkydivingFromLaunchPad=~
    bIsSkydivingFromBus=~
    bIsInVortex=~
    bIsDiving=~
    bIsParachuteOpen=~
    bIsSkydiving=~
    bIsParachuteLeaning=~
    bIsSkydiveLeaning=~
    bIsLeaning=~
    bIsSkydiveDiveMode=~
    bPlaySkydiveDrift=~
    bSkydiveDriftDelayActive=~
    bSkydiveDriftAnimAllowed=~
    bIsGliderRight=~
    bIsGliderCenter=~
    bIsGliderLeft=~
    bIsGliderForward=~
    bIsGliderBack=~
    bHasDeployedAndIsStandingUpright=~
    LocalAccelForward=~
    LocalAccelRight=~
    LocalVelocityRight=~
    SkydiveAimPitch=~
    SkydiveAimYaw=~
    DeployChuteAnimRate=~
    SkydiveDriftAnimRate=~
    SkydiveDriftAnimRateCurrent=~
    SkydiveFidgetAnimRate=~
    SkydiveFidgetAnimRateCurrent=~
    SkydiveAdditiveAlpha=~
    SkydiveDriftDelay=~
   )
   n3 = Blendspace Player 'BS_HornedJudgment_GLIDER'(X=@n2.SkydiveAimYaw, Y=@n2.LocalAccelForward) {
    blendSpace=/BRCosmetics/Animation/Game/MainPlayer/Skydive/ParaGlide/HornedJudgment/BS_HornedJudgment_GLIDER
    groupName=Gliding
    method=SyncGroup
   }
   Output Animation Pose(Result=Blend Poses by bool(
     BlendPose_0=Sequence Player 'HornedJudgment_Glide_FrontEnd_GLIDER'() {sequence=/BRCosmetics/Animation/Game/MainPlayer/Skydive/ParaGlide/HornedJudgment/HornedJudgment_Glide_FrontEnd_GLIDER}
     BlendPose_1=@n3
     BlendTime_0=0
     BlendTime_1=0
     bActiveValue=@n1
    ) {blendType=Linear})
  ]
 }
 Open_Into {
  path=AnimGraph.AnimGraphNode_StateMachine_0.Locomotion.AnimStateNode_1.Open_Into
  nodes [
   n1 = Get IsCosmeticPreview(self=Get OwnerGlider())
   Apply Additive(Alpha="To Float (Boolean)"(InBool=@n1)) {alphaBoolBlend {blendOption=Linear}}
   n2 = Break Fort Anim Input Skydiving(FortAnimInput_Skydiving=Get SkydivingState()) -> (
    ParaGilderRootOffset=~
    ParaGliderRootRotationOverride=~
    bUseParaGlideRootModifier=~
    bIsSkydivingFromLaunchPad=~
    bIsSkydivingFromBus=~
    bIsInVortex=~
    bIsDiving=~
    bIsParachuteOpen=~
    bIsSkydiving=~
    bIsParachuteLeaning=~
    bIsSkydiveLeaning=~
    bIsLeaning=~
    bIsSkydiveDiveMode=~
    bPlaySkydiveDrift=~
    bSkydiveDriftDelayActive=~
    bSkydiveDriftAnimAllowed=~
    bIsGliderRight=~
    bIsGliderCenter=~
    bIsGliderLeft=~
    bIsGliderForward=~
    bIsGliderBack=~
    bHasDeployedAndIsStandingUpright=~
    LocalAccelForward=~
    LocalAccelRight=~
    LocalVelocityRight=~
    SkydiveAimPitch=~
    SkydiveAimYaw=~
    DeployChuteAnimRate=~
    SkydiveDriftAnimRate=~
    SkydiveDriftAnimRateCurrent=~
    SkydiveFidgetAnimRate=~
    SkydiveFidgetAnimRateCurrent=~
    SkydiveAdditiveAlpha=~
    SkydiveDriftDelay=~
   )
   n3 = Blendspace Player 'BS_HornedJudgment_Into_Add_GLIDER'(X=@n2.SkydiveAimYaw, Y=@n2.LocalAccelForward) {blendSpace=/BRCosmetics/Animation/Game/MainPlayer/Skydive/ParaGlide/HornedJudgment/BS_HornedJudgment_Into_Add_GLIDER}
   n4 = Blendspace Player 'BS_HornedJudgment_Into_GLIDER'(X=@n2.SkydiveAimYaw, Y=@n2.LocalAccelForward) {
    bLoop=false
    bResetPlayTimeWhenBlendSpaceChanges=false
    blendSpace=/BRCosmetics/Animation/Game/MainPlayer/Skydive/ParaGlide/HornedJudgment/BS_HornedJudgment_Into_GLIDER
   }
   n5 = Sequence Player 'HornedJudgment_Into_FrontEnd_GLIDER'() {bLoopAnimation=false, sequence=/BRCosmetics/Animation/Game/MainPlayer/Skydive/ParaGlide/HornedJudgment/HornedJudgment_Into_FrontEnd_GLIDER}
   Output Animation Pose(Result=Blend Poses by bool(
     BlendPose_0=@n5
     BlendPose_1=Apply Additive!(Base=@n4, Additive=@n3, AlphaCurveName:Name=Into_FwdBack) -> (Pose) {alphaInputType=Curve, alphaBoolBlend {blendOption=Linear}}
     BlendTime_0=0
     BlendTime_1=0
     bActiveValue=@n1
    ) {blendType=Linear})
  ]
 }
 Closed {
  path=AnimGraph.AnimGraphNode_StateMachine_0.Locomotion.AnimStateNode_3.Closed
  nodes [
   n1 = Get OwnerGlider()
   Output Animation Pose(Result=Blend Poses by bool(
     BlendPose_0=Sequence Player 'HornedJudgment_Closed_FrontEnd_GLIDER'() {sequence=/BRCosmetics/Animation/Game/MainPlayer/Skydive/ParaGlide/HornedJudgment/HornedJudgment_Closed_FrontEnd_GLIDER}
     BlendPose_1=Sequence Player 'HornedJudgment_Closed_GLIDER'() {sequence=/BRCosmetics/Animation/Game/MainPlayer/Skydive/ParaGlide/HornedJudgment/HornedJudgment_Closed_GLIDER}
     bActiveValue=Get IsCosmeticPreview(self=@n1)
    ) {blendType=Linear})
  ]
 }
 Lean {
  path=AnimGraph.AnimGraphNode_StateMachine_0.Locomotion.AnimStateNode_4.Lean
  nodes [
   n1 = Break Fort Anim Input Skydiving(FortAnimInput_Skydiving=Get SkydivingState()) -> (
    ParaGilderRootOffset=~
    ParaGliderRootRotationOverride=~
    bUseParaGlideRootModifier=~
    bIsSkydivingFromLaunchPad=~
    bIsSkydivingFromBus=~
    bIsInVortex=~
    bIsDiving=~
    bIsParachuteOpen=~
    bIsSkydiving=~
    bIsParachuteLeaning=~
    bIsSkydiveLeaning=~
    bIsLeaning=~
    bIsSkydiveDiveMode=~
    bPlaySkydiveDrift=~
    bSkydiveDriftDelayActive=~
    bSkydiveDriftAnimAllowed=~
    bIsGliderRight=~
    bIsGliderCenter=~
    bIsGliderLeft=~
    bIsGliderForward=~
    bIsGliderBack=~
    bHasDeployedAndIsStandingUpright=~
    LocalAccelForward=~
    LocalAccelRight=~
    LocalVelocityRight=~
    SkydiveAimPitch=~
    SkydiveAimYaw=~
    DeployChuteAnimRate=~
    SkydiveDriftAnimRate=~
    SkydiveDriftAnimRateCurrent=~
    SkydiveFidgetAnimRate=~
    SkydiveFidgetAnimRateCurrent=~
    SkydiveAdditiveAlpha=~
    SkydiveDriftDelay=~
   )
   n2 = Blendspace Player 'BS_HornedJudgment_GLIDER'(X=@n1.SkydiveAimYaw, Y=@n1.LocalAccelForward) {
    blendSpace=&BS_HornedJudgment_GLIDER
    groupName=Gliding
    method=SyncGroup
   }
   Output Animation Pose(Result=@n2)
  ]
 }
 Lean_Into {
  path=AnimGraph.AnimGraphNode_StateMachine_0.Locomotion.AnimStateNode_9.Lean_Into
  nodes [
   n1 = Break Fort Anim Input Skydiving(FortAnimInput_Skydiving=Get SkydivingState()) -> (
    ParaGilderRootOffset=~
    ParaGliderRootRotationOverride=~
    bUseParaGlideRootModifier=~
    bIsSkydivingFromLaunchPad=~
    bIsSkydivingFromBus=~
    bIsInVortex=~
    bIsDiving=~
    bIsParachuteOpen=~
    bIsSkydiving=~
    bIsParachuteLeaning=~
    bIsSkydiveLeaning=~
    bIsLeaning=~
    bIsSkydiveDiveMode=~
    bPlaySkydiveDrift=~
    bSkydiveDriftDelayActive=~
    bSkydiveDriftAnimAllowed=~
    bIsGliderRight=~
    bIsGliderCenter=~
    bIsGliderLeft=~
    bIsGliderForward=~
    bIsGliderBack=~
    bHasDeployedAndIsStandingUpright=~
    LocalAccelForward=~
    LocalAccelRight=~
    LocalVelocityRight=~
    SkydiveAimPitch=~
    SkydiveAimYaw=~
    DeployChuteAnimRate=~
    SkydiveDriftAnimRate=~
    SkydiveDriftAnimRateCurrent=~
    SkydiveFidgetAnimRate=~
    SkydiveFidgetAnimRateCurrent=~
    SkydiveAdditiveAlpha=~
    SkydiveDriftDelay=~
   )
   n2 = Blendspace Player 'BS_HornedJudgment_Lean_Transition_GLIDER'!(X=@n1.LocalAccelRight, Y=@n1.LocalAccelForward, bLoop:bool=False) -> (Pose) {
    bLoop=false
    blendSpace=/BRCosmetics/Animation/Game/MainPlayer/Skydive/ParaGlide/HornedJudgment/BS_HornedJudgment_Lean_Transition_GLIDER
    groupRole=AlwaysLeader
   }
   Output Animation Pose(Result=@n2)
  ]
 }
 Transition {
  path=AnimGraph.AnimGraphNode_StateMachine_0.Locomotion.AnimStateTransitionNode_10.Transition
  nodes [
   n1 = Get bForceOpen() -> (bForceOpen:bool)
   n2 = Break Fort Anim Input Skydiving(FortAnimInput_Skydiving=Get SkydivingState()) -> (
    ParaGilderRootOffset=~
    ParaGliderRootRotationOverride=~
    bUseParaGlideRootModifier=~
    bIsSkydivingFromLaunchPad=~
    bIsSkydivingFromBus=~
    bIsInVortex=~
    bIsDiving=~
    bIsParachuteOpen=~
    bIsSkydiving=~
    bIsParachuteLeaning=~
    bIsSkydiveLeaning=~
    bIsLeaning=~
    bIsSkydiveDiveMode=~
    bPlaySkydiveDrift=~
    bSkydiveDriftDelayActive=~
    bSkydiveDriftAnimAllowed=~
    bIsGliderRight=~
    bIsGliderCenter=~
    bIsGliderLeft=~
    bIsGliderForward=~
    bIsGliderBack=~
    bHasDeployedAndIsStandingUpright=~
    LocalAccelForward=~
    LocalAccelRight=~
    LocalVelocityRight=~
    SkydiveAimPitch=~
    SkydiveAimYaw=~
    DeployChuteAnimRate=~
    SkydiveDriftAnimRate=~
    SkydiveDriftAnimRateCurrent=~
    SkydiveFidgetAnimRate=~
    SkydiveFidgetAnimRateCurrent=~
    SkydiveAdditiveAlpha=~
    SkydiveDriftDelay=~
   )
   Result(bCanEnterTransition=AND Boolean(A=NOT Boolean(A=@n2.bIsParachuteOpen), B=NOT Boolean(A=@n1)))
  ]
 }
 Transition {
  path=AnimGraph.AnimGraphNode_StateMachine_0.Locomotion.AnimStateTransitionNode_11.Transition
  nodes [
   n1 = Get bForceOpen()
   n2 = Break Fort Anim Input Skydiving(FortAnimInput_Skydiving=Get SkydivingState()) -> (
    ParaGilderRootOffset=~
    ParaGliderRootRotationOverride=~
    bUseParaGlideRootModifier=~
    bIsSkydivingFromLaunchPad=~
    bIsSkydivingFromBus=~
    bIsInVortex=~
    bIsDiving=~
    bIsParachuteOpen=~
    bIsSkydiving=~
    bIsParachuteLeaning=~
    bIsSkydiveLeaning=~
    bIsLeaning=~
    bIsSkydiveDiveMode=~
    bPlaySkydiveDrift=~
    bSkydiveDriftDelayActive=~
    bSkydiveDriftAnimAllowed=~
    bIsGliderRight=~
    bIsGliderCenter=~
    bIsGliderLeft=~
    bIsGliderForward=~
    bIsGliderBack=~
    bHasDeployedAndIsStandingUpright=~
    LocalAccelForward=~
    LocalAccelRight=~
    LocalVelocityRight=~
    SkydiveAimPitch=~
    SkydiveAimYaw=~
    DeployChuteAnimRate=~
    SkydiveDriftAnimRate=~
    SkydiveDriftAnimRateCurrent=~
    SkydiveFidgetAnimRate=~
    SkydiveFidgetAnimRateCurrent=~
    SkydiveAdditiveAlpha=~
    SkydiveDriftDelay=~
   )
   Result(bCanEnterTransition=AND Boolean(A=NOT Boolean(A=@n2.bIsParachuteOpen), B=NOT Boolean(A=@n1)))
  ]
 }
 Transition {
  path=AnimGraph.AnimGraphNode_StateMachine_0.Locomotion.AnimStateTransitionNode_12.Transition
  nodes [
   n1 = NOT Boolean(A=Get Flapping() -> (Flapping:bool))
   n2 = Get Direction() -> (Direction:ESkydivingDirection Enum=Center)
   n3 = "Not Equal (Enum)"!(A:ESkydivingDirection Enum=@n2, B:ESkydivingDirection Enum=Center) -> (ReturnValue:bool)
   n4 = "Not Equal (Enum)"(A=@n2, B=Left)
   n5 = "Not Equal (Enum)"(A=@n2, B=Right)
   Result(bCanEnterTransition=AND Boolean(A=AND Boolean!(A=@n4, B=@n5, C:bool=@n3, D:bool=@n1) -> (ReturnValue), B=Get LeanFromInto() -> (LeanFromInto:bool)))
   Get BlockAnimRule() -> (BlockAnimRule:bool)
  ]
 }
 Transition {
  path=AnimGraph.AnimGraphNode_StateMachine_0.Locomotion.AnimStateTransitionNode_13.Transition
  nodes [Result(bCanEnterTransition="Equal (Enum)"(A:ESkydivingDirection Enum=Get Direction() -> (Direction=Center), B:ESkydivingDirection Enum=Center))]
 }
 Transition {
  path=AnimGraph.AnimGraphNode_StateMachine_0.Locomotion.AnimStateTransitionNode_14.Transition
  nodes [
   Result(bCanEnterTransition=Get useLeanTransition() -> (useLeanTransition:bool))
   Get BlockAnimRule()
  ]
 }
 Transition {
  path=AnimGraph.AnimGraphNode_StateMachine_0.Locomotion.AnimStateTransitionNode_15.Transition
  nodes [
   n1 = Get useLeanTransition()
   n2 = Get Direction() -> (Direction=Center)
   Result(bCanEnterTransition=AND Boolean!(A="Not Equal (Enum)"(A=@n2, B=Center), B=NOT Boolean(A=@n1)) -> (ReturnValue))
  ]
 }
 Transition {
  path=AnimGraph.AnimGraphNode_StateMachine_0.Locomotion.AnimStateTransitionNode_21.Transition
  nodes [
   n1 = Get Flapping()
   n2 = NOT Boolean(A=Get TransitionAnimPlaying() -> (TransitionAnimPlaying:bool))
   n3 = "Get Relevant Anim Time Remaining (Lean_Into)"!() -> (ReturnValue:float) {is=title K2Node_AnimGetter}
   n4 = "Equal (Float)"!(A:double=@n3, B:double=0) -> (ReturnValue:bool) {is=title K2Node_CallFunction}
   n5 = Get Direction() -> (Direction=Center)
   Result(bCanEnterTransition=AND Boolean!(A=OR Boolean(A=@n2, B=@n4), B="Equal (Enum)"(A=@n5, B=Center), C:bool=NOT Boolean(A=@n1)) -> (ReturnValue))
   "float <= float"!(A:double=@n3, B:double=0.15) -> (ReturnValue:bool) {is=title K2Node_CallFunction}
   NOT Boolean(A=Get useLeanTransition())
   Get BlockAnimRule()
  ]
 }
 Transition {
  path=AnimGraph.AnimGraphNode_StateMachine_0.Locomotion.AnimStateTransitionNode_5.Transition
  nodes [
   n1 = Get bForceOpen()
   n2 = Break Fort Anim Input Skydiving(FortAnimInput_Skydiving=Get SkydivingState()) -> (
    ParaGilderRootOffset=~
    ParaGliderRootRotationOverride=~
    bUseParaGlideRootModifier=~
    bIsSkydivingFromLaunchPad=~
    bIsSkydivingFromBus=~
    bIsInVortex=~
    bIsDiving=~
    bIsParachuteOpen=~
    bIsSkydiving=~
    bIsParachuteLeaning=~
    bIsSkydiveLeaning=~
    bIsLeaning=~
    bIsSkydiveDiveMode=~
    bPlaySkydiveDrift=~
    bSkydiveDriftDelayActive=~
    bSkydiveDriftAnimAllowed=~
    bIsGliderRight=~
    bIsGliderCenter=~
    bIsGliderLeft=~
    bIsGliderForward=~
    bIsGliderBack=~
    bHasDeployedAndIsStandingUpright=~
    LocalAccelForward=~
    LocalAccelRight=~
    LocalVelocityRight=~
    SkydiveAimPitch=~
    SkydiveAimYaw=~
    DeployChuteAnimRate=~
    SkydiveDriftAnimRate=~
    SkydiveDriftAnimRateCurrent=~
    SkydiveFidgetAnimRate=~
    SkydiveFidgetAnimRateCurrent=~
    SkydiveAdditiveAlpha=~
    SkydiveDriftDelay=~
   )
   Result(bCanEnterTransition=AND Boolean!(A=NOT Boolean(A=@n2.bIsParachuteOpen), B=NOT Boolean(A=@n1)) -> (ReturnValue))
  ]
 }
 Transition {
  path=AnimGraph.AnimGraphNode_StateMachine_0.Locomotion.AnimStateTransitionNode_6.Transition
  nodes [Result(bCanEnterTransition="Not Equal (Enum)"(A=Get Direction() -> (Direction=Center), B=Center))]
 }
 Transition {path=AnimGraph.AnimGraphNode_StateMachine_0.Locomotion.AnimStateTransitionNode_7.Transition, nodes [Result()]}
 Transition {
  path=AnimGraph.AnimGraphNode_StateMachine_0.Locomotion.AnimStateTransitionNode_8.Transition
  nodes [
   n1 = Get SkydivingState()
   Result(bCanEnterTransition=OR Boolean(
     A=Break Fort Anim Input Skydiving(FortAnimInput_Skydiving=@n1) -> (
      ParaGilderRootOffset=~
      ParaGliderRootRotationOverride=~
      bUseParaGlideRootModifier=~
      bIsSkydivingFromLaunchPad=~
      bIsSkydivingFromBus=~
      bIsInVortex=~
      bIsDiving=~
      bIsParachuteOpen=~
      bIsSkydiving=~
      bIsParachuteLeaning=~
      bIsSkydiveLeaning=~
      bIsLeaning=~
      bIsSkydiveDiveMode=~
      bPlaySkydiveDrift=~
      bSkydiveDriftDelayActive=~
      bSkydiveDriftAnimAllowed=~
      bIsGliderRight=~
      bIsGliderCenter=~
      bIsGliderLeft=~
      bIsGliderForward=~
      bIsGliderBack=~
      bHasDeployedAndIsStandingUpright=~
      LocalAccelForward=~
      LocalAccelRight=~
      LocalVelocityRight=~
      SkydiveAimPitch=~
      SkydiveAimYaw=~
      DeployChuteAnimRate=~
      SkydiveDriftAnimRate=~
      SkydiveDriftAnimRateCurrent=~
      SkydiveFidgetAnimRate=~
      SkydiveFidgetAnimRateCurrent=~
      SkydiveAdditiveAlpha=~
      SkydiveDriftDelay=~
     ).bIsParachuteOpen
     B=Get bForceOpen()
    ))
  ]
 }
 Transition {
  path=AnimGraph.AnimGraphNode_StateMachine_0.Locomotion.AnimStateTransitionNode_9.Transition
  nodes [
   n1 = Get bForceOpen()
   n2 = Break Fort Anim Input Skydiving(FortAnimInput_Skydiving=Get SkydivingState()) -> (
    ParaGilderRootOffset=~
    ParaGliderRootRotationOverride=~
    bUseParaGlideRootModifier=~
    bIsSkydivingFromLaunchPad=~
    bIsSkydivingFromBus=~
    bIsInVortex=~
    bIsDiving=~
    bIsParachuteOpen=~
    bIsSkydiving=~
    bIsParachuteLeaning=~
    bIsSkydiveLeaning=~
    bIsLeaning=~
    bIsSkydiveDiveMode=~
    bPlaySkydiveDrift=~
    bSkydiveDriftDelayActive=~
    bSkydiveDriftAnimAllowed=~
    bIsGliderRight=~
    bIsGliderCenter=~
    bIsGliderLeft=~
    bIsGliderForward=~
    bIsGliderBack=~
    bHasDeployedAndIsStandingUpright=~
    LocalAccelForward=~
    LocalAccelRight=~
    LocalVelocityRight=~
    SkydiveAimPitch=~
    SkydiveAimYaw=~
    DeployChuteAnimRate=~
    SkydiveDriftAnimRate=~
    SkydiveDriftAnimRateCurrent=~
    SkydiveFidgetAnimRate=~
    SkydiveFidgetAnimRateCurrent=~
    SkydiveAdditiveAlpha=~
    SkydiveDriftDelay=~
   )
   Result(bCanEnterTransition=AND Boolean(A=NOT Boolean(A=@n2.bIsParachuteOpen), B=NOT Boolean(A=@n1)))
  ]
 }
 Variant_Roar {entry=Roar, path=AnimGraph.AnimGraphNode_StateMachine_1.Variant_Roar, ids=id, nodes [AnimStateNode_0 = Roar() {is=title AnimStateNode}]}
 Roar {
  path=AnimGraph.AnimGraphNode_StateMachine_1.Variant_Roar.AnimStateNode_0.Roar
  nodes [Output Animation Pose(Result=Sequence Player!(Sequence:Anim Sequence Base Object Reference=Get Roar_DragonVariant() -> (Roar_DragonVariant:Anim Sequence Object Reference)) -> (Pose) {bLoopAnimation=false})]
 }
 Flap {entry=Flapping, path=AnimGraph.AnimGraphNode_StateMachine_2.Flap, ids=id, nodes [AnimStateNode_0 = Flapping() {is=title AnimStateNode}]}
 Flapping {
  path=AnimGraph.AnimGraphNode_StateMachine_2.Flap.AnimStateNode_0.Flapping
  nodes [
   n1 = Break Fort Anim Input Skydiving(FortAnimInput_Skydiving=Get SkydivingState()) -> (
    ParaGilderRootOffset=~
    ParaGliderRootRotationOverride=~
    bUseParaGlideRootModifier=~
    bIsSkydivingFromLaunchPad=~
    bIsSkydivingFromBus=~
    bIsInVortex=~
    bIsDiving=~
    bIsParachuteOpen=~
    bIsSkydiving=~
    bIsParachuteLeaning=~
    bIsSkydiveLeaning=~
    bIsLeaning=~
    bIsSkydiveDiveMode=~
    bPlaySkydiveDrift=~
    bSkydiveDriftDelayActive=~
    bSkydiveDriftAnimAllowed=~
    bIsGliderRight=~
    bIsGliderCenter=~
    bIsGliderLeft=~
    bIsGliderForward=~
    bIsGliderBack=~
    bHasDeployedAndIsStandingUpright=~
    LocalAccelForward=~
    LocalAccelRight=~
    LocalVelocityRight=~
    SkydiveAimPitch=~
    SkydiveAimYaw=~
    DeployChuteAnimRate=~
    SkydiveDriftAnimRate=~
    SkydiveDriftAnimRateCurrent=~
    SkydiveFidgetAnimRate=~
    SkydiveFidgetAnimRateCurrent=~
    SkydiveAdditiveAlpha=~
    SkydiveDriftDelay=~
   )
   n2 = Get Flap_Rate() -> (Flap_Rate:double)
   n3 = "Sequence Player (additive) 'HornedJudgment_FlapAdd_FrontEnd_GLIDER'"!(PlayRate:float=@n2) -> (Pose) {
    bLoopAnimation=false
    sequence=/BRCosmetics/Animation/Game/MainPlayer/Skydive/ParaGlide/HornedJudgment/HornedJudgment_FlapAdd_FrontEnd_GLIDER
   }
   n4 = Get OwnerGlider()
   Output Animation Pose(Result=Apply Additive!(
     Base=Blendspace Player 'BS_HornedJudgment_FlapAdd_GLIDER'!(X=@n1.SkydiveAimYaw, Y=@n1.LocalAccelForward, PlayRate:float=@n2) -> (Pose) {blendSpace=/BRCosmetics/Animation/Game/MainPlayer/Skydive/ParaGlide/HornedJudgment/BS_HornedJudgment_FlapAdd_GLIDER}
     Additive=@n3
     bAlphaBoolEnabled:bool=Get IsCosmeticPreview(self=@n4)
    ) -> (Pose) {alphaInputType=Bool, alphaBoolBlend {blendInTime=0.2, blendOption=Linear}})
  ]
 }
 Original_Roar {entry=Roar, path=AnimGraph.AnimGraphNode_StateMachine_3.Original_Roar, ids=id, nodes [AnimStateNode_0 = Roar()]}
 Roar {
  path=AnimGraph.AnimGraphNode_StateMachine_3.Original_Roar.AnimStateNode_0.Roar
  nodes [
   n1 = Break Fort Anim Input Skydiving(FortAnimInput_Skydiving=Get SkydivingState()) -> (
    ParaGilderRootOffset=~
    ParaGliderRootRotationOverride=~
    bUseParaGlideRootModifier=~
    bIsSkydivingFromLaunchPad=~
    bIsSkydivingFromBus=~
    bIsInVortex=~
    bIsDiving=~
    bIsParachuteOpen=~
    bIsSkydiving=~
    bIsParachuteLeaning=~
    bIsSkydiveLeaning=~
    bIsLeaning=~
    bIsSkydiveDiveMode=~
    bPlaySkydiveDrift=~
    bSkydiveDriftDelayActive=~
    bSkydiveDriftAnimAllowed=~
    bIsGliderRight=~
    bIsGliderCenter=~
    bIsGliderLeft=~
    bIsGliderForward=~
    bIsGliderBack=~
    bHasDeployedAndIsStandingUpright=~
    LocalAccelForward=~
    LocalAccelRight=~
    LocalVelocityRight=~
    SkydiveAimPitch=~
    SkydiveAimYaw=~
    DeployChuteAnimRate=~
    SkydiveDriftAnimRate=~
    SkydiveDriftAnimRateCurrent=~
    SkydiveFidgetAnimRate=~
    SkydiveFidgetAnimRateCurrent=~
    SkydiveAdditiveAlpha=~
    SkydiveDriftDelay=~
   )
   Output Animation Pose(Result=Blendspace Player 'BS_HornedJudgment_RoarAdd_GLIDER'(X=@n1.SkydiveAimYaw, Y=@n1.LocalAccelForward) {bLoop=false, blendSpace=/BRCosmetics/Animation/Game/MainPlayer/Skydive/ParaGlide/HornedJudgment/BS_HornedJudgment_RoarAdd_GLIDER})
  ]
 }
 CheckAnimCurves {nodes [
   Inputs()
   > Set DisableFlap(DisableFlap:bool="float >= float"(A=GetCurveValue(CurveName=DisableFlap), B=0.25) {is=title K2Node_CallFunction}) -> (Output_Get:bool)
   n1 = GetCurveValue(CurveName=Flap)
   > n2 = Set Flapping(Flapping:bool="float > float"(A=@n1, B=0) {is=title K2Node_CallFunction}) -> (Output_Get:bool)
   > n3 = Branch(Condition=@n2)
   n4 = GetCurveValue(CurveName=LeanTransition)
   Set TransitionAnimPlaying(execute=[@n3.then, @n7.then], TransitionAnimPlaying:bool="float > float"(A=@n4, B=0.09)) -> (Output_Get:bool)
   > Set Roaring(Roaring:bool="Equal (Float)"(A=GetCurveValue(CurveName=Roar), B=1)) -> (Output_Get:bool)
   > n5 = Branch(Condition=NOT Boolean(A=Get TransitionAnimPlaying()))
   > n6 = Set useLeanTransition(useLeanTransition:bool=false) -> (Output_Get:bool)
   Outputs!(then:Exec=[@n5.else, @n6.then])
   n7 = Set canFlap(execute=@n3.else, canFlap:bool=false) -> (Output_Get:bool)
   "float >= float"(A=@n4, B=0.15)
   "Equal (Float)"(A=@n1, B=1)
   "Equal (Float)"(A=@n4, B=1)
   PrintString!(execute, InString="To String (Float)"(InDouble=@n4), bPrintToScreen, bPrintToLog, TextColor, Duration, Key) -> (then)
   "float >= float"(A=@n1, B=0.25)
  ]}
 CheckLeanFromInto {nodes [
   Inputs()
   > n1 = Branch(Condition="Equal (Float)"(A=GetCurveValue(CurveName=ParaGlide_Lean), B=1))
   > n2 = Set LeanFromInto(LeanFromInto:bool=true) -> (Output_Get:bool)
   Outputs(then=[@n2.then, @n3.then])
   n3 = Set LeanFromInto(execute=@n1.else, LeanFromInto=false)
  ]}
 CheckLeanTransition {nodes [
   Inputs()
   > n1 = Sequence()
   n2 = Get Flapping()
   n3 = Get Direction() -> (Direction=Center)
   > n4 = Branch(Condition=AND Boolean(A="Not Equal (Enum)"(A=@n3, B=Center), B=NOT Boolean(A=@n2)))
   > Do Once(Reset=@n5.then)
   > Set useLeanTransition(useLeanTransition=true)
   > Outputs()
   Set useLeanTransition(execute=@n4.else, useLeanTransition=false)
   n5 = Branch(execute=@n1.then_1, Condition="Equal (Enum)"(A=Get Direction() -> (Direction=Center), B=Center))
  ]}
 CheckVariation {nodes [
   n1 = Get OwnerGlider()
   n2 = GetDisplayName(Object=GetItemDefinitionFromLoadoutWithActor!(
     Actor:Actor Object Reference=GetFortPlayerPawn!(self:Fort Player Parachute Object Reference=@n1) -> (ReturnValue:Fort Player Pawn Object Reference)
     LoadoutSlot:Gameplay Tag Structure=(TagName="LoadoutSlot.Glider")
    ) -> (ReturnValue:Athena Cosmetic Item Definition Object Reference))
   n3 = PrintString(InString=@n2, TextColor=(R=1.7767,G=5,B=0.165,A=1), Duration=20)
   Do Once(execute=[@n3.then, @n11.execute])
   > Set IsFireVariant(IsFireVariant:bool=Contains(SearchIn=@n2, Substring=Glider_ID_134_DarkVikingFire)) -> (Output_Get:bool)
   > Set IsKnightRemix(IsKnightRemix:bool=Contains(SearchIn=@n2, Substring=Glider_ID_165_KnightRemix)) -> (Output_Get:bool)
   > Set IsGoldenSkeleton(IsGoldenSkeleton:bool=Contains(SearchIn=@n2, Substring=Glider_ID_193_TheGoldenSkeleton)) -> (Output_Get:bool)
   > n4 = Branch(Condition=Get IsFireVariant() -> (IsFireVariant:bool))
   > Set DV_DragonGlider_Index(DV_DragonGlider_Index:int=1) -> (Output_Get:int=0)
   > n5 = Set UseVariantRoar(UseVariantRoar:bool=false) -> (Output_Get:bool)
   Outputs(then=[@n10.then, @n9.then, @n7.then, @n5.then])
   n6 = Branch(execute=@n4.else, Condition=Get IsKnightRemix() -> (IsKnightRemix:bool))
   > Set DV_DragonGlider_Index(DV_DragonGlider_Index=2) -> (Output_Get=0)
   > n7 = Set UseVariantRoar(UseVariantRoar=true)
   n8 = Branch(execute=@n6.else, Condition=Get IsGoldenSkeleton() -> (IsGoldenSkeleton:bool))
   > Set DV_DragonGlider_Index(DV_DragonGlider_Index=3) -> (Output_Get=0)
   > n9 = Set UseVariantRoar(UseVariantRoar=true)
   Set DV_DragonGlider_Index(execute=@n8.else, DV_DragonGlider_Index=0) -> (Output_Get=0)
   > n10 = Set UseVariantRoar(UseVariantRoar=false)
   n11 = Inputs()
  ]}
 DirectionalFlap {nodes [
   Inputs()
   > n1 = Branch(Condition=Get TransitionAnimPlaying())
   n2 = "Equal (Enum)"(A=Get Direction() -> (Direction=Center), B=Left)
   n3 = "Equal (Enum)"(A=Get Direction_Buffer() -> (Direction_Buffer:ESkydivingDirection Enum=Center), B=Right)
   n4 = Get Direction_Buffer() -> (Direction_Buffer=Center)
   n5 = OR Boolean(A="Equal (Enum)"(A=@n4, B=Forward), B="Equal (Enum)"(A=@n4, B=Back))
   n6 = Get Direction() -> (Direction=Center)
   n7 = OR Boolean(A="Equal (Enum)"(A=@n6, B=Right), B="Equal (Enum)"(A=@n6, B=Left))
   n8 = "Equal (Enum)"(A=Get Direction() -> (Direction=Center), B=Right)
   n9 = "Equal (Enum)"(A=Get Direction_Buffer() -> (Direction_Buffer=Center), B=Left)
   n10 = Get Direction() -> (Direction=Center)
   n11 = OR Boolean(A="Equal (Enum)"(A=@n10, B=Forward), B="Equal (Enum)"(A=@n10, B=Back))
   n12 = Get Direction_Buffer() -> (Direction_Buffer=Center)
   n13 = OR Boolean(A="Equal (Enum)"(A=@n12, B=Right), B="Equal (Enum)"(A=@n12, B=Left))
   n14 = Get Direction_Buffer() -> (Direction_Buffer=Center)
   n15 = OR Boolean!(
    A="Equal (Enum)"(A=@n14, B=Forward)
    B="Equal (Enum)"(A=@n14, B=Back)
    C:bool="Equal (Enum)"(A=@n14, B=Right)
    D:bool="Equal (Enum)"(A=@n14, B=Left)
   ) -> (ReturnValue)
   n16 = "Equal (Enum)"(A=Get Direction() -> (Direction=Center), B=Center)
   n17 = Branch(
    execute=@n1.else
    Condition=OR Boolean!(
     A=AND Boolean(A=@n7, B=@n5)
     B=AND Boolean(A=@n11, B=@n13)
     C=AND Boolean(A=@n16, B=@n15)
     D=AND Boolean(A=@n8, B=@n9)
     E:bool=AND Boolean(A=@n2, B=@n3)
    ) -> (ReturnValue)
   )
   > n18 = Set DirectionChange(DirectionChange:bool=true) -> (Output_Get:bool)
   Branch(execute=[@n18.then, @n21.then], Condition=Get DirectionChange() -> (DirectionChange:bool))
   > n19 = Sequence()
   > Do Once(Reset=@n20.then)
   > Set FlapReset(FlapReset:bool=false) -> (Output_Get:bool)
   > Set canFlap(canFlap=true)
   > Set Flap Alpha!(execute:Exec) -> (then:Exec) {is=title K2Node_MacroInstance}
   > Outputs()
   n20 = Branch(execute=@n19.then_1, Condition=Get DirectionChange())
   n21 = Set DirectionChange(execute=@n17.else, DirectionChange=false)
  ]}
 EventGraph {nodes [
   n1 = Event BlueprintUpdateAnimation()
   > Set Time(Time:double=@n1.DeltaTimeX) -> (Output_Get:double)
   > Set Direction(Direction:ESkydivingDirection Enum=Break Fort Anim Input Skydiving(FortAnimInput_Skydiving=Get SkydivingState()) -> (
     ParaGilderRootOffset=~
     ParaGliderRootRotationOverride=~
     bUseParaGlideRootModifier=~
     bIsSkydivingFromLaunchPad=~
     bIsSkydivingFromBus=~
     bIsInVortex=~
     bIsDiving=~
     bIsParachuteOpen=~
     bIsSkydiving=~
     bIsParachuteLeaning=~
     bIsSkydiveLeaning=~
     bIsLeaning=~
     bIsSkydiveDiveMode=~
     bPlaySkydiveDrift=~
     bSkydiveDriftDelayActive=~
     bSkydiveDriftAnimAllowed=~
     bIsGliderRight=~
     bIsGliderCenter=~
     bIsGliderLeft=~
     bIsGliderForward=~
     bIsGliderBack=~
     bHasDeployedAndIsStandingUpright=~
     LocalAccelForward=~
     LocalAccelRight=~
     LocalVelocityRight=~
     SkydiveAimPitch=~
     SkydiveAimYaw=~
     DeployChuteAnimRate=~
     SkydiveDriftAnimRate=~
     SkydiveDriftAnimRateCurrent=~
     SkydiveFidgetAnimRate=~
     SkydiveFidgetAnimRateCurrent=~
     SkydiveAdditiveAlpha=~
     SkydiveDriftDelay=~
    ).LocalAccelDir) -> (Output_Get:ESkydivingDirection Enum=Center)
   > n2 = Sequence!(execute) -> (then_0, then_1, then_2:Exec, then_3:Exec, then_4:Exec, then_5:Exec, then_6:Exec, then_7:Exec)
   > Check Variation!(execute:Exec) -> (then:Exec) {is=title K2Node_MacroInstance}
   Check Anim Curves!(execute:Exec=@n2.then_1) -> (then:Exec) {is=title K2Node_MacroInstance}
   > Check Lean from Into!(execute:Exec) -> (then:Exec) {is=title K2Node_MacroInstance}
   > Check Lean Transition!(execute:Exec) -> (then:Exec) {is=title K2Node_MacroInstance}
   Directional Flap!(execute:Exec=@n2.then_2) -> (then:Exec) {is=title K2Node_MacroInstance}
   Random Flap!(execute:Exec=@n2.then_3) -> (then:Exec) {is=title K2Node_MacroInstance}
   Set Last Direction!(execute:Exec=@n2.then_5) -> (then:Exec) {is=title K2Node_MacroInstance}
   > Flap State Check!(execute:Exec) -> (then:Exec) {is=title K2Node_MacroInstance}
   > Branch(Condition=OR Boolean!(A=Get TransitionAnimPlaying(), B=Get DisableFlap() -> (DisableFlap:bool)) -> (ReturnValue))
   > Set canFlap(canFlap=false)
   Do Once(execute=@n2.then_6, Reset=[@n6.then_0, @n9.then_0])
   > Set ResetBackupTimer(ResetBackupTimer:bool=false) -> (Output_Get:bool)
   > RetriggerableDelay(Duration=10)
   > Set ResetBackupTimer(ResetBackupTimer=true)
   n3 = Get Flapping()
   n4 = Get TransitionAnimPlaying()
   n5 = Get Roaring() -> (Roaring:bool)
   Branch(
    execute=@n2.then_7
    Condition=AND Boolean!(A=Get ResetBackupTimer() -> (ResetBackupTimer:bool), B=NOT Boolean(A=@n3), C:bool=NOT Boolean(A=@n4), D:bool=NOT Boolean(A=@n5)) -> (ReturnValue)
   )
   > n6 = Sequence!(execute) -> (then_0, then_1)
   Set FlapReset(execute=[@n8.then_0, @n6.then_1, @n9.then_1], FlapReset=true)
   > Set canFlap(canFlap=false)
   > Set canFlapAgain(canFlapAgain:bool=false) -> (Output_Get:bool)
   > Ramp Flap Alpha!(execute:Exec) -> (then:Exec) {is=title K2Node_MacroInstance}
   Random Roar!(execute:Exec=@n2.then_4) -> (then:Exec) {is=title K2Node_MacroInstance}
   AnimNotify_TurnONEngines() {is=title K2Node_Event}
   > SetGliderFullyDeployed!(execute, self:Fort Glider Anim Instance Object Reference, bIsFullyDeployed:bool=true) -> (then)
   AnimNotify_FX_RoarOn() {is=title K2Node_Event}
   > Set FX_Roar(FX_Roar:bool=true) -> (Output_Get:bool)
   > Call Roar Event!(execute:Exec, self:Horned Judgment Glider Anim BP Object Reference) -> (then:Exec) {is=title K2Node_CallDelegate}
   AnimNotify_FX_RoarOff() {is=title K2Node_Event}
   > Set FX_Roar(FX_Roar=false)
   > Call Roar Event End!(execute:Exec, self:Horned Judgment Glider Anim BP Object Reference) -> (then:Exec) {is=title K2Node_CallDelegate}
   n7 = AnimNotify_RoarOff() {is=title K2Node_Event}
   Set RoarReset(execute=[@n7.then, @n8.then_1], RoarReset:bool=true) -> (Output_Get:bool)
   > Set Roaring(Roaring=false)
   > Set canRoar(canRoar:bool=false) -> (Output_Get:bool)
   AnimNotify_FlapEnable() {is=title K2Node_Event}
   > Do Once()
   > n8 = Sequence()
   AnimNotify_FlapOff() {is=title K2Node_Event}
   > n9 = Sequence()
   Set canFlapAgain(canFlapAgain=true)
   > Branch(Condition=AND Boolean!(A=Get Flapping(), B=Get DirectionChange(), C=Get canFlapAgain() -> (canFlapAgain:bool)) -> (ReturnValue))
   > Set FlapReset(FlapReset=true)
   PrintString(InString="To String (Boolean)"(InBool=Get IsKnightRemix()))
   K2Node_Composite_0 = Debug Text!(then_text:Exec) {is=title K2Node_Composite}
   AnimNotify_EnableFlapTransition() {is=title K2Node_Event}
  ]}
 Debug Text {
  path=EventGraph.K2Node_Composite_0.Debug Text
  nodes [
   Inputs!() -> (then_text:Exec)
   n1 = Get useLeanTransition()
   > PrintString(InString=Append(A="Use Lean Transition: ", B="To String (Boolean)"(InBool=@n1)))
   n2 = Get Direction() -> (Direction=Center)
   > PrintString(
    InString=Append(A="Direction: ", B=Enum to String!(Enumerator:ESkydivingDirection Enum=@n2) -> (ReturnValue:String))
    TextColor=(R=0.0887,G=0.3958,B=0.2521,A=1)
   )
   n3 = Get Direction_Last() -> (Direction_Last:ESkydivingDirection Enum=Center)
   > PrintString(InString=Append(A="DirectionLast: ", B=Enum to String(Enumerator=@n3)), TextColor=(R=0.0887,G=0.3958,B=0.2521,A=1))
   n4 = Get canFlap() -> (canFlap:bool)
   > PrintString(InString=Append(A="Can Flap: ", B="To String (Boolean)"(InBool=@n4)), TextColor=(R=0.6458,G=0.3484,B=0.1043,A=1))
   n5 = Get Flapping()
   > PrintString(InString=Append(A="Flapping: ", B="To String (Boolean)"(InBool=@n5)), TextColor=(R=0.6042,G=0.0566,B=0.1288,A=1))
   n6 = Get Roaring()
   > PrintString(InString=Append(A=Roaring:, B="To String (Boolean)"(InBool=@n6)), TextColor=(R=0.6042,G=0.0566,B=0.1288,A=1))
   n7 = Get TransitionAnimPlaying()
   > PrintString(InString=Append(A="Transition Anim: ", B="To String (Boolean)"(InBool=@n7)), TextColor=(R=0.6042,G=0.0566,B=0.1288,A=1))
   n8 = Get canRoar()
   > PrintString(InString=Append(A="Can Roar: ", B="To String (Boolean)"(InBool=@n8)), TextColor=(R=0.073,G=0.4844,B=0.0969,A=1))
   n9 = GetCurveValue(CurveName=Into_FwdBack)
   > PrintString(InString=Append(A="Into_FwdBack: ", B="To String (Float)"(InDouble=@n9)))
   n10 = Get Flap_Alpha()
   > PrintString(InString=Append(A="Flap Alpha: ", B="To String (Float)"(InDouble=@n10)))
   n11 = Get DisableFlap()
   > PrintString(InString=Append(A=DisableFlap:, B="To String (Boolean)"(InBool=@n11)), TextColor=(R=0.073,G=0.4844,B=0.0969,A=1))
   Break Fort Anim Input Skydiving(FortAnimInput_Skydiving=Get SkydivingState()) -> (
    ParaGilderRootOffset=~
    ParaGliderRootRotationOverride=~
    bUseParaGlideRootModifier=~
    bIsSkydivingFromLaunchPad=~
    bIsSkydivingFromBus=~
    bIsInVortex=~
    bIsDiving=~
    bIsParachuteOpen=~
    bIsSkydiving=~
    bIsParachuteLeaning=~
    bIsSkydiveLeaning=~
    bIsLeaning=~
    bIsSkydiveDiveMode=~
    bPlaySkydiveDrift=~
    bSkydiveDriftDelayActive=~
    bSkydiveDriftAnimAllowed=~
    bIsGliderRight=~
    bIsGliderCenter=~
    bIsGliderLeft=~
    bIsGliderForward=~
    bIsGliderBack=~
    bHasDeployedAndIsStandingUpright=~
    LocalAccelForward=~
    LocalAccelRight=~
    LocalVelocityRight=~
    SkydiveAimPitch=~
    SkydiveAimYaw=~
    DeployChuteAnimRate=~
    SkydiveDriftAnimRate=~
    SkydiveDriftAnimRateCurrent=~
    SkydiveFidgetAnimRate=~
    SkydiveFidgetAnimRateCurrent=~
    SkydiveAdditiveAlpha=~
    SkydiveDriftDelay=~
   )
   Outputs!()
   Get useLeanState() -> (useLeanState:bool)
  ]
 }
 FlapStateCheck {nodes [
   Inputs!() -> (execute:Exec)
   > n1 = Cast To B_Athena_Glider_Legacy_DarkViking!(execute, Object=Get OwnerGlider()) -> (then, CastFailed, AsB Athena Glider Legacy Dark Viking:B Athena Glider Legacy Dark Viking Object Reference) {is=title K2Node_DynamicCast}
   n2 = Get TransitionAnimPlaying()
   > n3 = On Flap State Changed!(
    execute
    self:B Athena Glider Legacy Dark Viking Object Reference=@n1
    New Flap State:bool=Get Flapping()
    Can Flap:bool=AND Boolean!(A=Get canFlap(), B=NOT Boolean(A=@n2)) -> (ReturnValue)
    Fast Flap:bool="Not Equal (Float)"!(A:double=Get Flap_Rate(), B:double=1) -> (ReturnValue:bool) {is=title K2Node_CallFunction}
   ) -> (then)
   Outputs!(then:Exec=[@n3.then, @n1.CastFailed])
  ]}
 RampFlapAlpha {nodes [
   Inputs()
   > Set Flap_Alpha(Flap_Alpha:double=0) -> (Output_Get:double)
   > Outputs()
   FInterpTo(Current=Get Flap_Alpha(), Target=0, DeltaTime=Get Time() -> (Time:double), InterpSpeed=12)
   Get Flap_Alpha_Target() -> (Flap_Alpha_Target:double)
  ]}
 RandomFlap {nodes [
   Inputs()
   > n1 = Branch(Condition=Get TransitionAnimPlaying())
   n2 = Sequence(execute=@n1.else)
   > Do Once(Reset=@n7.then)
   > Set FlapReset(FlapReset=false)
   n3 = Get SkydivingState()
   > n4 = Branch(Condition="float < float"!(
     A:double=Break Fort Anim Input Skydiving(FortAnimInput_Skydiving=@n3) -> (
      ParaGilderRootOffset=~
      ParaGliderRootRotationOverride=~
      bUseParaGlideRootModifier=~
      bIsSkydivingFromLaunchPad=~
      bIsSkydivingFromBus=~
      bIsInVortex=~
      bIsDiving=~
      bIsParachuteOpen=~
      bIsSkydiving=~
      bIsParachuteLeaning=~
      bIsSkydiveLeaning=~
      bIsLeaning=~
      bIsSkydiveDiveMode=~
      bPlaySkydiveDrift=~
      bSkydiveDriftDelayActive=~
      bSkydiveDriftAnimAllowed=~
      bIsGliderRight=~
      bIsGliderCenter=~
      bIsGliderLeft=~
      bIsGliderForward=~
      bIsGliderBack=~
      bHasDeployedAndIsStandingUpright=~
      LocalAccelForward=~
      LocalAccelRight=~
      LocalVelocityRight=~
      SkydiveAimPitch=~
      SkydiveAimYaw=~
      DeployChuteAnimRate=~
      SkydiveDriftAnimRate=~
      SkydiveDriftAnimRateCurrent=~
      SkydiveFidgetAnimRate=~
      SkydiveFidgetAnimRateCurrent=~
      SkydiveAdditiveAlpha=~
      SkydiveDriftDelay=~
     ).LocalAccelForward
     B:double=-250
    ) -> (ReturnValue:bool) {is=title K2Node_CallFunction})
   > n5 = Set Flap_Delay(Flap_Delay:double=0.25) -> (Output_Get:double)
   Delay(execute=[@n5.then, @n6.then], Duration=Get Flap_Delay() -> (Flap_Delay:double))
   > Set canFlap(canFlap=true)
   > Set Flap Alpha()
   > Outputs()
   n6 = Set Flap_Delay(execute=@n4.else, Flap_Delay=RandomFloatInRange(Min=3, Max=5))
   n7 = Branch(execute=@n2.then_1, Condition=Get FlapReset() -> (FlapReset:bool))
  ]}
 RandomRoar {nodes [
   Delay(Duration=5)
   > n1 = Set canFlap(canFlap=true)
   n2 = Set canRoar(execute=[@n11.else, @n11.then, @n1.then], canRoar=true)
   Outputs(then=[@n13.then, @n2.then])
   n3 = Branch(Condition=Get Flapping())
   n4 = Sequence(execute=[@n3.else, @n14.execute])
   > Do Once(Reset=@n12.then)
   > Set RoarReset(RoarReset=false)
   > Update Roar Side Counter!(" ":Exec) -> (then:Exec) {is=title K2Node_MacroInstance}
   > Delay(Duration=RandomFloatInRange(Min=7.5, Max=15))
   n5 = Get Counter() -> (Counter:int=0)
   n6 = Get Counter() -> (Counter=0)
   > Set Roar_DragonVariant(Roar_DragonVariant:Anim Sequence Object Reference=Select!(
     Option 0:Anim Sequence Object Reference=/BRCosmetics/Animation/Game/MainPlayer/Skydive/ParaGlide/HornedJudgment/HornedJudgment_RoarAdd_GLIDER
     Option 1:Anim Sequence Object Reference=&HornedJudgment_RoarAdd_GLIDER
     Option 2:Anim Sequence Object Reference=Select!(
      Option 0=/BRCosmetics/Animation/Game/MainPlayer/Skydive/ParaGlide/DarkVikingDragon_KnightRemix/Knight_Remix_RoarAdd_Single_Left_GLIDER
      Option 1=/BRCosmetics/Animation/Game/MainPlayer/Skydive/ParaGlide/DarkVikingDragon_KnightRemix/Knight_Remix_RoarAdd_Single_Right_GLIDER
      Index=@n6
     ) -> (ReturnValue)
     Option 3:Anim Sequence Object Reference=Select(
      Option 0=&Knight_Remix_RoarAdd_Single_Left_GLIDER
      Option 1=&Knight_Remix_RoarAdd_Single_Right_GLIDER
      Index=@n5
     )
     Index:int=Get DV_DragonGlider_Index() -> (DV_DragonGlider_Index:int=0)
    ) -> (ReturnValue:Anim Sequence Object Reference)) -> (Output_Get:Anim Sequence Object Reference)
   n7 = Get TransitionAnimPlaying()
   n8 = Get Flapping()
   > n9 = Branch(Condition=AND Boolean!(A=Get UseVariantRoar(), B=NOT Boolean(A=@n8), C:bool=NOT Boolean(A=@n7)) -> (ReturnValue))
   > Set canFlap(canFlap=true)
   > n10 = Set canRoar(canRoar=true)
   n11 = Branch(execute=[@n9.else, @n10.then], Condition=AND Boolean!(A=Get UseVariantRoar(), B=Get Flapping()) -> (ReturnValue))
   n12 = Branch(execute=@n4.then_1, Condition=Get RoarReset() -> (RoarReset:bool))
   n13 = Set Flap Alpha()
   n14 = Inputs()
  ]}
 RoarEvent {nodes [RoarEvent() {is=title K2Node_FunctionEntry}]}
 RoarEventEnd {nodes [RoarEventEnd() {is=title K2Node_FunctionEntry}]}
 SetFlapAlpha {nodes [
   Inputs()
   n1 = Get TransitionAnimPlaying()
   n2 = Get DisableFlap()
   > n3 = Branch(Condition=AND Boolean!(A=Get canFlap(), B=NOT Boolean(A=@n1), C:bool=NOT Boolean(A=@n2)) -> (ReturnValue))
   n4 = AND Boolean!(A=Get IsKnightRemix(), B=Get Roaring()) -> (ReturnValue)
   n5 = Select(Option 0:double=RandomFloatInRange(Min=0.5, Max=1), Option 1:double=1, Index=RandomInteger!(Max:int=2) -> (ReturnValue:int=0)) -> (ReturnValue:double)
   > n6 = Set Flap_Alpha(Flap_Alpha=Select(Option 0=Select(Option 0=@n5, Option 1=0.8, Index=@n4), Option 1=0.75, Index:bool=Get DirectionChange()))
   Outputs(then=[@n6.then, @n7.then])
   n7 = Set Flap_Alpha(execute=@n3.else, Flap_Alpha=0)
  ]}
 SetLastDirection {nodes [
   Inputs()
   > n1 = Sequence()
   > Do Once(Reset=@n2.then)
   > Set Direction_Last(Direction_Last:ESkydivingDirection Enum=Get Direction_Buffer() -> (Direction_Buffer=Center)) -> (Output_Get:ESkydivingDirection Enum=Center)
   > Set Direction_Buffer(Direction_Buffer:ESkydivingDirection Enum=Get Direction() -> (Direction=Center)) -> (Output_Get:ESkydivingDirection Enum=Center)
   > Set DirectionChange(DirectionChange=true)
   > Outputs()
   n2 = Branch(
    execute=@n1.then_1
    Condition="Not Equal (Enum)"(A=Get Direction() -> (Direction=Center), B=Get Direction_Buffer() -> (Direction_Buffer=Center))
   )
   Set DirectionChange(DirectionChange=false)
  ]}
 UpdateRoarSideCounter {nodes [
   Inputs!() -> (" ":Exec)
   > n1 = Increment Int(Value=Get Counter() -> (Counter=0))
   > n2 = Branch(Condition="integer > integer"(A=@n1, B=1) {is=title K2Node_CallFunction})
   > n3 = Set Counter(Counter:int=0) -> (Output_Get:int=0)
   Outputs(then=[@n2.else, @n3.then])
  ]}
]
variables [
 AimPitch {type=number}
 BlockAnimRule {type=boolean}
 canFlap {type=boolean}
 canFlapAgain {type=boolean}
 canRoar {type=boolean}
 Counter {type=integer}
 DeployAnimRate {type=number}
 Direction {type=ESkydivingDirection, default=Center}
 Direction_Buffer {type=ESkydivingDirection, default=Center}
 Direction_Last {type=ESkydivingDirection, default=Center}
 DirectionChange {type=boolean}
 DisableFlap {type=boolean}
 DV_DragonGlider_Index {type=integer}
 Flap_Alpha {type=number}
 Flap_Alpha_Target {type=number}
 Flap_Delay {type=number}
 Flap_Rate {type=number, default=1}
 Flapping {type=boolean}
 FlapReset {type=boolean}
 Forward {type=number}
 FX_Roar {type=boolean}
 IsFireVariant {type=boolean}
 IsGoldenSkeleton {type=boolean}
 IsKnightRemix {type=boolean}
 IsParachuteOpen {type=boolean}
 LeanFromInto {type=boolean}
 ResetBackupTimer {type=boolean}
 Right {type=number}
 Roar_DragonVariant {type=AnimSequence, default=null}
 RoarEvent {default=()}
 RoarEventEnd {default=()}
 Roaring {type=boolean}
 RoarReset {type=boolean}
 SkydiveYaw {type=number}
 Time {type=number}
 TransitionAnimPlaying {type=boolean}
 UseDynamicsAlpha {type=number}
 useLeanState {type=boolean}
 useLeanTransition {type=boolean}
 UseVariantRoar {type=boolean}
]
