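# Anim Blueprint for the Maverick glider, animating deploy, aim-driven glide and toggleable S-foil wings.
# Locomotion moves through closed, opening and open states, swapping in front-end clips for cosmetic preview; lean and turn additives use an inverted curve alpha.
# An additive SFoil_Pose machine layers on top, toggled by primary-fire input unless map, hard-deck or preview checks block it.
# Notifies hide or show the glider bone; Locomotion_GliderItem is unconnected.
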AnimBlueprint /BRCosmetics/Gadgets/Assets/VinderTech_GliderChute/Blueprints/Maverick_Glider_AnimBP
target_skeleton=/Game/Gadgets/Assets/VinderTech_GliderChute/Glider_Maverick/Meshes/Maverick_Skeleton
graphs [
 AnimGraph {nodes [
   Apply Additive() {alphaBoolBlend {blendOption=Linear}}
   n1 = Break Fort Anim Input Skydiving!(FortAnimInput_Skydiving:"Fort Anim Input Skydiving Structure (by ref)"=Get SkydivingState() -> (SkydivingState:Fort Anim Input Skydiving Structure)) -> (
    ParaGilderRootOffset:Vector=~
    ParaGliderRootRotationOverride:Rotator=~
    bUseParaGlideRootModifier:bool=~
    bIsSkydivingFromLaunchPad:bool=~
    bIsSkydivingFromBus:bool=~
    bIsInVortex:bool=~
    bIsDiving:bool=~
    bIsParachuteOpen:bool=~
    bIsSkydiving:bool=~
    bIsParachuteLeaning:bool=~
    bIsSkydiveLeaning:bool=~
    bIsLeaning:bool=~
    bIsSkydiveDiveMode:bool=~
    bPlaySkydiveDrift:bool=~
    bSkydiveDriftDelayActive:bool=~
    bSkydiveDriftAnimAllowed:bool=~
    bIsGliderRight:bool=~
    bIsGliderCenter:bool=~
    bIsGliderLeft:bool=~
    bIsGliderForward:bool=~
    bIsGliderBack:bool=~
    bHasDeployedAndIsStandingUpright:bool=~
    LocalAccelForward:float=~
    LocalAccelRight:float=~
    LocalVelocityRight:float=~
    SkydiveAimPitch:float=~
    SkydiveAimYaw:float=~
    DeployChuteAnimRate:float=~
    SkydiveDriftAnimRate:float=~
    SkydiveDriftAnimRateCurrent:float=~
    SkydiveFidgetAnimRate:float=~
    SkydiveFidgetAnimRateCurrent:float=~
    SkydiveAdditiveAlpha:float=~
    SkydiveDriftDelay:float=~
    SkydiveDriftAnim:int
    SkydiveDriftAnimMax:int
    LaunchpadAnim:int
    LocalAccelDir:ESkydivingDirection Enum
    DirectionLast:ESkydivingDirection Enum
    PoseAdditive:Anim Sequence Object Reference
   ) {is=title K2Node_BreakStruct}
   AnimGraphNode_StateMachine_0 = Locomotion() {is=title AnimGraphNode_StateMachine}
   n2 = Apply Additive!(
    Base=Apply Additive(
     Base=@AnimGraphNode_StateMachine_0
     Additive=Blendspace Player 'BS_Maverick_Lean_GLIDER'(X=@n1.LocalAccelRight, Y=@n1.LocalAccelForward) {blendSpace=/BRCosmetics/Animation/Game/MainPlayer/Skydive/ParaGlide/Maverick/BS_Maverick_Lean_GLIDER}
     AlphaCurveName=ParaGlide_Lean
    ) {alphaInputType=Curve, alphaBoolBlend {blendOption=Linear}, alphaScaleBiasClamp {bias=1, scale=-1}}
    Additive=Blendspace Player 'BS_Maverick_Turn_GLIDER'(X=@n1.SkydiveAimYaw, Y=@n1.SkydiveAimPitch) {blendSpace=/BRCosmetics/Animation/Game/MainPlayer/Skydive/ParaGlide/Maverick/BS_Maverick_Turn_GLIDER}
    AlphaCurveName:Name=ParaGlide_Lean
   ) -> (Pose) {alphaInputType=Curve, alphaBoolBlend {blendOption=Linear}, alphaScaleBiasClamp {bias=1, scale=-1}}
   AnimGraphNode_StateMachine_2 = SFoil_Pose() {is=title AnimGraphNode_StateMachine}
   Output Pose(Result=Slot 'DefaultSlot'(Source=Apply Additive!(Base=@n2, Additive=@AnimGraphNode_StateMachine_2, Alpha:float) -> (Pose) {alphaBoolBlend {blendOption=Linear}})) {layerGroup=None}
   AnimGraphNode_StateMachine_1 = Locomotion_GliderItem() {is=title AnimGraphNode_StateMachine}
  ]}
 Locomotion {
  entry=Closed
  path=AnimGraph.AnimGraphNode_StateMachine_0.Locomotion
  ids=id
  nodes [
   AnimStateNode_0 = Idle_Open() {is=title AnimStateNode}
   AnimStateNode_2 = Closed() {is=title AnimStateNode, stateEntered {guid=#1, notifyName=ClosedEntered}}
   AnimStateNode_3 = Open_Into() {is=title AnimStateNode, stateEntered {guid=#2, notifyName=DeployEntered}}
   AnimStateTransitionNode_0 = AnimStateTransitionNode() {bAutomaticRuleBasedOnSequencePlayerInState=true, blendMode=Linear, crossfadeDuration=0}
   AnimStateTransitionNode_1 = AnimStateTransitionNode() {crossfadeDuration=0}
   AnimStateTransitionNode_2 = AnimStateTransitionNode() {crossfadeDuration=0}
   AnimStateTransitionNode_3 = AnimStateTransitionNode() {crossfadeDuration=0, priorityOrder=2}
  ]
  links [
   {
    dst=Open_Into
    rule=AnimGraph.AnimGraphNode_StateMachine_0.Locomotion.AnimStateTransitionNode_1.Transition
    src=Closed
    via=AnimStateTransitionNode_1
   }
   {
    dst=Closed
    rule=AnimGraph.AnimGraphNode_StateMachine_0.Locomotion.AnimStateTransitionNode_2.Transition
    src=Idle_Open
    via=AnimStateTransitionNode_2
   }
   {
    dst=Closed
    rule=AnimGraph.AnimGraphNode_StateMachine_0.Locomotion.AnimStateTransitionNode_3.Transition
    src=Open_Into
    via=AnimStateTransitionNode_3
   }
   {
    dst=Idle_Open
    rule=AnimGraph.AnimGraphNode_StateMachine_0.Locomotion.AnimStateTransitionNode_0.Transition
    src=Open_Into
    via=AnimStateTransitionNode_0
   }
  ]
 }
 Idle_Open {
  path=AnimGraph.AnimGraphNode_StateMachine_0.Locomotion.AnimStateNode_0.Idle_Open
  nodes [
   n1 = Break Fort Anim Input Skydiving(FortAnimInput_Skydiving=Get SkydivingState()) -> (
    ParaGilderRootOffset=~
    ParaGliderRootRotationOverride=~
    bUseParaGlideRootModifier=~
    bIsSkydivingFromLaunchPad=~
    bIsSkydivingFromBus=~
    bIsInVortex=~
    bIsDiving=~
    bIsParachuteOpen=~
    bIsSkydiving=~
    bIsParachuteLeaning=~
    bIsSkydiveLeaning=~
    bIsLeaning=~
    bIsSkydiveDiveMode=~
    bPlaySkydiveDrift=~
    bSkydiveDriftDelayActive=~
    bSkydiveDriftAnimAllowed=~
    bIsGliderRight=~
    bIsGliderCenter=~
    bIsGliderLeft=~
    bIsGliderForward=~
    bIsGliderBack=~
    bHasDeployedAndIsStandingUpright=~
    LocalAccelForward=~
    LocalAccelRight=~
    LocalVelocityRight=~
    SkydiveAimPitch=~
    SkydiveAimYaw=~
    DeployChuteAnimRate=~
    SkydiveDriftAnimRate=~
    SkydiveDriftAnimRateCurrent=~
    SkydiveFidgetAnimRate=~
    SkydiveFidgetAnimRateCurrent=~
    SkydiveAdditiveAlpha=~
    SkydiveDriftDelay=~
   )
   n2 = Get OwnerGlider() -> (OwnerGlider:Fort Player Parachute Object Reference)
   Output Animation Pose(Result=Blend Poses by bool(
     BlendPose_0=Sequence Player 'Maverick_Glide_FrontEnd_GLIDER'() {sequence=/BRCosmetics/Animation/Game/MainPlayer/Skydive/ParaGlide/Maverick/Maverick_Glide_FrontEnd_GLIDER}
     BlendPose_1=Blendspace Player 'BS_Maverick_GLIDER'(X=@n1.SkydiveAimYaw, Y=@n1.SkydiveAimPitch) {blendSpace=/BRCosmetics/Animation/Game/MainPlayer/Skydive/ParaGlide/Maverick/BS_Maverick_GLIDER}
     bActiveValue=Get IsCosmeticPreview!(self:Fort Player Parachute Object Reference=@n2) -> (IsCosmeticPreview:bool)
    ) {blendType=Linear})
  ]
 }
 Closed {
  path=AnimGraph.AnimGraphNode_StateMachine_0.Locomotion.AnimStateNode_2.Closed
  nodes [
   n1 = Get OwnerGlider()
   Output Animation Pose(Result=Blend Poses by bool(
     BlendPose_0=Sequence Player 'Maverick_Closed_GLIDER'() {sequence=/BRCosmetics/Animation/Game/MainPlayer/Skydive/ParaGlide/Maverick/Maverick_Closed_GLIDER}
     BlendPose_1=Sequence Player 'Maverick_Closed_GLIDER'() {sequence=&Maverick_Closed_GLIDER}
     bActiveValue=Get IsCosmeticPreview(self=@n1)
    ) {blendType=Linear})
  ]
 }
 Open_Into {
  path=AnimGraph.AnimGraphNode_StateMachine_0.Locomotion.AnimStateNode_3.Open_Into
  nodes [
   n1 = Break Fort Anim Input Skydiving(FortAnimInput_Skydiving=Get SkydivingState()) -> (
    ParaGilderRootOffset=~
    ParaGliderRootRotationOverride=~
    bUseParaGlideRootModifier=~
    bIsSkydivingFromLaunchPad=~
    bIsSkydivingFromBus=~
    bIsInVortex=~
    bIsDiving=~
    bIsParachuteOpen=~
    bIsSkydiving=~
    bIsParachuteLeaning=~
    bIsSkydiveLeaning=~
    bIsLeaning=~
    bIsSkydiveDiveMode=~
    bPlaySkydiveDrift=~
    bSkydiveDriftDelayActive=~
    bSkydiveDriftAnimAllowed=~
    bIsGliderRight=~
    bIsGliderCenter=~
    bIsGliderLeft=~
    bIsGliderForward=~
    bIsGliderBack=~
    bHasDeployedAndIsStandingUpright=~
    LocalAccelForward=~
    LocalAccelRight=~
    LocalVelocityRight=~
    SkydiveAimPitch=~
    SkydiveAimYaw=~
    DeployChuteAnimRate=~
    SkydiveDriftAnimRate=~
    SkydiveDriftAnimRateCurrent=~
    SkydiveFidgetAnimRate=~
    SkydiveFidgetAnimRateCurrent=~
    SkydiveAdditiveAlpha=~
    SkydiveDriftDelay=~
   )
   n2 = Get OwnerGlider()
   Output Animation Pose(Result=Blend Poses by bool(
     BlendPose_0=Sequence Player 'Maverick_Into_FrontEnd_GLIDER'() {bLoopAnimation=false, sequence=/BRCosmetics/Animation/Game/MainPlayer/Skydive/ParaGlide/Maverick/Maverick_Into_FrontEnd_GLIDER}
     BlendPose_1=Blendspace Player 'BS_Maverick_Into_GLIDER'(X=@n1.SkydiveAimYaw, Y=@n1.SkydiveAimPitch) {bLoop=false, blendSpace=/BRCosmetics/Animation/Game/MainPlayer/Skydive/ParaGlide/Maverick/BS_Maverick_Into_GLIDER}
     bActiveValue=Get IsCosmeticPreview(self=@n2)
    ) {blendType=Linear})
  ]
 }
 Transition {
  path=AnimGraph.AnimGraphNode_StateMachine_0.Locomotion.AnimStateTransitionNode_0.Transition
  nodes [
   Result()
   "Equal (Float)"!(A:double="Get Relevant Anim Time Remaining (Open_Into)"!() -> (ReturnValue:float) {is=title K2Node_AnimGetter}, B:double=0) -> (ReturnValue:bool) {is=title K2Node_CallFunction}
  ]
 }
 Transition {
  path=AnimGraph.AnimGraphNode_StateMachine_0.Locomotion.AnimStateTransitionNode_1.Transition
  nodes [
   n1 = Get SkydivingState()
   Result(bCanEnterTransition=OR Boolean(
     A=Break Fort Anim Input Skydiving(FortAnimInput_Skydiving=@n1) -> (
      ParaGilderRootOffset=~
      ParaGliderRootRotationOverride=~
      bUseParaGlideRootModifier=~
      bIsSkydivingFromLaunchPad=~
      bIsSkydivingFromBus=~
      bIsInVortex=~
      bIsDiving=~
      bIsParachuteOpen=~
      bIsSkydiving=~
      bIsParachuteLeaning=~
      bIsSkydiveLeaning=~
      bIsLeaning=~
      bIsSkydiveDiveMode=~
      bPlaySkydiveDrift=~
      bSkydiveDriftDelayActive=~
      bSkydiveDriftAnimAllowed=~
      bIsGliderRight=~
      bIsGliderCenter=~
      bIsGliderLeft=~
      bIsGliderForward=~
      bIsGliderBack=~
      bHasDeployedAndIsStandingUpright=~
      LocalAccelForward=~
      LocalAccelRight=~
      LocalVelocityRight=~
      SkydiveAimPitch=~
      SkydiveAimYaw=~
      DeployChuteAnimRate=~
      SkydiveDriftAnimRate=~
      SkydiveDriftAnimRateCurrent=~
      SkydiveFidgetAnimRate=~
      SkydiveFidgetAnimRateCurrent=~
      SkydiveAdditiveAlpha=~
      SkydiveDriftDelay=~
     ).bIsParachuteOpen
     B=Get bForceOpen() -> (bForceOpen:bool)
    ))
  ]
 }
 Transition {
  path=AnimGraph.AnimGraphNode_StateMachine_0.Locomotion.AnimStateTransitionNode_2.Transition
  nodes [
   n1 = Get bForceOpen()
   n2 = Break Fort Anim Input Skydiving(FortAnimInput_Skydiving=Get SkydivingState()) -> (
    ParaGilderRootOffset=~
    ParaGliderRootRotationOverride=~
    bUseParaGlideRootModifier=~
    bIsSkydivingFromLaunchPad=~
    bIsSkydivingFromBus=~
    bIsInVortex=~
    bIsDiving=~
    bIsParachuteOpen=~
    bIsSkydiving=~
    bIsParachuteLeaning=~
    bIsSkydiveLeaning=~
    bIsLeaning=~
    bIsSkydiveDiveMode=~
    bPlaySkydiveDrift=~
    bSkydiveDriftDelayActive=~
    bSkydiveDriftAnimAllowed=~
    bIsGliderRight=~
    bIsGliderCenter=~
    bIsGliderLeft=~
    bIsGliderForward=~
    bIsGliderBack=~
    bHasDeployedAndIsStandingUpright=~
    LocalAccelForward=~
    LocalAccelRight=~
    LocalVelocityRight=~
    SkydiveAimPitch=~
    SkydiveAimYaw=~
    DeployChuteAnimRate=~
    SkydiveDriftAnimRate=~
    SkydiveDriftAnimRateCurrent=~
    SkydiveFidgetAnimRate=~
    SkydiveFidgetAnimRateCurrent=~
    SkydiveAdditiveAlpha=~
    SkydiveDriftDelay=~
   )
   Result(bCanEnterTransition=AND Boolean(A=NOT Boolean(A=@n2.bIsParachuteOpen), B=NOT Boolean(A=@n1)))
  ]
 }
 Transition {
  path=AnimGraph.AnimGraphNode_StateMachine_0.Locomotion.AnimStateTransitionNode_3.Transition
  nodes [
   n1 = Get bForceOpen()
   n2 = Break Fort Anim Input Skydiving(FortAnimInput_Skydiving=Get SkydivingState()) -> (
    ParaGilderRootOffset=~
    ParaGliderRootRotationOverride=~
    bUseParaGlideRootModifier=~
    bIsSkydivingFromLaunchPad=~
    bIsSkydivingFromBus=~
    bIsInVortex=~
    bIsDiving=~
    bIsParachuteOpen=~
    bIsSkydiving=~
    bIsParachuteLeaning=~
    bIsSkydiveLeaning=~
    bIsLeaning=~
    bIsSkydiveDiveMode=~
    bPlaySkydiveDrift=~
    bSkydiveDriftDelayActive=~
    bSkydiveDriftAnimAllowed=~
    bIsGliderRight=~
    bIsGliderCenter=~
    bIsGliderLeft=~
    bIsGliderForward=~
    bIsGliderBack=~
    bHasDeployedAndIsStandingUpright=~
    LocalAccelForward=~
    LocalAccelRight=~
    LocalVelocityRight=~
    SkydiveAimPitch=~
    SkydiveAimYaw=~
    DeployChuteAnimRate=~
    SkydiveDriftAnimRate=~
    SkydiveDriftAnimRateCurrent=~
    SkydiveFidgetAnimRate=~
    SkydiveFidgetAnimRateCurrent=~
    SkydiveAdditiveAlpha=~
    SkydiveDriftDelay=~
   )
   Result(bCanEnterTransition=AND Boolean(A=NOT Boolean(A=@n2.bIsParachuteOpen), B=NOT Boolean(A=@n1)))
  ]
 }
 Locomotion_GliderItem {
  entry=Closed
  path=AnimGraph.AnimGraphNode_StateMachine_1.Locomotion_GliderItem
  ids=id
  nodes [
   AnimStateNode_0 = Idle_Open()
   AnimStateNode_1 = ForceOpen_or_GliderItem() {is=title AnimStateNode, stateEntered {guid=#3, notifyName=DeployEntered}}
   AnimStateNode_2 = Closed() {stateEntered {guid=#1, notifyName=ClosedEntered}}
   AnimStateNode_3 = Open_Into() {stateEntered {guid=#2, notifyName=DeployEntered}}
   AnimStateTransitionNode_0 = AnimStateTransitionNode() {bAutomaticRuleBasedOnSequencePlayerInState=true, blendMode=Linear, crossfadeDuration=0}
   AnimStateTransitionNode_1 = AnimStateTransitionNode() {crossfadeDuration=0}
   AnimStateTransitionNode_2 = AnimStateTransitionNode() {crossfadeDuration=0}
   AnimStateTransitionNode_3 = AnimStateTransitionNode() {crossfadeDuration=0, priorityOrder=2}
   AnimStateTransitionNode_4 = AnimStateTransitionNode() {crossfadeDuration=0}
   AnimStateTransitionNode_5 = AnimStateTransitionNode() {bAutomaticRuleBasedOnSequencePlayerInState=true, crossfadeDuration=0}
   AnimStateTransitionNode_6 = AnimStateTransitionNode() {crossfadeDuration=0}
  ]
  links [
   {
    dst=ForceOpen_or_GliderItem
    rule=AnimGraph.AnimGraphNode_StateMachine_1.Locomotion_GliderItem.AnimStateTransitionNode_4.Transition
    src=Closed
    via=AnimStateTransitionNode_4
   }
   {
    dst=Open_Into
    rule=AnimGraph.AnimGraphNode_StateMachine_1.Locomotion_GliderItem.AnimStateTransitionNode_1.Transition
    src=Closed
    via=AnimStateTransitionNode_1
   }
   {
    dst=Closed
    rule=AnimGraph.AnimGraphNode_StateMachine_1.Locomotion_GliderItem.AnimStateTransitionNode_6.Transition
    src=ForceOpen_or_GliderItem
    via=AnimStateTransitionNode_6
   }
   {
    dst=Idle_Open
    rule=AnimGraph.AnimGraphNode_StateMachine_1.Locomotion_GliderItem.AnimStateTransitionNode_5.Transition
    src=ForceOpen_or_GliderItem
    via=AnimStateTransitionNode_5
   }
   {
    dst=Closed
    rule=AnimGraph.AnimGraphNode_StateMachine_1.Locomotion_GliderItem.AnimStateTransitionNode_2.Transition
    src=Idle_Open
    via=AnimStateTransitionNode_2
   }
   {
    dst=Closed
    rule=AnimGraph.AnimGraphNode_StateMachine_1.Locomotion_GliderItem.AnimStateTransitionNode_3.Transition
    src=Open_Into
    via=AnimStateTransitionNode_3
   }
   {
    dst=Idle_Open
    rule=AnimGraph.AnimGraphNode_StateMachine_1.Locomotion_GliderItem.AnimStateTransitionNode_0.Transition
    src=Open_Into
    via=AnimStateTransitionNode_0
   }
  ]
 }
 Idle_Open {
  path=AnimGraph.AnimGraphNode_StateMachine_1.Locomotion_GliderItem.AnimStateNode_0.Idle_Open
  nodes [
   Apply Additive!(Base, Additive, bAlphaBoolEnabled:bool=Get IsCosmeticPreview(self=Get OwnerGlider())) -> (Pose) {alphaInputType=Bool, alphaBoolBlend {blendOption=Linear}}
   n1 = Break Fort Anim Input Skydiving(FortAnimInput_Skydiving=Get SkydivingState()) -> (
    ParaGilderRootOffset=~
    ParaGliderRootRotationOverride=~
    bUseParaGlideRootModifier=~
    bIsSkydivingFromLaunchPad=~
    bIsSkydivingFromBus=~
    bIsInVortex=~
    bIsDiving=~
    bIsParachuteOpen=~
    bIsSkydiving=~
    bIsParachuteLeaning=~
    bIsSkydiveLeaning=~
    bIsLeaning=~
    bIsSkydiveDiveMode=~
    bPlaySkydiveDrift=~
    bSkydiveDriftDelayActive=~
    bSkydiveDriftAnimAllowed=~
    bIsGliderRight=~
    bIsGliderCenter=~
    bIsGliderLeft=~
    bIsGliderForward=~
    bIsGliderBack=~
    bHasDeployedAndIsStandingUpright=~
    LocalAccelForward=~
    LocalAccelRight=~
    LocalVelocityRight=~
    SkydiveAimPitch=~
    SkydiveAimYaw=~
    DeployChuteAnimRate=~
    SkydiveDriftAnimRate=~
    SkydiveDriftAnimRateCurrent=~
    SkydiveFidgetAnimRate=~
    SkydiveFidgetAnimRateCurrent=~
    SkydiveAdditiveAlpha=~
    SkydiveDriftDelay=~
   )
   n2 = Get OwnerGlider()
   Output Animation Pose(Result=Blend Poses by bool(
     BlendPose_0=Sequence Player 'Maverick_Glide_GLIDER'() {sequence=/BRCosmetics/Animation/Game/MainPlayer/Skydive/ParaGlide/Maverick/Maverick_Glide_GLIDER}
     BlendPose_1=Blendspace Player 'BS_Maverick_GLIDER'(X=@n1.SkydiveAimYaw, Y=@n1.SkydiveAimPitch) {blendSpace=&BS_Maverick_GLIDER}
     bActiveValue=Get IsCosmeticPreview(self=@n2)
    ) {blendType=Linear})
  ]
 }
 ForceOpen_or_GliderItem {
  path=AnimGraph.AnimGraphNode_StateMachine_1.Locomotion_GliderItem.AnimStateNode_1.ForceOpen_or_GliderItem
  nodes [
   n1 = Break Fort Anim Input Skydiving(FortAnimInput_Skydiving=Get SkydivingState()) -> (
    ParaGilderRootOffset=~
    ParaGliderRootRotationOverride=~
    bUseParaGlideRootModifier=~
    bIsSkydivingFromLaunchPad=~
    bIsSkydivingFromBus=~
    bIsInVortex=~
    bIsDiving=~
    bIsParachuteOpen=~
    bIsSkydiving=~
    bIsParachuteLeaning=~
    bIsSkydiveLeaning=~
    bIsLeaning=~
    bIsSkydiveDiveMode=~
    bPlaySkydiveDrift=~
    bSkydiveDriftDelayActive=~
    bSkydiveDriftAnimAllowed=~
    bIsGliderRight=~
    bIsGliderCenter=~
    bIsGliderLeft=~
    bIsGliderForward=~
    bIsGliderBack=~
    bHasDeployedAndIsStandingUpright=~
    LocalAccelForward=~
    LocalAccelRight=~
    LocalVelocityRight=~
    SkydiveAimPitch=~
    SkydiveAimYaw=~
    DeployChuteAnimRate=~
    SkydiveDriftAnimRate=~
    SkydiveDriftAnimRateCurrent=~
    SkydiveFidgetAnimRate=~
    SkydiveFidgetAnimRateCurrent=~
    SkydiveAdditiveAlpha=~
    SkydiveDriftDelay=~
   )
   n2 = Blendspace Player 'BS_Maverick_Into_GLIDER'(X=@n1.SkydiveAimYaw, Y=@n1.SkydiveAimPitch) {bLoop=false, blendSpace=&BS_Maverick_Into_GLIDER, startPosition=0.0063}
   Output Animation Pose(Result=@n2)
  ]
 }
 Closed {
  path=AnimGraph.AnimGraphNode_StateMachine_1.Locomotion_GliderItem.AnimStateNode_2.Closed
  nodes [
   n1 = Get OwnerGlider()
   Output Animation Pose(Result=Blend Poses by bool(
     BlendPose_0=Sequence Player 'Maverick_Closed_GLIDER'() {sequence=&Maverick_Closed_GLIDER}
     BlendPose_1=Sequence Player 'Maverick_Closed_GLIDER'() {sequence=&Maverick_Closed_GLIDER}
     bActiveValue=Get IsCosmeticPreview(self=@n1)
    ) {blendType=Linear})
  ]
 }
 Open_Into {
  path=AnimGraph.AnimGraphNode_StateMachine_1.Locomotion_GliderItem.AnimStateNode_3.Open_Into
  nodes [
   Apply Additive(bAlphaBoolEnabled=Get IsCosmeticPreview(self=Get OwnerGlider())) {alphaInputType=Bool, alphaBoolBlend {blendOption=Linear}}
   n1 = Break Fort Anim Input Skydiving(FortAnimInput_Skydiving=Get SkydivingState()) -> (
    ParaGilderRootOffset=~
    ParaGliderRootRotationOverride=~
    bUseParaGlideRootModifier=~
    bIsSkydivingFromLaunchPad=~
    bIsSkydivingFromBus=~
    bIsInVortex=~
    bIsDiving=~
    bIsParachuteOpen=~
    bIsSkydiving=~
    bIsParachuteLeaning=~
    bIsSkydiveLeaning=~
    bIsLeaning=~
    bIsSkydiveDiveMode=~
    bPlaySkydiveDrift=~
    bSkydiveDriftDelayActive=~
    bSkydiveDriftAnimAllowed=~
    bIsGliderRight=~
    bIsGliderCenter=~
    bIsGliderLeft=~
    bIsGliderForward=~
    bIsGliderBack=~
    bHasDeployedAndIsStandingUpright=~
    LocalAccelForward=~
    LocalAccelRight=~
    LocalVelocityRight=~
    SkydiveAimPitch=~
    SkydiveAimYaw=~
    DeployChuteAnimRate=~
    SkydiveDriftAnimRate=~
    SkydiveDriftAnimRateCurrent=~
    SkydiveFidgetAnimRate=~
    SkydiveFidgetAnimRateCurrent=~
    SkydiveAdditiveAlpha=~
    SkydiveDriftDelay=~
   )
   n2 = Get OwnerGlider()
   Output Animation Pose(Result=Blend Poses by bool(
     BlendPose_0=Sequence Player 'Maverick_Into_GLIDER'() {bLoopAnimation=false, sequence=/BRCosmetics/Animation/Game/MainPlayer/Skydive/ParaGlide/Maverick/Maverick_Into_GLIDER}
     BlendPose_1=Blendspace Player 'BS_Maverick_Into_GLIDER'(X=@n1.SkydiveAimYaw, Y=@n1.SkydiveAimPitch) {bLoop=false, blendSpace=&BS_Maverick_Into_GLIDER}
     bActiveValue=Get IsCosmeticPreview(self=@n2)
    ) {blendType=Linear})
  ]
 }
 Transition {
  path=AnimGraph.AnimGraphNode_StateMachine_1.Locomotion_GliderItem.AnimStateTransitionNode_0.Transition
  nodes [
   Result()
   "Equal (Float)"(A="Get Relevant Anim Time Remaining (Open_Into)"(), B=0)
  ]
 }
 Transition {
  path=AnimGraph.AnimGraphNode_StateMachine_1.Locomotion_GliderItem.AnimStateTransitionNode_1.Transition
  nodes [
   n1 = Break Fort Anim Input Skydiving(FortAnimInput_Skydiving=Get SkydivingState()) -> (
    ParaGilderRootOffset=~
    ParaGliderRootRotationOverride=~
    bUseParaGlideRootModifier=~
    bIsSkydivingFromLaunchPad=~
    bIsSkydivingFromBus=~
    bIsInVortex=~
    bIsDiving=~
    bIsParachuteOpen=~
    bIsSkydiving=~
    bIsParachuteLeaning=~
    bIsSkydiveLeaning=~
    bIsLeaning=~
    bIsSkydiveDiveMode=~
    bPlaySkydiveDrift=~
    bSkydiveDriftDelayActive=~
    bSkydiveDriftAnimAllowed=~
    bIsGliderRight=~
    bIsGliderCenter=~
    bIsGliderLeft=~
    bIsGliderForward=~
    bIsGliderBack=~
    bHasDeployedAndIsStandingUpright=~
    LocalAccelForward=~
    LocalAccelRight=~
    LocalVelocityRight=~
    SkydiveAimPitch=~
    SkydiveAimYaw=~
    DeployChuteAnimRate=~
    SkydiveDriftAnimRate=~
    SkydiveDriftAnimRateCurrent=~
    SkydiveFidgetAnimRate=~
    SkydiveFidgetAnimRateCurrent=~
    SkydiveAdditiveAlpha=~
    SkydiveDriftDelay=~
   )
   n2 = Get IsCosmeticPreview(self=Get OwnerGlider())
   Result(bCanEnterTransition=AND Boolean(
     A=OR Boolean!(A=@n1.bIsSkydivingFromLaunchPad, B=@n1.bIsSkydivingFromBus, C:bool=@n1.bIsInVortex, D:bool=@n2) -> (ReturnValue)
     B=@n1.bIsParachuteOpen
    ))
  ]
 }
 Transition {
  path=AnimGraph.AnimGraphNode_StateMachine_1.Locomotion_GliderItem.AnimStateTransitionNode_2.Transition
  nodes [
   n1 = Get bForceOpen()
   n2 = Break Fort Anim Input Skydiving(FortAnimInput_Skydiving=Get SkydivingState()) -> (
    ParaGilderRootOffset=~
    ParaGliderRootRotationOverride=~
    bUseParaGlideRootModifier=~
    bIsSkydivingFromLaunchPad=~
    bIsSkydivingFromBus=~
    bIsInVortex=~
    bIsDiving=~
    bIsParachuteOpen=~
    bIsSkydiving=~
    bIsParachuteLeaning=~
    bIsSkydiveLeaning=~
    bIsLeaning=~
    bIsSkydiveDiveMode=~
    bPlaySkydiveDrift=~
    bSkydiveDriftDelayActive=~
    bSkydiveDriftAnimAllowed=~
    bIsGliderRight=~
    bIsGliderCenter=~
    bIsGliderLeft=~
    bIsGliderForward=~
    bIsGliderBack=~
    bHasDeployedAndIsStandingUpright=~
    LocalAccelForward=~
    LocalAccelRight=~
    LocalVelocityRight=~
    SkydiveAimPitch=~
    SkydiveAimYaw=~
    DeployChuteAnimRate=~
    SkydiveDriftAnimRate=~
    SkydiveDriftAnimRateCurrent=~
    SkydiveFidgetAnimRate=~
    SkydiveFidgetAnimRateCurrent=~
    SkydiveAdditiveAlpha=~
    SkydiveDriftDelay=~
   )
   Result(bCanEnterTransition=AND Boolean(A=NOT Boolean(A=@n2.bIsParachuteOpen), B=NOT Boolean(A=@n1)))
  ]
 }
 Transition {
  path=AnimGraph.AnimGraphNode_StateMachine_1.Locomotion_GliderItem.AnimStateTransitionNode_3.Transition
  nodes [
   n1 = Get bForceOpen()
   n2 = Break Fort Anim Input Skydiving(FortAnimInput_Skydiving=Get SkydivingState()) -> (
    ParaGilderRootOffset=~
    ParaGliderRootRotationOverride=~
    bUseParaGlideRootModifier=~
    bIsSkydivingFromLaunchPad=~
    bIsSkydivingFromBus=~
    bIsInVortex=~
    bIsDiving=~
    bIsParachuteOpen=~
    bIsSkydiving=~
    bIsParachuteLeaning=~
    bIsSkydiveLeaning=~
    bIsLeaning=~
    bIsSkydiveDiveMode=~
    bPlaySkydiveDrift=~
    bSkydiveDriftDelayActive=~
    bSkydiveDriftAnimAllowed=~
    bIsGliderRight=~
    bIsGliderCenter=~
    bIsGliderLeft=~
    bIsGliderForward=~
    bIsGliderBack=~
    bHasDeployedAndIsStandingUpright=~
    LocalAccelForward=~
    LocalAccelRight=~
    LocalVelocityRight=~
    SkydiveAimPitch=~
    SkydiveAimYaw=~
    DeployChuteAnimRate=~
    SkydiveDriftAnimRate=~
    SkydiveDriftAnimRateCurrent=~
    SkydiveFidgetAnimRate=~
    SkydiveFidgetAnimRateCurrent=~
    SkydiveAdditiveAlpha=~
    SkydiveDriftDelay=~
   )
   Result(bCanEnterTransition=AND Boolean(A=NOT Boolean(A=@n2.bIsParachuteOpen), B=NOT Boolean(A=@n1)))
  ]
 }
 Transition {
  path=AnimGraph.AnimGraphNode_StateMachine_1.Locomotion_GliderItem.AnimStateTransitionNode_4.Transition
  nodes [
   n1 = Break Fort Anim Input Skydiving(FortAnimInput_Skydiving=Get SkydivingState()) -> (
    ParaGilderRootOffset=~
    ParaGliderRootRotationOverride=~
    bUseParaGlideRootModifier=~
    bIsSkydivingFromLaunchPad=~
    bIsSkydivingFromBus=~
    bIsInVortex=~
    bIsDiving=~
    bIsParachuteOpen=~
    bIsSkydiving=~
    bIsParachuteLeaning=~
    bIsSkydiveLeaning=~
    bIsLeaning=~
    bIsSkydiveDiveMode=~
    bPlaySkydiveDrift=~
    bSkydiveDriftDelayActive=~
    bSkydiveDriftAnimAllowed=~
    bIsGliderRight=~
    bIsGliderCenter=~
    bIsGliderLeft=~
    bIsGliderForward=~
    bIsGliderBack=~
    bHasDeployedAndIsStandingUpright=~
    LocalAccelForward=~
    LocalAccelRight=~
    LocalVelocityRight=~
    SkydiveAimPitch=~
    SkydiveAimYaw=~
    DeployChuteAnimRate=~
    SkydiveDriftAnimRate=~
    SkydiveDriftAnimRateCurrent=~
    SkydiveFidgetAnimRate=~
    SkydiveFidgetAnimRateCurrent=~
    SkydiveAdditiveAlpha=~
    SkydiveDriftDelay=~
   )
   Result(bCanEnterTransition=AND Boolean!(
     A=NOT Boolean(A=@n1.bIsSkydivingFromLaunchPad)
     B=NOT Boolean(A=@n1.bIsSkydivingFromBus)
     C:bool=NOT Boolean(A=@n1.bIsInVortex)
     D:bool=Get bForceOpen()
    ) -> (ReturnValue))
  ]
 }
 Transition {path=AnimGraph.AnimGraphNode_StateMachine_1.Locomotion_GliderItem.AnimStateTransitionNode_5.Transition, nodes [Result()]}
 Transition {
  path=AnimGraph.AnimGraphNode_StateMachine_1.Locomotion_GliderItem.AnimStateTransitionNode_6.Transition
  nodes [
   n1 = Get bForceOpen()
   n2 = Break Fort Anim Input Skydiving(FortAnimInput_Skydiving=Get SkydivingState()) -> (
    ParaGilderRootOffset=~
    ParaGliderRootRotationOverride=~
    bUseParaGlideRootModifier=~
    bIsSkydivingFromLaunchPad=~
    bIsSkydivingFromBus=~
    bIsInVortex=~
    bIsDiving=~
    bIsParachuteOpen=~
    bIsSkydiving=~
    bIsParachuteLeaning=~
    bIsSkydiveLeaning=~
    bIsLeaning=~
    bIsSkydiveDiveMode=~
    bPlaySkydiveDrift=~
    bSkydiveDriftDelayActive=~
    bSkydiveDriftAnimAllowed=~
    bIsGliderRight=~
    bIsGliderCenter=~
    bIsGliderLeft=~
    bIsGliderForward=~
    bIsGliderBack=~
    bHasDeployedAndIsStandingUpright=~
    LocalAccelForward=~
    LocalAccelRight=~
    LocalVelocityRight=~
    SkydiveAimPitch=~
    SkydiveAimYaw=~
    DeployChuteAnimRate=~
    SkydiveDriftAnimRate=~
    SkydiveDriftAnimRateCurrent=~
    SkydiveFidgetAnimRate=~
    SkydiveFidgetAnimRateCurrent=~
    SkydiveAdditiveAlpha=~
    SkydiveDriftDelay=~
   )
   Result(bCanEnterTransition=AND Boolean!(A=NOT Boolean(A=@n2.bIsParachuteOpen), B=NOT Boolean(A=@n1)) -> (ReturnValue))
  ]
 }
 SFoil_Pose {
  entry=Closed
  path=AnimGraph.AnimGraphNode_StateMachine_2.SFoil_Pose
  ids=id
  nodes [
   AnimStateNode_0 = Closed()
   AnimStateNode_1 = Attack_Position() {is=title AnimStateNode}
   AnimStateNode_2 = Into() {is=title AnimStateNode}
   AnimStateNode_3 = Out() {is=title AnimStateNode}
   AnimStateTransitionNode_0 = AnimStateTransitionNode() {crossfadeDuration=0}
   AnimStateTransitionNode_1 = AnimStateTransitionNode() {bAutomaticRuleBasedOnSequencePlayerInState=true, crossfadeDuration=0}
   AnimStateTransitionNode_2 = AnimStateTransitionNode() {crossfadeDuration=0, priorityOrder=2}
   AnimStateTransitionNode_3 = AnimStateTransitionNode() {bAutomaticRuleBasedOnSequencePlayerInState=true, crossfadeDuration=0}
   AnimStateTransitionNode_4 = AnimStateTransitionNode() {crossfadeDuration=0}
  ]
  links [
   {
    dst=Closed
    rule=AnimGraph.AnimGraphNode_StateMachine_2.SFoil_Pose.AnimStateTransitionNode_4.Transition
    src=Attack_Position
    via=AnimStateTransitionNode_4
   }
   {
    dst=Out
    rule=AnimGraph.AnimGraphNode_StateMachine_2.SFoil_Pose.AnimStateTransitionNode_2.Transition
    src=Attack_Position
    via=AnimStateTransitionNode_2
   }
   {
    dst=Into
    rule=AnimGraph.AnimGraphNode_StateMachine_2.SFoil_Pose.AnimStateTransitionNode_0.Transition
    src=Closed
    via=AnimStateTransitionNode_0
   }
   {
    dst=Attack_Position
    rule=AnimGraph.AnimGraphNode_StateMachine_2.SFoil_Pose.AnimStateTransitionNode_1.Transition
    src=Into
    via=AnimStateTransitionNode_1
   }
   {
    dst=Closed
    rule=AnimGraph.AnimGraphNode_StateMachine_2.SFoil_Pose.AnimStateTransitionNode_3.Transition
    src=Out
    via=AnimStateTransitionNode_3
   }
  ]
 }
 Closed {
  path=AnimGraph.AnimGraphNode_StateMachine_2.SFoil_Pose.AnimStateNode_0.Closed
  nodes [Output Animation Pose(Result="Sequence Player (additive) 'Maverick_SFoils_Closed_Pose_GLIDER'"() {sequence=/BRCosmetics/Animation/Game/MainPlayer/Skydive/ParaGlide/Maverick/Maverick_SFoils_Closed_Pose_GLIDER})]
 }
 Attack_Position {
  path=AnimGraph.AnimGraphNode_StateMachine_2.SFoil_Pose.AnimStateNode_1.Attack_Position
  nodes [Output Animation Pose(Result="Sequence Player (additive) 'Maverick_SFoils_Attack_Pose_GLIDER'"() {sequence=/BRCosmetics/Animation/Game/MainPlayer/Skydive/ParaGlide/Maverick/Maverick_SFoils_Attack_Pose_GLIDER})]
 }
 Into {
  path=AnimGraph.AnimGraphNode_StateMachine_2.SFoil_Pose.AnimStateNode_2.Into
  nodes [Output Animation Pose(Result="Sequence Player (additive) 'Maverick_SFoils_Into_GLIDER'"() {bLoopAnimation=false, sequence=/BRCosmetics/Animation/Game/MainPlayer/Skydive/ParaGlide/Maverick/Maverick_SFoils_Into_GLIDER})]
 }
 Out {
  path=AnimGraph.AnimGraphNode_StateMachine_2.SFoil_Pose.AnimStateNode_3.Out
  nodes [Output Animation Pose(Result="Sequence Player (additive) 'Maverick_SFoils_Out_GLIDER'"() {bLoopAnimation=false, sequence=/BRCosmetics/Animation/Game/MainPlayer/Skydive/ParaGlide/Maverick/Maverick_SFoils_Out_GLIDER})]
 }
 Transition {
  path=AnimGraph.AnimGraphNode_StateMachine_2.SFoil_Pose.AnimStateTransitionNode_0.Transition
  nodes [Result(bCanEnterTransition=Get SFoil_Attack_Position() -> (SFoil_Attack_Position:bool))]
 }
 Transition {path=AnimGraph.AnimGraphNode_StateMachine_2.SFoil_Pose.AnimStateTransitionNode_1.Transition, nodes [Result()]}
 Transition {
  path=AnimGraph.AnimGraphNode_StateMachine_2.SFoil_Pose.AnimStateTransitionNode_2.Transition
  nodes [
   n1 = Get SFoil_SnapClosed() -> (SFoil_SnapClosed:bool)
   n2 = Get SFoil_Attack_Position()
   Result(bCanEnterTransition=AND Boolean(A=NOT Boolean(A=@n1), B=NOT Boolean(A=@n2)))
  ]
 }
 Transition {path=AnimGraph.AnimGraphNode_StateMachine_2.SFoil_Pose.AnimStateTransitionNode_3.Transition, nodes [Result()]}
 Transition {
  path=AnimGraph.AnimGraphNode_StateMachine_2.SFoil_Pose.AnimStateTransitionNode_4.Transition
  nodes [
   n1 = Get SFoil_Attack_Position()
   Result(bCanEnterTransition=AND Boolean(A=Get SFoil_SnapClosed(), B=NOT Boolean(A=@n1)))
  ]
 }
 EventGraph {nodes [
   AnimNotify_TurnONEngines() {is=title K2Node_Event}
   > SetGliderFullyDeployed!(execute, self:Fort Glider Anim Instance Object Reference, bIsFullyDeployed:bool=true) -> (then)
   n1 = AnimNotify_ClosedEntered() {is=title K2Node_Event}
   n2 = GetOwningComponent()
   HideBoneByName!(
    execute=[@n1.then, @n4.then]
    self:Skinned Mesh Component Object Reference=@n2
    BoneName:Name=Glider_Base
    PhysBodyOption:EPhysBodyOp Enum=PBO_None
   ) -> (then)
   AnimNotify_DeployEntered() {is=title K2Node_Event}
   > Set SFoil_Attack_Position_Int(SFoil_Attack_Position_Int:int=0) -> (Output_Get:int=0)
   > Set SFoil_SnapClosed(SFoil_SnapClosed:bool=true) -> (Output_Get:bool)
   > Set SFoil_Attack_Position(SFoil_Attack_Position:bool=false) -> (Output_Get:bool)
   > Delay(Duration=0.1)
   > UnHideBoneByName!(execute, self:Skinned Mesh Component Object Reference=@n2, BoneName:Name=Glider_Base) -> (then)
   n3 = Event BlueprintProcessAnimInputEvent()
   > Branch(Condition=IsNamed(self=@n3.AnimInputEvent, InEventName=aie.Glider.PrimaryFire))
   > Set AllowPlayerInput(AllowPlayerInput:bool=true) -> (Output_Get:bool)
   > Set AllowHardDeckClose(AllowHardDeckClose:bool=false) -> (Output_Get:bool)
   > Get Player Info!(execute:Exec) -> (then:Exec) {is=title K2Node_MacroInstance}
   AnimNotify_Lock_SFoils_InAttackPosition() {is=title K2Node_Event}
   > Set SFoil_Attack_Position_Int(SFoil_Attack_Position_Int=1) -> (Output_Get=0)
   > Set SFoil_Attack_Position(SFoil_Attack_Position=true)
   n4 = Event BlueprintInitializeAnimation()
   Set Time(Time:double=0) -> (Output_Get:double)
   > n5 = ForwardBack_Value_Interp!(
    execute
    Time:double=Get Time() -> (Time:double)
    Currnet_Value:double=Get Example() -> (Example:double)
    Idle_Value:double=1
    Idle_InterpSpeed:double=1
    Forward_Value:double=2
    Forward_InterpSpeed:double=2
    Back_Value:double=1
    Back_InterpSpeed:double=2
    LocalAccelForward:double=Break Fort Anim Input Skydiving(FortAnimInput_Skydiving=Get SkydivingState()) -> (
     ParaGilderRootOffset=~
     ParaGliderRootRotationOverride=~
     bUseParaGlideRootModifier=~
     bIsSkydivingFromLaunchPad=~
     bIsSkydivingFromBus=~
     bIsInVortex=~
     bIsDiving=~
     bIsParachuteOpen=~
     bIsSkydiving=~
     bIsParachuteLeaning=~
     bIsSkydiveLeaning=~
     bIsLeaning=~
     bIsSkydiveDiveMode=~
     bPlaySkydiveDrift=~
     bSkydiveDriftDelayActive=~
     bSkydiveDriftAnimAllowed=~
     bIsGliderRight=~
     bIsGliderCenter=~
     bIsGliderLeft=~
     bIsGliderForward=~
     bIsGliderBack=~
     bHasDeployedAndIsStandingUpright=~
     LocalAccelForward=~
     LocalAccelRight=~
     LocalVelocityRight=~
     SkydiveAimPitch=~
     SkydiveAimYaw=~
     DeployChuteAnimRate=~
     SkydiveDriftAnimRate=~
     SkydiveDriftAnimRateCurrent=~
     SkydiveFidgetAnimRate=~
     SkydiveFidgetAnimRateCurrent=~
     SkydiveAdditiveAlpha=~
     SkydiveDriftDelay=~
    ).LocalAccelForward
   ) -> (then, New_Value:double)
   > Set Example(Example:double=@n5) -> (Output_Get:double)
   Break Fort Anim Input Skydiving(FortAnimInput_Skydiving=Get SkydivingState()) -> (
    ParaGilderRootOffset=~
    ParaGliderRootRotationOverride=~
    bUseParaGlideRootModifier=~
    bIsSkydivingFromLaunchPad=~
    bIsSkydivingFromBus=~
    bIsInVortex=~
    bIsDiving=~
    bIsParachuteOpen=~
    bIsSkydiving=~
    bIsParachuteLeaning=~
    bIsSkydiveLeaning=~
    bIsLeaning=~
    bIsSkydiveDiveMode=~
    bPlaySkydiveDrift=~
    bSkydiveDriftDelayActive=~
    bSkydiveDriftAnimAllowed=~
    bIsGliderRight=~
    bIsGliderCenter=~
    bIsGliderLeft=~
    bIsGliderForward=~
    bIsGliderBack=~
    bHasDeployedAndIsStandingUpright=~
    LocalAccelForward=~
    LocalAccelRight=~
    LocalVelocityRight=~
    SkydiveAimPitch=~
    SkydiveAimYaw=~
    DeployChuteAnimRate=~
    SkydiveDriftAnimRate=~
    SkydiveDriftAnimRateCurrent=~
    SkydiveFidgetAnimRate=~
    SkydiveFidgetAnimRateCurrent=~
    SkydiveAdditiveAlpha=~
    SkydiveDriftDelay=~
   )
   n6 = Get OwnerGlider()
   Set Actor Hidden In Game!(execute, self:Actor Object Reference=@n6, bNewHidden:bool=true) -> (then) {is=title K2Node_CallFunction}
   Set Actor Hidden In Game(self=@n6, bNewHidden=true)
   Set SFoil_SnapClosed(SFoil_SnapClosed=false)
  ]}
 GetPlayerInfo {nodes [
   Reroute Node()
   > n1 = Branch(Condition=Get UserClosed() -> (UserClosed:bool))
   > Set SFoil_Attack_Position_Int(SFoil_Attack_Position_Int=0) -> (Output_Get=0)
   > Set SFoil_Attack_Position(SFoil_Attack_Position=false)
   Branch(execute=@n1.else, Condition=Get UserOpen() -> (UserOpen:bool))
   > Set SFoil_Attack_Position_Int(SFoil_Attack_Position_Int=1) -> (Output_Get=0)
   > Set SFoil_Attack_Position(SFoil_Attack_Position=true)
   Inputs()
   > Set UserClosed(UserClosed:bool=false) -> (Output_Get:bool)
   > Set UserOpen(UserOpen:bool=false) -> (Output_Get:bool)
   > n2 = Sequence!(execute) -> (then_0, then_1, then_2:Exec)
   n3 = Branch(execute=@n2.then_1, Condition=Get IsDeployAnimPlaying() -> (IsDeployAnimPlaying:bool))
   Set SFoil_SnapClosed(execute=@n3.else, SFoil_SnapClosed=false)
   n4 = Get IsMapOpen() -> (IsMapOpen:bool)
   n5 = Get IsHardDeck() -> (IsHardDeck:bool)
   n6 = Get IsCosmeticPreview(self=Get OwnerGlider())
   > n7 = Branch(Condition=AND Boolean!(A=NOT Boolean(A=@n4), B=NOT Boolean(A=@n5), C:bool=NOT Boolean(A=@n6), D:bool=Get AllowPlayerInput() -> (AllowPlayerInput:bool)) -> (ReturnValue))
   > n8 = Increment Int(Value=Get SFoil_Attack_Position_Int() -> (SFoil_Attack_Position_Int:int=0))
   > n9 = Set SFoil_Attack_Position_Int(SFoil_Attack_Position_Int=@n8) -> (Output_Get=0)
   > n10 = Branch(Condition="integer > integer"(A=@n9, B=1) {is=title K2Node_CallFunction})
   > Set SFoil_Attack_Position_Int(SFoil_Attack_Position_Int=0) -> (Output_Get=0)
   > Set SFoil_Attack_Position(SFoil_Attack_Position=false)
   Set SFoil_Attack_Position_Int(execute=@n10.else, SFoil_Attack_Position_Int=1) -> (Output_Get=0)
   > Set SFoil_Attack_Position(SFoil_Attack_Position=true)
   Branch(execute=@n7.else, Condition=OR Boolean!(A=Get IsHardDeck(), B=Get IsDeployAnimPlaying()) -> (ReturnValue))
   > Set SFoil_Attack_Position(SFoil_Attack_Position=false)
   Outputs!(then:Exec=@n2.then_2)
   n11 = GetCurveValue(CurveName=Deploy)
   Set IsDeployAnimPlaying(execute=@n2.then_0, IsDeployAnimPlaying:bool="Nearly Equal (Float)"(A=@n11, B=1, ErrorTolerance=0.01)) -> (Output_Get:bool)
   > n12 = Branch(Condition=Get IsCosmeticPreview(self=Get OwnerGlider()))
   n13 = Cast To FortPlayerController(
    execute=@n12.else
    Object=GetControllerFromActor!(Actor:Actor Object Reference=Get OwnerFortPawn() -> (OwnerFortPawn:Fort Player Pawn Object Reference)) -> (ReturnValue:Controller Object Reference)
   )
   n14 = GetMiniMapComponent!(self:Fort Player Controller Object Reference=@n13) -> (ReturnValue:Fort Controller Component Minimap Object Reference)
   > Is Valid(InputObject=@n14)
   > Set IsMapOpen(IsMapOpen:bool=IsFullscreenMapVisible!(self:Fort Controller Component Minimap Object Reference=@n14) -> (ReturnValue:bool)) -> (Output_Get:bool)
   n15 = Get OwnerFortPawn()
   > n16 = Set RawHeight(RawHeight:double=BreakVector(InVec=Get Actor Location(self=@n15)).Z) -> (Output_Get:double)
   > Branch(Condition=Get AllowHardDeckClose() -> (AllowHardDeckClose:bool))
   > n17 = Set IsHardDeck(IsHardDeck:bool="float <= float"!(A:double=@n16, B:double=Get HardDeck() -> (HardDeck:double)) -> (ReturnValue:bool) {is=title K2Node_CallFunction}) -> (Output_Get:bool)
   "To String (Float)"(InDouble=0)
   "To String (Boolean)"(InBool=@n17)
   "Equal (Float)"(A=@n11, B=1)
   OR Boolean(A=Get UserOpen(), B=Get UserClosed())
  ]}
 SetUseAltGliderCam {nodes [
   Inputs()
   > n1 = Branch(Condition=Get IsCosmeticPreview(self=Get OwnerGlider()))
   Outputs(then=[@n2.then, @n1.then])
   n2 = SetUseAltGliderCamera!(execute=@n1.else, self:Fort Glider Anim Instance Object Reference, bUseAltCamera:bool=true) -> (then)
  ]}
]
variables [
 AimPitch {type=number}
 AllowHardDeckClose {type=boolean}
 AllowPlayerInput {type=boolean}
 DeployAnimRate {type=number}
 Example {type=number}
 Forward {type=number}
 HardDeck {type=number, default=550}
 IsDeployAnimPlaying {type=boolean, default=true}
 IsHardDeck {type=boolean}
 IsMapOpen {type=boolean}
 IsParachuteOpen {type=boolean}
 PlayerInput {type=boolean}
 RawHeight {type=number}
 Right {type=number}
 SFoil_Attack_Position {type=boolean}
 SFoil_Attack_Position_Int {type=integer}
 SFoil_SnapClosed {type=boolean, default=true}
 SkydiveYaw {type=number}
 Time {type=number}
 UseDynamicsAlpha {type=number}
 UserClosed {type=boolean}
 UserOpen {type=boolean}
]
