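# Animation Blueprint for the Nosh glider cosmetic, switching between closed, deploying and gliding poses, with front-end variants for cosmetic preview.
# Additive layers add lean, turn, jitter, flaps, noshing and chain sway, many gated by animation curves, before Copy Bone corrects frame and door.
# The event graph scales slime and mouth speeds with forward acceleration.
# Notifies and random cooldowns time flaps and noshes, reveal random junk child meshes, and hide the glider when closed.
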AnimBlueprint /BRCosmetics/Gadgets/Assets/VinderTech_GliderChute/Blueprints/Nosh_Glider_AnimBP
target_skeleton=/BRCosmetics/Gadgets/Assets/VinderTech_GliderChute/Glider_Nosh/Meshes/Nosh_Glider_Skeleton
graphs [
 AnimGraph {nodes [
   n1 = Break Fort Anim Input Skydiving!(FortAnimInput_Skydiving:"Fort Anim Input Skydiving Structure (by ref)"=Get SkydivingState() -> (SkydivingState:Fort Anim Input Skydiving Structure)) -> (
    ParaGilderRootOffset:Vector=~
    ParaGliderRootRotationOverride:Rotator=~
    bUseParaGlideRootModifier:bool=~
    bIsSkydivingFromLaunchPad:bool=~
    bIsSkydivingFromBus:bool=~
    bIsInVortex:bool=~
    bIsDiving:bool=~
    bIsParachuteOpen:bool=~
    bIsSkydiving:bool=~
    bIsParachuteLeaning:bool=~
    bIsSkydiveLeaning:bool=~
    bIsLeaning:bool=~
    bIsSkydiveDiveMode:bool=~
    bPlaySkydiveDrift:bool=~
    bSkydiveDriftDelayActive:bool=~
    bSkydiveDriftAnimAllowed:bool=~
    bIsGliderRight:bool=~
    bIsGliderCenter:bool=~
    bIsGliderLeft:bool=~
    bIsGliderForward:bool=~
    bIsGliderBack:bool=~
    bHasDeployedAndIsStandingUpright:bool=~
    LocalAccelForward:float=~
    LocalAccelRight:float=~
    LocalVelocityRight:float=~
    SkydiveAimPitch:float=~
    SkydiveAimYaw:float=~
    DeployChuteAnimRate:float=~
    SkydiveDriftAnimRate:float=~
    SkydiveDriftAnimRateCurrent:float=~
    SkydiveFidgetAnimRate:float=~
    SkydiveFidgetAnimRateCurrent:float=~
    SkydiveAdditiveAlpha:float=~
    SkydiveDriftDelay:float=~
    SkydiveDriftAnim:int
    SkydiveDriftAnimMax:int
    LaunchpadAnim:int
    LocalAccelDir:ESkydivingDirection Enum
    DirectionLast:ESkydivingDirection Enum
    PoseAdditive:Anim Sequence Object Reference
   ) {is=title K2Node_BreakStruct}
   AnimGraphNode_StateMachine_0 = Locomotion() {is=title AnimGraphNode_StateMachine}
   n2 = Apply Additive!(
    Base=@AnimGraphNode_StateMachine_0
    Additive=Blendspace Player 'BS_Nosh_Lean_GLIDER'(X=@n1.LocalAccelRight, Y=@n1.LocalAccelForward) {blendSpace=/BRCosmetics/Animation/Game/MainPlayer/Skydive/ParaGlide/Nosh/BS_Nosh_Lean_GLIDER}
    AlphaCurveName:Name=ParaGlide_Lean
   ) -> (Pose) {alphaInputType=Curve, alphaBoolBlend {blendOption=Linear}, alphaScaleBiasClamp {bias=1, scale=-1}}
   n3 = Blendspace Player 'BS_Nosh_Turn_GLIDER'(X=@n1.SkydiveAimYaw, Y=@n1.SkydiveAimPitch) {blendSpace=/BRCosmetics/Animation/Game/MainPlayer/Skydive/ParaGlide/Nosh/BS_Nosh_Turn_GLIDER}
   n4 = Apply Additive(
    Base=Apply Additive!(
     Base=Apply Additive(Base=@n2, Additive=@n3) {alphaBoolBlend {blendOption=Linear}}
     Additive="Sequence Player (additive) 'Nosh_Jitter_GLIDER'"() {sequence=/BRCosmetics/Animation/Game/MainPlayer/Skydive/ParaGlide/Nosh/Nosh_Jitter_GLIDER}
     Alpha:float
    ) -> (Pose) {alphaBoolBlend {blendOption=Linear}}
    Additive="Sequence Player (additive) 'Nosh_TrayHide_Pose_GLIDER'"() {sequence=/BRCosmetics/Animation/Game/MainPlayer/Skydive/ParaGlide/Nosh/Nosh_TrayHide_Pose_GLIDER}
    AlphaCurveName=Noshin
   ) {alphaInputType=Curve, alphaBoolBlend {blendOption=Linear}, alphaScaleBiasClamp {bias=1, scale=-1}}
   n5 = "Sequence Player (additive) 'Nosh_DoorHide_Pose_GLIDER'"() {sequence=/BRCosmetics/Animation/Game/MainPlayer/Skydive/ParaGlide/Nosh/Nosh_DoorHide_Pose_GLIDER}
   n6 = Get SlimySpeed() -> (SlimySpeed:double)
   AnimGraphNode_StateMachine_1 = Flap() {is=title AnimGraphNode_StateMachine}
   n7 = Apply Additive(
    Base=Apply Additive(
     Base=Apply Additive!(Base=@n4, Additive=@n5, AlphaCurveName:Name=Deploy) -> (Pose) {alphaInputType=Curve, alphaBoolBlend {blendOption=Linear}, alphaScaleBiasClamp {bias=1, scale=-1}}
     Additive="Sequence Player (additive) 'Nosh_Slimies_GLIDER'"!(PlayRate:float=@n6) -> (Pose) {sequence=/BRCosmetics/Animation/Game/MainPlayer/Skydive/ParaGlide/Nosh/Nosh_Slimies_GLIDER}
    ) {alphaBoolBlend {blendOption=Linear}}
    Additive=@AnimGraphNode_StateMachine_1
    Alpha=Get FlapAlpha() -> (FlapAlpha:double)
   ) {alphaBoolBlend {blendOption=Linear}}
   n8 = Random Sequence Player!() -> (Pose:Pose Link Structure) {
    is=title AnimGraphNode_RandomPlayer
    bShuffleMode=true
    entries [
     {
      chanceToPlay=1
      maxLoopCount=0
      maxPlayRate=1
      minLoopCount=0
      minPlayRate=1
      sequence=/BRCosmetics/Animation/Game/MainPlayer/Skydive/ParaGlide/Nosh/Nosh_NoshRight_GLIDER
      blendIn {blendOption=Linear, blendTime=0.05, customCurve=null}
     }
     {
      chanceToPlay=1
      maxLoopCount=0
      maxPlayRate=1
      minLoopCount=0
      minPlayRate=1
      sequence=/BRCosmetics/Animation/Game/MainPlayer/Skydive/ParaGlide/Nosh/Nosh_NoshLeft_GLIDER
      blendIn {blendOption=Linear, blendTime=0.05, customCurve=null}
     }
    ]
   }
   n9 = Get CanNosh() -> (CanNosh:bool)
   n10 = Get MouthJitterSpeed() -> (MouthJitterSpeed:double)
   n11 = Apply Additive(
    Base=Apply Additive(
     Base=Apply Additive!(Base=@n7, Additive=@n8, bAlphaBoolEnabled:bool=@n9) -> (Pose) {alphaInputType=Bool, alphaBoolBlend {blendOption=Linear}}
     Additive="Sequence Player (additive) 'Nosh_MouthJitter_GLIDER'"!(PlayRate:float=@n10) -> (Pose) {sequence=/BRCosmetics/Animation/Game/MainPlayer/Skydive/ParaGlide/Nosh/Nosh_MouthJitter_GLIDER}
     AlphaCurveName=Noshin
    ) {alphaInputType=Curve, alphaBoolBlend {blendOption=Linear}, alphaScaleBiasClamp {bias=1, scale=-1}}
    Additive="Sequence Player (additive) 'Nosh_MouthBreather_GLIDER'"() {sequence=/BRCosmetics/Animation/Game/MainPlayer/Skydive/ParaGlide/Nosh/Nosh_MouthBreather_GLIDER}
    AlphaCurveName=Noshin
   ) {alphaInputType=Curve, alphaBoolBlend {blendOption=Linear}, alphaScaleBiasClamp {bias=1, scale=-1}}
   n12 = "Sequence Player (additive) 'Nosh_Chains_FrontEnd_GLIDER'"() {sequence=/BRCosmetics/Animation/Game/MainPlayer/Skydive/ParaGlide/Nosh/Nosh_Chains_FrontEnd_GLIDER}
   n13 = Get ChainSwayAlpha() -> (ChainSwayAlpha:double)
   n14 = Copy Bone - Source Bone: VB Root_Frame - Target Bone: Frame!(
    ComponentPose:Component Space Pose Link Structure=Local To Component(LocalPose=Apply Additive!(Base=@n11, Additive=@n12, Alpha:float=@n13) -> (Pose) {alphaBoolBlend {blendOption=Linear}})
    AlphaCurveName:Name=Deploy
    bCopyTranslation:bool=true
    bCopyRotation:bool=true
    bCopyScale:bool=true
   ) -> (Pose:Component Space Pose Link Structure) {
    is=title AnimGraphNode_CopyBone
    alphaInputType=Curve
    alphaBoolBlend {blendOption=Linear}
    sourceBone {boneName=VB Root_Frame}
    targetBone {boneName=Frame}
   }
   n15 = Copy Bone - Source Bone: VB Root_door_01 - Target Bone: door_01!(
    ComponentPose:Component Space Pose Link Structure=@n14
    AlphaCurveName:Name=Deploy
    bCopyTranslation:bool=true
    bCopyRotation:bool=true
    bCopyScale:bool=true
   ) -> (Pose:Component Space Pose Link Structure) {
    is=title AnimGraphNode_CopyBone
    alphaInputType=Curve
    alphaBoolBlend {blendOption=Linear}
    sourceBone {boneName=VB Root_door_01}
    targetBone {boneName=door_01}
   }
   Output Pose(Result=Slot 'DefaultSlot'(Source=Component To Local(ComponentPose=@n15))) {layerGroup=None}
   Sequence Player!(Sequence:Anim Sequence Base Object Reference=Get NoshSide() -> (NoshSide:Anim Sequence Base Object Reference)) -> (Pose)
  ]}
 Locomotion {
  entry=Closed
  path=AnimGraph.AnimGraphNode_StateMachine_0.Locomotion
  ids=id
  nodes [
   AnimStateNode_0 = Idle_Open() {is=title AnimStateNode}
   AnimStateNode_1 = Open_Into() {is=title AnimStateNode, stateEntered {guid=#1, notifyName=DeployEntered}}
   AnimStateNode_2 = Closed() {is=title AnimStateNode, stateEntered {guid=#2, notifyName=ClosedEntered}}
   AnimStateTransitionNode_0 = AnimStateTransitionNode() {bAutomaticRuleBasedOnSequencePlayerInState=true, blendMode=Linear, crossfadeDuration=0}
   AnimStateTransitionNode_1 = AnimStateTransitionNode() {crossfadeDuration=0}
   AnimStateTransitionNode_2 = AnimStateTransitionNode() {crossfadeDuration=0}
   AnimStateTransitionNode_3 = AnimStateTransitionNode() {crossfadeDuration=0, priorityOrder=2}
  ]
  links [
   {
    dst=Open_Into
    rule=AnimGraph.AnimGraphNode_StateMachine_0.Locomotion.AnimStateTransitionNode_1.Transition
    src=Closed
    via=AnimStateTransitionNode_1
   }
   {
    dst=Closed
    rule=AnimGraph.AnimGraphNode_StateMachine_0.Locomotion.AnimStateTransitionNode_2.Transition
    src=Idle_Open
    via=AnimStateTransitionNode_2
   }
   {
    dst=Closed
    rule=AnimGraph.AnimGraphNode_StateMachine_0.Locomotion.AnimStateTransitionNode_3.Transition
    src=Open_Into
    via=AnimStateTransitionNode_3
   }
   {
    dst=Idle_Open
    rule=AnimGraph.AnimGraphNode_StateMachine_0.Locomotion.AnimStateTransitionNode_0.Transition
    src=Open_Into
    via=AnimStateTransitionNode_0
   }
  ]
 }
 Idle_Open {
  path=AnimGraph.AnimGraphNode_StateMachine_0.Locomotion.AnimStateNode_0.Idle_Open
  nodes [
   Blend Poses by bool(
    BlendPose_0=Sequence Player 'Nosh_Glide_GLIDER'() {sequence=/BRCosmetics/Animation/Game/MainPlayer/Skydive/ParaGlide/Nosh/Nosh_Glide_GLIDER}
    bActiveValue=Get IsCosmeticPreview!(self:Fort Player Parachute Object Reference=Get OwnerGlider() -> (OwnerGlider:Fort Player Parachute Object Reference)) -> (IsCosmeticPreview:bool)
   ) {blendType=Linear}
   n1 = Break Fort Anim Input Skydiving(FortAnimInput_Skydiving=Get SkydivingState()) -> (
    ParaGilderRootOffset=~
    ParaGliderRootRotationOverride=~
    bUseParaGlideRootModifier=~
    bIsSkydivingFromLaunchPad=~
    bIsSkydivingFromBus=~
    bIsInVortex=~
    bIsDiving=~
    bIsParachuteOpen=~
    bIsSkydiving=~
    bIsParachuteLeaning=~
    bIsSkydiveLeaning=~
    bIsLeaning=~
    bIsSkydiveDiveMode=~
    bPlaySkydiveDrift=~
    bSkydiveDriftDelayActive=~
    bSkydiveDriftAnimAllowed=~
    bIsGliderRight=~
    bIsGliderCenter=~
    bIsGliderLeft=~
    bIsGliderForward=~
    bIsGliderBack=~
    bHasDeployedAndIsStandingUpright=~
    LocalAccelForward=~
    LocalAccelRight=~
    LocalVelocityRight=~
    SkydiveAimPitch=~
    SkydiveAimYaw=~
    DeployChuteAnimRate=~
    SkydiveDriftAnimRate=~
    SkydiveDriftAnimRateCurrent=~
    SkydiveFidgetAnimRate=~
    SkydiveFidgetAnimRateCurrent=~
    SkydiveAdditiveAlpha=~
    SkydiveDriftDelay=~
   )
   n2 = Get OwnerGlider()
   Output Animation Pose(Result=Apply Additive!(
     Base=Blendspace Player 'BS_Nosh_GLIDER'(X=@n1.SkydiveAimYaw, Y=@n1.SkydiveAimPitch) {blendSpace=/BRCosmetics/Animation/Game/MainPlayer/Skydive/ParaGlide/Nosh/BS_Nosh_GLIDER}
     Additive="Sequence Player (additive) 'Nosh_Glide_FrontEnd_GLIDER'"() {sequence=/BRCosmetics/Animation/Game/MainPlayer/Skydive/ParaGlide/Nosh/Nosh_Glide_FrontEnd_GLIDER}
     bAlphaBoolEnabled:bool=Get IsCosmeticPreview(self=@n2)
    ) -> (Pose) {alphaInputType=Bool, alphaBoolBlend {blendOption=Linear}})
  ]
 }
 Open_Into {
  path=AnimGraph.AnimGraphNode_StateMachine_0.Locomotion.AnimStateNode_1.Open_Into
  nodes [
   Apply Additive(bAlphaBoolEnabled=Get IsCosmeticPreview(self=Get OwnerGlider())) {alphaInputType=Bool, alphaBoolBlend {blendOption=Linear}}
   n1 = Break Fort Anim Input Skydiving(FortAnimInput_Skydiving=Get SkydivingState()) -> (
    ParaGilderRootOffset=~
    ParaGliderRootRotationOverride=~
    bUseParaGlideRootModifier=~
    bIsSkydivingFromLaunchPad=~
    bIsSkydivingFromBus=~
    bIsInVortex=~
    bIsDiving=~
    bIsParachuteOpen=~
    bIsSkydiving=~
    bIsParachuteLeaning=~
    bIsSkydiveLeaning=~
    bIsLeaning=~
    bIsSkydiveDiveMode=~
    bPlaySkydiveDrift=~
    bSkydiveDriftDelayActive=~
    bSkydiveDriftAnimAllowed=~
    bIsGliderRight=~
    bIsGliderCenter=~
    bIsGliderLeft=~
    bIsGliderForward=~
    bIsGliderBack=~
    bHasDeployedAndIsStandingUpright=~
    LocalAccelForward=~
    LocalAccelRight=~
    LocalVelocityRight=~
    SkydiveAimPitch=~
    SkydiveAimYaw=~
    DeployChuteAnimRate=~
    SkydiveDriftAnimRate=~
    SkydiveDriftAnimRateCurrent=~
    SkydiveFidgetAnimRate=~
    SkydiveFidgetAnimRateCurrent=~
    SkydiveAdditiveAlpha=~
    SkydiveDriftDelay=~
   )
   n2 = Get OwnerGlider()
   Output Animation Pose(Result=Blend Poses by bool(
     BlendPose_0=Sequence Player 'Nosh_Into_FrontEnd_GLIDER'() {bLoopAnimation=false, sequence=/BRCosmetics/Animation/Game/MainPlayer/Skydive/ParaGlide/Nosh/Nosh_Into_FrontEnd_GLIDER}
     BlendPose_1=Blendspace Player 'BS_Nosh_Into_GLIDER'(X=@n1.SkydiveAimYaw, Y=@n1.SkydiveAimPitch) {bLoop=false, blendSpace=/BRCosmetics/Animation/Game/MainPlayer/Skydive/ParaGlide/Nosh/BS_Nosh_Into_GLIDER}
     bActiveValue=Get IsCosmeticPreview(self=@n2)
    ) {blendType=Linear})
  ]
 }
 Closed {
  path=AnimGraph.AnimGraphNode_StateMachine_0.Locomotion.AnimStateNode_2.Closed
  nodes [
   n1 = Get OwnerGlider()
   Output Animation Pose(Result=Blend Poses by bool(
     BlendPose_0=Sequence Player 'Nosh_Closed_FrontEnd_GLIDER'() {sequence=/BRCosmetics/Animation/Game/MainPlayer/Skydive/ParaGlide/Nosh/Nosh_Closed_FrontEnd_GLIDER}
     BlendPose_1=Sequence Player 'Nosh_Closed_GLIDER'() {sequence=/BRCosmetics/Animation/Game/MainPlayer/Skydive/ParaGlide/Nosh/Nosh_Closed_GLIDER}
     bActiveValue=Get IsCosmeticPreview(self=@n1)
    ) {blendType=Linear})
  ]
 }
 Transition {
  path=AnimGraph.AnimGraphNode_StateMachine_0.Locomotion.AnimStateTransitionNode_0.Transition
  nodes [
   Result()
   "Equal (Float)"!(A:double="Get Relevant Anim Time Remaining (Open_Into)"!() -> (ReturnValue:float) {is=title K2Node_AnimGetter}, B:double=0) -> (ReturnValue:bool) {is=title K2Node_CallFunction}
  ]
 }
 Transition {
  path=AnimGraph.AnimGraphNode_StateMachine_0.Locomotion.AnimStateTransitionNode_1.Transition
  nodes [
   n1 = Get SkydivingState()
   Result(bCanEnterTransition=OR Boolean(
     A=Break Fort Anim Input Skydiving(FortAnimInput_Skydiving=@n1) -> (
      ParaGilderRootOffset=~
      ParaGliderRootRotationOverride=~
      bUseParaGlideRootModifier=~
      bIsSkydivingFromLaunchPad=~
      bIsSkydivingFromBus=~
      bIsInVortex=~
      bIsDiving=~
      bIsParachuteOpen=~
      bIsSkydiving=~
      bIsParachuteLeaning=~
      bIsSkydiveLeaning=~
      bIsLeaning=~
      bIsSkydiveDiveMode=~
      bPlaySkydiveDrift=~
      bSkydiveDriftDelayActive=~
      bSkydiveDriftAnimAllowed=~
      bIsGliderRight=~
      bIsGliderCenter=~
      bIsGliderLeft=~
      bIsGliderForward=~
      bIsGliderBack=~
      bHasDeployedAndIsStandingUpright=~
      LocalAccelForward=~
      LocalAccelRight=~
      LocalVelocityRight=~
      SkydiveAimPitch=~
      SkydiveAimYaw=~
      DeployChuteAnimRate=~
      SkydiveDriftAnimRate=~
      SkydiveDriftAnimRateCurrent=~
      SkydiveFidgetAnimRate=~
      SkydiveFidgetAnimRateCurrent=~
      SkydiveAdditiveAlpha=~
      SkydiveDriftDelay=~
     ).bIsParachuteOpen
     B=Get bForceOpen() -> (bForceOpen:bool)
    ))
  ]
 }
 Transition {
  path=AnimGraph.AnimGraphNode_StateMachine_0.Locomotion.AnimStateTransitionNode_2.Transition
  nodes [
   n1 = Get bForceOpen()
   n2 = Break Fort Anim Input Skydiving(FortAnimInput_Skydiving=Get SkydivingState()) -> (
    ParaGilderRootOffset=~
    ParaGliderRootRotationOverride=~
    bUseParaGlideRootModifier=~
    bIsSkydivingFromLaunchPad=~
    bIsSkydivingFromBus=~
    bIsInVortex=~
    bIsDiving=~
    bIsParachuteOpen=~
    bIsSkydiving=~
    bIsParachuteLeaning=~
    bIsSkydiveLeaning=~
    bIsLeaning=~
    bIsSkydiveDiveMode=~
    bPlaySkydiveDrift=~
    bSkydiveDriftDelayActive=~
    bSkydiveDriftAnimAllowed=~
    bIsGliderRight=~
    bIsGliderCenter=~
    bIsGliderLeft=~
    bIsGliderForward=~
    bIsGliderBack=~
    bHasDeployedAndIsStandingUpright=~
    LocalAccelForward=~
    LocalAccelRight=~
    LocalVelocityRight=~
    SkydiveAimPitch=~
    SkydiveAimYaw=~
    DeployChuteAnimRate=~
    SkydiveDriftAnimRate=~
    SkydiveDriftAnimRateCurrent=~
    SkydiveFidgetAnimRate=~
    SkydiveFidgetAnimRateCurrent=~
    SkydiveAdditiveAlpha=~
    SkydiveDriftDelay=~
   )
   Result(bCanEnterTransition=AND Boolean(A=NOT Boolean(A=@n2.bIsParachuteOpen), B=NOT Boolean(A=@n1)))
  ]
 }
 Transition {
  path=AnimGraph.AnimGraphNode_StateMachine_0.Locomotion.AnimStateTransitionNode_3.Transition
  nodes [
   n1 = Get bForceOpen()
   n2 = Break Fort Anim Input Skydiving(FortAnimInput_Skydiving=Get SkydivingState()) -> (
    ParaGilderRootOffset=~
    ParaGliderRootRotationOverride=~
    bUseParaGlideRootModifier=~
    bIsSkydivingFromLaunchPad=~
    bIsSkydivingFromBus=~
    bIsInVortex=~
    bIsDiving=~
    bIsParachuteOpen=~
    bIsSkydiving=~
    bIsParachuteLeaning=~
    bIsSkydiveLeaning=~
    bIsLeaning=~
    bIsSkydiveDiveMode=~
    bPlaySkydiveDrift=~
    bSkydiveDriftDelayActive=~
    bSkydiveDriftAnimAllowed=~
    bIsGliderRight=~
    bIsGliderCenter=~
    bIsGliderLeft=~
    bIsGliderForward=~
    bIsGliderBack=~
    bHasDeployedAndIsStandingUpright=~
    LocalAccelForward=~
    LocalAccelRight=~
    LocalVelocityRight=~
    SkydiveAimPitch=~
    SkydiveAimYaw=~
    DeployChuteAnimRate=~
    SkydiveDriftAnimRate=~
    SkydiveDriftAnimRateCurrent=~
    SkydiveFidgetAnimRate=~
    SkydiveFidgetAnimRateCurrent=~
    SkydiveAdditiveAlpha=~
    SkydiveDriftDelay=~
   )
   Result(bCanEnterTransition=AND Boolean(A=NOT Boolean(A=@n2.bIsParachuteOpen), B=NOT Boolean(A=@n1)))
  ]
 }
 Flap {entry=FlappyNosh, path=AnimGraph.AnimGraphNode_StateMachine_1.Flap, ids=id, nodes [AnimStateNode_0 = FlappyNosh() {is=title AnimStateNode}]}
 FlappyNosh {
  path=AnimGraph.AnimGraphNode_StateMachine_1.Flap.AnimStateNode_0.FlappyNosh
  nodes [Output Animation Pose(Result="Sequence Player (additive) 'Nosh_FlappyNosh_GLIDER'"!(PlayRate:float=Get FlapRate() -> (FlapRate:double)) -> (Pose) {bLoopAnimation=false, sequence=/BRCosmetics/Animation/Game/MainPlayer/Skydive/ParaGlide/Nosh/Nosh_FlappyNosh_GLIDER})]
 }
 ChainSwayFrontend {nodes [
   Inputs()
   > n1 = Branch(Condition=Get IsCosmeticPreview(self=Get OwnerGlider()))
   > n2 = Branch(Condition="Nearly Equal (Float)"(A=GetCurveValue(CurveName=Deploy), B=0, ErrorTolerance=0))
   > n3 = Set ChainSwayAlpha(ChainSwayAlpha:double=FInterpTo(Current=Get ChainSwayAlpha(), Target=1, DeltaTime=Get Time() -> (Time:double), InterpSpeed=1)) -> (Output_Get:double)
   Outputs!(then:Exec=[@n3.then, @n4.then])
   n4 = Set ChainSwayAlpha(execute=[@n1.else, @n2.else], ChainSwayAlpha=0)
  ]}
 CheckNoFlapNoNosh {nodes [
   Inputs()
   > Set NoFlap(NoFlap:bool="Equal (Float)"(A=GetCurveValue(CurveName=NoFlap), B=1)) -> (Output_Get:bool)
   > Set NoNosh(NoNosh:bool="Equal (Float)"(A=GetCurveValue(CurveName=NoNosh), B=1)) -> (Output_Get:bool)
   > Outputs()
  ]}
 EventGraph {nodes [
   n1 = Event BlueprintUpdateAnimation()
   > Set Time(Time:double=@n1.DeltaTimeX) -> (Output_Get:double)
   > Chain Sway Frontend!(execute:Exec) -> (then:Exec) {is=title K2Node_MacroInstance}
   > n2 = ForwardBack_Value_Interp!(
    execute
    Time:double=Get Time()
    Currnet_Value:double=Get SlimySpeed()
    Idle_Value:double=1
    Idle_InterpSpeed:double=1
    Forward_Value:double=1.5
    Forward_InterpSpeed:double=2
    Back_Value:double=1
    Back_InterpSpeed:double=2
    LocalAccelForward:double=Break Fort Anim Input Skydiving(FortAnimInput_Skydiving=Get SkydivingState()) -> (
     ParaGilderRootOffset=~
     ParaGliderRootRotationOverride=~
     bUseParaGlideRootModifier=~
     bIsSkydivingFromLaunchPad=~
     bIsSkydivingFromBus=~
     bIsInVortex=~
     bIsDiving=~
     bIsParachuteOpen=~
     bIsSkydiving=~
     bIsParachuteLeaning=~
     bIsSkydiveLeaning=~
     bIsLeaning=~
     bIsSkydiveDiveMode=~
     bPlaySkydiveDrift=~
     bSkydiveDriftDelayActive=~
     bSkydiveDriftAnimAllowed=~
     bIsGliderRight=~
     bIsGliderCenter=~
     bIsGliderLeft=~
     bIsGliderForward=~
     bIsGliderBack=~
     bHasDeployedAndIsStandingUpright=~
     LocalAccelForward=~
     LocalAccelRight=~
     LocalVelocityRight=~
     SkydiveAimPitch=~
     SkydiveAimYaw=~
     DeployChuteAnimRate=~
     SkydiveDriftAnimRate=~
     SkydiveDriftAnimRateCurrent=~
     SkydiveFidgetAnimRate=~
     SkydiveFidgetAnimRateCurrent=~
     SkydiveAdditiveAlpha=~
     SkydiveDriftDelay=~
    ).LocalAccelForward
   ) -> (then, New_Value:double)
   > Set SlimySpeed(SlimySpeed:double=@n2) -> (Output_Get:double)
   > n3 = ForwardBack_Value_Interp(
    Time=Get Time()
    Currnet_Value=Get MouthJitterSpeed()
    Idle_Value=1
    Idle_InterpSpeed=1
    Forward_Value=1.5
    Forward_InterpSpeed=2
    Back_Value=1
    Back_InterpSpeed=2
    LocalAccelForward=Break Fort Anim Input Skydiving(FortAnimInput_Skydiving=Get SkydivingState()) -> (
     ParaGilderRootOffset=~
     ParaGliderRootRotationOverride=~
     bUseParaGlideRootModifier=~
     bIsSkydivingFromLaunchPad=~
     bIsSkydivingFromBus=~
     bIsInVortex=~
     bIsDiving=~
     bIsParachuteOpen=~
     bIsSkydiving=~
     bIsParachuteLeaning=~
     bIsSkydiveLeaning=~
     bIsLeaning=~
     bIsSkydiveDiveMode=~
     bPlaySkydiveDrift=~
     bSkydiveDriftDelayActive=~
     bSkydiveDriftAnimAllowed=~
     bIsGliderRight=~
     bIsGliderCenter=~
     bIsGliderLeft=~
     bIsGliderForward=~
     bIsGliderBack=~
     bHasDeployedAndIsStandingUpright=~
     LocalAccelForward=~
     LocalAccelRight=~
     LocalVelocityRight=~
     SkydiveAimPitch=~
     SkydiveAimYaw=~
     DeployChuteAnimRate=~
     SkydiveDriftAnimRate=~
     SkydiveDriftAnimRateCurrent=~
     SkydiveFidgetAnimRate=~
     SkydiveFidgetAnimRateCurrent=~
     SkydiveAdditiveAlpha=~
     SkydiveDriftDelay=~
    ).LocalAccelForward
   )
   > Set MouthJitterSpeed(MouthJitterSpeed:double=@n3) -> (Output_Get:double)
   > n4 = Sequence()
   > Check No Flap No Nosh!(execute:Exec) -> (then:Exec) {is=title K2Node_MacroInstance}
   > n5 = Branch(Condition=Get NoFlap() -> (NoFlap:bool))
   Do Once(execute=@n5.else, Reset=[@n8.then, @n10.Completed], Start Closed=true)
   > Set FlapRate(FlapRate:double=RandomFloatInRange(Min=1, Max=1.2)) -> (Output_Get:double)
   > Set FlapAlpha(FlapAlpha:double=1) -> (Output_Get:double)
   n6 = Branch(execute=@n4.then_1, Condition=Get NoNosh() -> (NoNosh:bool))
   Do Once(execute=@n6.else, Reset=[@n13.then, @n12.Completed], Start Closed=true)
   > Spawn Junk!(execute:Exec) -> (then:Exec) {is=title K2Node_MacroInstance}
   > Set CanNosh(CanNosh:bool=true) -> (Output_Get:bool)
   AnimNotify_TurnONEngines() {is=title K2Node_Event}
   > SetGliderFullyDeployed!(execute, self:Fort Glider Anim Instance Object Reference, bIsFullyDeployed:bool=true) -> (then)
   AnimNotify_ClosedEntered() {is=title K2Node_Event}
   n7 = Get OwnerGlider()
   > Set Actor Hidden In Game!(execute, self:Actor Object Reference=@n7, bNewHidden:bool=true) -> (then) {is=title K2Node_CallFunction}
   > Hide Junk!(Exec:Exec) -> (then:Exec) {is=title K2Node_MacroInstance}
   > Set CanNosh(CanNosh=false)
   AnimNotify_DeployEntered() {is=title K2Node_Event}
   > Delay(Duration=0.1)
   > Set Actor Hidden In Game(self=@n7, bNewHidden=false)
   AnimNotify_FlapOff() {is=title K2Node_Event}
   > Set FlapAlpha(FlapAlpha=0)
   > n8 = Delay(Duration=RandomFloatInRange(Min=3, Max=5))
   AnimNotify_FlapEnable() {is=title K2Node_Event}
   > n9 = Sequence()
   > n10 = Do Once()
   Delay(execute=@n9.then_1, Duration=1)
   > n11 = Sequence()
   n12 = Do Once(execute=@n11.then_1, Reset=@n11.then_0)
   AnimNotify_NoshDone() {is=title K2Node_Event}
   > Hide Junk()
   > Set CanNosh(CanNosh=false)
   > n13 = Delay(Duration=RandomFloatInRange(Min=3, Max=5))
   SetHiddenInGame!(
    execute
    self:Scene Component Object Reference=Get NoshJunk() -> (NoshJunk:Scene Component Object Reference)
    NewHidden:bool=true
    bPropagateToChildren:bool=false
   ) -> (then)
   Sequence()
  ]}
 HideJunk {nodes [
   Inputs!() -> (Exec:Exec)
   n1 = Get ParachuteMesh!(self:Fort Player Parachute Object Reference=Get OwnerGlider()) -> (ParachuteMesh:Skeletal Mesh Component Budgeted Object Reference)
   > n2 = For Loop(
    FirstIndex=1
    LastIndex=int - int(A=GetNumChildrenComponents!(self:Scene Component Object Reference=@n1) -> (ReturnValue:int=0), B=1) {is=title K2Node_CallFunction}
   )
   Outputs(then=@n2.Completed)
   > SetHiddenInGame(
    self=GetChildComponent!(self:Scene Component Object Reference=@n1, ChildIndex:int=@n2) -> (ReturnValue:Scene Component Object Reference)
    NewHidden=true
    bPropagateToChildren=false
   )
  ]}
 SpawnJunk {nodes [
   Inputs!() -> (execute:Exec)
   n1 = Get OwnerGlider()
   > n2 = Set NoshJunk(NoshJunk:Scene Component Object Reference=GetChildComponent(self=Get ParachuteMesh(self=@n1), ChildIndex=RandomIntegerInRange!(Min:int=1, Max:int=6) -> (ReturnValue:int=0))) -> (Output_Get:Scene Component Object Reference)
   > SetHiddenInGame(self=@n2, NewHidden=false, bPropagateToChildren=false)
   > Outputs()
   n3 = GetNumChildrenComponents(self=Get ParachuteMesh(self=Get OwnerGlider())) -> (ReturnValue=0)
   PrintString!(execute, InString="To String (Integer)"(InInt=@n3), bPrintToScreen, bPrintToLog, TextColor, Duration, Key) -> (then)
   int - int(A=@n3, B=1)
  ]}
]
variables [
 AimPitch {type=number}
 CanNosh {type=boolean}
 ChainSwayAlpha {type=number}
 DeployAnimRate {type=number}
 FlapAlpha {type=number}
 FlapRate {type=number}
 Forward {type=number}
 IsParachuteOpen {type=boolean}
 MouthJitterSpeed {type=number, default=1}
 NoFlap {type=boolean}
 NoNosh {type=boolean}
 NoshJunk {type=SceneComponent, default=null}
 NoshSide {type=AnimSequenceBase, default=&Nosh_NoshRight_GLIDER}
 Right {type=number}
 SkydiveYaw {type=number}
 SlimySpeed {type=number, default=0.9995}
 Time {type=number, default=0.0167}
 UseDynamicsAlpha {type=number}
]
