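# Animation Blueprint for the Realm_Male glider cosmetic, animating deployment, gliding and closing during skydives.
# Locomotion switches closed, opening and open states from skydiving flags or force-open, using aim-driven blendspaces and front-end preview sequences.
# Curve-scaled lean and aim-driven turn additives layer on top before DefaultSlot.
# Notifies hide or reveal Glider_Base; event logic drives a Galileo Rocket Turret state machine, which is unconnected to the output pose.
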AnimBlueprint /BRCosmetics/Gadgets/Assets/VinderTech_GliderChute/Blueprints/Realm_Male_Glider_AnimBP
target_skeleton=/Game/Gadgets/Assets/VinderTech_GliderChute/Glider_Realm_Male/Meshes/Realm_Male_Glider_Skeleton
graphs [
 AnimGraph {nodes [
   Apply Additive(Additive="Sequence Player (additive) 'Realm_Male_Turret_GLIDER'"() {sequence=/Game/Animation/Game/MainPlayer/Skydive/ParaGlide/Realm_Male/Realm_Male_Turret_GLIDER}) {alphaBoolBlend {blendOption=Linear}}
   n1 = Break Fort Anim Input Skydiving!(FortAnimInput_Skydiving:"Fort Anim Input Skydiving Structure (by ref)"=Get SkydivingState() -> (SkydivingState:Fort Anim Input Skydiving Structure)) -> (
    ParaGilderRootOffset:Vector=~
    ParaGliderRootRotationOverride:Rotator=~
    bUseParaGlideRootModifier:bool=~
    bIsSkydivingFromLaunchPad:bool=~
    bIsSkydivingFromBus:bool=~
    bIsInVortex:bool=~
    bIsDiving:bool=~
    bIsParachuteOpen:bool=~
    bIsSkydiving:bool=~
    bIsParachuteLeaning:bool=~
    bIsSkydiveLeaning:bool=~
    bIsLeaning:bool=~
    bIsSkydiveDiveMode:bool=~
    bPlaySkydiveDrift:bool=~
    bSkydiveDriftDelayActive:bool=~
    bSkydiveDriftAnimAllowed:bool=~
    bIsGliderRight:bool=~
    bIsGliderCenter:bool=~
    bIsGliderLeft:bool=~
    bIsGliderForward:bool=~
    bIsGliderBack:bool=~
    bHasDeployedAndIsStandingUpright:bool=~
    LocalAccelForward:float=~
    LocalAccelRight:float=~
    LocalVelocityRight:float=~
    SkydiveAimPitch:float=~
    SkydiveAimYaw:float=~
    DeployChuteAnimRate:float=~
    SkydiveDriftAnimRate:float=~
    SkydiveDriftAnimRateCurrent:float=~
    SkydiveFidgetAnimRate:float=~
    SkydiveFidgetAnimRateCurrent:float=~
    SkydiveAdditiveAlpha:float=~
    SkydiveDriftDelay:float=~
    SkydiveDriftAnim:int
    SkydiveDriftAnimMax:int
    LaunchpadAnim:int
    LocalAccelDir:ESkydivingDirection Enum
    DirectionLast:ESkydivingDirection Enum
    PoseAdditive:Anim Sequence Object Reference
   ) {is=title K2Node_BreakStruct}
   AnimGraphNode_StateMachine_0 = Locomotion() {is=title AnimGraphNode_StateMachine}
   n2 = Apply Additive!(
    Base=@AnimGraphNode_StateMachine_0
    Additive=Blendspace Player 'BS_Realm_Male_Lean_GLIDER'(X=@n1.LocalAccelRight, Y=@n1.LocalAccelForward) {blendSpace=/BRCosmetics/Animation/Game/MainPlayer/Skydive/ParaGlide/Realm_Male/BS_Realm_Male_Lean_GLIDER}
    AlphaCurveName:Name=ParaGlide_Lean
   ) -> (Pose) {alphaInputType=Curve, alphaBoolBlend {blendOption=Linear}, alphaScaleBiasClamp {bias=1, scale=-1}}
   n3 = Blendspace Player 'BS_Realm_Male_Turn_GLIDER'(X=@n1.SkydiveAimYaw, Y=@n1.SkydiveAimPitch) {blendSpace=/BRCosmetics/Animation/Game/MainPlayer/Skydive/ParaGlide/Realm_Male/BS_Realm_Male_Turn_GLIDER}
   Output Pose(Result=Slot 'DefaultSlot'(Source=Apply Additive!(Base=@n2, Additive=@n3, Alpha:float) -> (Pose) {alphaBoolBlend {blendOption=Linear}})) {layerGroup=None}
   AnimGraphNode_StateMachine_1 = Turret() {is=title AnimGraphNode_StateMachine}
  ]}
 Locomotion {
  entry=Closed
  path=AnimGraph.AnimGraphNode_StateMachine_0.Locomotion
  ids=id
  nodes [
   AnimStateNode_0 = Idle_Open() {is=title AnimStateNode}
   AnimStateNode_1 = ForceOpen_or_GliderItem() {is=title AnimStateNode, stateEntered {guid=#1, notifyName=DeployEntered}}
   AnimStateNode_2 = Closed() {is=title AnimStateNode, stateEntered {guid=#2, notifyName=ClosedEntered}}
   AnimStateNode_3 = Open_Into() {is=title AnimStateNode, stateEntered {guid=#3, notifyName=DeployEntered}}
   AnimStateTransitionNode_0 = AnimStateTransitionNode() {bAutomaticRuleBasedOnSequencePlayerInState=true, blendMode=Linear, crossfadeDuration=0}
   AnimStateTransitionNode_1 = AnimStateTransitionNode() {crossfadeDuration=0}
   AnimStateTransitionNode_2 = AnimStateTransitionNode() {crossfadeDuration=0}
   AnimStateTransitionNode_3 = AnimStateTransitionNode() {crossfadeDuration=0, priorityOrder=2}
   AnimStateTransitionNode_4 = AnimStateTransitionNode() {crossfadeDuration=0}
   AnimStateTransitionNode_5 = AnimStateTransitionNode() {crossfadeDuration=0}
   AnimStateTransitionNode_6 = AnimStateTransitionNode() {bAutomaticRuleBasedOnSequencePlayerInState=true, blendMode=Linear, crossfadeDuration=0}
  ]
  links [
   {
    dst=ForceOpen_or_GliderItem
    rule=AnimGraph.AnimGraphNode_StateMachine_0.Locomotion.AnimStateTransitionNode_4.Transition
    src=Closed
    via=AnimStateTransitionNode_4
   }
   {
    dst=Open_Into
    rule=AnimGraph.AnimGraphNode_StateMachine_0.Locomotion.AnimStateTransitionNode_1.Transition
    src=Closed
    via=AnimStateTransitionNode_1
   }
   {
    dst=Closed
    rule=AnimGraph.AnimGraphNode_StateMachine_0.Locomotion.AnimStateTransitionNode_5.Transition
    src=ForceOpen_or_GliderItem
    via=AnimStateTransitionNode_5
   }
   {
    dst=Idle_Open
    rule=AnimGraph.AnimGraphNode_StateMachine_0.Locomotion.AnimStateTransitionNode_6.Transition
    src=ForceOpen_or_GliderItem
    via=AnimStateTransitionNode_6
   }
   {
    dst=Closed
    rule=AnimGraph.AnimGraphNode_StateMachine_0.Locomotion.AnimStateTransitionNode_2.Transition
    src=Idle_Open
    via=AnimStateTransitionNode_2
   }
   {
    dst=Closed
    rule=AnimGraph.AnimGraphNode_StateMachine_0.Locomotion.AnimStateTransitionNode_3.Transition
    src=Open_Into
    via=AnimStateTransitionNode_3
   }
   {
    dst=Idle_Open
    rule=AnimGraph.AnimGraphNode_StateMachine_0.Locomotion.AnimStateTransitionNode_0.Transition
    src=Open_Into
    via=AnimStateTransitionNode_0
   }
  ]
 }
 Idle_Open {
  path=AnimGraph.AnimGraphNode_StateMachine_0.Locomotion.AnimStateNode_0.Idle_Open
  nodes [
   Apply Additive!(
    Base
    Additive
    bAlphaBoolEnabled:bool=Get IsCosmeticPreview!(self:Fort Player Parachute Object Reference=Get OwnerGlider() -> (OwnerGlider:Fort Player Parachute Object Reference)) -> (IsCosmeticPreview:bool)
   ) -> (Pose) {alphaInputType=Bool, alphaBoolBlend {blendOption=Linear}}
   n1 = Break Fort Anim Input Skydiving(FortAnimInput_Skydiving=Get SkydivingState()) -> (
    ParaGilderRootOffset=~
    ParaGliderRootRotationOverride=~
    bUseParaGlideRootModifier=~
    bIsSkydivingFromLaunchPad=~
    bIsSkydivingFromBus=~
    bIsInVortex=~
    bIsDiving=~
    bIsParachuteOpen=~
    bIsSkydiving=~
    bIsParachuteLeaning=~
    bIsSkydiveLeaning=~
    bIsLeaning=~
    bIsSkydiveDiveMode=~
    bPlaySkydiveDrift=~
    bSkydiveDriftDelayActive=~
    bSkydiveDriftAnimAllowed=~
    bIsGliderRight=~
    bIsGliderCenter=~
    bIsGliderLeft=~
    bIsGliderForward=~
    bIsGliderBack=~
    bHasDeployedAndIsStandingUpright=~
    LocalAccelForward=~
    LocalAccelRight=~
    LocalVelocityRight=~
    SkydiveAimPitch=~
    SkydiveAimYaw=~
    DeployChuteAnimRate=~
    SkydiveDriftAnimRate=~
    SkydiveDriftAnimRateCurrent=~
    SkydiveFidgetAnimRate=~
    SkydiveFidgetAnimRateCurrent=~
    SkydiveAdditiveAlpha=~
    SkydiveDriftDelay=~
   )
   n2 = Get OwnerGlider()
   Output Animation Pose(Result=Blend Poses by bool(
     BlendPose_0=Sequence Player 'Realm_Male_Glide_FrontEnd_GLIDER'() {sequence=/BRCosmetics/Animation/Game/MainPlayer/Skydive/ParaGlide/Realm_Male/Realm_Male_Glide_FrontEnd_GLIDER}
     BlendPose_1=Blendspace Player 'BS_Realm_Male_GLIDER'(X=@n1.SkydiveAimYaw, Y=@n1.SkydiveAimPitch) {blendSpace=/BRCosmetics/Animation/Game/MainPlayer/Skydive/ParaGlide/Realm_Male/BS_Realm_Male_GLIDER}
     bActiveValue=Get IsCosmeticPreview(self=@n2)
    ) {blendType=Linear})
  ]
 }
 ForceOpen_or_GliderItem {
  path=AnimGraph.AnimGraphNode_StateMachine_0.Locomotion.AnimStateNode_1.ForceOpen_or_GliderItem
  nodes [
   n1 = Break Fort Anim Input Skydiving(FortAnimInput_Skydiving=Get SkydivingState()) -> (
    ParaGilderRootOffset=~
    ParaGliderRootRotationOverride=~
    bUseParaGlideRootModifier=~
    bIsSkydivingFromLaunchPad=~
    bIsSkydivingFromBus=~
    bIsInVortex=~
    bIsDiving=~
    bIsParachuteOpen=~
    bIsSkydiving=~
    bIsParachuteLeaning=~
    bIsSkydiveLeaning=~
    bIsLeaning=~
    bIsSkydiveDiveMode=~
    bPlaySkydiveDrift=~
    bSkydiveDriftDelayActive=~
    bSkydiveDriftAnimAllowed=~
    bIsGliderRight=~
    bIsGliderCenter=~
    bIsGliderLeft=~
    bIsGliderForward=~
    bIsGliderBack=~
    bHasDeployedAndIsStandingUpright=~
    LocalAccelForward=~
    LocalAccelRight=~
    LocalVelocityRight=~
    SkydiveAimPitch=~
    SkydiveAimYaw=~
    DeployChuteAnimRate=~
    SkydiveDriftAnimRate=~
    SkydiveDriftAnimRateCurrent=~
    SkydiveFidgetAnimRate=~
    SkydiveFidgetAnimRateCurrent=~
    SkydiveAdditiveAlpha=~
    SkydiveDriftDelay=~
   )
   n2 = Get Lightspeed() -> (Lightspeed:bool)
   Output Animation Pose(Result=Blendspace Player!(
     X=@n1.SkydiveAimYaw
     Y=@n1.SkydiveAimPitch
     BlendSpace:Blend Space Object Reference=Select(
      Option 0:Blend Space Object Reference=/BRCosmetics/Animation/Game/MainPlayer/Skydive/ParaGlide/Realm_Male/BS_Realm_Male_Into_GLIDER
      Option 1:Blend Space Object Reference=&BS_Realm_Male_Into_GLIDER
      Index=@n2
     ) -> (ReturnValue:Blend Space Object Reference)
    ) -> (Pose) {bLoop=false, startPosition=0.0044})
   Get Into_B_Side() -> (Into_B_Side:Blend Space Object Reference)
  ]
 }
 Closed {
  path=AnimGraph.AnimGraphNode_StateMachine_0.Locomotion.AnimStateNode_2.Closed
  nodes [
   n1 = Get OwnerGlider()
   Output Animation Pose(Result=Blend Poses by bool(
     BlendPose_0=Sequence Player 'Realm_Male_Closed_FrontEnd_GLIDER'() {sequence=/BRCosmetics/Animation/Game/MainPlayer/Skydive/ParaGlide/Realm_Male/Realm_Male_Closed_FrontEnd_GLIDER}
     BlendPose_1=Sequence Player 'Realm_Male_Closed_GLIDER'() {sequence=/BRCosmetics/Animation/Game/MainPlayer/Skydive/ParaGlide/Realm_Male/Realm_Male_Closed_GLIDER}
     bActiveValue=Get IsCosmeticPreview(self=@n1)
    ) {blendType=Linear})
  ]
 }
 Open_Into {
  path=AnimGraph.AnimGraphNode_StateMachine_0.Locomotion.AnimStateNode_3.Open_Into
  nodes [
   Apply Additive(bAlphaBoolEnabled=Get IsCosmeticPreview(self=Get OwnerGlider())) {alphaInputType=Bool, alphaBoolBlend {blendOption=Linear}}
   n1 = Break Fort Anim Input Skydiving(FortAnimInput_Skydiving=Get SkydivingState()) -> (
    ParaGilderRootOffset=~
    ParaGliderRootRotationOverride=~
    bUseParaGlideRootModifier=~
    bIsSkydivingFromLaunchPad=~
    bIsSkydivingFromBus=~
    bIsInVortex=~
    bIsDiving=~
    bIsParachuteOpen=~
    bIsSkydiving=~
    bIsParachuteLeaning=~
    bIsSkydiveLeaning=~
    bIsLeaning=~
    bIsSkydiveDiveMode=~
    bPlaySkydiveDrift=~
    bSkydiveDriftDelayActive=~
    bSkydiveDriftAnimAllowed=~
    bIsGliderRight=~
    bIsGliderCenter=~
    bIsGliderLeft=~
    bIsGliderForward=~
    bIsGliderBack=~
    bHasDeployedAndIsStandingUpright=~
    LocalAccelForward=~
    LocalAccelRight=~
    LocalVelocityRight=~
    SkydiveAimPitch=~
    SkydiveAimYaw=~
    DeployChuteAnimRate=~
    SkydiveDriftAnimRate=~
    SkydiveDriftAnimRateCurrent=~
    SkydiveFidgetAnimRate=~
    SkydiveFidgetAnimRateCurrent=~
    SkydiveAdditiveAlpha=~
    SkydiveDriftDelay=~
   )
   n2 = Select(
    Option 0=&BS_Realm_Male_Into_GLIDER
    Option 1=&BS_Realm_Male_Into_GLIDER
    Index=Get Lightspeed()
   )
   n3 = Get OwnerGlider()
   Output Animation Pose(Result=Blend Poses by bool(
     BlendPose_0=Sequence Player 'Realm_Male_Into_FrontEnd_GLIDER'() {bLoopAnimation=false, sequence=/BRCosmetics/Animation/Game/MainPlayer/Skydive/ParaGlide/Realm_Male/Realm_Male_Into_FrontEnd_GLIDER}
     BlendPose_1=Blendspace Player(X=@n1.SkydiveAimYaw, Y=@n1.SkydiveAimPitch, BlendSpace=@n2) {bLoop=false}
     bActiveValue=Get IsCosmeticPreview(self=@n3)
    ) {blendType=Linear})
   Get Into_B_Side()
  ]
 }
 Transition {
  path=AnimGraph.AnimGraphNode_StateMachine_0.Locomotion.AnimStateTransitionNode_0.Transition
  nodes [
   Result()
   "Equal (Float)"!(A:double="Get Relevant Anim Time Remaining (Open_Into)"!() -> (ReturnValue:float) {is=title K2Node_AnimGetter}, B:double=0) -> (ReturnValue:bool) {is=title K2Node_CallFunction}
  ]
 }
 Transition {
  path=AnimGraph.AnimGraphNode_StateMachine_0.Locomotion.AnimStateTransitionNode_1.Transition
  nodes [
   n1 = Break Fort Anim Input Skydiving(FortAnimInput_Skydiving=Get SkydivingState()) -> (
    ParaGilderRootOffset=~
    ParaGliderRootRotationOverride=~
    bUseParaGlideRootModifier=~
    bIsSkydivingFromLaunchPad=~
    bIsSkydivingFromBus=~
    bIsInVortex=~
    bIsDiving=~
    bIsParachuteOpen=~
    bIsSkydiving=~
    bIsParachuteLeaning=~
    bIsSkydiveLeaning=~
    bIsLeaning=~
    bIsSkydiveDiveMode=~
    bPlaySkydiveDrift=~
    bSkydiveDriftDelayActive=~
    bSkydiveDriftAnimAllowed=~
    bIsGliderRight=~
    bIsGliderCenter=~
    bIsGliderLeft=~
    bIsGliderForward=~
    bIsGliderBack=~
    bHasDeployedAndIsStandingUpright=~
    LocalAccelForward=~
    LocalAccelRight=~
    LocalVelocityRight=~
    SkydiveAimPitch=~
    SkydiveAimYaw=~
    DeployChuteAnimRate=~
    SkydiveDriftAnimRate=~
    SkydiveDriftAnimRateCurrent=~
    SkydiveFidgetAnimRate=~
    SkydiveFidgetAnimRateCurrent=~
    SkydiveAdditiveAlpha=~
    SkydiveDriftDelay=~
   )
   n2 = Get IsCosmeticPreview(self=Get OwnerGlider())
   Result(bCanEnterTransition=AND Boolean(
     A=OR Boolean!(A=@n1.bIsSkydivingFromLaunchPad, B=@n1.bIsSkydivingFromBus, C:bool=@n1.bIsInVortex, D:bool=@n2) -> (ReturnValue)
     B=@n1.bIsParachuteOpen
    ))
   NOT Boolean(A=Get bForceOpen() -> (bForceOpen:bool))
  ]
 }
 Transition {
  path=AnimGraph.AnimGraphNode_StateMachine_0.Locomotion.AnimStateTransitionNode_2.Transition
  nodes [
   n1 = Get bForceOpen()
   n2 = Break Fort Anim Input Skydiving(FortAnimInput_Skydiving=Get SkydivingState()) -> (
    ParaGilderRootOffset=~
    ParaGliderRootRotationOverride=~
    bUseParaGlideRootModifier=~
    bIsSkydivingFromLaunchPad=~
    bIsSkydivingFromBus=~
    bIsInVortex=~
    bIsDiving=~
    bIsParachuteOpen=~
    bIsSkydiving=~
    bIsParachuteLeaning=~
    bIsSkydiveLeaning=~
    bIsLeaning=~
    bIsSkydiveDiveMode=~
    bPlaySkydiveDrift=~
    bSkydiveDriftDelayActive=~
    bSkydiveDriftAnimAllowed=~
    bIsGliderRight=~
    bIsGliderCenter=~
    bIsGliderLeft=~
    bIsGliderForward=~
    bIsGliderBack=~
    bHasDeployedAndIsStandingUpright=~
    LocalAccelForward=~
    LocalAccelRight=~
    LocalVelocityRight=~
    SkydiveAimPitch=~
    SkydiveAimYaw=~
    DeployChuteAnimRate=~
    SkydiveDriftAnimRate=~
    SkydiveDriftAnimRateCurrent=~
    SkydiveFidgetAnimRate=~
    SkydiveFidgetAnimRateCurrent=~
    SkydiveAdditiveAlpha=~
    SkydiveDriftDelay=~
   )
   Result(bCanEnterTransition=AND Boolean(A=NOT Boolean(A=@n2.bIsParachuteOpen), B=NOT Boolean(A=@n1)))
  ]
 }
 Transition {
  path=AnimGraph.AnimGraphNode_StateMachine_0.Locomotion.AnimStateTransitionNode_3.Transition
  nodes [
   n1 = Get bForceOpen()
   n2 = Break Fort Anim Input Skydiving(FortAnimInput_Skydiving=Get SkydivingState()) -> (
    ParaGilderRootOffset=~
    ParaGliderRootRotationOverride=~
    bUseParaGlideRootModifier=~
    bIsSkydivingFromLaunchPad=~
    bIsSkydivingFromBus=~
    bIsInVortex=~
    bIsDiving=~
    bIsParachuteOpen=~
    bIsSkydiving=~
    bIsParachuteLeaning=~
    bIsSkydiveLeaning=~
    bIsLeaning=~
    bIsSkydiveDiveMode=~
    bPlaySkydiveDrift=~
    bSkydiveDriftDelayActive=~
    bSkydiveDriftAnimAllowed=~
    bIsGliderRight=~
    bIsGliderCenter=~
    bIsGliderLeft=~
    bIsGliderForward=~
    bIsGliderBack=~
    bHasDeployedAndIsStandingUpright=~
    LocalAccelForward=~
    LocalAccelRight=~
    LocalVelocityRight=~
    SkydiveAimPitch=~
    SkydiveAimYaw=~
    DeployChuteAnimRate=~
    SkydiveDriftAnimRate=~
    SkydiveDriftAnimRateCurrent=~
    SkydiveFidgetAnimRate=~
    SkydiveFidgetAnimRateCurrent=~
    SkydiveAdditiveAlpha=~
    SkydiveDriftDelay=~
   )
   Result(bCanEnterTransition=AND Boolean(A=NOT Boolean(A=@n2.bIsParachuteOpen), B=NOT Boolean(A=@n1)))
  ]
 }
 Transition {
  path=AnimGraph.AnimGraphNode_StateMachine_0.Locomotion.AnimStateTransitionNode_4.Transition
  nodes [
   n1 = Break Fort Anim Input Skydiving(FortAnimInput_Skydiving=Get SkydivingState()) -> (
    ParaGilderRootOffset=~
    ParaGliderRootRotationOverride=~
    bUseParaGlideRootModifier=~
    bIsSkydivingFromLaunchPad=~
    bIsSkydivingFromBus=~
    bIsInVortex=~
    bIsDiving=~
    bIsParachuteOpen=~
    bIsSkydiving=~
    bIsParachuteLeaning=~
    bIsSkydiveLeaning=~
    bIsLeaning=~
    bIsSkydiveDiveMode=~
    bPlaySkydiveDrift=~
    bSkydiveDriftDelayActive=~
    bSkydiveDriftAnimAllowed=~
    bIsGliderRight=~
    bIsGliderCenter=~
    bIsGliderLeft=~
    bIsGliderForward=~
    bIsGliderBack=~
    bHasDeployedAndIsStandingUpright=~
    LocalAccelForward=~
    LocalAccelRight=~
    LocalVelocityRight=~
    SkydiveAimPitch=~
    SkydiveAimYaw=~
    DeployChuteAnimRate=~
    SkydiveDriftAnimRate=~
    SkydiveDriftAnimRateCurrent=~
    SkydiveFidgetAnimRate=~
    SkydiveFidgetAnimRateCurrent=~
    SkydiveAdditiveAlpha=~
    SkydiveDriftDelay=~
   )
   Result(bCanEnterTransition=AND Boolean!(
     A=NOT Boolean(A=@n1.bIsSkydivingFromLaunchPad)
     B=NOT Boolean(A=@n1.bIsSkydivingFromBus)
     C:bool=NOT Boolean(A=@n1.bIsInVortex)
     D:bool=Get bForceOpen()
    ) -> (ReturnValue))
  ]
 }
 Transition {
  path=AnimGraph.AnimGraphNode_StateMachine_0.Locomotion.AnimStateTransitionNode_5.Transition
  nodes [
   n1 = Get bForceOpen()
   n2 = Break Fort Anim Input Skydiving(FortAnimInput_Skydiving=Get SkydivingState()) -> (
    ParaGilderRootOffset=~
    ParaGliderRootRotationOverride=~
    bUseParaGlideRootModifier=~
    bIsSkydivingFromLaunchPad=~
    bIsSkydivingFromBus=~
    bIsInVortex=~
    bIsDiving=~
    bIsParachuteOpen=~
    bIsSkydiving=~
    bIsParachuteLeaning=~
    bIsSkydiveLeaning=~
    bIsLeaning=~
    bIsSkydiveDiveMode=~
    bPlaySkydiveDrift=~
    bSkydiveDriftDelayActive=~
    bSkydiveDriftAnimAllowed=~
    bIsGliderRight=~
    bIsGliderCenter=~
    bIsGliderLeft=~
    bIsGliderForward=~
    bIsGliderBack=~
    bHasDeployedAndIsStandingUpright=~
    LocalAccelForward=~
    LocalAccelRight=~
    LocalVelocityRight=~
    SkydiveAimPitch=~
    SkydiveAimYaw=~
    DeployChuteAnimRate=~
    SkydiveDriftAnimRate=~
    SkydiveDriftAnimRateCurrent=~
    SkydiveFidgetAnimRate=~
    SkydiveFidgetAnimRateCurrent=~
    SkydiveAdditiveAlpha=~
    SkydiveDriftDelay=~
   )
   Result(bCanEnterTransition=AND Boolean!(A=NOT Boolean(A=@n2.bIsParachuteOpen), B=NOT Boolean(A=@n1)) -> (ReturnValue))
  ]
 }
 Transition {path=AnimGraph.AnimGraphNode_StateMachine_0.Locomotion.AnimStateTransitionNode_6.Transition, nodes [Result()]}
 Turret {
  entry=Off
  path=AnimGraph.AnimGraphNode_StateMachine_1.Turret
  ids=id
  nodes [
   AnimStateNode_0 = Off() {is=title AnimStateNode}
   AnimStateNode_1 = Center() {is=title AnimStateNode}
   AnimStateNode_10 = L_to_R() {is=title AnimStateNode}
   AnimStateNode_11 = R_to_L() {is=title AnimStateNode}
   AnimStateNode_12 = L_to_Off() {is=title AnimStateNode}
   AnimStateNode_13 = Off_to_L() {is=title AnimStateNode}
   AnimStateNode_14 = R_to_Off() {is=title AnimStateNode}
   AnimStateNode_15 = Off_to_R() {is=title AnimStateNode}
   AnimStateNode_2 = Left() {is=title AnimStateNode}
   AnimStateNode_3 = Right() {is=title AnimStateNode}
   AnimStateNode_4 = Into() {is=title AnimStateNode}
   AnimStateNode_5 = Out() {is=title AnimStateNode}
   AnimStateNode_6 = C_to_L() {is=title AnimStateNode}
   AnimStateNode_7 = C_to_R() {is=title AnimStateNode}
   AnimStateNode_8 = L_to_C() {is=title AnimStateNode}
   AnimStateNode_9 = R_to_C() {is=title AnimStateNode}
   AnimStateTransitionNode_0 = AnimStateTransitionNode() {crossfadeDuration=0}
   AnimStateTransitionNode_1 = AnimStateTransitionNode() {bAutomaticRuleBasedOnSequencePlayerInState=true, crossfadeDuration=0}
   AnimStateTransitionNode_10 = AnimStateTransitionNode() {crossfadeDuration=0.1}
   AnimStateTransitionNode_11 = AnimStateTransitionNode() {bAutomaticRuleBasedOnSequencePlayerInState=true, crossfadeDuration=0}
   AnimStateTransitionNode_12 = AnimStateTransitionNode() {crossfadeDuration=0.1}
   AnimStateTransitionNode_13 = AnimStateTransitionNode() {bAutomaticRuleBasedOnSequencePlayerInState=true, crossfadeDuration=0}
   AnimStateTransitionNode_14 = AnimStateTransitionNode() {crossfadeDuration=0.1}
   AnimStateTransitionNode_15 = AnimStateTransitionNode() {bAutomaticRuleBasedOnSequencePlayerInState=true, crossfadeDuration=0}
   AnimStateTransitionNode_16 = AnimStateTransitionNode() {crossfadeDuration=0}
   AnimStateTransitionNode_17 = AnimStateTransitionNode() {bAutomaticRuleBasedOnSequencePlayerInState=true, crossfadeDuration=0}
   AnimStateTransitionNode_18 = AnimStateTransitionNode() {crossfadeDuration=0.1}
   AnimStateTransitionNode_19 = AnimStateTransitionNode() {bAutomaticRuleBasedOnSequencePlayerInState=true, crossfadeDuration=0}
   AnimStateTransitionNode_2 = AnimStateTransitionNode() {crossfadeDuration=0.1}
   AnimStateTransitionNode_20 = AnimStateTransitionNode() {crossfadeDuration=0}
   AnimStateTransitionNode_21 = AnimStateTransitionNode() {bAutomaticRuleBasedOnSequencePlayerInState=true, crossfadeDuration=0}
   AnimStateTransitionNode_22 = AnimStateTransitionNode() {crossfadeDuration=0.1}
   AnimStateTransitionNode_23 = AnimStateTransitionNode() {bAutomaticRuleBasedOnSequencePlayerInState=true, crossfadeDuration=0}
   AnimStateTransitionNode_3 = AnimStateTransitionNode() {bAutomaticRuleBasedOnSequencePlayerInState=true, crossfadeDuration=0}
   AnimStateTransitionNode_4 = AnimStateTransitionNode() {crossfadeDuration=0.1}
   AnimStateTransitionNode_5 = AnimStateTransitionNode() {bAutomaticRuleBasedOnSequencePlayerInState=true, crossfadeDuration=0}
   AnimStateTransitionNode_6 = AnimStateTransitionNode() {crossfadeDuration=0.1}
   AnimStateTransitionNode_7 = AnimStateTransitionNode() {bAutomaticRuleBasedOnSequencePlayerInState=true, crossfadeDuration=0}
   AnimStateTransitionNode_8 = AnimStateTransitionNode() {crossfadeDuration=0.1}
   AnimStateTransitionNode_9 = AnimStateTransitionNode() {bAutomaticRuleBasedOnSequencePlayerInState=true, crossfadeDuration=0}
  ]
  links [
   {dst=Left, rule=AnimGraph.AnimGraphNode_StateMachine_1.Turret.AnimStateTransitionNode_5.Transition, src=C_to_L, via=AnimStateTransitionNode_5}
   {
    dst=Right
    rule=AnimGraph.AnimGraphNode_StateMachine_1.Turret.AnimStateTransitionNode_9.Transition
    src=C_to_R
    via=AnimStateTransitionNode_9
   }
   {
    dst=C_to_L
    rule=AnimGraph.AnimGraphNode_StateMachine_1.Turret.AnimStateTransitionNode_4.Transition
    src=Center
    via=AnimStateTransitionNode_4
   }
   {
    dst=C_to_R
    rule=AnimGraph.AnimGraphNode_StateMachine_1.Turret.AnimStateTransitionNode_8.Transition
    src=Center
    via=AnimStateTransitionNode_8
   }
   {dst=Out, rule=AnimGraph.AnimGraphNode_StateMachine_1.Turret.AnimStateTransitionNode_2.Transition, src=Center, via=AnimStateTransitionNode_2}
   {dst=Center, rule=AnimGraph.AnimGraphNode_StateMachine_1.Turret.AnimStateTransitionNode_1.Transition, src=Into, via=AnimStateTransitionNode_1}
   {
    dst=Center
    rule=AnimGraph.AnimGraphNode_StateMachine_1.Turret.AnimStateTransitionNode_7.Transition
    src=L_to_C
    via=AnimStateTransitionNode_7
   }
   {
    dst=Off
    rule=AnimGraph.AnimGraphNode_StateMachine_1.Turret.AnimStateTransitionNode_23.Transition
    src=L_to_Off
    via=AnimStateTransitionNode_23
   }
   {
    dst=Right
    rule=AnimGraph.AnimGraphNode_StateMachine_1.Turret.AnimStateTransitionNode_13.Transition
    src=L_to_R
    via=AnimStateTransitionNode_13
   }
   {dst=L_to_C, rule=AnimGraph.AnimGraphNode_StateMachine_1.Turret.AnimStateTransitionNode_6.Transition, src=Left, via=AnimStateTransitionNode_6}
   {
    dst=L_to_Off
    rule=AnimGraph.AnimGraphNode_StateMachine_1.Turret.AnimStateTransitionNode_22.Transition
    src=Left
    via=AnimStateTransitionNode_22
   }
   {
    dst=L_to_R
    rule=AnimGraph.AnimGraphNode_StateMachine_1.Turret.AnimStateTransitionNode_12.Transition
    src=Left
    via=AnimStateTransitionNode_12
   }
   {dst=Into, rule=AnimGraph.AnimGraphNode_StateMachine_1.Turret.AnimStateTransitionNode_0.Transition, src=Off, via=AnimStateTransitionNode_0}
   {
    dst=Off_to_L
    rule=AnimGraph.AnimGraphNode_StateMachine_1.Turret.AnimStateTransitionNode_20.Transition
    src=Off
    via=AnimStateTransitionNode_20
   }
   {
    dst=Off_to_R
    rule=AnimGraph.AnimGraphNode_StateMachine_1.Turret.AnimStateTransitionNode_16.Transition
    src=Off
    via=AnimStateTransitionNode_16
   }
   {
    dst=Left
    rule=AnimGraph.AnimGraphNode_StateMachine_1.Turret.AnimStateTransitionNode_21.Transition
    src=Off_to_L
    via=AnimStateTransitionNode_21
   }
   {
    dst=Right
    rule=AnimGraph.AnimGraphNode_StateMachine_1.Turret.AnimStateTransitionNode_17.Transition
    src=Off_to_R
    via=AnimStateTransitionNode_17
   }
   {dst=Off, rule=AnimGraph.AnimGraphNode_StateMachine_1.Turret.AnimStateTransitionNode_3.Transition, src=Out, via=AnimStateTransitionNode_3}
   {
    dst=Center
    rule=AnimGraph.AnimGraphNode_StateMachine_1.Turret.AnimStateTransitionNode_11.Transition
    src=R_to_C
    via=AnimStateTransitionNode_11
   }
   {
    dst=Left
    rule=AnimGraph.AnimGraphNode_StateMachine_1.Turret.AnimStateTransitionNode_15.Transition
    src=R_to_L
    via=AnimStateTransitionNode_15
   }
   {
    dst=Off
    rule=AnimGraph.AnimGraphNode_StateMachine_1.Turret.AnimStateTransitionNode_19.Transition
    src=R_to_Off
    via=AnimStateTransitionNode_19
   }
   {
    dst=R_to_C
    rule=AnimGraph.AnimGraphNode_StateMachine_1.Turret.AnimStateTransitionNode_10.Transition
    src=Right
    via=AnimStateTransitionNode_10
   }
   {
    dst=R_to_L
    rule=AnimGraph.AnimGraphNode_StateMachine_1.Turret.AnimStateTransitionNode_14.Transition
    src=Right
    via=AnimStateTransitionNode_14
   }
   {
    dst=R_to_Off
    rule=AnimGraph.AnimGraphNode_StateMachine_1.Turret.AnimStateTransitionNode_18.Transition
    src=Right
    via=AnimStateTransitionNode_18
   }
  ]
 }
 Off {
  path=AnimGraph.AnimGraphNode_StateMachine_1.Turret.AnimStateNode_0.Off
  nodes [Output Animation Pose(Result="Sequence Player (additive) 'Galileo_Rocket_Turret_Off_GLIDER'"() {sequence=/BRCosmetics/Animation/Game/MainPlayer/Skydive/ParaGlide/GalileoRocket/Galileo_Rocket_Turret_Off_GLIDER})]
 }
 Center {
  path=AnimGraph.AnimGraphNode_StateMachine_1.Turret.AnimStateNode_1.Center
  nodes [
   "Sequence Player (additive) 'Galileo_Rocket_Turret_Center_GLIDER'"() {sequence=/Game/Animation/Game/MainPlayer/Skydive/ParaGlide/GalileoRocket/Galileo_Rocket_Turret_Center_GLIDER}
   Output Animation Pose(Result="Sequence Player (additive) 'Galileo_Rocket_Turret_C_Idle_GLIDER'"() {sequence=/BRCosmetics/Animation/Game/MainPlayer/Skydive/ParaGlide/GalileoRocket/Galileo_Rocket_Turret_C_Idle_GLIDER})
  ]
 }
 L_to_R {
  path=AnimGraph.AnimGraphNode_StateMachine_1.Turret.AnimStateNode_10.L_to_R
  nodes [
   n1 = "Sequence Player (additive) 'Galileo_Rocket_Turret_L_to_R_GLIDER'"() {bLoopAnimation=false, sequence=/BRCosmetics/Animation/Game/MainPlayer/Skydive/ParaGlide/GalileoRocket/Galileo_Rocket_Turret_L_to_R_GLIDER}
   Output Animation Pose(Result=@n1)
  ]
 }
 R_to_L {
  path=AnimGraph.AnimGraphNode_StateMachine_1.Turret.AnimStateNode_11.R_to_L
  nodes [
   n1 = "Sequence Player (additive) 'Galileo_Rocket_Turret_R_to_L_GLIDER'"() {bLoopAnimation=false, sequence=/BRCosmetics/Animation/Game/MainPlayer/Skydive/ParaGlide/GalileoRocket/Galileo_Rocket_Turret_R_to_L_GLIDER}
   Output Animation Pose(Result=@n1)
  ]
 }
 L_to_Off {
  path=AnimGraph.AnimGraphNode_StateMachine_1.Turret.AnimStateNode_12.L_to_Off
  nodes [
   n1 = "Sequence Player (additive) 'Galileo_Rocket_Turret_L_to_Off_GLIDER'"() {bLoopAnimation=false, sequence=/BRCosmetics/Animation/Game/MainPlayer/Skydive/ParaGlide/GalileoRocket/Galileo_Rocket_Turret_L_to_Off_GLIDER}
   Output Animation Pose(Result=@n1)
  ]
 }
 Off_to_L {
  path=AnimGraph.AnimGraphNode_StateMachine_1.Turret.AnimStateNode_13.Off_to_L
  nodes [
   n1 = "Sequence Player (additive) 'Galileo_Rocket_Turret_Off_to_L_GLIDER'"() {bLoopAnimation=false, sequence=/BRCosmetics/Animation/Game/MainPlayer/Skydive/ParaGlide/GalileoRocket/Galileo_Rocket_Turret_Off_to_L_GLIDER}
   Output Animation Pose(Result=@n1)
  ]
 }
 R_to_Off {
  path=AnimGraph.AnimGraphNode_StateMachine_1.Turret.AnimStateNode_14.R_to_Off
  nodes [
   n1 = "Sequence Player (additive) 'Galileo_Rocket_Turret_R_to_Off_GLIDER'"() {bLoopAnimation=false, sequence=/BRCosmetics/Animation/Game/MainPlayer/Skydive/ParaGlide/GalileoRocket/Galileo_Rocket_Turret_R_to_Off_GLIDER}
   Output Animation Pose(Result=@n1)
  ]
 }
 Off_to_R {
  path=AnimGraph.AnimGraphNode_StateMachine_1.Turret.AnimStateNode_15.Off_to_R
  nodes [
   n1 = "Sequence Player (additive) 'Galileo_Rocket_Turret_Off_to_R_GLIDER'"() {bLoopAnimation=false, sequence=/BRCosmetics/Animation/Game/MainPlayer/Skydive/ParaGlide/GalileoRocket/Galileo_Rocket_Turret_Off_to_R_GLIDER}
   Output Animation Pose(Result=@n1)
  ]
 }
 Left {
  path=AnimGraph.AnimGraphNode_StateMachine_1.Turret.AnimStateNode_2.Left
  nodes [
   "Sequence Player (additive) 'Galileo_Rocket_Turret_Left_GLIDER'"() {sequence=/Game/Animation/Game/MainPlayer/Skydive/ParaGlide/GalileoRocket/Galileo_Rocket_Turret_Left_GLIDER}
   Output Animation Pose(Result="Sequence Player (additive) 'Galileo_Rocket_Turret_L_Idle_GLIDER'"() {sequence=/BRCosmetics/Animation/Game/MainPlayer/Skydive/ParaGlide/GalileoRocket/Galileo_Rocket_Turret_L_Idle_GLIDER})
  ]
 }
 Right {
  path=AnimGraph.AnimGraphNode_StateMachine_1.Turret.AnimStateNode_3.Right
  nodes [
   "Sequence Player (additive) 'Galileo_Rocket_Turret_Right_GLIDER'"() {sequence=/Game/Animation/Game/MainPlayer/Skydive/ParaGlide/GalileoRocket/Galileo_Rocket_Turret_Right_GLIDER}
   Output Animation Pose(Result="Sequence Player (additive) 'Galileo_Rocket_Turret_R_Idle_GLIDER'"() {sequence=/BRCosmetics/Animation/Game/MainPlayer/Skydive/ParaGlide/GalileoRocket/Galileo_Rocket_Turret_R_Idle_GLIDER})
  ]
 }
 Into {
  path=AnimGraph.AnimGraphNode_StateMachine_1.Turret.AnimStateNode_4.Into
  nodes [
   n1 = "Sequence Player (additive) 'Galileo_Rocket_Turret_Into_GLIDER'"() {bLoopAnimation=false, sequence=/BRCosmetics/Animation/Game/MainPlayer/Skydive/ParaGlide/GalileoRocket/Galileo_Rocket_Turret_Into_GLIDER}
   Output Animation Pose(Result=@n1)
  ]
 }
 Out {
  path=AnimGraph.AnimGraphNode_StateMachine_1.Turret.AnimStateNode_5.Out
  nodes [
   n1 = "Sequence Player (additive) 'Galileo_Rocket_Turret_Out_GLIDER'"() {bLoopAnimation=false, sequence=/BRCosmetics/Animation/Game/MainPlayer/Skydive/ParaGlide/GalileoRocket/Galileo_Rocket_Turret_Out_GLIDER}
   Output Animation Pose(Result=@n1)
  ]
 }
 C_to_L {
  path=AnimGraph.AnimGraphNode_StateMachine_1.Turret.AnimStateNode_6.C_to_L
  nodes [
   n1 = "Sequence Player (additive) 'Galileo_Rocket_Turret_C_to_L_GLIDER'"() {bLoopAnimation=false, sequence=/BRCosmetics/Animation/Game/MainPlayer/Skydive/ParaGlide/GalileoRocket/Galileo_Rocket_Turret_C_to_L_GLIDER}
   Output Animation Pose(Result=@n1)
  ]
 }
 C_to_R {
  path=AnimGraph.AnimGraphNode_StateMachine_1.Turret.AnimStateNode_7.C_to_R
  nodes [
   n1 = "Sequence Player (additive) 'Galileo_Rocket_Turret_C_to_R_GLIDER'"() {bLoopAnimation=false, sequence=/BRCosmetics/Animation/Game/MainPlayer/Skydive/ParaGlide/GalileoRocket/Galileo_Rocket_Turret_C_to_R_GLIDER}
   Output Animation Pose(Result=@n1)
  ]
 }
 L_to_C {
  path=AnimGraph.AnimGraphNode_StateMachine_1.Turret.AnimStateNode_8.L_to_C
  nodes [
   n1 = "Sequence Player (additive) 'Galileo_Rocket_Turret_L_to_C_GLIDER'"() {bLoopAnimation=false, sequence=/BRCosmetics/Animation/Game/MainPlayer/Skydive/ParaGlide/GalileoRocket/Galileo_Rocket_Turret_L_to_C_GLIDER}
   Output Animation Pose(Result=@n1)
  ]
 }
 R_to_C {
  path=AnimGraph.AnimGraphNode_StateMachine_1.Turret.AnimStateNode_9.R_to_C
  nodes [
   n1 = "Sequence Player (additive) 'Galileo_Rocket_Turret_R_to_C_GLIDER'"() {bLoopAnimation=false, sequence=/BRCosmetics/Animation/Game/MainPlayer/Skydive/ParaGlide/GalileoRocket/Galileo_Rocket_Turret_R_to_C_GLIDER}
   Output Animation Pose(Result=@n1)
  ]
 }
 Transition {
  path=AnimGraph.AnimGraphNode_StateMachine_1.Turret.AnimStateTransitionNode_0.Transition
  nodes [
   n1 = Get TurretMode() -> (TurretMode:Galileo_Rocket_Turret_Mode Enum=NewEnumerator0)
   Result(bCanEnterTransition=NOT Boolean(A="Equal (Enum)"(A:Galileo_Rocket_Turret_Mode Enum=@n1, B:Galileo_Rocket_Turret_Mode Enum=NewEnumerator0)))
  ]
 }
 Transition {path=AnimGraph.AnimGraphNode_StateMachine_1.Turret.AnimStateTransitionNode_1.Transition, nodes [Result()]}
 Transition {
  path=AnimGraph.AnimGraphNode_StateMachine_1.Turret.AnimStateTransitionNode_10.Transition
  nodes [
   n1 = Get TurretMode() -> (TurretMode=NewEnumerator0)
   Result(bCanEnterTransition=OR Boolean!(A="Equal (Enum)"(A=@n1, B=NewEnumerator1), B="Equal (Enum)"(A=@n1, B=NewEnumerator0)) -> (ReturnValue))
  ]
 }
 Transition {path=AnimGraph.AnimGraphNode_StateMachine_1.Turret.AnimStateTransitionNode_11.Transition, nodes [Result()]}
 Transition {
  path=AnimGraph.AnimGraphNode_StateMachine_1.Turret.AnimStateTransitionNode_12.Transition
  nodes [Result(bCanEnterTransition="Equal (Enum)"(A=Get TurretMode() -> (TurretMode=NewEnumerator0), B=NewEnumerator3))]
 }
 Transition {path=AnimGraph.AnimGraphNode_StateMachine_1.Turret.AnimStateTransitionNode_13.Transition, nodes [Result()]}
 Transition {
  path=AnimGraph.AnimGraphNode_StateMachine_1.Turret.AnimStateTransitionNode_14.Transition
  nodes [Result(bCanEnterTransition="Equal (Enum)"(A=Get TurretMode() -> (TurretMode=NewEnumerator0), B=NewEnumerator2))]
 }
 Transition {path=AnimGraph.AnimGraphNode_StateMachine_1.Turret.AnimStateTransitionNode_15.Transition, nodes [Result()]}
 Transition {
  path=AnimGraph.AnimGraphNode_StateMachine_1.Turret.AnimStateTransitionNode_16.Transition
  nodes [Result(bCanEnterTransition="Equal (Enum)"(A=Get TurretMode() -> (TurretMode=NewEnumerator0), B=NewEnumerator3))]
 }
 Transition {path=AnimGraph.AnimGraphNode_StateMachine_1.Turret.AnimStateTransitionNode_17.Transition, nodes [Result()]}
 Transition {
  path=AnimGraph.AnimGraphNode_StateMachine_1.Turret.AnimStateTransitionNode_18.Transition
  nodes [Result(bCanEnterTransition="Equal (Enum)"(A=Get TurretMode() -> (TurretMode=NewEnumerator0), B=NewEnumerator0))]
 }
 Transition {path=AnimGraph.AnimGraphNode_StateMachine_1.Turret.AnimStateTransitionNode_19.Transition, nodes [Result()]}
 Transition {
  path=AnimGraph.AnimGraphNode_StateMachine_1.Turret.AnimStateTransitionNode_2.Transition
  nodes [Result(bCanEnterTransition="Equal (Enum)"(A=Get TurretMode() -> (TurretMode=NewEnumerator0), B=NewEnumerator0))]
 }
 Transition {
  path=AnimGraph.AnimGraphNode_StateMachine_1.Turret.AnimStateTransitionNode_20.Transition
  nodes [Result(bCanEnterTransition="Equal (Enum)"(A=Get TurretMode() -> (TurretMode=NewEnumerator0), B=NewEnumerator2))]
 }
 Transition {path=AnimGraph.AnimGraphNode_StateMachine_1.Turret.AnimStateTransitionNode_21.Transition, nodes [Result()]}
 Transition {
  path=AnimGraph.AnimGraphNode_StateMachine_1.Turret.AnimStateTransitionNode_22.Transition
  nodes [Result(bCanEnterTransition="Equal (Enum)"(A=Get TurretMode() -> (TurretMode=NewEnumerator0), B=NewEnumerator0))]
 }
 Transition {path=AnimGraph.AnimGraphNode_StateMachine_1.Turret.AnimStateTransitionNode_23.Transition, nodes [Result()]}
 Transition {path=AnimGraph.AnimGraphNode_StateMachine_1.Turret.AnimStateTransitionNode_3.Transition, nodes [Result()]}
 Transition {
  path=AnimGraph.AnimGraphNode_StateMachine_1.Turret.AnimStateTransitionNode_4.Transition
  nodes [Result(bCanEnterTransition="Equal (Enum)"(A=Get TurretMode() -> (TurretMode=NewEnumerator0), B=NewEnumerator2))]
 }
 Transition {path=AnimGraph.AnimGraphNode_StateMachine_1.Turret.AnimStateTransitionNode_5.Transition, nodes [Result()]}
 Transition {
  path=AnimGraph.AnimGraphNode_StateMachine_1.Turret.AnimStateTransitionNode_6.Transition
  nodes [
   n1 = Get TurretMode() -> (TurretMode=NewEnumerator0)
   Result(bCanEnterTransition=OR Boolean(A="Equal (Enum)"(A=@n1, B=NewEnumerator1), B="Equal (Enum)"(A=@n1, B=NewEnumerator0)))
  ]
 }
 Transition {path=AnimGraph.AnimGraphNode_StateMachine_1.Turret.AnimStateTransitionNode_7.Transition, nodes [Result()]}
 Transition {
  path=AnimGraph.AnimGraphNode_StateMachine_1.Turret.AnimStateTransitionNode_8.Transition
  nodes [Result(bCanEnterTransition="Equal (Enum)"(A=Get TurretMode() -> (TurretMode=NewEnumerator0), B=NewEnumerator3))]
 }
 Transition {path=AnimGraph.AnimGraphNode_StateMachine_1.Turret.AnimStateTransitionNode_9.Transition, nodes [Result()]}
 EventGraph {nodes [
   n1 = ForwardBack_Value_Interp!(
    execute
    Time:double=Get Time() -> (Time:double)
    Currnet_Value:double=Get Example() -> (Example:double)
    Idle_Value:double=1
    Idle_InterpSpeed:double=1
    Forward_Value:double=2
    Forward_InterpSpeed:double=2
    Back_Value:double=1
    Back_InterpSpeed:double=2
    LocalAccelForward:double=Break Fort Anim Input Skydiving(FortAnimInput_Skydiving=Get SkydivingState()) -> (
     ParaGilderRootOffset=~
     ParaGliderRootRotationOverride=~
     bUseParaGlideRootModifier=~
     bIsSkydivingFromLaunchPad=~
     bIsSkydivingFromBus=~
     bIsInVortex=~
     bIsDiving=~
     bIsParachuteOpen=~
     bIsSkydiving=~
     bIsParachuteLeaning=~
     bIsSkydiveLeaning=~
     bIsLeaning=~
     bIsSkydiveDiveMode=~
     bPlaySkydiveDrift=~
     bSkydiveDriftDelayActive=~
     bSkydiveDriftAnimAllowed=~
     bIsGliderRight=~
     bIsGliderCenter=~
     bIsGliderLeft=~
     bIsGliderForward=~
     bIsGliderBack=~
     bHasDeployedAndIsStandingUpright=~
     LocalAccelForward=~
     LocalAccelRight=~
     LocalVelocityRight=~
     SkydiveAimPitch=~
     SkydiveAimYaw=~
     DeployChuteAnimRate=~
     SkydiveDriftAnimRate=~
     SkydiveDriftAnimRateCurrent=~
     SkydiveFidgetAnimRate=~
     SkydiveFidgetAnimRateCurrent=~
     SkydiveAdditiveAlpha=~
     SkydiveDriftDelay=~
    ).LocalAccelForward
   ) -> (then, New_Value:double)
   > Set Example(Example:double=@n1) -> (Output_Get:double)
   n2 = Event BlueprintUpdateAnimation()
   > Set Time(Time:double=@n2.DeltaTimeX) -> (Output_Get:double)
   > n3 = Sequence!(execute) -> (then_0, then_1, then_2:Exec, then_3:Exec, then_4:Exec)
   > Get Player Open Glider Bool!(execute:Exec) -> (then:Exec) {is=title K2Node_MacroInstance}
   > Branch(Condition="float >= float"(A=GetCurveValue(CurveName=TurretOff), B=0.9) {is=title K2Node_CallFunction})
   > Set TurretActive(TurretActive:bool=false) -> (Output_Get:bool)
   Set Turret Direction!(execute:Exec=@n3.then_1) -> (then:Exec) {is=title K2Node_MacroInstance}
   Set Turret Mode!(execute:Exec=@n3.then_2) {is=title K2Node_MacroInstance}
   Random Dir Non Repeating!(execute:Exec=@n3.then_3) -> (then:Exec) {is=title K2Node_MacroInstance}
   AnimNotify_TurretOn() {is=title K2Node_Event}
   > Delay(Duration=5)
   > Set TurretActive(TurretActive=true)
   AnimNotify_ClosedEntered() {is=title K2Node_Event}
   > Set TurretActive(TurretActive=false)
   > Set TurretMode(TurretMode:Galileo_Rocket_Turret_Mode Enum=NewEnumerator0) -> (Output_Get:Galileo_Rocket_Turret_Mode Enum=NewEnumerator0)
   n4 = GetOwningComponent()
   > HideBoneByName!(execute, self:Skinned Mesh Component Object Reference=@n4, BoneName:Name=Glider_Base, PhysBodyOption:EPhysBodyOp Enum=PBO_None) -> (then)
   AnimNotify_DeployEntered() {is=title K2Node_Event}
   > Set TurretActive(TurretActive=false)
   > Set TurretMode(TurretMode=NewEnumerator0) -> (Output_Get=NewEnumerator0)
   > Delay(Duration=0.1)
   > UnHideBoneByName!(execute, self:Skinned Mesh Component Object Reference=@n4, BoneName:Name=Glider_Base) -> (then)
   AnimNotify_EnginesOn() {is=title K2Node_Event}
   > SetGliderFullyDeployed!(execute, self:Fort Glider Anim Instance Object Reference, bIsFullyDeployed:bool=true) -> (then)
   n5 = Increment Int(Value=Get TimesDeployed() -> (TimesDeployed:int=0))
   > Set TimesDeployed(TimesDeployed:int=@n5) -> (Output_Get:int=0)
   PrintString!(
    execute
    InString="To String (Integer)"(InInt=Get TimesDeployed() -> (TimesDeployed=0))
    bPrintToScreen
    bPrintToLog
    TextColor
    Duration
    Key
   ) -> (then)
  ]}
 GetPlayerOpenGliderBool {nodes [
   Inputs()
   > n1 = Branch(Condition=Get IsCosmeticPreview(self=Get OwnerGlider()))
   Outputs!(then:Exec=[@n2.then, @n1.then])
   n2 = Set bForceOpen(execute=@n1.else, bForceOpen:bool=ShouldForceOpenGlider!(self:Fort Glider Anim Instance Object Reference) -> (ReturnValue:bool)) -> (Output_Get:bool)
  ]}
 RandomDirNonRepeating {nodes [
   Inputs()
   > Set TurretCounter(TurretCounter:int=RandomIntegerInRange!(Min:int=Get TurretMinDirection() -> (TurretMinDirection:int=0), Max:int=2) -> (ReturnValue:int=0)) -> (Output_Get:int=0)
   > n1 = Branch(Condition="Equal (Integer)"(A=Get TurretCounter() -> (TurretCounter:int=0), B=Get TurretCurrentDir() -> (TurretCurrentDir:int=0)) {is=title K2Node_CallFunction})
   > Increment Int(Value=Get TurretCounter() -> (TurretCounter=0))
   > n2 = Branch(Condition="integer > integer"(A=Get TurretCounter() -> (TurretCounter=0), B=2) {is=title K2Node_CallFunction})
   > n3 = Set TurretCurrentDir(TurretCurrentDir:int=Get TurretMinDirection() -> (TurretMinDirection=0)) -> (Output_Get:int=0)
   Outputs(then=[@n5.then, @n4.then, @n3.then])
   n4 = Set TurretCurrentDir(execute=@n2.else, TurretCurrentDir=Get TurretCounter() -> (TurretCounter=0)) -> (Output_Get=0)
   n5 = Set TurretCurrentDir(execute=@n1.else, TurretCurrentDir=Get TurretCounter() -> (TurretCounter=0)) -> (Output_Get=0)
  ]}
 SetTurretDirection {nodes [
   Inputs()
   > n1 = Branch(Condition=Get TurretActive() -> (TurretActive:bool))
   > n2 = Sequence!(execute) -> (then_0, then_1)
   > Do Once(Reset=@n3.then)
   > Set TurretMode(TurretMode=Select!(
     Option 0:Galileo_Rocket_Turret_Mode Enum=NewEnumerator1
     Option 1:Galileo_Rocket_Turret_Mode Enum=NewEnumerator2
     Option 2:Galileo_Rocket_Turret_Mode Enum=NewEnumerator3
     Index:int=Get TurretCurrentDir() -> (TurretCurrentDir=0)
    ) -> (ReturnValue:Galileo_Rocket_Turret_Mode Enum)) -> (Output_Get=NewEnumerator0)
   > Outputs()
   n3 = Delay(execute=@n2.then_1, Duration=RandomFloatInRange(Min=5, Max=7))
   Set TurretMode(execute=@n1.else, TurretMode=NewEnumerator0) -> (Output_Get=NewEnumerator0)
   RandomIntegerInRange(Min=0, Max=2) -> (ReturnValue=0)
  ]}
 SetTurretMode {nodes [
   Inputs()
   > n1 = Branch(Condition="float >= float"(A=GetCurveValue(CurveName=TurretOff), B=0.9))
   Delay(
    execute=@n1.else
    Duration=Select!(Option 0:double=5, Option 1:double=RandomFloatInRange(Min=12, Max=17), Index:bool=Get TurretActive()) -> (ReturnValue:double)
   )
   > n2 = Branch(Condition=Get TurretActive())
   > Set TurretActive(TurretActive=false)
   Set TurretActive(execute=@n2.else, TurretActive=true)
   Outputs!()
  ]}
]
variables [
 AimPitch {type=number}
 DeployAnimRate {type=number}
 Example {type=number}
 Forward {type=number}
 Into_B_Side {type=BlendSpace, default=&BS_Realm_Male_Into_GLIDER}
 IsParachuteOpen {type=boolean}
 Lightspeed {type=boolean, default=true}
 ReActivateTurret {type=boolean, default=true}
 Right {type=number}
 SkydiveYaw {type=number}
 Time {type=number}
 TimesDeployed {type=integer}
 TurretActive {type=boolean}
 TurretCounter {type=integer}
 TurretCurrentDir {type=integer}
 TurretMinDirection {type=integer, default=1}
 TurretMode {type=Galileo_Rocket_Turret_Mode, default=Off}
 UseAlternatePlayerInto {type=boolean}
 UseDynamicsAlpha {type=number}
]
