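# Animation Blueprint for the Rumble glider, animating its deployment and in-flight motion from the skydiving state.
# A state machine runs Closed, Into and Glide from parachute-open or force-open, snapping shut when both clear.
# Lean, aim-turn and spiral jitter (faster and stronger when gliding forward) layer additively, weighted by the inverted ParaGlide_Lean curve, before DefaultSlot.
# The TurnONEngines notify marks the glider fully deployed.
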AnimBlueprint /BRCosmetics/Gadgets/Assets/VinderTech_GliderChute/Blueprints/Rumble_Glider_AnimBP
target_skeleton=/BRCosmetics/Gadgets/Assets/VinderTech_GliderChute/Glider_Rumble_Female/Meshes/Rumble_Female_Glider_Skeleton
graphs [
 AnimGraph {nodes [
   n1 = Break Fort Anim Input Skydiving!(FortAnimInput_Skydiving:"Fort Anim Input Skydiving Structure (by ref)"=Get SkydivingState() -> (SkydivingState:Fort Anim Input Skydiving Structure)) -> (
    ParaGilderRootOffset:Vector=~
    ParaGliderRootRotationOverride:Rotator=~
    bUseParaGlideRootModifier:bool=~
    bIsSkydivingFromLaunchPad:bool=~
    bIsSkydivingFromBus:bool=~
    bIsInVortex:bool=~
    bIsDiving:bool=~
    bIsParachuteOpen:bool=~
    bIsSkydiving:bool=~
    bIsParachuteLeaning:bool=~
    bIsSkydiveLeaning:bool=~
    bIsLeaning:bool=~
    bIsSkydiveDiveMode:bool=~
    bPlaySkydiveDrift:bool=~
    bSkydiveDriftDelayActive:bool=~
    bSkydiveDriftAnimAllowed:bool=~
    bIsGliderRight:bool=~
    bIsGliderCenter:bool=~
    bIsGliderLeft:bool=~
    bIsGliderForward:bool=~
    bIsGliderBack:bool=~
    bHasDeployedAndIsStandingUpright:bool=~
    LocalAccelForward:float=~
    LocalAccelRight:float=~
    LocalVelocityRight:float=~
    SkydiveAimPitch:float=~
    SkydiveAimYaw:float=~
    DeployChuteAnimRate:float=~
    SkydiveDriftAnimRate:float=~
    SkydiveDriftAnimRateCurrent:float=~
    SkydiveFidgetAnimRate:float=~
    SkydiveFidgetAnimRateCurrent:float=~
    SkydiveAdditiveAlpha:float=~
    SkydiveDriftDelay:float=~
    SkydiveDriftAnim:int
    SkydiveDriftAnimMax:int
    LaunchpadAnim:int
    LocalAccelDir:ESkydivingDirection Enum
    DirectionLast:ESkydivingDirection Enum
    PoseAdditive:Anim Sequence Object Reference
   ) {is=title K2Node_BreakStruct}
   AnimGraphNode_StateMachine_0 = New State Machine() {is=title AnimGraphNode_StateMachine}
   n2 = Apply Additive!(
    Base=Apply Additive(
     Base=@AnimGraphNode_StateMachine_0
     Additive=Blendspace Player 'Rumble_Glider_Lean_BS'(X=@n1.LocalAccelRight, Y=@n1.LocalAccelForward) {blendSpace=/BRCosmetics/Animation/Game/MainPlayer/Skydive/ParaGlide/Rumble/Rumble_Glider_Lean_BS}
     AlphaCurveName=ParaGlide_Lean
    ) {alphaInputType=Curve, alphaBoolBlend {blendOption=Linear}, alphaScaleBiasClamp {bias=1, scale=-1}}
    Additive=Blendspace Player 'Rumble_Glider_Turn_BS'(X=@n1.SkydiveAimYaw, Y=@n1.SkydiveAimPitch) {blendSpace=/BRCosmetics/Animation/Game/MainPlayer/Skydive/ParaGlide/Rumble/Rumble_Glider_Turn_BS}
    AlphaCurveName:Name=ParaGlide_Lean
   ) -> (Pose) {alphaInputType=Curve, alphaBoolBlend {blendOption=Linear}, alphaScaleBiasClamp {bias=1, scale=-1}}
   n3 = "Sequence Player (additive) 'Rumble_Glide_Spiral_GLIDER'"!(PlayRate:float=2) -> (Pose) {groupName=JItter, method=SyncGroup, sequence=/BRCosmetics/Animation/Game/MainPlayer/Skydive/ParaGlide/Rumble/Rumble_Glide_Spiral_GLIDER}
   n4 = "Sequence Player (additive) 'Rumble_Glide_Spiral_GLIDER'"(PlayRate=1) {groupName=JItter, method=SyncGroup, sequence=&Rumble_Glide_Spiral_GLIDER}
   n5 = Blend Poses by bool(
    BlendPose_0=@n3
    BlendPose_1=Two Way Blend(A=Additive Identity Pose(), B=@n4, Alpha=0.5) {blendNode {alphaBoolBlend {blendOption=Linear}}}
    bActiveValue=@n1.bIsGliderForward
   ) {blendType=Linear}
   Output Pose(Result=Slot 'DefaultSlot'(Source=Apply Additive(Base=@n2, Additive=@n5, AlphaCurveName=ParaGlide_Lean) {alphaInputType=Curve, alphaBoolBlend {blendOption=Linear}, alphaScaleBiasClamp {bias=1, scale=-1}})) {layerGroup=None}
  ]}
 New State Machine {
  entry=Closed
  path=AnimGraph.AnimGraphNode_StateMachine_0.New State Machine
  ids=id
  nodes [
   AnimStateNode_0 = Closed() {is=title AnimStateNode}
   AnimStateNode_1 = Into() {is=title AnimStateNode}
   AnimStateNode_2 = Glide() {is=title AnimStateNode}
   AnimStateTransitionNode_0 = AnimStateTransitionNode()
   AnimStateTransitionNode_1 = AnimStateTransitionNode() {bAutomaticRuleBasedOnSequencePlayerInState=true}
   AnimStateTransitionNode_2 = AnimStateTransitionNode() {crossfadeDuration=0}
   AnimStateTransitionNode_3 = AnimStateTransitionNode() {crossfadeDuration=0}
  ]
  links [
   {
    dst=Into
    rule=AnimGraph.AnimGraphNode_StateMachine_0.New State Machine.AnimStateTransitionNode_0.Transition
    src=Closed
    via=AnimStateTransitionNode_0
   }
   {
    dst=Closed
    rule=AnimGraph.AnimGraphNode_StateMachine_0.New State Machine.AnimStateTransitionNode_2.Transition
    src=Glide
    via=AnimStateTransitionNode_2
   }
   {
    dst=Closed
    rule=AnimGraph.AnimGraphNode_StateMachine_0.New State Machine.AnimStateTransitionNode_3.Transition
    src=Into
    via=AnimStateTransitionNode_3
   }
   {
    dst=Glide
    rule=AnimGraph.AnimGraphNode_StateMachine_0.New State Machine.AnimStateTransitionNode_1.Transition
    src=Into
    via=AnimStateTransitionNode_1
   }
  ]
 }
 Closed {
  path=AnimGraph.AnimGraphNode_StateMachine_0.New State Machine.AnimStateNode_0.Closed
  nodes [Output Animation Pose(Result=Sequence Player 'Rumble_Closed_GLIDER'() {sequence=/BRCosmetics/Animation/Game/MainPlayer/Skydive/ParaGlide/Rumble/Rumble_Closed_GLIDER})]
 }
 Into {
  path=AnimGraph.AnimGraphNode_StateMachine_0.New State Machine.AnimStateNode_1.Into
  nodes [Output Animation Pose(Result=Sequence Player 'Rumble_Into_GLIDER'!(bLoopAnimation:bool=False) -> (Pose) {bLoopAnimation=false, sequence=/BRCosmetics/Animation/Game/MainPlayer/Skydive/ParaGlide/Rumble/Rumble_Into_GLIDER})]
 }
 Glide {
  path=AnimGraph.AnimGraphNode_StateMachine_0.New State Machine.AnimStateNode_2.Glide
  nodes [Output Animation Pose(Result=Sequence Player 'Rumble_Base_Pose_GLIDER'() {sequence=/BRCosmetics/Animation/Game/MainPlayer/Skydive/ParaGlide/Rumble/Rumble_Base_Pose_GLIDER})]
 }
 Transition {
  path=AnimGraph.AnimGraphNode_StateMachine_0.New State Machine.AnimStateTransitionNode_0.Transition
  nodes [
   n1 = Get SkydivingState()
   Result(bCanEnterTransition=OR Boolean(
     A=Break Fort Anim Input Skydiving(FortAnimInput_Skydiving=@n1) -> (
      ParaGilderRootOffset=~
      ParaGliderRootRotationOverride=~
      bUseParaGlideRootModifier=~
      bIsSkydivingFromLaunchPad=~
      bIsSkydivingFromBus=~
      bIsInVortex=~
      bIsDiving=~
      bIsParachuteOpen=~
      bIsSkydiving=~
      bIsParachuteLeaning=~
      bIsSkydiveLeaning=~
      bIsLeaning=~
      bIsSkydiveDiveMode=~
      bPlaySkydiveDrift=~
      bSkydiveDriftDelayActive=~
      bSkydiveDriftAnimAllowed=~
      bIsGliderRight=~
      bIsGliderCenter=~
      bIsGliderLeft=~
      bIsGliderForward=~
      bIsGliderBack=~
      bHasDeployedAndIsStandingUpright=~
      LocalAccelForward=~
      LocalAccelRight=~
      LocalVelocityRight=~
      SkydiveAimPitch=~
      SkydiveAimYaw=~
      DeployChuteAnimRate=~
      SkydiveDriftAnimRate=~
      SkydiveDriftAnimRateCurrent=~
      SkydiveFidgetAnimRate=~
      SkydiveFidgetAnimRateCurrent=~
      SkydiveAdditiveAlpha=~
      SkydiveDriftDelay=~
     ).bIsParachuteOpen
     B=Get bForceOpen() -> (bForceOpen:bool)
    ))
  ]
 }
 Transition {path=AnimGraph.AnimGraphNode_StateMachine_0.New State Machine.AnimStateTransitionNode_1.Transition, nodes [Result()]}
 Transition {
  path=AnimGraph.AnimGraphNode_StateMachine_0.New State Machine.AnimStateTransitionNode_2.Transition
  nodes [
   n1 = Get bForceOpen()
   n2 = Break Fort Anim Input Skydiving(FortAnimInput_Skydiving=Get SkydivingState()) -> (
    ParaGilderRootOffset=~
    ParaGliderRootRotationOverride=~
    bUseParaGlideRootModifier=~
    bIsSkydivingFromLaunchPad=~
    bIsSkydivingFromBus=~
    bIsInVortex=~
    bIsDiving=~
    bIsParachuteOpen=~
    bIsSkydiving=~
    bIsParachuteLeaning=~
    bIsSkydiveLeaning=~
    bIsLeaning=~
    bIsSkydiveDiveMode=~
    bPlaySkydiveDrift=~
    bSkydiveDriftDelayActive=~
    bSkydiveDriftAnimAllowed=~
    bIsGliderRight=~
    bIsGliderCenter=~
    bIsGliderLeft=~
    bIsGliderForward=~
    bIsGliderBack=~
    bHasDeployedAndIsStandingUpright=~
    LocalAccelForward=~
    LocalAccelRight=~
    LocalVelocityRight=~
    SkydiveAimPitch=~
    SkydiveAimYaw=~
    DeployChuteAnimRate=~
    SkydiveDriftAnimRate=~
    SkydiveDriftAnimRateCurrent=~
    SkydiveFidgetAnimRate=~
    SkydiveFidgetAnimRateCurrent=~
    SkydiveAdditiveAlpha=~
    SkydiveDriftDelay=~
   )
   Result(bCanEnterTransition=AND Boolean(A=NOT Boolean(A=@n2.bIsParachuteOpen), B=NOT Boolean(A=@n1)))
  ]
 }
 Transition {
  path=AnimGraph.AnimGraphNode_StateMachine_0.New State Machine.AnimStateTransitionNode_3.Transition
  nodes [
   n1 = Get bForceOpen()
   n2 = Break Fort Anim Input Skydiving(FortAnimInput_Skydiving=Get SkydivingState()) -> (
    ParaGilderRootOffset=~
    ParaGliderRootRotationOverride=~
    bUseParaGlideRootModifier=~
    bIsSkydivingFromLaunchPad=~
    bIsSkydivingFromBus=~
    bIsInVortex=~
    bIsDiving=~
    bIsParachuteOpen=~
    bIsSkydiving=~
    bIsParachuteLeaning=~
    bIsSkydiveLeaning=~
    bIsLeaning=~
    bIsSkydiveDiveMode=~
    bPlaySkydiveDrift=~
    bSkydiveDriftDelayActive=~
    bSkydiveDriftAnimAllowed=~
    bIsGliderRight=~
    bIsGliderCenter=~
    bIsGliderLeft=~
    bIsGliderForward=~
    bIsGliderBack=~
    bHasDeployedAndIsStandingUpright=~
    LocalAccelForward=~
    LocalAccelRight=~
    LocalVelocityRight=~
    SkydiveAimPitch=~
    SkydiveAimYaw=~
    DeployChuteAnimRate=~
    SkydiveDriftAnimRate=~
    SkydiveDriftAnimRateCurrent=~
    SkydiveFidgetAnimRate=~
    SkydiveFidgetAnimRateCurrent=~
    SkydiveAdditiveAlpha=~
    SkydiveDriftDelay=~
   )
   Result(bCanEnterTransition=AND Boolean(A=NOT Boolean(A=@n2.bIsParachuteOpen), B=NOT Boolean(A=@n1)))
  ]
 }
 EventGraph {nodes [
   AnimNotify_TurnONEngines() {is=title K2Node_Event}
   > SetGliderFullyDeployed!(execute, self:Fort Glider Anim Instance Object Reference, bIsFullyDeployed:bool=true) -> (then)
   TryGetPawnOwner()
   Event BlueprintUpdateAnimation()
  ]}
]
variables []
