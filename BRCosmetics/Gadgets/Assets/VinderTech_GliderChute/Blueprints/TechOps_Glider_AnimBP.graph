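# Animation Blueprint for the VinderTech TechOps glider, animating deployment, flight and spinning rotors.
# A Closed/Open_Into/Idle_Open state machine follows parachute state or bForceOpen, using aim blendspaces or cosmetic-preview sequences, plus curve-driven additive lean.
# EventGraph resets rotors on opening, then notify-opened gates integrate warm-up spin, scaled by forward acceleration (top) and smoothed yaw (tail).
# Bone modifiers apply these angles, counter-rotating the two top rotors.
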AnimBlueprint /BRCosmetics/Gadgets/Assets/VinderTech_GliderChute/Blueprints/TechOps_Glider_AnimBP
target_skeleton=/BRCosmetics/Gadgets/Assets/VinderTech_GliderChute/Glider_Tech_Ops/Meshes/SK_Glider_TechOps_Skeleton
graphs [
 AnimGraph {nodes [
   n1 = Break Fort Anim Input Skydiving!(FortAnimInput_Skydiving:"Fort Anim Input Skydiving Structure (by ref)"=Get SkydivingState() -> (SkydivingState:Fort Anim Input Skydiving Structure)) -> (
    ParaGilderRootOffset:Vector=~
    ParaGliderRootRotationOverride:Rotator=~
    bUseParaGlideRootModifier:bool=~
    bIsSkydivingFromLaunchPad:bool=~
    bIsSkydivingFromBus:bool=~
    bIsInVortex:bool=~
    bIsDiving:bool=~
    bIsParachuteOpen:bool=~
    bIsSkydiving:bool=~
    bIsParachuteLeaning:bool=~
    bIsSkydiveLeaning:bool=~
    bIsLeaning:bool=~
    bIsSkydiveDiveMode:bool=~
    bPlaySkydiveDrift:bool=~
    bSkydiveDriftDelayActive:bool=~
    bSkydiveDriftAnimAllowed:bool=~
    bIsGliderRight:bool=~
    bIsGliderCenter:bool=~
    bIsGliderLeft:bool=~
    bIsGliderForward:bool=~
    bIsGliderBack:bool=~
    bHasDeployedAndIsStandingUpright:bool=~
    LocalAccelForward:float=~
    LocalAccelRight:float=~
    LocalVelocityRight:float=~
    SkydiveAimPitch:float=~
    SkydiveAimYaw:float=~
    DeployChuteAnimRate:float=~
    SkydiveDriftAnimRate:float=~
    SkydiveDriftAnimRateCurrent:float=~
    SkydiveFidgetAnimRate:float=~
    SkydiveFidgetAnimRateCurrent:float=~
    SkydiveAdditiveAlpha:float=~
    SkydiveDriftDelay:float=~
    SkydiveDriftAnim:int
    SkydiveDriftAnimMax:int
    LaunchpadAnim:int
    LocalAccelDir:ESkydivingDirection Enum
    DirectionLast:ESkydivingDirection Enum
    PoseAdditive:Anim Sequence Object Reference
   ) {is=title K2Node_BreakStruct}
   n2 = Blendspace Player 'BS_TechOps_Lean_GLIDER'(X=@n1.LocalAccelRight, Y=@n1.LocalAccelForward) {blendSpace=/BRCosmetics/Animation/Game/MainPlayer/Skydive/ParaGlide/TechOps/BS_TechOps_Lean_GLIDER}
   AnimGraphNode_StateMachine_0 = Locomotion() {is=title AnimGraphNode_StateMachine}
   n3 = "Transform (Modify) Bone - Bone: TailRotor_l"!(
    ComponentPose=Local To Component(LocalPose=Apply Additive!(Base=@AnimGraphNode_StateMachine_0, Additive=@n2, AlphaCurveName:Name=ParaGlide_Lean) -> (Pose) {alphaInputType=Curve, alphaBoolBlend {blendOption=Linear}, alphaScaleBiasClamp {bias=1, scale=-1}})
    Alpha=1
    Rotation_Roll:float=Get Rotor_Left_Rot() -> (Rotor_Left_Rot:double)
    Rotation_Pitch:float=0
    Rotation_Yaw:float=0
   ) -> (Pose) {
    rotationMode=BMM_Additive
    rotationSpace=BCS_BoneSpace
    scaleSpace=BCS_BoneSpace
    alphaBoolBlend {blendOption=Linear}
    boneToModify {boneName=TailRotor_l}
   }
   n4 = "Transform (Modify) Bone - Bone: TailRotor_r"!(
    ComponentPose=@n3
    Alpha=1
    Rotation_Roll:float=Get Rotor_Right_Rot() -> (Rotor_Right_Rot:double)
    Rotation_Pitch:float=0
    Rotation_Yaw:float=0
   ) -> (Pose) {
    rotationMode=BMM_Additive
    rotationSpace=BCS_BoneSpace
    scaleSpace=BCS_BoneSpace
    alphaBoolBlend {blendOption=Linear}
    boneToModify {boneName=TailRotor_r}
   }
   n5 = "Transform (Modify) Bone - Bone: Rotor_A"!(
    ComponentPose=@n4
    Alpha=1
    Rotation_Roll:float=float * float(A=Get Rotor_Top_Rot() -> (Rotor_Top_Rot:double), B=-1)
    Rotation_Pitch:float=0
    Rotation_Yaw:float=0
   ) -> (Pose) {
    rotationMode=BMM_Additive
    rotationSpace=BCS_BoneSpace
    scaleSpace=BCS_BoneSpace
    alphaBoolBlend {blendOption=Linear}
    boneToModify {boneName=Rotor_A}
   }
   n6 = "Transform (Modify) Bone - Bone: Rotor_B"!(ComponentPose=@n5, Alpha=1, Rotation_Roll:float=Get Rotor_Top_Rot(), Rotation_Pitch:float=0, Rotation_Yaw:float=0) -> (Pose) {
    rotationMode=BMM_Additive
    rotationSpace=BCS_BoneSpace
    scaleSpace=BCS_BoneSpace
    alphaBoolBlend {blendOption=Linear}
    boneToModify {boneName=Rotor_B}
   }
   Output Pose(Result=Slot 'DefaultSlot'(Source=Component To Local(ComponentPose=@n6))) {layerGroup=None}
  ]}
 Locomotion {
  entry=Closed
  path=AnimGraph.AnimGraphNode_StateMachine_0.Locomotion
  ids=id
  nodes [
   AnimStateNode_0 = Idle_Open() {is=title AnimStateNode}
   AnimStateNode_1 = Open_Into() {is=title AnimStateNode, stateEntered {guid=#1, notifyName=Entered_Open_Into}}
   AnimStateNode_2 = Closed() {is=title AnimStateNode}
   AnimStateTransitionNode_0 = AnimStateTransitionNode() {bAutomaticRuleBasedOnSequencePlayerInState=true, blendMode=Linear, crossfadeDuration=0.088}
   AnimStateTransitionNode_1 = AnimStateTransitionNode() {crossfadeDuration=0}
   AnimStateTransitionNode_2 = AnimStateTransitionNode() {crossfadeDuration=0}
   AnimStateTransitionNode_3 = AnimStateTransitionNode() {crossfadeDuration=0, priorityOrder=2}
  ]
  links [
   {
    dst=Open_Into
    rule=AnimGraph.AnimGraphNode_StateMachine_0.Locomotion.AnimStateTransitionNode_1.Transition
    src=Closed
    via=AnimStateTransitionNode_1
   }
   {
    dst=Closed
    rule=AnimGraph.AnimGraphNode_StateMachine_0.Locomotion.AnimStateTransitionNode_2.Transition
    src=Idle_Open
    via=AnimStateTransitionNode_2
   }
   {
    dst=Closed
    rule=AnimGraph.AnimGraphNode_StateMachine_0.Locomotion.AnimStateTransitionNode_3.Transition
    src=Open_Into
    via=AnimStateTransitionNode_3
   }
   {
    dst=Idle_Open
    rule=AnimGraph.AnimGraphNode_StateMachine_0.Locomotion.AnimStateTransitionNode_0.Transition
    src=Open_Into
    via=AnimStateTransitionNode_0
   }
  ]
 }
 Idle_Open {
  path=AnimGraph.AnimGraphNode_StateMachine_0.Locomotion.AnimStateNode_0.Idle_Open
  nodes [
   n1 = Break Fort Anim Input Skydiving(FortAnimInput_Skydiving=Get SkydivingState()) -> (
    ParaGilderRootOffset=~
    ParaGliderRootRotationOverride=~
    bUseParaGlideRootModifier=~
    bIsSkydivingFromLaunchPad=~
    bIsSkydivingFromBus=~
    bIsInVortex=~
    bIsDiving=~
    bIsParachuteOpen=~
    bIsSkydiving=~
    bIsParachuteLeaning=~
    bIsSkydiveLeaning=~
    bIsLeaning=~
    bIsSkydiveDiveMode=~
    bPlaySkydiveDrift=~
    bSkydiveDriftDelayActive=~
    bSkydiveDriftAnimAllowed=~
    bIsGliderRight=~
    bIsGliderCenter=~
    bIsGliderLeft=~
    bIsGliderForward=~
    bIsGliderBack=~
    bHasDeployedAndIsStandingUpright=~
    LocalAccelForward=~
    LocalAccelRight=~
    LocalVelocityRight=~
    SkydiveAimPitch=~
    SkydiveAimYaw=~
    DeployChuteAnimRate=~
    SkydiveDriftAnimRate=~
    SkydiveDriftAnimRateCurrent=~
    SkydiveFidgetAnimRate=~
    SkydiveFidgetAnimRateCurrent=~
    SkydiveAdditiveAlpha=~
    SkydiveDriftDelay=~
   )
   n2 = Blendspace Player 'BS_TechOps_Turn_GLIDER'(X=@n1.SkydiveAimYaw, Y=@n1.SkydiveAimPitch) {blendSpace=/BRCosmetics/Animation/Game/MainPlayer/Skydive/ParaGlide/TechOps/BS_TechOps_Turn_GLIDER}
   n3 = Sequence Player 'TechOps_Base_Pose_GLIDER'() {sequence=/BRCosmetics/Animation/Game/MainPlayer/Skydive/ParaGlide/TechOps/TechOps_Base_Pose_GLIDER}
   n4 = Get OwnerGlider() -> (OwnerGlider:Fort Player Parachute Object Reference)
   Output Animation Pose(Result=Blend Poses by bool(
     BlendPose_0=Sequence Player 'TechOps_Glide_FrontEnd_GLIDER'() {sequence=/BRCosmetics/Animation/Game/MainPlayer/Skydive/ParaGlide/TechOps/TechOps_Glide_FrontEnd_GLIDER}
     BlendPose_1=Apply Additive!(Base=@n3, Additive=@n2, Alpha:float) -> (Pose) {alphaBoolBlend {blendOption=Linear}}
     bActiveValue=Get IsCosmeticPreview!(self:Fort Player Parachute Object Reference=@n4) -> (IsCosmeticPreview:bool)
    ) {blendType=Linear})
  ]
 }
 Open_Into {
  path=AnimGraph.AnimGraphNode_StateMachine_0.Locomotion.AnimStateNode_1.Open_Into
  nodes [
   n1 = Break Fort Anim Input Skydiving(FortAnimInput_Skydiving=Get SkydivingState()) -> (
    ParaGilderRootOffset=~
    ParaGliderRootRotationOverride=~
    bUseParaGlideRootModifier=~
    bIsSkydivingFromLaunchPad=~
    bIsSkydivingFromBus=~
    bIsInVortex=~
    bIsDiving=~
    bIsParachuteOpen=~
    bIsSkydiving=~
    bIsParachuteLeaning=~
    bIsSkydiveLeaning=~
    bIsLeaning=~
    bIsSkydiveDiveMode=~
    bPlaySkydiveDrift=~
    bSkydiveDriftDelayActive=~
    bSkydiveDriftAnimAllowed=~
    bIsGliderRight=~
    bIsGliderCenter=~
    bIsGliderLeft=~
    bIsGliderForward=~
    bIsGliderBack=~
    bHasDeployedAndIsStandingUpright=~
    LocalAccelForward=~
    LocalAccelRight=~
    LocalVelocityRight=~
    SkydiveAimPitch=~
    SkydiveAimYaw=~
    DeployChuteAnimRate=~
    SkydiveDriftAnimRate=~
    SkydiveDriftAnimRateCurrent=~
    SkydiveFidgetAnimRate=~
    SkydiveFidgetAnimRateCurrent=~
    SkydiveAdditiveAlpha=~
    SkydiveDriftDelay=~
   )
   n2 = Get OwnerGlider()
   Output Animation Pose(Result=Blend Poses by bool(
     BlendPose_0=Sequence Player 'TechOps_Into_FrontEnd_GLIDER'!(bLoopAnimation:bool=false) -> (Pose) {sequence=/BRCosmetics/Animation/Game/MainPlayer/Skydive/ParaGlide/TechOps/TechOps_Into_FrontEnd_GLIDER}
     BlendPose_1=Blendspace Player 'BS_TechOps_Into_GLIDER'!(X=@n1.SkydiveAimYaw, Y=@n1.SkydiveAimPitch, bLoop:bool=false) -> (Pose) {blendSpace=/BRCosmetics/Animation/Game/MainPlayer/Skydive/ParaGlide/TechOps/BS_TechOps_Into_GLIDER}
     bActiveValue=Get IsCosmeticPreview(self=@n2)
    ) {blendType=Linear})
  ]
 }
 Closed {
  path=AnimGraph.AnimGraphNode_StateMachine_0.Locomotion.AnimStateNode_2.Closed
  nodes [
   n1 = Get OwnerGlider()
   Output Animation Pose(Result=Blend Poses by bool(
     BlendPose_0=Sequence Player 'TechOps_Closed_FrontEnd_GLIDER'() {sequence=/BRCosmetics/Animation/Game/MainPlayer/Skydive/ParaGlide/TechOps/TechOps_Closed_FrontEnd_GLIDER}
     BlendPose_1=Sequence Player 'TechOps_Closed_GLIDER'() {sequence=/BRCosmetics/Animation/Game/MainPlayer/Skydive/ParaGlide/TechOps/TechOps_Closed_GLIDER}
     bActiveValue=Get IsCosmeticPreview(self=@n1)
    ) {blendType=Linear})
  ]
 }
 Transition {path=AnimGraph.AnimGraphNode_StateMachine_0.Locomotion.AnimStateTransitionNode_0.Transition, nodes [Result()]}
 Transition {
  path=AnimGraph.AnimGraphNode_StateMachine_0.Locomotion.AnimStateTransitionNode_1.Transition
  nodes [
   n1 = Get SkydivingState()
   Result(bCanEnterTransition=OR Boolean(
     A=Break Fort Anim Input Skydiving(FortAnimInput_Skydiving=@n1) -> (
      ParaGilderRootOffset=~
      ParaGliderRootRotationOverride=~
      bUseParaGlideRootModifier=~
      bIsSkydivingFromLaunchPad=~
      bIsSkydivingFromBus=~
      bIsInVortex=~
      bIsDiving=~
      bIsParachuteOpen=~
      bIsSkydiving=~
      bIsParachuteLeaning=~
      bIsSkydiveLeaning=~
      bIsLeaning=~
      bIsSkydiveDiveMode=~
      bPlaySkydiveDrift=~
      bSkydiveDriftDelayActive=~
      bSkydiveDriftAnimAllowed=~
      bIsGliderRight=~
      bIsGliderCenter=~
      bIsGliderLeft=~
      bIsGliderForward=~
      bIsGliderBack=~
      bHasDeployedAndIsStandingUpright=~
      LocalAccelForward=~
      LocalAccelRight=~
      LocalVelocityRight=~
      SkydiveAimPitch=~
      SkydiveAimYaw=~
      DeployChuteAnimRate=~
      SkydiveDriftAnimRate=~
      SkydiveDriftAnimRateCurrent=~
      SkydiveFidgetAnimRate=~
      SkydiveFidgetAnimRateCurrent=~
      SkydiveAdditiveAlpha=~
      SkydiveDriftDelay=~
     ).bIsParachuteOpen
     B=Get bForceOpen() -> (bForceOpen:bool)
    ))
  ]
 }
 Transition {
  path=AnimGraph.AnimGraphNode_StateMachine_0.Locomotion.AnimStateTransitionNode_2.Transition
  nodes [
   n1 = Get bForceOpen()
   n2 = Break Fort Anim Input Skydiving(FortAnimInput_Skydiving=Get SkydivingState()) -> (
    ParaGilderRootOffset=~
    ParaGliderRootRotationOverride=~
    bUseParaGlideRootModifier=~
    bIsSkydivingFromLaunchPad=~
    bIsSkydivingFromBus=~
    bIsInVortex=~
    bIsDiving=~
    bIsParachuteOpen=~
    bIsSkydiving=~
    bIsParachuteLeaning=~
    bIsSkydiveLeaning=~
    bIsLeaning=~
    bIsSkydiveDiveMode=~
    bPlaySkydiveDrift=~
    bSkydiveDriftDelayActive=~
    bSkydiveDriftAnimAllowed=~
    bIsGliderRight=~
    bIsGliderCenter=~
    bIsGliderLeft=~
    bIsGliderForward=~
    bIsGliderBack=~
    bHasDeployedAndIsStandingUpright=~
    LocalAccelForward=~
    LocalAccelRight=~
    LocalVelocityRight=~
    SkydiveAimPitch=~
    SkydiveAimYaw=~
    DeployChuteAnimRate=~
    SkydiveDriftAnimRate=~
    SkydiveDriftAnimRateCurrent=~
    SkydiveFidgetAnimRate=~
    SkydiveFidgetAnimRateCurrent=~
    SkydiveAdditiveAlpha=~
    SkydiveDriftDelay=~
   )
   Result(bCanEnterTransition=AND Boolean(A=NOT Boolean(A=@n2.bIsParachuteOpen), B=NOT Boolean(A=@n1)))
  ]
 }
 Transition {
  path=AnimGraph.AnimGraphNode_StateMachine_0.Locomotion.AnimStateTransitionNode_3.Transition
  nodes [
   n1 = Get bForceOpen()
   n2 = Break Fort Anim Input Skydiving(FortAnimInput_Skydiving=Get SkydivingState()) -> (
    ParaGilderRootOffset=~
    ParaGliderRootRotationOverride=~
    bUseParaGlideRootModifier=~
    bIsSkydivingFromLaunchPad=~
    bIsSkydivingFromBus=~
    bIsInVortex=~
    bIsDiving=~
    bIsParachuteOpen=~
    bIsSkydiving=~
    bIsParachuteLeaning=~
    bIsSkydiveLeaning=~
    bIsLeaning=~
    bIsSkydiveDiveMode=~
    bPlaySkydiveDrift=~
    bSkydiveDriftDelayActive=~
    bSkydiveDriftAnimAllowed=~
    bIsGliderRight=~
    bIsGliderCenter=~
    bIsGliderLeft=~
    bIsGliderForward=~
    bIsGliderBack=~
    bHasDeployedAndIsStandingUpright=~
    LocalAccelForward=~
    LocalAccelRight=~
    LocalVelocityRight=~
    SkydiveAimPitch=~
    SkydiveAimYaw=~
    DeployChuteAnimRate=~
    SkydiveDriftAnimRate=~
    SkydiveDriftAnimRateCurrent=~
    SkydiveFidgetAnimRate=~
    SkydiveFidgetAnimRateCurrent=~
    SkydiveAdditiveAlpha=~
    SkydiveDriftDelay=~
   )
   Result(bCanEnterTransition=AND Boolean(A=NOT Boolean(A=@n2.bIsParachuteOpen), B=NOT Boolean(A=@n1)))
  ]
 }
 EventGraph {nodes [
   Event BlueprintUpdateAnimation()
   n1 = Break Fort Anim Input Skydiving(FortAnimInput_Skydiving=Get SkydivingState()) -> (
    ParaGilderRootOffset=~
    ParaGliderRootRotationOverride=~
    bUseParaGlideRootModifier=~
    bIsSkydivingFromLaunchPad=~
    bIsSkydivingFromBus=~
    bIsInVortex=~
    bIsDiving=~
    bIsParachuteOpen=~
    bIsSkydiving=~
    bIsParachuteLeaning=~
    bIsSkydiveLeaning=~
    bIsLeaning=~
    bIsSkydiveDiveMode=~
    bPlaySkydiveDrift=~
    bSkydiveDriftDelayActive=~
    bSkydiveDriftAnimAllowed=~
    bIsGliderRight=~
    bIsGliderCenter=~
    bIsGliderLeft=~
    bIsGliderForward=~
    bIsGliderBack=~
    bHasDeployedAndIsStandingUpright=~
    LocalAccelForward=~
    LocalAccelRight=~
    LocalVelocityRight=~
    SkydiveAimPitch=~
    SkydiveAimYaw=~
    DeployChuteAnimRate=~
    SkydiveDriftAnimRate=~
    SkydiveDriftAnimRateCurrent=~
    SkydiveFidgetAnimRate=~
    SkydiveFidgetAnimRateCurrent=~
    SkydiveAdditiveAlpha=~
    SkydiveDriftDelay=~
   )
   > Set LocalAccelForward(LocalAccelForward:double=@n1.LocalAccelForward) -> (Output_Get:double)
   > Set SkydiveAimYaw(SkydiveAimYaw:double=@n1.SkydiveAimYaw) -> (Output_Get:double)
   > Gate!(Enter:Exec, Open:Exec=@n14.then, Close:Exec=@n16.then_0, Toggle:Exec, bStartClosed:bool=true) -> (Exit:Exec) {is=title K2Node_MacroInstance}
   > Set WarmUpScaleTop(WarmUpScaleTop:double=FInterpTo(Current=Get WarmUpScaleTop() -> (WarmUpScaleTop:double), Target=1, DeltaTime=Get DeltaTime() -> (DeltaTime:float), InterpSpeed=2)) -> (Output_Get:double)
   > Set ForwardAccelScale(ForwardAccelScale:double=FInterpTo(
     Current=Get ForwardAccelScale() -> (ForwardAccelScale:double)
     Target=Get LocalAccelForward() -> (LocalAccelForward:double)
     DeltaTime=Get DeltaTime()
     InterpSpeed=1
    )) -> (Output_Get:double)
   n2 = Get ForwardAccelScale()
   n3 = Get DeltaTime()
   n4 = float * float(
    A=float * float(A=float * float(A=@n3, B=1000), B=Get WarmUpScaleTop())
    B=float + float(A=float / float(A=@n2, B=2048) {is=title K2Node_CallFunction}, B=1.5) {is=title K2Node_CommutativeAssociativeBinaryOperator}
   )
   n5 = Get Rotor_Top_Rot()
   > Set Rotor_Top_Rot(Rotor_Top_Rot:double="% (Float)"!(A:double=float + float(A=@n5, B=@n4), B:double=360) -> (ReturnValue:double)) -> (Output_Get:double)
   > Gate(Open=@n15.then, Close=@n16.then_1, bStartClosed=true)
   > Set SkydiveYaw_Interp(SkydiveYaw_Interp:double=FInterpTo(
     Current=Get SkydiveYaw_Interp() -> (SkydiveYaw_Interp:double)
     Target=Get SkydiveAimYaw() -> (SkydiveAimYaw:double)
     DeltaTime=Get DeltaTime()
     InterpSpeed=5
    )) -> (Output_Get:double)
   > Set WarmUpScaleTail(WarmUpScaleTail:double=FInterpTo(Current=Get WarmUpScaleTail() -> (WarmUpScaleTail:double), Target=1, DeltaTime=Get DeltaTime(), InterpSpeed=5)) -> (Output_Get:double)
   n6 = Get SkydiveYaw_Interp()
   > Set TurnLeftScale(TurnLeftScale:double=SelectFloat(
     A=1
     B=NormalizeToRange!(Value:double=@n6, RangeMin:double=-45, RangeMax:double=0) -> (ReturnValue:double)
     bPickA="float >= float"(A=@n6, B=0) {is=title K2Node_CallFunction}
    )) -> (Output_Get:double)
   n7 = NormalizeToRange(Value=@n6, RangeMin=0, RangeMax=45)
   > Set TurnRightScale(TurnRightScale:double=SelectFloat(
     A=1
     B=float - float(B=@n7) {is=title K2Node_CallFunction}
     bPickA="float <= float"!(A:double=@n6, B:double=0) -> (ReturnValue:bool) {is=title K2Node_CallFunction}
    )) -> (Output_Get:double)
   n8 = Get DeltaTime()
   n9 = float * float(A=float * float(A=float * float(A=@n8, B=-2500), B=Get WarmUpScaleTail()), B=Get TurnLeftScale() -> (TurnLeftScale:double))
   n10 = Get Rotor_Left_Rot()
   > Set Rotor_Left_Rot(Rotor_Left_Rot:double="% (Float)"(A=float + float(A=@n10, B=@n9), B=360)) -> (Output_Get:double)
   n11 = Get DeltaTime()
   n12 = float * float(A=float * float(A=float * float(A=@n11, B=2500), B=Get WarmUpScaleTail()), B=Get TurnRightScale() -> (TurnRightScale:double))
   n13 = Get Rotor_Right_Rot()
   > Set Rotor_Right_Rot(Rotor_Right_Rot:double="% (Float)"(A=float + float(A=@n13, B=@n12), B=360)) -> (Output_Get:double)
   AnimNotify_TurnONEngines() {is=title K2Node_Event}
   > SetGliderFullyDeployed!(execute, self:Fort Glider Anim Instance Object Reference, bIsFullyDeployed:bool=true) -> (then)
   n14 = AnimNotify_StartMainRotors() {is=title K2Node_Event}
   n15 = AnimNotify_StartTailRotors() {is=title K2Node_Event}
   AnimNotify_Entered_Open_Into() {is=title K2Node_Event}
   > Set ForwardAccelScale(ForwardAccelScale=0)
   > Set Rotor_Top_Rot(Rotor_Top_Rot=0)
   > Set SkydiveYaw_Interp(SkydiveYaw_Interp=0)
   > Set TurnLeftScale(TurnLeftScale=0)
   > Set TurnRightScale(TurnRightScale=0)
   > Set Rotor_Left_Rot(Rotor_Left_Rot=0)
   > Set Rotor_Right_Rot(Rotor_Right_Rot=0)
   > Set WarmUpScaleTop(WarmUpScaleTop=0)
   > Set WarmUpScaleTail(WarmUpScaleTail=0)
   > n16 = Sequence()
   Break Fort Anim Input Skydiving(FortAnimInput_Skydiving=Get SkydivingState()) -> (
    ParaGilderRootOffset=~
    ParaGliderRootRotationOverride=~
    bUseParaGlideRootModifier=~
    bIsSkydivingFromLaunchPad=~
    bIsSkydivingFromBus=~
    bIsInVortex=~
    bIsDiving=~
    bIsParachuteOpen=~
    bIsSkydiving=~
    bIsParachuteLeaning=~
    bIsSkydiveLeaning=~
    bIsLeaning=~
    bIsSkydiveDiveMode=~
    bPlaySkydiveDrift=~
    bSkydiveDriftDelayActive=~
    bSkydiveDriftAnimAllowed=~
    bIsGliderRight=~
    bIsGliderCenter=~
    bIsGliderLeft=~
    bIsGliderForward=~
    bIsGliderBack=~
    bHasDeployedAndIsStandingUpright=~
    LocalAccelForward=~
    LocalAccelRight=~
    LocalVelocityRight=~
    SkydiveAimPitch=~
    SkydiveAimYaw=~
    DeployChuteAnimRate=~
    SkydiveDriftAnimRate=~
    SkydiveDriftAnimRateCurrent=~
    SkydiveFidgetAnimRate=~
    SkydiveFidgetAnimRateCurrent=~
    SkydiveAdditiveAlpha=~
    SkydiveDriftDelay=~
   )
  ]}
]
variables [
 ForwardAccelScale {type=number}
 LocalAccelForward {type=number}
 Rotor_Left_Rot {type=number}
 Rotor_Right_Rot {type=number}
 Rotor_Top_Rot {type=number}
 SkydiveAimYaw {type=number}
 SkydiveYaw_Interp {type=number}
 TurnLeftScale {type=number}
 TurnRightScale {type=number}
 WarmUpScaleTail {type=number}
 WarmUpScaleTop {type=number}
]
