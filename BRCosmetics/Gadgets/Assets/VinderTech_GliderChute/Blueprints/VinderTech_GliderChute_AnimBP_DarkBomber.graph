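# Animation Blueprint for the Dark Bomber VinderTech glider, targeting the GliderChute skeleton.
# When the TurnONEngine anim notify fires, it marks the glider as fully deployed through the Fort Glider anim instance.
# The update event and pawn owner lookup are present but unconnected, and it declares no variables.
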AnimBlueprint /BRCosmetics/Gadgets/Assets/VinderTech_GliderChute/Blueprints/VinderTech_GliderChute_AnimBP_DarkBomber
target_skeleton=/Game/Gadgets/Assets/VinderTech_GliderChute/Meshes/SK_M_VinderTech_GliderChute_Skeleton
graphs [EventGraph {nodes [
   AnimNotify_TurnONEngine() {is=title K2Node_Event}
   > SetGliderFullyDeployed!(execute, self:Fort Glider Anim Instance Object Reference, bIsFullyDeployed:bool=true) -> (then)
   TryGetPawnOwner()
   Event BlueprintUpdateAnimation()
  ]}]
variables []
