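# Animation Blueprint for the StreetRacerBlack variant of the VinderTech glider chute, targeting the shared glider chute skeleton.
# Its event graph handles the DynamicsOK anim notify by broadcasting the Dynamics OK dispatcher, then calling SetGliderFullyDeployed with true on itself as a Fort Glider Anim Instance.
# The animation timeline therefore decides when the glider counts as fully deployed.
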AnimBlueprint /BRCosmetics/Gadgets/Assets/VinderTech_GliderChute/Blueprints/VinderTech_GliderChute_AnimBP_StreetRacerBlack
target_skeleton=/Game/Gadgets/Assets/VinderTech_GliderChute/Meshes/SK_M_VinderTech_GliderChute_Skeleton
graphs [EventGraph {nodes [
   AnimNotify_DynamicsOK() {is=title K2Node_Event}
   > Call Dynamics OK!(execute:Exec, self:Vinder Tech Glider Chute Anim BP Street Racer Black Object Reference) -> (then:Exec) {is=title K2Node_CallDelegate}
   > SetGliderFullyDeployed!(execute, self:Fort Glider Anim Instance Object Reference, bIsFullyDeployed:bool=true) -> (then)
  ]}]
variables []
