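# Master material for the VinderTech glider chute, using Default Lit material attributes on skeletal meshes with mesh deformer support.
# A glider base shader function builds the attributes, which feed the shared weapon skin FX function.
# That result passes through the TechArt accessories override function.
# The graph holds no shading logic of its own, only this chain of three material function calls.
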Material /BRCosmetics/Gadgets/Assets/VinderTech_GliderChute/Materials/M_FN_Glider_MASTER
graph {
 properties {
  bUseMaterialAttributes=true
  bUsedWithMeshDeformer=true
  bUsedWithSkeletalMesh=true
  pixelDepthOffsetMode=PDOM_Legacy
  usedShadingModels=Default Lit
 }
 nodes [MaterialFunctionCall(In=MaterialFunctionCall!(
    Pre-Skinned World Normal
    Gmap
    Customization Mask
    IsCharacter
    IsStatsB
    Material Attributes=MaterialFunctionCall!(UV 1 coordinates) {materialFunction=/Game/Gadgets/Assets/VinderTech_GliderChute/Materials/MF_GliderBaseShader}.Result Material Attributes
   ) {materialFunction=/Game/Weapons/WeaponSkins/MaterialFunction/Weapon_SkinFXFunction}.Result) {materialFunction=&MF_TechArtOverrides_accessories}]
}
