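# Subsurface material for the Gumdrop variant of the VinderTech glider chute, giving masked areas a glossy, sparkly candy look.
# The candy layer builds fresnel-tinted color from two bump-offset noise samples for fake depth, adds tiled detail and sparkle normals, and drives subsurface color, with cheaper quality-switch paths.
# The E texture's green channel blends this layer over the textured base, and its blue channel masks a tinted emissive.
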Material /BRCosmetics/Gadgets/Assets/VinderTech_GliderChute/Materials/M_VinderTech_GliderChute_Gumdrop
graph {
 properties {
  bUseMaterialAttributes=true
  bUsedWithMeshDeformer=true
  bUsedWithSkeletalMesh=true
  pixelDepthOffsetMode=PDOM_Legacy
  shadingModel=MSM_Subsurface
  translucencyLightingMode=TLM_Surface
  usedShadingModels=Subsurface
 }
 nodes [
  n1 = TextureSample() {texture=/BRCosmetics/Gadgets/Assets/VinderTech_GliderChute/Textures/T_VinderTech_GliderChute_Gumdrop_D}
  n2 = TextureSample() {samplerType=SAMPLERTYPE_Masks, texture=/BRCosmetics/Gadgets/Assets/VinderTech_GliderChute/Textures/T_VinderTech_GliderChute_Gumdrop_M}
  n3 = Constant() {r=1}
  n4 = TextureSample() {samplerType=SAMPLERTYPE_Normal, texture=/BRCosmetics/Gadgets/Assets/VinderTech_GliderChute/Textures/T_VinderTech_GliderChute_Gumdrop_N}
  n5 = Constant()
  n6 = TextureSample() {texture=/Game/Gadgets/Assets/VinderTech_GliderChute/Textures/VinderTech_GliderChute_D}
  n7 = VertexNormalWS()
  n8 = Fresnel(ExponentIn=ScalarParameter() {defaultValue=2.0232, parameterName=Fresnel}, Normal=@n7) {baseReflectFraction=0}
  n9 = LinearInterpolate(
   A=VectorParameter() {parameterName=Color_8, defaultValue=<rgba 0.125 0.0124 0.0124 1>}.RGB
   B=VectorParameter() {parameterName=Color_9, defaultValue=<rgba 1 0.1198 0.1286 1>}.RGB
   Alpha=@n8
  )
  n10 = LinearInterpolate(
   A=VectorParameter() {parameterName=Color_6, defaultValue=<rgba 0.276 0.0201 0.0201 1>}.RGB
   B=VectorParameter() {parameterName=Color_7, defaultValue=<rgba 0.026 0 0 1>}.RGB
   Alpha=@n8
  )
  n11 = BumpOffset(
   Coordinate=Multiply(A=TextureCoordinate(), B=ScalarParameter() {defaultValue=0.1252, parameterName=Tiling2})
   Height=LinearInterpolate(
    A=ScalarParameter() {defaultValue=-4.433, parameterName=Offset}
    B=ScalarParameter() {defaultValue=-8.0362, parameterName=Offset_1}
    Alpha=@n8
   )
  )
  n12 = BumpOffset(
   Coordinate=Multiply(A=TextureCoordinate(), B=ScalarParameter() {defaultValue=0.2027, parameterName=Tiling2_1})
   Height=LinearInterpolate(
    A=ScalarParameter() {defaultValue=-8.689, parameterName=Offset_2}
    B=ScalarParameter() {defaultValue=-20.152, parameterName=Offset_3}
    Alpha=@n8
   )
  )
  n13 = Clamp(None=Multiply(
    A=TextureSample(UVs=@n11) {samplerType=SAMPLERTYPE_LinearColor, texture=/BRCosmetics/Gadgets/Assets/VinderTech_GliderChute/Textures/T_Gumdrop_M}.R
    B=TextureSample(UVs=@n12) {samplerType=SAMPLERTYPE_LinearColor, texture=&T_Gumdrop_M}.R
   ))
  n14 = ScalarParameter() {defaultValue=0.0941, parameterName=Fresnel_1}
  n15 = ComponentMask(None=LinearInterpolate(
    A=VectorParameter() {parameterName=Color_3, defaultValue=<rgba 1 0.0394 0 1>}.RGB
    B=LinearInterpolate(A=@n9, B=@n10, Alpha=@n13)
    Alpha=Fresnel(ExponentIn=@n14, Normal=@n7) {baseReflectFraction=0}
   )) {b=true, g=true, r=true}
  n16 = Multiply(
   A=ShadingPathSwitch(
    Default=ScalarParameter() {defaultValue=0.085, parameterName=Diffuse}
    Mobile=ScalarParameter() {defaultValue=0.8, parameterName=Diffuse_Switch}
   )
   B=@n15
  )
  n17 = Multiply(A=ScalarParameter() {defaultValue=0.8, parameterName=MobileDiffuse}, B=@n15)
  n18 = ScalarParameter() {defaultValue=1, parameterName=Transition_1}
  n19 = ScalarParameter() {defaultValue=1, parameterName=Transition_2}
  n20 = TextureSample() {texture=&T_VinderTech_GliderChute_Gumdrop_D}
  n21 = ScalarParameter() {defaultValue=1, parameterName=Translucent}
  n22 = ScalarParameter() {defaultValue=1, parameterName=Translucent_1}
  n23 = TextureCoordinate()
  n24 = ScalarParameter() {defaultValue=2.7962, parameterName=OffsetT}
  n25 = TextureSample(UVs=Multiply(A=BumpOffset(Coordinate=@n23, Height=@n24), B=ScalarParameter() {defaultValue=1.2997, parameterName=TransparentTiling})) {samplerType=SAMPLERTYPE_LinearColor, texture=&T_Gumdrop_M}
  n26 = Multiply(A=Clamp(None=LinearInterpolate(A=@n18, B=@n19, Alpha=@n20.R)), B=Clamp(None=LinearInterpolate(A=@n21, B=@n22, Alpha=@n25.R)))
  n27 = LinearInterpolate(A=@n6.RGB, B=QualitySwitch(Default=@n16, Low=@n17, High=@n16, Medium=@n17, Epic=@n16), Alpha=@n26)
  n28 = LinearInterpolate(
   A=TextureSample() {samplerType=SAMPLERTYPE_Masks, texture=/Game/Gadgets/Assets/VinderTech_GliderChute/Textures/VinderTech_GliderChute_S}.G
   B=ScalarParameter() {parameterName=Metallic}
   Alpha=@n26
  )
  n29 = Multiply(A=TextureCoordinate(), B=ScalarParameter() {defaultValue=3.9549, parameterName=Tiling_1})
  n30 = TextureSample(UVs=@n29) {samplerType=SAMPLERTYPE_Masks, texture=/Game/Gadgets/Assets/VinderTech_GliderChute/Textures/T_Gumdrop_Sparkle_M}
  n31 = LinearInterpolate(A=ScalarParameter() {defaultValue=0.4211, parameterName=Spec}, B=ScalarParameter() {defaultValue=1, parameterName=Spec_1}, Alpha=@n30.R)
  n32 = Clamp(None=LinearInterpolate(A=ScalarParameter() {defaultValue=1.3289, parameterName=R_1}, B=ScalarParameter() {defaultValue=0.283, parameterName=R}, Alpha=@n30.R))
  n33 = Add(
   A=Multiply(A=OneMinus(None=@n26), B=ScalarParameter() {defaultValue=1, parameterName=SS_Solid})
   B=ScalarParameter() {defaultValue=0.2621, parameterName=SS}
  )
  n34 = TextureSample() {samplerType=SAMPLERTYPE_Normal, texture=&T_VinderTech_GliderChute_Gumdrop_N}
  n35 = Multiply(A=TextureCoordinate(), B=ScalarParameter() {defaultValue=1.4132, parameterName=Tiling})
  n36 = MaterialFunctionCall(
   BaseNormal=MaterialFunctionCall(
    BaseNormal=@n34.RGB
    AdditionalNormal=TextureSample(UVs=@n35) {samplerType=SAMPLERTYPE_Normal, texture=/BRCosmetics/Gadgets/Assets/VinderTech_GliderChute/Textures/T_Gumdrop_N}.RGB
   ) {materialFunction=&BlendAngleCorrectedNormals}.Result
   AdditionalNormal=MaterialFunctionCall(
    Normal=TextureSample(UVs=@n29) {samplerType=SAMPLERTYPE_Normal, texture=/BRCosmetics/Gadgets/Assets/VinderTech_GliderChute/Textures/T_Gumdrop_Sparkle}.RGB
    Flatness=ScalarParameter() {defaultValue=-1.0998, parameterName=Sparkelintensity}
   ) {materialFunction=&FlattenNormal}.Result
  ) {materialFunction=&BlendAngleCorrectedNormals}
  n37 = LinearInterpolate(A=@n6.RGB, B=Multiply(A=@n15, B=VectorParameter() {parameterName=SS, defaultValue=<rgba 1 1 1 1>}.RGB), Alpha=@n26)
  n38 = VectorParameter() {parameterName=SSS, defaultValue=<rgba 0.33 0 0.0056 1>}
  n39 = QualitySwitch(Default=@n37, Low=@n38.RGB, High=@n37, Medium=@n38.RGB)
  n40 = TextureSample() {samplerType=SAMPLERTYPE_Masks, texture=/BRCosmetics/Gadgets/Assets/VinderTech_GliderChute/Textures/T_VinderTech_GliderChute_Gumdrop_E}
  MaterialFunctionCall!(
   Material=MaterialFunctionCall!(
    Base Material=MakeMaterialAttributes!(
     BaseColor=@n1.RGB
     Metallic=@n2.G
     Specular=@n2.G
     Roughness=@n2.G
     Anisotropy
     EmissiveColor
     Opacity=@n3
     OpacityMask
     Normal=@n4.RGB
     Tangent
     WorldPositionOffset
     SubsurfaceColor=@n5
     ClearCoat
     ClearCoatRoughness
     AmbientOcclusion
     Refraction
     CustomizedUVs_0
     CustomizedUVs_1
     CustomizedUVs_2
     CustomizedUVs_3
     CustomizedUVs_4
     CustomizedUVs_5
     CustomizedUVs_6
     CustomizedUVs_7
     PixelDepthOffset
     ShadingModel
     Displacement
    )
    Top Material=MakeMaterialAttributes(BaseColor=@n27, Metallic=@n28, Specular=@n31, Roughness=@n32, Opacity=@n33, Normal=@n36.Result, SubsurfaceColor=@n39)
    Alpha=@n40.G
   ) {materialFunction=/Engine/Functions/MaterialLayerFunctions/MatLayerBlend_Standard}.Blended Material
   Emissive=Multiply(A=VectorParameter() {parameterName=Emissive, defaultValue=<rgba 0.6042 4 4 1>}.RGB, B=@n40.B)
  ) {materialFunction=/Engine/Functions/MaterialLayerFunctions/MatLayerBlend_Emissive}
 ]
}
