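# Material for the StormEye VinderTech glider chute, blending standard textured surface and emissive lights with an animated storm through a mask.
# The storm maps pre-skinned XY position through a sky twist function, then composites two panning swirl-mask cloud layers over a vertical gradient background.
# Three HDR-texture lightning channels flicker via a timeline texture, two jumping in UV on stepped time, lighting the clouds.
# The blend finishes in the accessories tech-art override function.
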Material /BRCosmetics/Gadgets/Assets/VinderTech_GliderChute/Materials/M_VinderTech_GliderChute_StormEye
graph {
 properties {
  bUseMaterialAttributes=true
  bUsedWithMeshDeformer=true
  bUsedWithSkeletalMesh=true
  pixelDepthOffsetMode=PDOM_Legacy
  usedShadingModels=Default Lit
 }
 nodes [
  n1 = Constant2Vector() {r=1}
  n2 = VertexInterpolator(VS=PreSkinnedPosition())
  n3 = Add(
   A=Divide(A=ComponentMask(None=@n2.PS) {g=true, r=true}, B=ScalarParameter() {defaultValue=135.4167, parameterName=Size})
   B=Constant2Vector() {g=0.5, r=0.5}
  )
  n4 = Multiply(A=AppendVector(A=ScalarParameter() {defaultValue=0.6476, parameterName=TilingX}, B=Constant() {r=1}), B=@n3)
  n5 = Add(
   A=AppendVector(A=ScalarParameter() {defaultValue=-0.3, parameterName=TwistOffsetX}, B=ScalarParameter() {defaultValue=-1.1705, parameterName=TwistOffsetY})
   B=Multiply(A=ScalarParameter() {defaultValue=1.1704, parameterName=TilingTwirl}, B=@n4)
  )
  n6 = ScalarParameter() {defaultValue=1.7433, parameterName=SineAnimSpeed}
  n7 = ScalarParameter() {defaultValue=4, parameterName=GridNoiseTiling}
  n8 = AppendVector(A=@n7, B=@n7)
  n9 = ScalarParameter() {defaultValue=0.6212, parameterName=Rotation}
  n10 = ScalarParameter() {defaultValue=-0.0571, parameterName=SineAnimSpeed_1}
  n11 = ComponentMask(None=Add(
    A=MaterialFunctionCall!(
     SineAnimOffset
     Flow Direction=@n1
     Time offset
     EdgeMaskSize
     UVS * size=@n5
     SineAnimSpeed=@n6
     Grid Noise Tile Rep=@n8
     Rotation amount modulator=@n9
     un frac Time=@n10
    ) {materialFunction=/Game/Packages/Fortress_Sky/Fortnite_SkyTwistFunction}.Rotated UVs
    B=@n4
   )) {g=true, r=true}
  n12 = ComponentMask(None=CameraVectorWS()) {g=true, r=true}
  Add(A=@n11, B=Multiply(A=@n12, B=ScalarParameter() {defaultValue=-0.6476, parameterName=BumpOffset}))
  Add(A=Multiply(A=@n12, B=ScalarParameter() {defaultValue=0.1143, parameterName=BumpOffset_2}), B=@n11)
  n13 = ComponentMask(None=@n3) {g=true}
  n14 = VectorParameter() {parameterName=CloudColor_1, defaultValue=<rgba 0.0101 0.0032 0.0156 1>}
  n15 = VectorParameter() {parameterName=CloudColor_2, defaultValue=<rgba 0.031 0.009 0.0573 1>}
  n16 = Multiply(A=@n11, B=ScalarParameter() {defaultValue=0.7591, parameterName=ScaleCloudLayer2})
  n17 = Time()
  n18 = Panner(Coordinate=@n16, Time=Multiply(A=@n17, B=ScalarParameter() {defaultValue=0.0857, parameterName=SpeedPanCloud1_1})) {speedX=-0.33, speedY=-0.043}
  n19 = Panner(Coordinate=@n16, Time=Multiply(A=@n17, B=ScalarParameter() {defaultValue=0.0376, parameterName=SpeedPanCloud2_1})) {speedX=-0.44, speedY=0.1}
  n20 = Divide(A=Add(
    A=TextureSample(UVs=@n18) {samplerType=SAMPLERTYPE_LinearGrayscale, texture=/Game/Global/Textures/Default/Flowmaps/T_StormSwirlMask}.R
    B=TextureSample(UVs=@n19) {samplerType=SAMPLERTYPE_LinearGrayscale, texture=&T_StormSwirlMask}.R
   ))
  n21 = LinearInterpolate(A=@n14.RGB, B=@n15.RGB, Alpha=Clamp(None=@n20))
  n22 = Multiply(A=@n20, B=ScalarParameter() {defaultValue=-0.2095, parameterName=CloudmaskInfluence_1})
  n23 = ComponentMask(None=@n3) {g=true}
  n24 = Clamp(None=LinearInterpolate(
    A=ScalarParameter() {defaultValue=2.269, parameterName=ContrastClouds_2}
    B=ScalarParameter() {defaultValue=-0.9425, parameterName=ContrastClouds_3}
    Alpha=Add(A=@n22, B=@n23)
   ))
  n25 = LinearInterpolate(
   A=LinearInterpolate(
    A=VectorParameter() {parameterName=BackGround, defaultValue=<rgba 0.0155 0.0279 0.1354 1>}.RGB
    B=VectorParameter() {parameterName=BackGround_1, defaultValue=<rgba 0.01 0.0002 0 1>}.RGB
    Alpha=@n13
   )
   B=Multiply(A=ScalarParameter() {defaultValue=0.1333, parameterName=CloudLayer2Brightness}, B=@n21)
   Alpha=@n24
  )
  n26 = Time()
  n27 = Panner(Coordinate=@n11, Time=Multiply(A=@n26, B=ScalarParameter() {defaultValue=0.0381, parameterName=SpeedPanCloud1})) {speedX=-0.33, speedY=-0.043}
  n28 = Panner(Coordinate=@n11, Time=Multiply(A=@n26, B=ScalarParameter() {defaultValue=0.0853, parameterName=SpeedPanCloud2})) {speedX=-0.44, speedY=0.1}
  n29 = Divide(A=Add(
    A=TextureSample(UVs=@n27) {samplerType=SAMPLERTYPE_LinearGrayscale, texture=&T_StormSwirlMask}.R
    B=TextureSample(UVs=@n28) {samplerType=SAMPLERTYPE_LinearGrayscale, texture=&T_StormSwirlMask}.R
   ))
  n30 = Power(Base=@n13, Exp=ScalarParameter() {defaultValue=0.5484, parameterName=FalloffGradient})
  n31 = LinearInterpolate(A=@n14.RGB, B=@n15.RGB, Alpha=Clamp(None=Multiply(A=@n29, B=@n30)))
  n32 = Multiply(A=@n29, B=ScalarParameter() {defaultValue=0.2857, parameterName=CloudmaskInfluence})
  n33 = Clamp(None=LinearInterpolate(
    A=ScalarParameter() {defaultValue=4.5217, parameterName=ContrastClouds}
    B=ScalarParameter() {defaultValue=-4.035, parameterName=ContrastClouds_1}
    Alpha=Add(A=@n32, B=@n13)
   ))
  n34 = Multiply(A=@n31, B=@n33)
  n35 = Time()
  n36 = ScalarParameter() {defaultValue=0.08, parameterName=LightingOffset}
  n37 = AppendVector(A=Multiply(A=Add(A=@n35, B=@n36), B=ScalarParameter() {defaultValue=0.0514, parameterName=LightingSpeed}), B=Constant())
  n38 = TextureSample(UVs=@n3) {
   samplerType=SAMPLERTYPE_LinearColor
   texture=/BRCosmetics/Gadgets/Assets/VinderTech_GliderChute/Textures/VinderTech_GliderChute_StormEye_HDR_M
  }
  n39 = Multiply(A=Time(), B=ScalarParameter() {defaultValue=1.1596, parameterName=TimeShift})
  n40 = Frac(None=@n39)
  n41 = OneMinus(None=@n40)
  n42 = Clamp(None=Multiply(A=Multiply(A=@n40, B=@n41), B=ScalarParameter() {defaultValue=7.8132, parameterName=FadeSpeedbetweenShit}))
  n43 = Multiply(A=Ceil(None=@n39), B=ScalarParameter() {defaultValue=0.1595, parameterName=TimeShiftOffset})
  n44 = Constant()
  n45 = Multiply(
   A=ScalarParameter() {defaultValue=3.6641, parameterName=UVOffsetLayer2}
   B=TextureSample(UVs=AppendVector(A=@n43, B=@n44)) {samplerType=SAMPLERTYPE_Masks, texture=/Game/Gadgets/Assets/VinderTech_GliderChute/Textures/T_LightingTimeline_M}.G
  )
  n46 = Constant()
  n47 = TextureSample(UVs=Add(A=@n3, B=AppendVector(A=@n45, B=@n46))) {
   samplerType=SAMPLERTYPE_LinearColor
   texture=&VinderTech_GliderChute_StormEye_HDR_M
  }
  n48 = Time()
  n49 = ScalarParameter() {defaultValue=3.68, parameterName=LightingOffset_1}
  n50 = AppendVector(A=Multiply(A=Add(A=@n48, B=@n49), B=ScalarParameter() {defaultValue=0.07, parameterName=LightingSpeed_1}), B=Constant())
  n51 = Add(
   A=Multiply(
    A=TextureSample(UVs=@n37) {samplerType=SAMPLERTYPE_Masks, texture=&T_LightingTimeline_M}.R
    B=@n38.R
   )
   B=Multiply(
    A=Multiply(A=@n42, B=@n47.G)
    B=TextureSample(UVs=@n50) {samplerType=SAMPLERTYPE_Masks, texture=&T_LightingTimeline_M}.R
   )
  )
  n52 = ScalarParameter() {defaultValue=10.2356, parameterName=ClougLightingBackground}
  n53 = Multiply(A=@n21, B=@n24)
  n54 = Multiply(A=@n53, B=@n51)
  n55 = Multiply(A=Time(), B=ScalarParameter() {defaultValue=1.1596, parameterName=TimeShift_1})
  n56 = Frac(None=@n55)
  n57 = Multiply(A=Multiply(A=@n56, B=OneMinus(None=@n56)), B=ScalarParameter() {defaultValue=7.8132, parameterName=FadeSpeedbetweenShit_1})
  n58 = Multiply(A=Ceil(None=@n55), B=ScalarParameter() {defaultValue=0.1595, parameterName=TimeShiftOffset_1})
  n59 = Constant()
  n60 = Multiply(
   A=ScalarParameter() {defaultValue=3.6641, parameterName=UVOffsetLayer2_1}
   B=TextureSample(UVs=AppendVector(A=@n58, B=@n59)) {samplerType=SAMPLERTYPE_Masks, texture=&T_LightingTimeline_M}.G
  )
  n61 = Constant()
  n62 = TextureSample(UVs=Add(A=@n3, B=AppendVector(A=@n60, B=@n61))) {
   samplerType=SAMPLERTYPE_LinearColor
   texture=&VinderTech_GliderChute_StormEye_HDR_M
  }
  n63 = Multiply(
   A=Add(A=Time(), B=ScalarParameter() {defaultValue=4.697, parameterName=LightingOffset_2})
   B=ScalarParameter() {defaultValue=0.07, parameterName=LightingSpeed_2}
  )
  n64 = Constant()
  n65 = Multiply(
   A=Multiply(A=Clamp(None=@n57), B=@n62.B)
   B=TextureSample(UVs=AppendVector(A=@n63, B=@n64)) {samplerType=SAMPLERTYPE_Masks, texture=&T_LightingTimeline_M}.R
  )
  n66 = Multiply(A=ScalarParameter() {defaultValue=1, parameterName=ClougLightingBackground_1}, B=Multiply(A=@n53, B=@n65))
  n67 = VectorParameter() {parameterName=Emissive, defaultValue=<rgba 1.3539 0.5156 1.5 1>}
  n68 = Multiply(A=@n51, B=@n67.RGB)
  n69 = Add(
   A=Multiply(A=ScalarParameter() {defaultValue=23.9454, parameterName=ClougLightingForeground}, B=Multiply(A=@n34, B=@n51))
   B=Add(A=Multiply(A=@n52, B=@n54), B=Add(A=@n66, B=@n68))
  )
  n70 = Constant()
  n71 = Multiply(A=@n65, B=@n67.RGB)
  n72 = Add(
   A=Multiply(A=ScalarParameter() {defaultValue=1.0232, parameterName=EmissiveBackground}, B=LinearInterpolate(A=@n25, B=@n31, Alpha=@n33))
   B=Add(A=LinearInterpolate(A=@n69, B=@n70, Alpha=@n33), B=LinearInterpolate(A=@n71, B=@n70, Alpha=@n24))
  )
  n73 = ComponentMask(None=@n72) {r=true}
  LinearInterpolate(A=ScalarParameter() {defaultValue=1, parameterName=R_2}, B=ScalarParameter() {defaultValue=-3.842, parameterName=R_1}, Alpha=@n73)
  LinearInterpolate(A=ScalarParameter() {defaultValue=-0.1333, parameterName=S_2}, B=ScalarParameter() {defaultValue=16.2532, parameterName=S_1}, Alpha=@n73)
  n74 = TextureSample() {texture=/BRCosmetics/Gadgets/Assets/VinderTech_GliderChute/Textures/T_VinderTech_GliderChute_StormEye_D}
  n75 = TextureSample() {samplerType=SAMPLERTYPE_Masks, texture=/BRCosmetics/Gadgets/Assets/VinderTech_GliderChute/Textures/T_VinderTech_GliderChute_StormEye_M}
  n76 = Multiply(
   A=TextureSample() {texture=/BRCosmetics/Gadgets/Assets/VinderTech_GliderChute/Textures/T_VinderTech_GliderChute_StormEye_E}.B
   B=VectorParameter() {parameterName=EmissiveLights, defaultValue=<rgba 3 1.2811 0 1>}.RGB
  )
  n77 = TextureSample() {samplerType=SAMPLERTYPE_Normal, texture=/BRCosmetics/Gadgets/Assets/VinderTech_GliderChute/Textures/T_VinderTech_GliderChute_StormEye_N}
  n78 = Constant()
  n79 = ScalarParameter() {defaultValue=0.1286, parameterName=S}
  n80 = ScalarParameter() {defaultValue=-0.8571, parameterName=R}
  n81 = TextureSample() {samplerType=SAMPLERTYPE_Masks, texture=/BRCosmetics/Gadgets/Assets/VinderTech_GliderChute/Textures/T_VinderTech_GliderChute_StormEye_M2}
  MaterialFunctionCall(In=MaterialFunctionCall!(
    Base Material=MakeMaterialAttributes!(
     BaseColor=@n74.RGB
     Metallic=@n75.G
     Specular=@n75.G
     Roughness=@n75.G
     Anisotropy
     EmissiveColor=@n76
     Opacity
     OpacityMask
     Normal=@n77.RGB
     Tangent
     WorldPositionOffset
     SubsurfaceColor
     ClearCoat
     ClearCoatRoughness
     AmbientOcclusion
     Refraction
     CustomizedUVs_0
     CustomizedUVs_1
     CustomizedUVs_2
     CustomizedUVs_3
     CustomizedUVs_4
     CustomizedUVs_5
     CustomizedUVs_6
     CustomizedUVs_7
     PixelDepthOffset
     ShadingModel
     Displacement
    )
    Top Material=MakeMaterialAttributes(BaseColor=@n78, Specular=@n79, Roughness=@n80, EmissiveColor=@n72, Normal=@n77.RGB)
    Alpha=@n81.R
   ) {materialFunction=/Engine/Functions/MaterialLayerFunctions/MatLayerBlend_Standard}.Blended Material) {materialFunction=&MF_TechArtOverrides_accessories}
  Multiply(A=Time(), B=@n10)
 ]
}
