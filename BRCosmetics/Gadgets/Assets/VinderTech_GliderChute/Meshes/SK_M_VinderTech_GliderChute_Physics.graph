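# Physics asset for the VinderTech GliderChute gadget's skeletal mesh, providing its simplified collision shape.
# A single body on the Frame bone holds one large box, offset and rotated relative to that bone.
# With no other bodies and no constraints, the asset defines no jointed ragdoll or secondary-motion chain.
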PhysicsAsset /BRCosmetics/Gadgets/Assets/VinderTech_GliderChute/Meshes/SK_M_VinderTech_GliderChute_Physics
bodies {Frame [{shapeType=Box, extentX=190.6205, extentY=80.8157, extentZ=206.6446, center=<1.3723 -12.7696 -16.5267>, rotation=<rot 0 50.4229 -87.6776>}]}
constraints []
